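# Niagara system for the aura effect of the RealCrown player emote.
# Big_aura pins sprites to body bones by spawn group, with per-group sizes and time-of-day dimming.
# Embers and EnergyAbsorb spawn in brief windows and get pulled inward by point attraction, while HandGlow bursts a sprite with lights, shifting into a color selected by System.RankPreview.
# Small_aura, Upward_energy, Glow_Mesh_Inhale and Flow are disabled.
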NiagaraSystem /BRCosmetics/Animation/Game/MainPlayer/Emotes/RealCrown/FX/NS_Emote_RealCrown_Aura
dependencies {
 usedDataInterfaces [
  &NiagaraDataInterfaceCurve
  &NiagaraDataInterfaceSkeletalMesh
  &NiagaraDataInterfaceArrayFloat
  &NiagaraDataInterfaceArrayFloat3
  &NiagaraDataInterfaceColorCurve
  &NiagaraDataInterfaceArrayColor
 ]
 usedDynamicInputs [
  &MakeCustomFloatFromBool
  &MakeCustomIntFromBool
  &SetBoolByFloatComparison
  &MakeFloatFromInt
  &Add_Integer
 ]
 usedModules [
  &SystemState
  &self:SystemSpawnScript.SystemScriptSource.SystemScriptGraph.NiagaraNodeAssignment_0.SetVariables_#1
  &self:SystemSpawnScript.SystemScriptSource.SystemScriptGraph.NiagaraNodeAssignment_1.SetVariables_#2
  &self:SystemSpawnScript.SystemScriptSource.SystemScriptGraph.NiagaraNodeAssignment_2.SetVariables_#3
  &self:SystemSpawnScript.SystemScriptSource.SystemScriptGraph.NiagaraNodeAssignment_4.SetVariables_#4
  &self:SystemSpawnScript.SystemScriptSource.SystemScriptGraph.NiagaraNodeAssignment_5.SetVariables_#5
 ]
}
emitters {
 Big_aura {
  emitterUpdateScript [
   &EmitterState(Life Cycle Mode=::NewEnumerator1, Loop Behavior=::NewEnumerator1, Loop Duration=3, UseLoopDelay=false)
   &SpawnBurst_Instantaneous(Spawn Count=8, Use Spawn Probability=false, UseLoopCountLimit=false)
  ]
  particleSpawnScript [
   &FN_SetSpawnGroup(Number of Spawn Groups=15)
   &InitializeParticle(
    Color=<rgba 2 1.9553 1.9751 0.35>
    Lifetime=2
    Sprite Size Mode=::NewEnumerator3
    Sprite UV Mode=::NewEnumerator1
    Uniform Sprite Size=&SelectFloatFromArray(
     Direct Array Index=$Particles.SpawnGroup
     Float Selection Array=DataInterface() {propertyValues {FloatData [100,70,70,70,70,50,50,130,130,100,100,100,100]}}
    )
    UsePositionOffset=false
    Write Material Random=true
   )
   &CameraOffset(Camera Offset Amount=-10)
  ]
  particleUpdateScript [
   &ParticleState()
   &SkeletalMeshLocation(
    Bone / Socket Index=$Particles.SpawnGroup
    Bone Sampling Mode=::NewEnumerator4
    MaskSpawnGroup=false
    Sampled Position Offset=&SelectVectorFromArray(
     Direct Array Index=$Particles.SpawnGroup
     Vector Selection Array=DataInterface() {propertyValues {FloatData [<0 0 0>, <0 0 0>, <0 0 0>, <0 0 0>, <0 0 0>, <5 0 0>, <-5 0 0>, <0 0 0>, <0 0 0>, <0 0 0>, <0 0 0>, <0 0 0>, <0 0 0>]}}
    )
    Skeletal Mesh=DataInterface() {propertyValues {
      DefaultMesh=null
      ExcludeBoneName=root
      OptionalComponentClass=null
      PreviewMesh=&M_Med_Soldier_04_Preview
      SkinningMode=SkinOnTheFly
      SoftSourceActor=null
      SourceMode=Default
      UvSetIndex=0
      WholeMeshLOD=-1
      bExcludeBone=false
      bReadDeformedGeometry=true
      bRequireCurrentFrameData=true
      ComponentTags []
      FilteredBones [head,upperarm_l,upperarm_r,lowerarm_l,lowerarm_r,hand_l,hand_r,thigh_l,thigh_r,calf_l,calf_r,foot_l,foot_r]
      FilteredSockets []
      MeshUserParameter {parameter {name=None, typeDefHandle {registeredTypeIndex=1}, varData []}}
      SamplingRegions []
     }}
   )
   &ScaleSpriteSize(
    Scale Factor=&Multiply_Vector2D_ByFloat(Float=2, Vector2D=&RandomRangeVector2D(Minimum=<0.7 1>, Maximum=<0.7 1>))
    Scale Sprite Size Mode=::NewEnumerator2
   ) {module=ScaleSpriteSize001}
   &DynamicMaterialParameters(
    Index 0 Param 4=&RandomRangeFloat(Minimum=0, Maximum=1)
    Module.Param0WriteEnabled=true
    Module.Param1WriteEnabled=true
    Module.Param2WriteEnabled=true
    Module.Param3WriteEnabled=true
   )
   &ScaleColor(
    Scale Alpha=&Multiply_Float(
     A=&Multiply_Float(A=1, B=&FN_FrontEndFloatModifier(Gameplay=1, Front End=1))
     B=&FloatFromCurve(FloatCurve=DataInterface() {propertyValues {Curve {keys [
          {tangentMode=RCTM_User, time=0, value=0}
          <at 0.2 1>
          {arriveTangent=-0.0405, interpMode=RCIM_Cubic, leaveTangent=-0.0405, tangentMode=RCTM_User, time=0.7009, value=1}
          {arriveTangent=-8.0089, interpMode=RCIM_Cubic, leaveTangent=-8.0089, tangentMode=RCTM_User, time=1, value=0}
         ]}}})
    )
    Scale RGB=&Multiply_VectorByFloat(
     Vector=<1 1 1>
     Float=&Multiply_Float(A=$System.TOD, B=&FN_FrontEndFloatModifier(Gameplay=1, Front End=1))
    )
    ScaleA=true
    ScaleRGB=true
   )
  ]
  rendererClasses [&NiagaraSpriteRendererProperties]
  renderers [{rendererClass=&NiagaraSpriteRendererProperties}]
 }
 Embers {
  emitterUpdateScript [
   &EmitterState()
   &SpawnRate(
    Spawn Probability=&FloatFromCurve(
     FloatCurve=DataInterface() {propertyValues {Curve {keys [<at 0.2 0 RCIM_Constant>,<at 0.5 1 RCIM_Constant>,<at 0.8 1 RCIM_Constant>,<at 1.4 0>]}}}
     CurveIndex=$System.Age
    )
    SpawnRate=40
    Use Spawn Probability=true
   )
  ]
  particleSpawnScript [
   /Niagara/Modules/Spawn/Initialization/InitializeParticle(
    Color=&RandomRangeLinearColor(Minimum=<rgba 2 1.9886 1.9 1>, Maximum=<rgba 0.1262 1.7405 2 1>)
    Lifetime=&RandomRangeFloat(Minimum=0.8, Maximum=1)
    Module.Write Color=true
    Module.Write Lifetime=true
    Module.Write Mass=true
    Module.Write Material Random=true
    Module.Write Position=true
    Module.Write Scale=false
    Module.Write SpriteRotation=false
    Module.Write SpriteSize=true
    Position=&SimulationPosition()
    Sprite Size=&Vector2DFromFloat(Value=&RandomRangeFloat(Minimum=2, Maximum=3))
   ) {enabled=false}
   /Niagara/Modules/Spawn/Initialization/V2/InitializeParticle(
    Color=&RandomRangeLinearColor(Minimum=<rgba 2 1.9886 1.9 1>, Maximum=<rgba 0.1262 1.7405 2 1>)
    Lifetime=&RandomRangeFloat(Minimum=0.8, Maximum=1)
    UsePositionOffset=false
    Write Material Random=true
   ) {module=InitializeParticle001}
   &SphereLocation(
    MaskSpawnGroup=false
    Override Seed=false
    Sphere Origin:NiagaraPosition=$Engine.Emitter.SimulationPosition
    Sphere Radius=150
    Surface Only=false
    UseNonUniformScale=false
    UseOffset=false
    UseSphereOrientation=false
   )
  ]
  particleUpdateScript [
   &ParticleState()
   /Niagara/Modules/Update/Forces/PointAttractionForce(
    Attraction Radius=5000
    AttractionStrength=6
    AttractorPosition:NiagaraPosition=$Engine.Emitter.SimulationPosition
    Kill Radius:NiagaraFloat=&Max_Float(
     A:NiagaraFloat=5
     B:NiagaraFloat=&VectorLength(Vector=&Multiply_VectorByFloat(Vector=$Particles.Velocity, Float=$Engine.DeltaTime))
    )
    Module.Kill Within Radius=true
    Module.Use Falloff=false
   )
   &SolveForcesAndVelocity(Clamp Velocity=false, Limit Acceleration=false)
   &ScaleColor(
    Scale Mode=::NewEnumerator1
    Scale RGBA=&MakeVector4_FromLinearColor(LinearColor=&ColorFromCurve(ColorCurve=DataInterface() {propertyValues {
        AlphaCurve {keys [<at 0.2971 1 RCIM_Cubic>,<at 0.7549 0>]}
        BlueCurve {keys [<at 0.0145 20>, {arriveTangent=-5.8718, interpMode=RCIM_Cubic, leaveTangent=-5.8718, time=0.1199, value=2.9051}, <at 1 14.2131 RCIM_Cubic>]}
        GreenCurve {keys [<at 0.0145 20>, {arriveTangent=-18.5572, interpMode=RCIM_Cubic, leaveTangent=-18.5572, time=0.1199, value=2.9343}, <at 1 1.7112 RCIM_Cubic>]}
        RedCurve {keys [<at 0.0145 20>, {arriveTangent=5.0734, interpMode=RCIM_Cubic, leaveTangent=5.0734, time=0.1199, value=3}, <at 1 25 RCIM_Cubic>]}
       }}))
    ScaleRGBA=true
   )
   &ScaleSpriteSize(Scale Factor=&Vector2DFromFloat(Value=&FloatFromCurve(FloatCurve=DataInterface() {propertyValues {Curve {keys [
          <at 0 0 RCIM_Cubic>
          {arriveTangent=0.836, interpMode=RCIM_Cubic, leaveTangent=0.836, time=0.2, value=1}
          {arriveTangent=-1.8455, interpMode=RCIM_Cubic, leaveTangent=-1.8455, tangentMode=RCTM_User, time=0.6507, value=0.544}
          <at 1 0>
         ]}}})))
   &ScaleSpriteSizeBySpeed(Max Scale Factor=<1 20>, Min Scale Factor=<0.5 0.1>)
   &DynamicMaterialParameters(
    Index 0 Param 1=1
    Index 0 Param 2=1
    Index 0 Param 3=0.5
    Module.Param0WriteEnabled=true
    Module.Param1WriteEnabled=true
    Module.Param2WriteEnabled=true
    Module.Param3WriteEnabled=true
   )
  ]
  rendererClasses [&NiagaraSpriteRendererProperties]
  renderers [{rendererClass=&NiagaraSpriteRendererProperties}]
 }
 EnergyAbsorb {
  emitterUpdateScript [
   &EmitterState()
   &SpawnRate(
    Spawn Probability=&FloatFromCurve(
     FloatCurve=DataInterface() {propertyValues {Curve {keys [<at 0.2 0 RCIM_Constant>,<at 0.5 1 RCIM_Constant>,<at 0.8 1 RCIM_Constant>,<at 1 0>]}}}
     CurveIndex=$System.Age
    )
    SpawnRate=50
    Use Spawn Probability=true
   )
  ]
  particleSpawnScript [
   /Niagara/Modules/Spawn/Initialization/InitializeParticle(
    Color=&RandomRangeLinearColor(Minimum=<rgba 15.84 16 16 1>, Maximum=<rgba 17.0762 17.3116 20 1>, Color Channel Mode=::NewEnumerator0)
    Lifetime=1.5
    Module.Write Color=true
    Module.Write Lifetime=true
    Module.Write Mass=false
    Module.Write Material Random=true
    Module.Write Position=false
    Module.Write Scale=false
    Module.Write SpriteRotation=false
    Module.Write SpriteSize=true
    Position=&SimulationPosition()
    Sprite Size=&Vector2DFromFloat(Value=&RandomRangeFloat(Minimum=0.3, Maximum=0.9))
   ) {enabled=false}
   /Niagara/Modules/Spawn/Initialization/V2/InitializeParticle(
    Color=&RandomRangeLinearColor(Minimum=<rgba 15.84 16 16 1>, Maximum=<rgba 17.0762 17.3116 20 1>, Color Channel Mode=::NewEnumerator0)
    Lifetime=1.5
    UsePositionOffset=false
    Write Material Random=true
   ) {module=InitializeParticle001}
   &SphereLocation(
    MaskSpawnGroup=false
    Override Seed=false
    Sphere Origin:NiagaraPosition=$Engine.Owner.Position
    Surface Only=false
    UseNonUniformScale=false
    UseOffset=false
    UseSphereOrientation=false
   )
   &AddVelocity(Velocity=&RandomRangeVector(Minimum=<-500 -500 -500>, Maximum=<500 500 500>)) {enabled=false}
  ]
  particleUpdateScript [
   &ParticleState()
   &PointAttractionForce(
    Attraction Radius=200
    AttractionStrength=3
    AttractorPosition:NiagaraPosition=$Engine.Owner.Position
    Kill Radius=3
    Module.Kill Within Radius=true
    Module.Use Falloff=false
   )
   &SolveForcesAndVelocity(Clamp Velocity=false, Limit Acceleration=false)
   &ScaleSpriteSizeBySpeed(Max Scale Factor=<1 2>, Min Scale Factor=<0.75 1.5>)
   &ScaleColor(
    Scale Alpha=&FloatFromCurve(
     FloatCurve=DataInterface() {propertyValues {Curve {keys [<at 0.2 0 RCIM_Cubic>,<at 0.5 1>,<at 0.95 1 RCIM_Cubic>,<at 1 0>]}}}
     Scale Curve=&FloatFromCurve(FloatCurve=DataInterface() {propertyValues {Curve {keys [<at 0 1 RCIM_Constant>,<at 0.75 0>]}}}, CurveIndex=$Emitter.NormalizedLoopAge)
    )
    Scale RGB=&Multiply_Vector(
     A=&MakeVectorFromLinearColorRGB(Color=&ColorFromCurve(ColorCurve=DataInterface() {propertyValues {
         AlphaCurve {keys [<at 0 1 RCIM_Cubic>]}
         BlueCurve {keys [<at 0.8 3 RCIM_Cubic>,<at 0.95 8 RCIM_Cubic>]}
         GreenCurve {keys [<at 0.8 3 RCIM_Cubic>,<at 0.95 8 RCIM_Cubic>]}
         RedCurve {keys [<at 0.8 3 RCIM_Cubic>,<at 0.95 8 RCIM_Cubic>]}
        }}))
     B=<1 1 1>
    )
    ScaleA=true
    ScaleRGB=true
   )
  ]
  rendererClasses [&NiagaraSpriteRendererProperties]
  renderers [{rendererClass=&NiagaraSpriteRendererProperties}]
 }
 Flow {
  bEnabled=false
  emitterUpdateScript [
   &EmitterState(Life Cycle Mode=::NewEnumerator1, Loop Behavior=::NewEnumerator1, Loop Duration=1, UseLoopDelay=false)
   &SpawnBurst_Instantaneous(Spawn Count=1, Spawn Time=0.9, Use Spawn Probability=false)
  ]
  particleSpawnScript [
   &FN_SetSpawnGroup(Number of Spawn Groups=2)
   &InitializeParticle(
    Color=<rgba 0.4348 0.4647 0.5 0.5>
    Lifetime=1.4
    Mesh Scale Mode=::NewEnumerator3
    Mesh Uniform Scale=0.35
    UsePositionOffset=false
    Write Material Random=true
   )
   &InitialMeshOrientation(Orientation Axis=<0 0 1>, Orientation Vector=<0 0 1>, Rotation=<0 0.25 0>, Use Orientation Vector=true, Use Rotation Vector=true) {module=InitialMeshOrientation001}
   &InitialMeshOrientation(
    Rotation=&MakeCustomVectorFromBool(
     Bool=&SetBoolByIntComparison(A=$Particles.SpawnGroup, B=0, Comparison Type=::NewEnumerator4)
     True Vector=<0 -0.05 -0.05>
     False Vector=<0.5 0.05 0.05>
    )
    Use Orientation Vector=true
    Use Rotation Vector=true
   )
  ]
  particleUpdateScript [
   &ParticleState(Kill Particles When Lifetime Has Elapsed=false, Let Infinitely Lived Particles Die When Emitter Deactivates=true)
   &ScaleMeshSize(Scale Factor=&VectorFromFloat(Value=&FloatFromCurve(FloatCurve=DataInterface() {propertyValues {Curve {keys [
          <at 0 0.5 RCIM_Cubic>
          {arriveTangent=0.625, interpMode=RCIM_Cubic, leaveTangent=0.625, time=0.4, value=1}
          {arriveTangent=-0.8333, interpMode=RCIM_Cubic, leaveTangent=-0.8333, time=0.8, value=1}
          <at 1 0.5 RCIM_Cubic>
         ]}}})))
   &ScaleColor(
    Scale Alpha=&Multiply_Float(
     A=&FloatFromCurve(FloatCurve=DataInterface() {propertyValues {Curve {keys [
          <at 0.102 0 RCIM_Cubic>
          {interpMode=RCIM_Cubic, tangentMode=RCTM_User, time=0.3763, value=1}
          {arriveTangent=-2.4769, interpMode=RCIM_Cubic, leaveTangent=-2.4769, time=0.75, value=0.25}
          {arriveTangent=-1, interpMode=RCIM_Cubic, leaveTangent=-1, time=0.78, value=0}
          {interpMode=RCIM_Cubic, tangentMode=RCTM_User, time=1, value=0}
         ]}}})
     B=&FNTimeOfDayScalar(Night Scale=0.7, Morning Scale=0.5, Evening Scale=0.5)
    )
    ScaleA=true
    ScaleRGB=true
   )
   &Color(Color=&ColorFromCurve(ColorCurve=DataInterface() {propertyValues {
       AlphaCurve {keys [
         <at 0.0066 0>
         {arriveTangent=0.2556, interpMode=RCIM_Cubic, leaveTangent=0.2556, time=0.2379, value=1}
         <at 0.797 0.202>
         <at 0.8368 0 RCIM_Cubic>
        ]}
       BlueCurve {keys [<at 0 0.1905 RCIM_Cubic>,<at 0.2379 1>,<at 0.4653 0.1796>,<at 0.6946 0.531>,<at 1 1 RCIM_Cubic>]}
       GreenCurve {keys [<at 0 0.5574 RCIM_Cubic>,<at 0.2379 0.6441>,<at 0.4653 0.7814>,<at 0.6946 1>,<at 1 0.3486 RCIM_Cubic>]}
       RedCurve {keys [<at 0 1 RCIM_Cubic>,<at 0.2379 0.3897>,<at 0.4653 1>,<at 0.6946 0.2118>,<at 1 0.0904 RCIM_Cubic>]}
      }}))
  ]
  rendererClasses [&NiagaraMeshRendererProperties]
  renderers [{rendererClass=&NiagaraMeshRendererProperties, rendererIndex=0}]
 }
 Glow_Mesh_Inhale {
  bEnabled=false
  emitterUpdateScript [
   &EmitterState()
   &SpawnRate(
    Spawn Probability=&FloatFromCurve(
     FloatCurve=DataInterface() {propertyValues {Curve {keys [<at 0.6 0 RCIM_Cubic>,<at 1.3 1 RCIM_Cubic>,<at 1.7 0 RCIM_Cubic>]}}}
     CurveIndex=$System.Age
    )
    SpawnRate=15
    Use Spawn Probability=true
   )
  ]
  particleSpawnScript [
   /Niagara/Modules/Spawn/Initialization/InitializeParticle(
    Color=&RandomRangeLinearColor(Minimum=<rgba 0.5229 0.7128 1 0.5>, Maximum=<rgba 4.8704 7.5064 10 0.75>)
    Lifetime=0.45
    Mesh Scale=&VectorFromFloat(Value=0.1)
    Module.Write Color=true
    Module.Write Lifetime=true
    Module.Write Mass=true
    Module.Write Material Random=true
    Module.Write Position=true
    Module.Write Scale=true
    Module.Write SpriteRotation=false
    Module.Write SpriteSize=false
    Sprite Rotation=&RandomRangeFloat(Minimum=0, Maximum=360)
    Sprite Size=&Vector2DFromFloat(Value=&RandomRangeFloat(Minimum=3000, Maximum=3200))
   )
   &InitialMeshOrientation(
    Rotation=&RandomRangeVector(Minimum=<-1 -1 -1>, Maximum=<1 1 1>)
    Use Orientation Vector=true
    Use Rotation Vector=true
   ) {module=InitialMeshOrientation001}
   &DynamicMaterialParameters(
    Index 0 Param 1=1
    Index 0 Param 2=1
    Index 0 Param 3=1
    Index 0 Param 4=1
    Module.Param0WriteEnabled=true
    Module.Param1WriteEnabled=true
    Module.Param2WriteEnabled=true
    Module.Param3WriteEnabled=true
   )
  ]
  particleUpdateScript [
   &ParticleState()
   &ScaleColor(
    Scale Alpha=&FloatFromCurve(
     FloatCurve=DataInterface() {propertyValues {Curve {keys [
         <at 0.1 0 RCIM_Cubic>
         {interpMode=RCIM_Cubic, leaveTangent=-5.9941, tangentMode=RCTM_Break, time=0.5, value=1}
         {interpMode=RCIM_Cubic, leaveTangent=-23.3931, tangentMode=RCTM_Break, time=0.9, value=0}
        ]}}}
     Scale Curve=0.02
    )
    Scale RGB=&Multiply_VectorByFloat(
     Vector=&MakeVectorFromLinearColorRGB(Color=&ColorFromCurve(
       ColorCurve=DataInterface() {propertyValues {
         AlphaCurve {keys []}
         BlueCurve {keys [{arriveTangent=3.4035, interpMode=RCIM_Cubic, leaveTangent=3.4034, tangentMode=RCTM_User, time=0.1, value=4}, <at 0.75 4 RCIM_Cubic>]}
         GreenCurve {keys [<at 0.1 3.8351 RCIM_Cubic>,<at 0.75 1.2091 RCIM_Cubic>]}
         RedCurve {keys [
           <at 0.1 3.9561 RCIM_Cubic>
           {arriveTangent=-14.5406, interpMode=RCIM_Cubic, leaveTangent=-14.5406, tangentMode=RCTM_User, time=0.75, value=0}
          ]}
        }}
       CurveIndex=$Emitter.NormalizedLoopAge
      ))
     Float=&FloatFromCurve(FloatCurve=DataInterface() {propertyValues {Curve {keys [<at 0 2>,<at 1 0>]}}})
    )
    ScaleA=true
    ScaleRGB=true
   )
   &ScaleMeshSize(Scale Factor=&VectorFromFloat(Value=&FloatFromCurve(FloatCurve=DataInterface() {propertyValues {Curve {keys [<at 0 1 RCIM_Cubic>, {arriveTangent=-1.6059, interpMode=RCIM_Cubic, leaveTangent=-1.6059, tangentMode=RCTM_User, time=1, value=0}]}}})))
  ]
  rendererClasses [&NiagaraMeshRendererProperties]
  renderers [{rendererClass=&NiagaraMeshRendererProperties, rendererIndex=0}]
 }
 HandGlow {
  emitterUpdateScript [
   &EmitterState(
    Inactive Response=::NewEnumerator1
    Life Cycle Mode=::NewEnumerator1
    Loop Behavior=::NewEnumerator1
    Loop Duration Mode=::NewEnumerator1
    UseLoopDelay=false
   )
   &SpawnBurst_Instantaneous(Spawn Count=1, Spawn Time=0.9, Use Spawn Probability=false)
  ]
  particleSpawnScript [/Niagara/Modules/Spawn/Initialization/V2/InitializeParticle(
    Color=<rgba 2 0.628 0.0387 1>
    Lifetime=1.6
    Mass Mode=::NewEnumerator1
    Sprite Size Mode=::NewEnumerator3
    Uniform Sprite Size=100
    UsePositionOffset=false
    Write Material Random=true
   ) {module=InitializeParticle001}]
  particleUpdateScript [
   &ParticleState()
   &ScaleSpriteSize(Uniform Curve Sprite Scale=DataInterface() {propertyValues {Curve {keys [
        <at 0 0.5 RCIM_Cubic>
        {arriveTangent=0.9091, interpMode=RCIM_Cubic, leaveTangent=0.9091, time=0.4, value=1}
        {arriveTangent=-2.5, interpMode=RCIM_Cubic, leaveTangent=-2.5, time=0.55, value=1}
        <at 0.8 0 RCIM_Cubic>
       ]}}})
   &Color(Color=&MultiplyLinearColor(
     Linear Color=&LerpLinearColors(
      StartColor=&ColorFromCurve(ColorCurve=DataInterface() {propertyValues {
         AlphaCurve {keys [
           <at 0.0066 0>
           {arriveTangent=0.1927, interpMode=RCIM_Cubic, leaveTangent=0.1927, time=0.2379, value=1}
           <at 0.797 0.1523>
           <at 0.8349 0 RCIM_Cubic>
          ]}
         BlueCurve {keys [<at 0 0.1905 RCIM_Cubic>,<at 0.2379 1>,<at 0.4653 0.1796>,<at 0.7363 0.747>,<at 1 1>]}
         GreenCurve {keys [<at 0 0.5574 RCIM_Cubic>,<at 0.2379 0.6441>,<at 0.4653 0.7814>,<at 0.7363 1>,<at 1 0.4669>]}
         RedCurve {keys [<at 0 1 RCIM_Cubic>,<at 0.2379 0.3897>,<at 0.4653 1>,<at 0.7363 0.2062>,<at 1 0.2351>]}
        }})
      EndColor=&SelectLinearColorFromArray(
       Array Sampling Mode=::NewEnumerator1
       Direct Array Index=&MakeIntFromFloat(Float=$System.RankPreview)
       Color Selection Array=DataInterface() {propertyValues {ColorData [<rgba 0.3282 0.5 0.476 1>, <rgba 1 0.4679 0.1164 1>, <rgba 0.4588 0.0706 1 1>]}}
       Random Seed=?
      )
      LerpFactor=&FloatFromCurve(FloatCurve=DataInterface() {propertyValues {Curve {keys [<at 0.3342 0>,<at 0.6684 1.0128>]}}})
     )
     Scale Factor=&MakeFloat4FromVectorAndFloat(
      "Vector (RGB)"=&VectorFromFloat(Value=&FloatFromCurve(FloatCurve=DataInterface() {propertyValues {Curve {keys [<at 0.6 1>,<at 0.7 1.5>,<at 1 0>]}}}))
      "Float (Alpha)"=&FloatFromCurve(FloatCurve=DataInterface() {propertyValues {Curve {keys [<at 0.8 1>,<at 1 0>]}}})
     )
    ))
  ]
  rendererClasses [&NiagaraSpriteRendererProperties, &NiagaraLightRendererProperties]
  renderers [
   {rendererClass=&NiagaraSpriteRendererProperties}
   {rendererClass=&NiagaraLightRendererProperties, rendererIndex=1}
   {rendererClass=&NiagaraLightRendererProperties, rendererIndex=2}
  ]
 }
 Small_aura {
  bEnabled=false
  emitterUpdateScript [
   &EmitterState(Life Cycle Mode=::NewEnumerator1, Loop Behavior=::NewEnumerator1, Loop Delay=1.5, Loop Duration=3, UseLoopDelay=false)
   &SpawnBurst_Instantaneous(Spawn Count=13, Use Spawn Probability=false, UseLoopCountLimit=false)
  ]
  particleSpawnScript [
   &FN_SetSpawnGroup(Number of Spawn Groups=15)
   &InitializeParticle(
    Color=<rgba 3.758 3.9671 5 0.1>
    Lifetime=2
    Sprite Size Mode=::NewEnumerator3
    Sprite UV Mode=::NewEnumerator1
    Uniform Sprite Size=&SelectFloatFromArray(
     Direct Array Index=$Particles.SpawnGroup
     Float Selection Array=DataInterface() {propertyValues {FloatData [80,80,80,80,80,60,60,130,130,100,100,100,100]}}
    )
    UsePositionOffset=false
    Write Material Random=true
   )
   &CameraOffset(Camera Offset Amount=-10)
  ]
  particleUpdateScript [
   &ParticleState()
   &SkeletalMeshLocation(
    Bone / Socket Index=$Particles.SpawnGroup
    Bone Sampling Mode=::NewEnumerator4
    MaskSpawnGroup=false
    Sampled Position Offset=&SelectVectorFromArray(
     Direct Array Index=$Particles.SpawnGroup
     Vector Selection Array=DataInterface() {propertyValues {FloatData [<0 0 0>, <0 0 0>, <0 0 0>, <0 0 0>, <0 0 0>, <5 0 0>, <-5 0 0>, <0 0 0>, <0 0 0>, <0 0 0>, <0 0 0>, <0 0 0>, <0 0 0>]}}
    )
    Skeletal Mesh=DataInterface() {propertyValues {
      DefaultMesh=null
      ExcludeBoneName=root
      OptionalComponentClass=null
      PreviewMesh=&M_Med_Soldier_04_Preview
      SkinningMode=SkinOnTheFly
      SoftSourceActor=null
      SourceMode=Default
      UvSetIndex=0
      WholeMeshLOD=-1
      bExcludeBone=false
      bReadDeformedGeometry=true
      bRequireCurrentFrameData=true
      ComponentTags []
      FilteredBones [head,upperarm_l,upperarm_r,lowerarm_l,lowerarm_r,hand_l,hand_r,thigh_l,thigh_r,calf_l,calf_r,foot_l,foot_r]
      FilteredSockets []
      MeshUserParameter {parameter {name=None, typeDefHandle {registeredTypeIndex=1}, varData []}}
      SamplingRegions []
     }}
   )
   &ScaleSpriteSize(
    Scale Factor=&Multiply_Vector2D_ByFloat(Float=0.7, Vector2D=&RandomRangeVector2D(Minimum=<0.9 1>, Maximum=<0.9 1>))
    Scale Sprite Size Mode=::NewEnumerator2
   ) {module=ScaleSpriteSize001}
   &DynamicMaterialParameters(
    Index 0 Param 4=&RandomRangeFloat(Minimum=0, Maximum=1)
    Module.Param0WriteEnabled=true
    Module.Param1WriteEnabled=true
    Module.Param2WriteEnabled=true
    Module.Param3WriteEnabled=true
   )
   &ScaleColor(
    Scale Alpha=&Multiply_Float(
     A=&Multiply_Float(A=1, B=&FN_FrontEndFloatModifier(Gameplay=1, Front End=1))
     B=&FloatFromCurve(FloatCurve=DataInterface() {propertyValues {Curve {keys [
          {tangentMode=RCTM_User, time=0, value=0}
          <at 0.2 1>
          {arriveTangent=-0.0405, interpMode=RCIM_Cubic, leaveTangent=-0.0405, tangentMode=RCTM_User, time=0.7009, value=1}
          {arriveTangent=-8.0089, interpMode=RCIM_Cubic, leaveTangent=-8.0089, tangentMode=RCTM_User, time=1, value=0}
         ]}}})
    )
    Scale RGB=&Multiply_VectorByFloat(
     Vector=<1 1 1>
     Float=&Multiply_Float(A=$System.TOD, B=&FN_FrontEndFloatModifier(Gameplay=1, Front End=1))
    )
    ScaleA=true
    ScaleRGB=true
   )
  ]
  rendererClasses [&NiagaraSpriteRendererProperties]
  renderers [{rendererClass=&NiagaraSpriteRendererProperties}]
 }
 Upward_energy {
  bEnabled=false
  emitterUpdateScript [
   &EmitterState(
    Inactive Response=::NewEnumerator1
    Life Cycle Mode=::NewEnumerator1
    Loop Behavior=::NewEnumerator1
    Loop Delay=0.1
    Loop Duration=1.5
    UseLoopDelay=true
   )
   &SpawnRate(SpawnRate=10, Use Spawn Probability=false)
  ]
  particleSpawnScript [
   &InitializeParticle(
    Color=<rgba 1 1 1 0.2>
    Lifetime=0.3
    Mesh Scale=&RandomRangeVector(Minimum=<0.6 1 1>, Maximum=<0.8 1.3 1.3>)
    Mesh Scale Mode=::NewEnumerator4
    UsePositionOffset=false
    Write Material Random=true
   )
   &InitialMeshOrientation(Mesh Orientation Mode=::NewEnumerator2, Rotation=<0 0.25 0>, Use Rotation Vector=true)
   &InitialMeshOrientation(
    Mesh Orientation Mode=::NewEnumerator2
    Rotation=&RandomRangeVector(Minimum=<0 0 0>, Maximum=<0 0 1>)
    Use Rotation Vector=true
   ) {module=InitialMeshOrientation001}
  ]
  particleUpdateScript [
   &ParticleState()
   &DynamicMaterialParameters(
    Index 0 Param 1=1.5
    Index 0 Param 3=1
    Index 0 Param 4=1
    Module.Param0WriteEnabled=true
    Module.Param1WriteEnabled=true
    Module.Param2WriteEnabled=true
    Module.Param3WriteEnabled=true
   )
   &ScaleMeshSize(Scale Factor=&Multiply_VectorByFloat(
     Vector=<0.9 0.85 0.85>
     Float=&FloatFromCurve(FloatCurve=DataInterface() {propertyValues {Curve {keys [{interpMode=RCIM_Cubic, tangentMode=RCTM_User, time=0, value=0.4}, {interpMode=RCIM_Cubic, tangentMode=RCTM_User, time=1, value=1}]}}})
    ))
   &SkeletalMeshLocation(
    Bone Sampling Mode=::NewEnumerator4
    MaskSpawnGroup=false
    Sampled Position Offset=<0 0 -20>
    Skeletal Mesh=DataInterface() {propertyValues {
      DefaultMesh=null
      ExcludeBoneName=root
      OptionalComponentClass=null
      PreviewMesh=&M_Med_Soldier_04_Preview
      SkinningMode=SkinOnTheFly
      SoftSourceActor=null
      SourceMode=Default
      UvSetIndex=0
      WholeMeshLOD=-1
      bExcludeBone=false
      bReadDeformedGeometry=true
      bRequireCurrentFrameData=true
      ComponentTags []
      FilteredBones [root]
      FilteredSockets []
      MeshUserParameter {parameter {name=None, typeDefHandle {registeredTypeIndex=1}, varData []}}
      SamplingRegions []
     }}
   )
   &ScaleColor(
    Scale Alpha=&Multiply_Float(
     A=$System.TOD
     B=&FloatFromCurve(FloatCurve=DataInterface() {propertyValues {Curve {keys [
          {interpMode=RCIM_Cubic, tangentMode=RCTM_User, time=0, value=0}
          {interpMode=RCIM_Cubic, tangentMode=RCTM_User, time=0.3765, value=1}
          {tangentMode=RCTM_User, time=1, value=0}
         ]}}})
    )
    ScaleA=true
    ScaleRGB=true
   )
  ]
  rendererClasses [&NiagaraMeshRendererProperties]
  renderers [{rendererClass=&NiagaraMeshRendererProperties, rendererIndex=0}]
 }
}
summary {
 systemName=NS_Emote_RealCrown_Aura
 emitters [
  {bEnabled=false, emitterName=Upward_energy, rendererClasses [&NiagaraMeshRendererProperties]}
  {bEnabled=false, emitterName=Small_aura, rendererClasses [&NiagaraSpriteRendererProperties]}
  {emitterName=Big_aura, rendererClasses [&NiagaraSpriteRendererProperties]}
  {emitterName=Embers, rendererClasses [&NiagaraSpriteRendererProperties]}
  {emitterName=EnergyAbsorb, rendererClasses [&NiagaraSpriteRendererProperties]}
  {bEnabled=false, emitterName=Glow_Mesh_Inhale, rendererClasses [&NiagaraMeshRendererProperties]}
  {bEnabled=false, emitterName=Flow, rendererClasses [&NiagaraMeshRendererProperties]}
  {emitterName=HandGlow, rendererClasses [&NiagaraSpriteRendererProperties, &NiagaraLightRendererProperties]}
 ]
 userVariables [
  User.bIsFrontEnd {type=NiagaraBool}
  User.bIsFrontEndPreview {type=NiagaraBool}
  User.Rank_Champion {type=NiagaraInt32, defaultValue=888}
  User.Rank_Champion_Reload {type=NiagaraInt32, defaultValue=555}
  User.Rank_Elite {type=NiagaraInt32, defaultValue=777}
  User.Rank_Elite_Reload {type=NiagaraInt32, defaultValue=444}
  User.Rank_Unreal {type=NiagaraInt32, defaultValue=999}
  User.Rank_Unreal_Reload {type=NiagaraInt32, defaultValue=666}
  User.RibbonIDOffset {type=NiagaraInt32}
 ]
}
