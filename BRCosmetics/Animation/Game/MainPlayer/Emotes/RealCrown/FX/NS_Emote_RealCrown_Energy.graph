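# Niagara system for the RealCrown emote's energy effect, displaying the player's Elite, Champion and Unreal rank icons with counters amid a burst of energy.
# A sprite-and-light black hole color wave, expanding mesh waves and curl-noise sparks form the visible energy layers.
# RankIcons indexes layout arrays by System.RankInt and animates icons with curves.
# Its counters sum User rank and Reload values, cap them, and pass digit data to material parameters.
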NiagaraSystem /BRCosmetics/Animation/Game/MainPlayer/Emotes/RealCrown/FX/NS_Emote_RealCrown_Energy
dependencies {
 usedDataInterfaces [
  &NiagaraDataInterfaceCurve
  &NiagaraDataInterfaceColorCurve
  &NiagaraDataInterfaceArrayFloat3
  &NiagaraDataInterfaceArrayFloat2
  &NiagaraDataInterfaceArrayFloat
  &NiagaraDataInterfaceArrayColor
  &NiagaraDataInterfaceVector2DCurve
 ]
 usedDynamicInputs [&SimulationPosition, &MakeCustomIntFromBool]
 usedModules [
  &SystemState
  &self:SystemSpawnScript.SystemScriptSource.SystemScriptGraph.NiagaraNodeAssignment_2.SetVariables_#1
  &self:SystemSpawnScript.SystemScriptSource.SystemScriptGraph.NiagaraNodeAssignment_3.SetVariables_#2
  &self:SystemSpawnScript.SystemScriptSource.SystemScriptGraph.NiagaraNodeAssignment_0.SetVariables_#3
 ]
}
emitters {
 ColorWaveBlackhole001 {
  emitterSpawnScript [/SetVariables(Emitter.SortOrder=3) {isSetParameters=true}]
  emitterUpdateScript [
   &EmitterState(Inactive Response=::NewEnumerator1, Life Cycle Mode=::NewEnumerator1, Loop Behavior=::NewEnumerator1, Loop Duration=5.5, UseLoopDelay=false)
   &SpawnBurst_Instantaneous(Use Spawn Probability=false, UseLoopCountLimit=false)
  ]
  particleSpawnScript [&InitializeParticle(
    Color=<rgba 3 3 3 2>
    Lifetime=5.5
    Mesh Scale Mode=::NewEnumerator3
    Mesh Uniform Scale=-3.5
    Position Offset=$System.Position
    Sprite Rotation Angle=90
    Sprite Rotation Mode=::NewEnumerator0
    Sprite Size Mode=::NewEnumerator3
    Uniform Sprite Size=&MakeCustomFloatFromBool(
     Bool=&SetBoolByFloatComparison(A=&MakeFloatFromInt(Int=$User.Rank_Unreal), B=0)
     True Float=160
     False Float=150
    )
    UsePositionOffset=true
    Write Material Random=true
   )]
  particleUpdateScript [
   &ParticleState() {module=ParticleState001}
   &ScaleSpriteSize(Uniform Curve Sprite Scale=DataInterface() {propertyValues {Curve {keys [
        {arriveTangent=16.3505, interpMode=RCIM_Cubic, leaveTangent=16.3505, tangentMode=RCTM_User, time=0, value=0}
        {arriveTangent=0.0836, interpMode=RCIM_Cubic, leaveTangent=0.0836, tangentMode=RCTM_User, time=0.16, value=1}
        {arriveTangent=0.1415, interpMode=RCIM_Cubic, leaveTangent=0.1415, time=0.7668, value=1}
        <at 0.8668 1.1 RCIM_Cubic>
        <at 1 1 RCIM_Cubic>
       ]}}}) {module=ScaleSpriteSize001}
   /SetVariables(Particles.Color02:LinearColor=&MakeLinearColorFromVectorAndFloat(
     "Vector (RGB)"=&MakeVectorFromLinearColorRGB(Color=&MultiplyLinearColor(Linear Color=$System.ColorWaveCol, Scale Factor=<2 2 2 1>))
     "Float (Alpha)"=&FloatFromCurve(FloatCurve=DataInterface() {propertyValues {Curve {keys [
          {interpMode=RCIM_Cubic, tangentMode=RCTM_User, time=0, value=1}
          {interpMode=RCIM_Cubic, tangentMode=RCTM_User, time=0.5, value=1}
          {arriveTangent=-2, interpMode=RCIM_Cubic, leaveTangent=-2, time=0.71, value=1}
          {tangentMode=RCTM_User, time=1, value=0}
         ]}}})
    )) {isSetParameters=true}
   &DynamicMaterialParameters(
    Index 0 Param 1=1
    Index 0 Param 2=&FloatFromCurve(FloatCurve=DataInterface() {propertyValues {Curve {keys [
         {interpMode=RCIM_Cubic, tangentMode=RCTM_User, time=0, value=1}
         {arriveTangent=-1, interpMode=RCIM_Cubic, leaveTangent=-1, time=0.6, value=0.7}
         {interpMode=RCIM_Cubic, leaveTangent=-3.3753, time=1, value=0}
        ]}}})
    Index 0 Param 3=1
    Index 0 Param 4=1
    Module.Param0WriteEnabled=true
    Module.Param1WriteEnabled=true
    Module.Param2WriteEnabled=true
    Module.Param3WriteEnabled=true
   )
   &ScaleColor(
    Scale Alpha=&FloatFromCurve(
     FloatCurve=DataInterface() {propertyValues {Curve {keys [
         {interpMode=RCIM_Cubic, tangentMode=RCTM_User, time=0, value=0}
         {interpMode=RCIM_Cubic, tangentMode=RCTM_User, time=0.5, value=1}
         {tangentMode=RCTM_User, time=1, value=0}
        ]}}}
     Scale Curve=2
    )
    Scale RGB=&MakeVectorFromLinearColorRGB(Color=&MultiplyLinearColor(
      Linear Color=&ColorFromCurve(ColorCurve=DataInterface() {propertyValues {
         AlphaCurve {keys []}
         BlueCurve {keys [<at 0.1259 2 RCIM_Cubic>, <at 0.2277 5>, {interpMode=RCIM_Cubic, leaveTangent=-0.7175, time=0.4478, value=2}]}
         GreenCurve {keys [<at 0.1259 0.1681 RCIM_Cubic>, <at 0.2277 0.2344>, {interpMode=RCIM_Cubic, leaveTangent=-4.4353, time=0.4478, value=0.1399}]}
         RedCurve {keys [<at 0.1259 0.5589 RCIM_Cubic>, <at 0.2277 1.1257>, {interpMode=RCIM_Cubic, leaveTangent=2.2176, time=0.4478, value=0.5865}]}
        }})
      Scale Factor=<1 1 1 1>
     ))
    ScaleA=true
    ScaleRGB=true
   )
   &FN_Sprite_Attributes(Color=$System.BlackholeColor, Create Color Attribute=true, Create Sprite Rotation Attribute=true, Write to Module Namespace=true)
   /SetVariables(
    Particles.BlackHoleDyn1:Vector4f=&MakeVector4(X=0.2, Y=2, Z=1, W=1)
    Particles.BlackHoleDyn2:Vector4f=&MakeVector4(
     X=&MakeFloatFromVector2d(Vector 2D=$System.BlackHoleAtt)
     Y=&MakeFloatFromVector2d(Vector 2D=$System.BlackHoleAtt, Channel=::NewEnumerator1)
     Z=&FloatFromCurve(FloatCurve=DataInterface() {propertyValues {Curve {keys [<at 0.76 1>,<at 1 0>]}}})
     W=0.6
    )
   ) {isSetParameters=true}
   /SetVariables(Particles.BlackHoleSize:Vector2f=&Vector2DFromCurve(
     Vector2Curve=DataInterface() {propertyValues {
       XCurve {keys [
         {arriveTangent=13.4205, interpMode=RCIM_Cubic, leaveTangent=13.4205, tangentMode=RCTM_User, time=0, value=0}
         {arriveTangent=1.1111, interpMode=RCIM_Cubic, leaveTangent=1.1111, time=0.15, value=1}
         {arriveTangent=-0.3529, interpMode=RCIM_Cubic, leaveTangent=-0.3529, time=0.9, value=1}
         <at 1 0.7 RCIM_Cubic>
        ]}
       YCurve {keys [
         {arriveTangent=11.4842, interpMode=RCIM_Cubic, leaveTangent=11.4842, tangentMode=RCTM_User, time=0, value=0}
         {arriveTangent=1.1111, interpMode=RCIM_Cubic, leaveTangent=1.1111, time=0.15, value=1}
         {arriveTangent=-0.3529, interpMode=RCIM_Cubic, leaveTangent=-0.3529, time=0.9, value=1}
         <at 1 0.7 RCIM_Cubic>
        ]}
      }}
     ScaleCurve=&Vector2DFromFloat(Value=&MakeCustomFloatFromBool(
       Bool=&SetBoolByFloatComparison(A=&MakeFloatFromInt(Int=$User.Rank_Unreal), B=0)
       True Float=150
       False Float=100
      ))
    )) {isSetParameters=true}
   /SetVariables(Particles.SmokeSize:Vector2f=&Vector2DFromCurve(
     Vector2Curve=DataInterface() {propertyValues {
       XCurve {keys [
         {arriveTangent=2.9521, interpMode=RCIM_Cubic, leaveTangent=2.9521, tangentMode=RCTM_User, time=0, value=0}
         {arriveTangentWeight=0.2261, interpMode=RCIM_Cubic, leaveTangentWeight=0.0072, tangentMode=RCTM_User, time=0.6784, value=1.1}
        ]}
       YCurve {keys [
         {arriveTangent=2.8322, interpMode=RCIM_Cubic, leaveTangent=2.8322, tangentMode=RCTM_User, time=0, value=0}
         {interpMode=RCIM_Cubic, tangentMode=RCTM_User, time=0.6784, value=1.1}
        ]}
      }}
     CurveIndex=$System.Age
     ScaleCurve=&Vector2DFromFloat(Value=&MakeCustomFloatFromBool(
       Bool=&SetBoolByFloatComparison(A=&MakeFloatFromInt(Int=$User.Rank_Unreal), B=0)
       True Float=195
       False Float=185
      ))
    )) {isSetParameters=true}
   /SetVariables(Particles.BlackHoleCol:LinearColor=&MultiplyLinearColor(
     Linear Color=$System.ColorWaveCol
     Scale Factor=&MakeVector4(
      X=1
      Y=1
      Z=1
      W=&FloatFromCurve(FloatCurve=DataInterface() {propertyValues {Curve {keys [
           {arriveTangent=-2.4882, interpMode=RCIM_Cubic, leaveTangent=-2.4882, tangentMode=RCTM_User, time=0.8, value=0.5}
           {arriveTangent=-0.0344, interpMode=RCIM_Cubic, leaveTangent=-0.0344, tangentMode=RCTM_User, time=1, value=0}
          ]}}})
     )
    )) {isSetParameters=true}
   /SetVariables() {isSetParameters=true}
   &FN_Light_Attributes(
    Color=&MultiplyLinearColor(
     Linear Color=$System.LightColor
     Scale Factor=&MakeVector4(
      X=1
      Y=1
      Z=1
      W=&FloatFromCurve(FloatCurve=DataInterface() {propertyValues {Curve {keys [
           {interpMode=RCIM_Cubic, tangentMode=RCTM_User, time=0, value=0}
           {interpMode=RCIM_Cubic, tangentMode=RCTM_User, time=0.151, value=1}
           {arriveTangent=-1.1779, interpMode=RCIM_Cubic, leaveTangent=-1.1779, time=0.6, value=1}
           {tangentMode=RCTM_User, time=1, value=0}
          ]}}})
     )
    )
    Create Color Attribute=true
    Create Exponent Attribute=true
    Create Position Attribute=true
    Create Radius Attribute=true
    Enable Particle Light Enable=true
    Exponent=1
    Light Enabled=true
    Radius=75
   )
  ]
  rendererClasses [&NiagaraSpriteRendererProperties, &NiagaraLightRendererProperties]
  renderers [
   {rendererClass=&NiagaraSpriteRendererProperties}
   {rendererClass=&NiagaraLightRendererProperties, rendererIndex=1}
   {rendererClass=&NiagaraSpriteRendererProperties, rendererIndex=2}
   {rendererClass=&NiagaraSpriteRendererProperties, rendererIndex=3}
  ]
 }
 RankIcons {
  emitterSpawnScript [/SetVariables(Emitter.SortOrder=10, Emitter.SortOrder2=9) {isSetParameters=true}]
  emitterUpdateScript [
   &EmitterState(Inactive Response=::NewEnumerator1, Life Cycle Mode=::NewEnumerator1, Loop Behavior=::NewEnumerator1, Loop Duration=5.5, UseLoopDelay=false)
   &SpawnBurst_Instantaneous(Use Spawn Probability=false, UseLoopCountLimit=false)
   /SetVariables(Emitter.Hardness=0.35) {isSetParameters=true}
  ]
  particleSpawnScript [
   &InitializeParticle(
    Color=<rgba 1 1 1 2>
    Lifetime=5.5
    Mesh Scale Mode=::NewEnumerator3
    Mesh Uniform Scale=-3.5
    Position Offset=$System.Position
    Sprite Rotation Mode=::NewEnumerator0
    Sprite Size Mode=::NewEnumerator3
    Uniform Sprite Size=45
    UsePositionOffset=true
    Write Material Random=true
   )
   &SpriteFacingAndAlignment(WriteAlignment=false, WriteFacing=true) {enabled=false}
   /SetVariables(Particles.ChampionPosition:NiagaraPosition=&AddVectorToPosition(
     Position=$Particles.Position
     Vector=&SelectVectorFromArray(Direct Array Index=$System.RankInt, Vector Selection Array=DataInterface() {propertyValues {FloatData [<0 0 0>, <0 0 0>, <-50 0 0>]}})
    )) {enabled=false, isSetParameters=true}
   /SetVariables(
    Particles.ChampionOffset:Vector2f=&SelectVector2DFromArray(Direct Array Index=$System.RankInt, Vector 2D Selection Array=DataInterface() {propertyValues {FloatData [<0 0>, <0.5 0.5>, <1.75 0>]}})
    Particles.ChampionSize:Vector2f=&SelectVector2DFromArray(Direct Array Index=$System.RankInt, Vector 2D Selection Array=DataInterface() {propertyValues {FloatData [<0 0>, <45 45>, <31 31>]}})
    Particles.EliteOffset:Vector2f=&SelectVector2DFromArray(Direct Array Index=$System.RankInt, Vector 2D Selection Array=DataInterface() {propertyValues {FloatData [<0.5 0.5>, <1.6 0.3>, <-0.7 0>]}})
    Particles.EliteSize:Vector2f=&SelectVector2DFromArray(Direct Array Index=$System.RankInt, Vector 2D Selection Array=DataInterface() {propertyValues {FloatData [<45 45>, <31 31>, <31 31>]}})
    Particles.UnrealOffset=12
    Particles.UnrealSize:Vector2f=&SelectVector2DFromArray(Direct Array Index=$System.RankInt, Vector 2D Selection Array=DataInterface() {propertyValues {FloatData [<0 0>, <0 0>, <70 70>]}})
   ) {isSetParameters=true}
   /SetVariables(
    Particles.ChampionBackOffset:Vector2f=&SelectVector2DFromArray(Direct Array Index=$System.RankInt, Vector 2D Selection Array=DataInterface() {propertyValues {FloatData [<0 0>, <0.5 0.5>, <1 0.28>]}})
    Particles.ChampionBackSize:Vector2f=&SelectVector2DFromArray(Direct Array Index=$System.RankInt, Vector 2D Selection Array=DataInterface() {propertyValues {FloatData [<0 0>, <100 100>, <80 80>]}})
    Particles.EliteBackOffset:Vector2f=&SelectVector2DFromArray(
     Direct Array Index=$System.RankInt
     Vector 2D Selection Array=DataInterface() {propertyValues {FloatData [<0.5 0.5>, <1.12 0.37>, <0.02 0.3>]}}
    )
    Particles.EliteBackSize:Vector2f=&SelectVector2DFromArray(Direct Array Index=$System.RankInt, Vector 2D Selection Array=DataInterface() {propertyValues {FloatData [<80 80>, <60 60>, <80 80>]}})
    Particles.IconBackingCol:LinearColor=<rgba 0 0 0 0.2>
    Particles.UnrealBackSize:Vector2f=&SelectVector2DFromArray(Direct Array Index=$System.RankInt, Vector 2D Selection Array=DataInterface() {propertyValues {FloatData [<0 0>, <0 0>, <100 100>]}})
   ) {isSetParameters=true}
   /SetVariables(
    Particles.ChampCountColor:LinearColor=<rgba 1 0.3156 0.2083 1>
    Particles.ChampionCountOffset:Vector2f=&SelectVector2DFromArray(Direct Array Index=$System.RankInt, Vector 2D Selection Array=DataInterface() {propertyValues {FloatData [<0 0>, <0.5 -1.5>, <2 -3>]}})
    Particles.ChampionCountSize:Vector2f=&SelectVector2DFromArray(Direct Array Index=$System.RankInt, Vector 2D Selection Array=DataInterface() {propertyValues {FloatData [<0 0>, <11 11>, <9 9>]}})
    Particles.EliteCountColor:LinearColor=<rgba 0.1268 0.618 0.7656 1>
    Particles.EliteCountOffset:Vector2f=&SelectVector2DFromArray(
     Direct Array Index=$System.RankInt
     Vector 2D Selection Array=DataInterface() {propertyValues {FloatData [<0.35 -1.6>, <1.85 -2.3>, <-1.12 -3.15>]}}
    )
    Particles.EliteCountSize:Vector2f=&SelectVector2DFromArray(Direct Array Index=$System.RankInt, Vector 2D Selection Array=DataInterface() {propertyValues {FloatData [<11 11>, <9 9>, <9 9>]}})
    Particles.UnrealCountColor:LinearColor=<rgba 0.2957 0.3534 0.8125 1>
    Particles.UnrealCountOffset:Vector2f=&SelectVector2DFromArray(Direct Array Index=$System.RankInt, Vector 2D Selection Array=DataInterface() {propertyValues {FloatData [<0 0>, <0 0>, <0.5 -2>]}})
    Particles.UnrealCountSize:Vector2f=&SelectVector2DFromArray(Direct Array Index=$System.RankInt, Vector 2D Selection Array=DataInterface() {propertyValues {FloatData [<0 0>, <0 0>, <13 13>]}})
   ) {isSetParameters=true}
   /SetVariables(
    Particles.ChampionCountBackOffset:Vector2f=&SelectVector2DFromArray(Direct Array Index=$System.RankInt, Vector 2D Selection Array=DataInterface() {propertyValues {FloatData [<0 0>, <0.5 0.07>, <0.95 0.04>]}})
    Particles.ChampionCountBackSize:Vector2f=&SelectVector2DFromArray(Direct Array Index=$System.RankInt, Vector 2D Selection Array=DataInterface() {propertyValues {FloatData [<0 0>, <55 55>, <80 80>]}})
    Particles.EliteCountBackOffset:Vector2f=&SelectVector2DFromArray(
     Direct Array Index=$System.RankInt
     Vector 2D Selection Array=DataInterface() {propertyValues {FloatData [<0.5 0.1>, <1.31 -0.08>, <0 0.05>]}}
    )
    Particles.EliteCountBackSize:Vector2f=&SelectVector2DFromArray(Direct Array Index=$System.RankInt, Vector 2D Selection Array=DataInterface() {propertyValues {FloatData [<60 60>, <45 45>, <80 80>]}})
    Particles.UnrealCountBackOffset:Vector2f=&SelectVector2DFromArray(Direct Array Index=$System.RankInt, Vector 2D Selection Array=DataInterface() {propertyValues {FloatData [<0 0>, <0 0>, <0.5 0.1>]}})
    Particles.UnrealCountBackSize:Vector2f=&SelectVector2DFromArray(Direct Array Index=$System.RankInt, Vector 2D Selection Array=DataInterface() {propertyValues {FloatData [<0 0>, <0 0>, <90 90>]}})
   ) {isSetParameters=true}
  ]
  particleUpdateScript [
   &ParticleState() {module=ParticleState001}
   &ScaleSpriteSize(Uniform Curve Sprite Scale=DataInterface() {propertyValues {Curve {keys [
        {arriveTangent=8.9267, interpMode=RCIM_Cubic, leaveTangent=8.9267, tangentMode=RCTM_User, time=0, value=0}
        {interpMode=RCIM_Cubic, tangentMode=RCTM_User, time=0.25, value=1}
        {arriveTangent=-1.3333, interpMode=RCIM_Cubic, leaveTangent=-1.3333, time=0.7668, value=1}
        <at 1 0 RCIM_Cubic>
       ]}}}) {module=ScaleSpriteSize001}
   /SetVariables(Particles.RollingDigits_Elite:NiagaraInt32=&MakeIntFromFloat(Float=&FloatFromCurve(
      FloatCurve=DataInterface() {propertyValues {Curve {keys [<at 0 1>,<at 2 1>]}}}
      CurveIndex=&Add_Float(A=&Multiply_Float(A=$Emitter.Age, B=50), B=-0.2)
      Scale Curve=&MakeFloatFromInt(Int=/Niagara/DynamicInputs/Math/Min_Int(A:NiagaraInt32=$User.Rank_Elite, B:NiagaraInt32=99999))
     ))) {isSetParameters=true}
   /SetVariables(
    Particles.Counter_Champion:NiagaraInt32=&Min_Int(A:NiagaraInt32=&Add_Integer(A=$User.Rank_Champion, B=$User.Rank_Champion_Reload), B:NiagaraInt32=99999)
    Particles.Counter_Elite:NiagaraInt32=&Min_Int(A:NiagaraInt32=&Add_Integer(A=$User.Rank_Elite, B=$User.Rank_Elite_Reload), B:NiagaraInt32=99999)
    Particles.Counter_Unreal:NiagaraInt32=&Min_Int(A:NiagaraInt32=&Add_Integer(A=$User.Rank_Unreal, B=$User.Rank_Unreal_Reload), B:NiagaraInt32=99999)
   ) {isSetParameters=true}
   &self:CalcDigits_16(count:NiagaraInt32=$Particles.Counter_Elite) {module=CalcDigits_16}
   &self:CalcDigits_16(Counter Selector=1, count:NiagaraInt32=$Particles.Counter_Champion) {module=CalcDigits001}
   &self:CalcDigits_16(Counter Selector=2, count:NiagaraInt32=$Particles.Counter_Unreal) {module=CalcDigits002}
   &ScaleColor(
    Scale Alpha=&FloatFromCurve(
     FloatCurve=DataInterface() {propertyValues {Curve {keys [
         {interpMode=RCIM_Cubic, tangentMode=RCTM_User, time=0, value=0}
         {interpMode=RCIM_Cubic, tangentMode=RCTM_User, time=0.5, value=1}
         {tangentMode=RCTM_User, time=1, value=0}
        ]}}}
     Scale Curve=&Multiply_Float(A=2, B=&FNTimeOfDayScalar(Night Scale=0.5, Morning Scale=0.8, Evening Scale=0.8))
    )
    Scale RGB=&MakeVectorFromLinearColorRGB(Color=&MultiplyLinearColor(
      Linear Color=&ColorFromCurve(ColorCurve=DataInterface() {propertyValues {
         AlphaCurve {keys []}
         BlueCurve {keys [<at 0.1259 2 RCIM_Cubic>, <at 0.2277 5>, {interpMode=RCIM_Cubic, leaveTangent=-0.7175, time=0.4478, value=2}]}
         GreenCurve {keys [<at 0.1259 0.1681 RCIM_Cubic>, <at 0.2277 0.2344>, {interpMode=RCIM_Cubic, leaveTangent=-4.4353, time=0.4478, value=0.1399}]}
         RedCurve {keys [<at 0.1259 0.5589 RCIM_Cubic>, <at 0.2277 1.1257>, {interpMode=RCIM_Cubic, leaveTangent=2.2176, time=0.4478, value=0.5865}]}
        }})
      Scale Factor=<1 1 1 1>
     ))
    ScaleA=true
    ScaleRGB=true
   )
   /SetVariables(
    Particles.ChampionSize:Vector2f=&Lerp_Vector2(
     A=<0 0>
     B=$Particles.Initial.ChampionSize
     Alpha=&MakeCustomFloatFromBool(
      Bool=&SetBoolByFloatComparison(A=&MakeFloatFromInt(Int=$User.Rank_Unreal), B=0)
      True Float=&FloatFromCurve(FloatCurve=DataInterface() {propertyValues {Curve {keys [
           {arriveTangent=11.3256, interpMode=RCIM_Cubic, leaveTangent=11.3256, tangentMode=RCTM_User, time=0.065, value=0}
           {arriveTangent=0.568, interpMode=RCIM_Cubic, leaveTangent=0.568, tangentMode=RCTM_User, time=0.215, value=1}
           {arriveTangent=-1.2408, interpMode=RCIM_Cubic, leaveTangent=-1.2408, tangentMode=RCTM_User, time=0.85, value=1}
           {arriveTangent=-9.0183, interpMode=RCIM_Cubic, leaveTangent=-9.0183, tangentMode=RCTM_User, time=0.96, value=0}
          ]}}})
      False Float=&FloatFromCurve(FloatCurve=DataInterface() {propertyValues {Curve {keys [
           {arriveTangent=16.8732, interpMode=RCIM_Cubic, leaveTangent=16.8732, tangentMode=RCTM_User, time=0.0171, value=0}
           {arriveTangent=0.568, interpMode=RCIM_Cubic, leaveTangent=0.568, tangentMode=RCTM_User, time=0.14, value=1}
           {arriveTangent=-1.2408, interpMode=RCIM_Cubic, leaveTangent=-1.2408, tangentMode=RCTM_User, time=0.81, value=1}
           {arriveTangent=-4.0827, interpMode=RCIM_Cubic, leaveTangent=-4.0827, tangentMode=RCTM_User, time=1, value=0}
          ]}}})
     )
    )
    Particles.EliteSize:Vector2f=&Lerp_Vector2(
     A=<0 0>
     B=$Particles.Initial.EliteSize
     Alpha=&MakeCustomFloatFromBool(
      Bool=&SetBoolByFloatComparison(A=&MakeFloatFromInt(Int=$User.Rank_Champion), B=0)
      True Float=&FloatFromCurve(FloatCurve=DataInterface() {propertyValues {Curve {keys [
           {arriveTangent=10.9229, interpMode=RCIM_Cubic, leaveTangent=10.9229, tangentMode=RCTM_User, time=0.082, value=0}
           {arriveTangent=0.2188, interpMode=RCIM_Cubic, leaveTangent=0.2188, tangentMode=RCTM_User, time=0.24, value=1}
           {arriveTangent=-1.4286, interpMode=RCIM_Cubic, leaveTangent=-1.4286, time=0.81, value=1}
           {arriveTangent=-6.9943, interpMode=RCIM_Cubic, leaveTangent=-6.9943, tangentMode=RCTM_User, time=0.94, value=0}
          ]}}})
      False Float=&FloatFromCurve(FloatCurve=DataInterface() {propertyValues {Curve {keys [
           {arriveTangent=16.3012, interpMode=RCIM_Cubic, leaveTangent=16.3012, tangentMode=RCTM_User, time=0.0216, value=0}
           {arriveTangent=0.2188, interpMode=RCIM_Cubic, leaveTangent=0.2188, tangentMode=RCTM_User, time=0.15, value=1}
           {arriveTangent=-1.1765, interpMode=RCIM_Cubic, leaveTangent=-1.1765, time=0.81, value=1}
           {arriveTangent=-5.7216, interpMode=RCIM_Cubic, leaveTangent=-5.7216, tangentMode=RCTM_User, time=1, value=0}
          ]}}})
     )
    )
    Particles.UnrealSize:Vector2f=&Lerp_Vector2(
     A=<0 0>
     B=$Particles.Initial.UnrealSize
     Alpha=&FloatFromCurve(FloatCurve=DataInterface() {propertyValues {Curve {keys [
          {arriveTangent=17.8461, interpMode=RCIM_Cubic, leaveTangent=17.8461, tangentMode=RCTM_User, time=0, value=0}
          {arriveTangent=0.7732, interpMode=RCIM_Cubic, leaveTangent=0.7732, tangentMode=RCTM_User, time=0.09, value=1}
          {arriveTangent=-0.8496, interpMode=RCIM_Cubic, leaveTangent=-0.8496, tangentMode=RCTM_User, time=0.9, value=1}
          {arriveTangent=-4.3419, interpMode=RCIM_Cubic, leaveTangent=-4.3419, tangentMode=RCTM_User, time=1, value=0}
         ]}}})
    )
   ) {isSetParameters=true}
   &CameraOffset(Camera Offset Amount=11)
   /SetVariables(Particles.ChampionCamOffset=11, Particles.EliteCamOffset=10.5, Particles.UnrealCamOffset=12) {isSetParameters=true}
   /SetVariables(
    Particles.ChampionBackCamOffset=10.75
    Particles.ChampionBackSize:Vector2f=&Lerp_Vector2(
     A=<0 0>
     B=$Particles.Initial.ChampionBackSize
     Alpha=&MakeCustomFloatFromBool(
      Bool=&SetBoolByFloatComparison(A=&MakeFloatFromInt(Int=$User.Rank_Unreal), B=0)
      True Float=&FloatFromCurve(FloatCurve=DataInterface() {propertyValues {Curve {keys [
           {arriveTangent=11.3256, interpMode=RCIM_Cubic, leaveTangent=11.3256, tangentMode=RCTM_User, time=0.065, value=0}
           {arriveTangent=0.568, interpMode=RCIM_Cubic, leaveTangent=0.568, tangentMode=RCTM_User, time=0.215, value=1}
           {arriveTangent=-1.2408, interpMode=RCIM_Cubic, leaveTangent=-1.2408, tangentMode=RCTM_User, time=0.85, value=1}
           {arriveTangent=-9.0183, interpMode=RCIM_Cubic, leaveTangent=-9.0183, tangentMode=RCTM_User, time=0.96, value=0}
          ]}}})
      False Float=&FloatFromCurve(FloatCurve=DataInterface() {propertyValues {Curve {keys [
           {arriveTangent=18.7089, interpMode=RCIM_Cubic, leaveTangent=18.7089, tangentMode=RCTM_User, time=0.0171, value=0}
           {arriveTangent=-0.4445, interpMode=RCIM_Cubic, leaveTangent=-0.4445, tangentMode=RCTM_User, time=0.125, value=1}
           {arriveTangent=-1.1429, interpMode=RCIM_Cubic, leaveTangent=-1.1429, time=0.81, value=1}
           <at 1 0 RCIM_Cubic>
          ]}}})
     )
    )
    Particles.EliteBackSize:Vector2f=&Lerp_Vector2(
     A=<0 0>
     B=$Particles.Initial.EliteBackSize
     Alpha=&FloatFromCurve(FloatCurve=DataInterface() {propertyValues {Curve {keys [
          {arriveTangent=10.9229, interpMode=RCIM_Cubic, leaveTangent=10.9229, tangentMode=RCTM_User, time=0.082, value=0}
          {arriveTangent=0.2188, interpMode=RCIM_Cubic, leaveTangent=0.2188, tangentMode=RCTM_User, time=0.24, value=1}
          {arriveTangent=-1.4286, interpMode=RCIM_Cubic, leaveTangent=-1.4286, time=0.81, value=1}
          {arriveTangent=-6.9943, interpMode=RCIM_Cubic, leaveTangent=-6.9943, tangentMode=RCTM_User, time=0.94, value=0}
         ]}}})
    )
    Particles.UnrealBackCamOffset=11.5
    Particles.UnrealBackSize:Vector2f=&Lerp_Vector2(
     A=<0 0>
     B=$Particles.Initial.UnrealBackSize
     Alpha=&FloatFromCurve(FloatCurve=DataInterface() {propertyValues {Curve {keys [
          {arriveTangent=13.9076, interpMode=RCIM_Cubic, leaveTangent=13.9076, tangentMode=RCTM_User, time=0, value=0}
          {arriveTangent=0.2188, interpMode=RCIM_Cubic, leaveTangent=0.2188, tangentMode=RCTM_User, time=0.11, value=1}
          {arriveTangent=-0.7878, interpMode=RCIM_Cubic, leaveTangent=-0.7878, tangentMode=RCTM_User, time=0.88, value=1}
          <at 1 0 RCIM_Cubic>
         ]}}})
    )
   ) {isSetParameters=true}
   /SetVariables(
    Particles.ChampCountCamOffset=12.5
    Particles.ChampCountSize:Vector2f=&Multiply_Vector2D_ByFloat(
     Float=&MakeCustomFloatFromBool(
      Bool=&SetBoolByFloatComparison(A=&MakeFloatFromInt(Int=$User.Rank_Unreal), B=0)
      True Float=&FloatFromCurve(FloatCurve=DataInterface() {propertyValues {Curve {keys [
           {arriveTangent=11.3256, interpMode=RCIM_Cubic, leaveTangent=11.3256, tangentMode=RCTM_User, time=0.065, value=0}
           {arriveTangent=0.568, interpMode=RCIM_Cubic, leaveTangent=0.568, tangentMode=RCTM_User, time=0.215, value=1}
           {arriveTangent=-1.2408, interpMode=RCIM_Cubic, leaveTangent=-1.2408, tangentMode=RCTM_User, time=0.85, value=1}
           {arriveTangent=-9.0183, interpMode=RCIM_Cubic, leaveTangent=-9.0183, tangentMode=RCTM_User, time=0.96, value=0}
          ]}}})
      False Float=&FloatFromCurve(FloatCurve=DataInterface() {propertyValues {Curve {keys [
           {arriveTangent=16.8732, interpMode=RCIM_Cubic, leaveTangent=16.8732, tangentMode=RCTM_User, time=0.0171, value=0}
           {arriveTangent=0.568, interpMode=RCIM_Cubic, leaveTangent=0.568, tangentMode=RCTM_User, time=0.14, value=1}
           {arriveTangent=-1.2408, interpMode=RCIM_Cubic, leaveTangent=-1.2408, tangentMode=RCTM_User, time=0.81, value=1}
           {arriveTangent=-4.0827, interpMode=RCIM_Cubic, leaveTangent=-4.0827, tangentMode=RCTM_User, time=1, value=0}
          ]}}})
     )
     Vector2D=&MakeVector2D(
      X=&Multiply_Float(
       A=&MakeFloatFromVector(Vector=$Particles.DigitScale_Champion, Channel=::NewEnumerator1)
       B=&MakeFloatFromVector2d(Vector 2D=$Particles.Initial.ChampionCountSize, Channel=::NewEnumerator1)
      )
      Y=&MakeFloatFromVector2d(Vector 2D=$Particles.Initial.ChampionCountSize, Channel=::NewEnumerator1)
     )
    )
    Particles.EliteCountCamOffset=10
    Particles.EliteCountSize:Vector2f=&Multiply_Vector2D_ByFloat(
     Float=&MakeCustomFloatFromBool(
      Bool=&SetBoolByFloatComparison(A=&MakeFloatFromInt(Int=$User.Rank_Champion), B=0)
      True Float=&FloatFromCurve(FloatCurve=DataInterface() {propertyValues {Curve {keys [
           {arriveTangent=10.9229, interpMode=RCIM_Cubic, leaveTangent=10.9229, tangentMode=RCTM_User, time=0.082, value=0}
           {arriveTangent=0.2188, interpMode=RCIM_Cubic, leaveTangent=0.2188, tangentMode=RCTM_User, time=0.24, value=1}
           {arriveTangent=-1.4286, interpMode=RCIM_Cubic, leaveTangent=-1.4286, time=0.81, value=1}
           {arriveTangent=-6.9943, interpMode=RCIM_Cubic, leaveTangent=-6.9943, tangentMode=RCTM_User, time=0.94, value=0}
          ]}}})
      False Float=&FloatFromCurve(FloatCurve=DataInterface() {propertyValues {Curve {keys [
           {arriveTangent=16.3012, interpMode=RCIM_Cubic, leaveTangent=16.3012, tangentMode=RCTM_User, time=0.0216, value=0}
           {arriveTangent=0.2188, interpMode=RCIM_Cubic, leaveTangent=0.2188, tangentMode=RCTM_User, time=0.15, value=1}
           {arriveTangent=-1.1765, interpMode=RCIM_Cubic, leaveTangent=-1.1765, time=0.81, value=1}
           {arriveTangent=-5.7216, interpMode=RCIM_Cubic, leaveTangent=-5.7216, tangentMode=RCTM_User, time=1, value=0}
          ]}}})
     )
     Vector2D=&MakeVector2D(
      X=&Multiply_Float(
       A=&MakeFloatFromVector(Vector=$Particles.DigitScale_Elite, Channel=::NewEnumerator1)
       B=&MakeFloatFromVector2d(Vector 2D=$Particles.Initial.EliteCountSize, Channel=::NewEnumerator1)
      )
      Y=&MakeFloatFromVector2d(Vector 2D=$Particles.Initial.EliteCountSize, Channel=::NewEnumerator1)
     )
    )
    Particles.UnrealCountCamOffset=13
    Particles.UnrealCountSize:Vector2f=&Multiply_Vector2D_ByFloat(
     Float=&FloatFromCurve(FloatCurve=DataInterface() {propertyValues {Curve {keys [
          {arriveTangent=17.8461, interpMode=RCIM_Cubic, leaveTangent=17.8461, tangentMode=RCTM_User, time=0, value=0}
          {arriveTangent=0.7732, interpMode=RCIM_Cubic, leaveTangent=0.7732, tangentMode=RCTM_User, time=0.09, value=1}
          {arriveTangent=-0.8496, interpMode=RCIM_Cubic, leaveTangent=-0.8496, tangentMode=RCTM_User, time=0.9, value=1}
          {arriveTangent=-4.3419, interpMode=RCIM_Cubic, leaveTangent=-4.3419, tangentMode=RCTM_User, time=1, value=0}
         ]}}})
     Vector2D=&MakeVector2D(
      X=&Multiply_Float(
       A=&MakeFloatFromVector(Vector=$Particles.DigitScale_Unreal, Channel=::NewEnumerator1)
       B=&MakeFloatFromVector2d(Vector 2D=$Particles.Initial.UnrealCountSize, Channel=::NewEnumerator1)
      )
      Y=&MakeFloatFromVector2d(Vector 2D=$Particles.Initial.UnrealCountSize, Channel=::NewEnumerator1)
     )
    )
    Particles.UnrealCountSize001=13
   ) {isSetParameters=true}
   /SetVariables(
    Particles.CounterDynParams_Champion:Vector4f=&MakeVector4(X=$Output.CalcDigits001.digits, Y=&MakeFloatFromInt(Int=$Particles.Counter_Champion), Z=0, W=0)
    Particles.CounterDynParams_Elite:Vector4f=&MakeVector4(X=$Output.CalcDigits_16.digits, Y=&MakeFloatFromInt(Int=$Particles.Counter_Elite), Z=0, W=0)
    Particles.CounterDynParams_Unreal:Vector4f=&MakeVector4(X=$Output.CalcDigits002.digits, Y=&MakeFloatFromInt(Int=$Particles.Counter_Unreal), Z=0, W=0)
   ) {isSetParameters=true}
   /SetVariables(Particles.EliteOffsetSteady:Vector2f=&MakeVector2D(
     X=&Multiply_Float(
      A=&MakeFloatFromVector2d(Vector 2D=$Particles.EliteCountOffset)
      B=&MakeCustomFloatFromBool(
       Bool=&SetBoolByFloatComparison(A=&MakeFloatFromInt(Int=$System.RankInt), B=1, Comparison Type=::NewEnumerator2)
       True Float=&MakeCustomFloatFromBool(
        Bool=&SetBoolByFloatComparison(A=&MakeFloatFromInt(Int=$System.RankInt), B=2, Comparison Type=::NewEnumerator2)
        True Float=&SelectFloatFromArray(Direct Array Index=$Particles.DigitInt_Elite, Float Selection Array=DataInterface() {propertyValues {FloatData [1,3.2,1.4,0.82]}})
        False Float=&SelectFloatFromArray(Direct Array Index=$Particles.DigitInt_Elite, Float Selection Array=DataInterface() {propertyValues {FloatData [1,2.4,1.4,0.98]}})
       )
       False Float=1.5
      )
     )
     Y=&MakeFloatFromVector2d(Vector 2D=$Particles.EliteCountOffset, Channel=::NewEnumerator1)
    )) {isSetParameters=true}
   /SetVariables(Particles.ChampionOffsetSteady:Vector2f=&MakeVector2D(
     X=&Multiply_Float(
      A=&MakeFloatFromVector2d(Vector 2D=$Particles.ChampionCountOffset)
      B=&MakeCustomFloatFromBool(
       Bool=&SetBoolByFloatComparison(A=&MakeFloatFromInt(Int=$System.RankInt), B=2, Comparison Type=::NewEnumerator2)
       True Float=&SelectFloatFromArray(Direct Array Index=$Particles.DigitInt_Champion, Float Selection Array=DataInterface() {propertyValues {FloatData [1,2.4,1.33,0.98]}})
       False Float=1
      )
     )
     Y=&MakeFloatFromVector2d(Vector 2D=$Particles.ChampionCountOffset, Channel=::NewEnumerator1)
    )) {isSetParameters=true}
   /SetVariables(
    Particles.ChampionCountBackSize:Vector2f=&Lerp_Vector2(
     A=<0 0>
     B=$Particles.Initial.ChampionCountBackSize
     Alpha=&FloatFromCurve(FloatCurve=DataInterface() {propertyValues {Curve {keys [
          {arriveTangent=11.3256, interpMode=RCIM_Cubic, leaveTangent=11.3256, tangentMode=RCTM_User, time=0.065, value=0}
          {arriveTangent=0.568, interpMode=RCIM_Cubic, leaveTangent=0.568, tangentMode=RCTM_User, time=0.215, value=1}
          {arriveTangent=-1.2408, interpMode=RCIM_Cubic, leaveTangent=-1.2408, tangentMode=RCTM_User, time=0.85, value=1}
          {arriveTangent=-9.0183, interpMode=RCIM_Cubic, leaveTangent=-9.0183, tangentMode=RCTM_User, time=0.96, value=0}
         ]}}})
    )
    Particles.EliteCountBackSize:Vector2f=&Lerp_Vector2(
     A=<0 0>
     B=$Particles.Initial.EliteCountBackSize
     Alpha=&FloatFromCurve(FloatCurve=DataInterface() {propertyValues {Curve {keys [
          {arriveTangent=10.9229, interpMode=RCIM_Cubic, leaveTangent=10.9229, tangentMode=RCTM_User, time=0.082, value=0}
          {arriveTangent=0.2188, interpMode=RCIM_Cubic, leaveTangent=0.2188, tangentMode=RCTM_User, time=0.24, value=1}
          {arriveTangent=-1.4286, interpMode=RCIM_Cubic, leaveTangent=-1.4286, time=0.81, value=1}
          {arriveTangent=-6.9943, interpMode=RCIM_Cubic, leaveTangent=-6.9943, tangentMode=RCTM_User, time=0.94, value=0}
         ]}}})
    )
    Particles.UnrealCountBackSize:Vector2f=&Lerp_Vector2(
     A=<0 0>
     B=$Particles.Initial.UnrealCountBackSize
     Alpha=&FloatFromCurve(FloatCurve=DataInterface() {propertyValues {Curve {keys [
          {arriveTangent=7.4365, interpMode=RCIM_Cubic, leaveTangent=7.4365, tangentMode=RCTM_User, time=0, value=0}
          {arriveTangent=0.2188, interpMode=RCIM_Cubic, leaveTangent=0.2188, tangentMode=RCTM_User, time=0.13, value=1}
          {arriveTangent=-0.7878, interpMode=RCIM_Cubic, leaveTangent=-0.7878, tangentMode=RCTM_User, time=0.8, value=1}
          <at 1 0 RCIM_Cubic>
         ]}}})
    )
   ) {isSetParameters=true}
   /SetVariables(
    Particles.ChampionCountBackShadow:LinearColor=&SelectLinearColorFromArray(
     Array Sampling Mode=::NewEnumerator1
     Direct Array Index=$System.RankInt
     Color Selection Array=DataInterface() {propertyValues {ColorData [<rgba 0 0 0 0.2>, <rgba 0 0 0 0.3>, <rgba 0 0 0 0.1>]}}
     Random Seed=?
    )
    Particles.UnrealCountBackColor:LinearColor=&LerpLinearColors(
     StartColor=<rgba 0 0 0 0.65>
     EndColor=<rgba 0 0 0 0.2>
     LerpFactor=&FloatFromCurve(FloatCurve=DataInterface() {propertyValues {Curve {keys [<at 0 1>,<at 2 0>,<at 3 1>]}}}, CurveIndex=$Emitter.Age)
    )
   ) {isSetParameters=true}
   &ScaleColor(
    Scale RGB=&Lerp_Vector(A=<1 1 1>, B=<0.1 0.1 0.1>, Alpha=$NPC.FortniteNPC.IsPlayerIndoors)
    ScaleA=true
    ScaleRGB=true
   ) {enabled=false, module=ScaleColor001}
  ]
  rendererClasses [&NiagaraSpriteRendererProperties]
  renderers [
   {rendererClass=&NiagaraSpriteRendererProperties}
   {rendererClass=&NiagaraSpriteRendererProperties, rendererIndex=1}
   {rendererClass=&NiagaraSpriteRendererProperties, rendererIndex=2}
   {rendererClass=&NiagaraSpriteRendererProperties, rendererIndex=3}
   {rendererClass=&NiagaraSpriteRendererProperties, rendererIndex=4}
   {rendererClass=&NiagaraSpriteRendererProperties, rendererIndex=5}
   {rendererClass=&NiagaraSpriteRendererProperties, rendererIndex=6}
   {rendererClass=&NiagaraSpriteRendererProperties, rendererIndex=7}
   {rendererClass=&NiagaraSpriteRendererProperties, rendererIndex=8}
   {rendererClass=&NiagaraSpriteRendererProperties, rendererIndex=9}
   {rendererClass=&NiagaraSpriteRendererProperties, rendererIndex=10}
   {rendererClass=&NiagaraSpriteRendererProperties, rendererIndex=11}
  ]
 }
 SlowWaves {
  emitterSpawnScript [/SetVariables(Emitter.SortOrder=4) {isSetParameters=true}]
  emitterUpdateScript [
   &EmitterState()
   &SpawnRate(
    SpawnRate=&FloatFromCurve(FloatCurve=DataInterface() {propertyValues {Curve {keys [<at 0 3>,<at 3.5 0>]}}}, CurveIndex=$System.Age)
    Use Spawn Probability=false
   )
  ]
  particleSpawnScript [
   &InitializeParticle(
    Color=$System.SlowWavesCol
    Lifetime=&RandomRangeFloat(Minimum=1.3, Maximum=2)
    Mesh Scale=&Multiply_VectorByFloat(Vector=<1 1 1>, Float=1)
    Mesh Scale Mode=::NewEnumerator4
    UsePositionOffset=false
    Write Material Random=true
   ) {module=InitializeParticle001}
   &InitialMeshOrientation(
    Mesh Orientation Mode=::NewEnumerator2
    Orientation Vector=<0 0 1>
    Rotation=&RandomRangeVector(Minimum=<0 0 0>, Maximum=<0 0 1>)
    Use Rotation Vector=true
   )
   &CameraOffset(Camera Offset Amount=&FN_FrontEndFloatModifier(Gameplay=10, Front End=0))
  ]
  particleUpdateScript [
   &DynamicMaterialParameters(
    Index 0 Param 2=0.3
    Index 0 Param 3=3
    Index 0 Param 4=&FloatFromCurve(FloatCurve=DataInterface() {propertyValues {Curve {keys [<at 0 1>,<at 0.7 1>,<at 1 0>]}}}, Scale Curve=5)
    Module.Param0WriteEnabled=true
    Module.Param1WriteEnabled=true
    Module.Param2WriteEnabled=true
    Module.Param3WriteEnabled=true
   )
   &ParticleState()
   &ScaleMeshSize(Scale Factor=&VectorFromFloat(Value=&FloatFromCurve(
      FloatCurve=DataInterface() {propertyValues {Curve {keys [{tangentMode=RCTM_User, time=0, value=0.2}, {tangentMode=RCTM_User, time=1, value=1}]}}}
      Scale Curve=6
     ))) {module=ScaleMeshSize001}
   &ScaleColor(
    Scale Alpha=&FloatFromCurve(
     FloatCurve=DataInterface() {propertyValues {Curve {keys [
         {interpMode=RCIM_Cubic, tangentMode=RCTM_User, time=0.4821, value=0}
         {interpMode=RCIM_Cubic, tangentMode=RCTM_User, time=0.6995, value=1}
         {tangentMode=RCTM_User, time=1, value=0}
        ]}}}
     Scale Curve=&FN_FrontEndFloatModifier(Gameplay=1, Front End=0)
    )
    Scale RGB=&MakeVectorFromLinearColorRGB(Color=&FNTimeOfDayColor(Night Color=<rgba 0.5 0.104 0.114 1>, Morning Color=<rgba 1 1 1 1>, Day Color=<rgba 1 1 1 1>, Evening Color=<rgba 1 1 1 1>))
    ScaleA=true
    ScaleRGB=true
   )
   &ScaleMeshSize(Scale Factor=&Multiply_VectorByFloat(
     Vector=&Multiply_VectorByFloat(
      Vector=<1 1 1>
      Float=&FloatFromCurve(FloatCurve=DataInterface() {propertyValues {Curve {keys [<at 0 0>,<at 1 1.2>]}}})
     )
     Float=6
    )) {enabled=false}
   &SolveForcesAndVelocity(Clamp Velocity=false, Limit Acceleration=false)
  ]
  rendererClasses [&NiagaraMeshRendererProperties]
  renderers [{rendererClass=&NiagaraMeshRendererProperties, rendererIndex=0}]
 }
 Spark {
  emitterUpdateScript [
   &EmitterState(Life Cycle Mode=::NewEnumerator1, Loop Behavior=::NewEnumerator1, Loop Delay=0.05, Loop Duration=1, UseLoopDelay=true)
   &SpawnBurst_Instantaneous(Spawn Count=60, Use Spawn Probability=false)
  ]
  particleSpawnScript [
   &InitializeParticle(
    Color=$System.SparkCol
    Lifetime Max=6
    Lifetime Min=6
    Lifetime Mode=::NewEnumerator1
    Sprite Size Max=<3 8>
    Sprite Size Min=<2.5 3.5>
    Sprite Size Mode=::NewEnumerator2
    UsePositionOffset=false
    Write Material Random=true
   )
   &AddVelocityFromPoint(
    Module.Use Falloff Distance=false
    Velocity Strength=&Multiply_Float(
     A=&RandomRangeFloat(Minimum=600, Maximum=1200)
     B=&FN_FrontEndFloatModifier(Gameplay=1, Front End=0.5)
    )
   )
   &ShapeLocation(MaskSpawnGroup=false, Override Seed=false, Sphere Radius=1, UseOwnerScale=true)
  ]
  particleUpdateScript [
   &ParticleState()
   &ScaleColor(
    Linear Color Curve=DataInterface() {propertyValues {
      AlphaCurve {keys [<at 0.0074 1>,<at 0.8634 1>,<at 0.8905 1>]}
      BlueCurve {keys [<at 0 15>]}
      GreenCurve {keys [<at 0 0>]}
      RedCurve {keys [<at 0 5.5313>]}
     }}
    Scale Mode=::NewEnumerator2
   ) {enabled=false}
   &ScaleColor(
    Scale Alpha=&FNTimeOfDayScalar(Night Scale=1, Morning Scale=1, Evening Scale=1)
    Scale RGB=&VectorFromFloat(Value=10)
    ScaleA=true
    ScaleRGB=true
   ) {module=ScaleColor001}
   &Drag(Drag=&RandomRangeFloat(Minimum=8, Maximum=15), Use Linear Drag=true, Use Rotational Drag=false)
   &CurlNoiseForce(
    Noise Strength=&Multiply_Float(
     A=&RandomRangeFloat(Minimum=200, Maximum=400)
     B=&FloatFromCurve(FloatCurve=DataInterface() {propertyValues {Curve {keys [<at 0 1>,<at 1 0.2>]}}}, CurveIndex=$Emitter.Age)
    )
    Pan Noise=true
   )
   &GravityForce(Gravity=&RandomRangeVector(Minimum=<0 0 0>, Maximum=<0 0 -100>))
   &PointAttractionForce(
    Attraction Radius=200
    Attraction Strength=&FloatFromCurve(
     FloatCurve=DataInterface() {propertyValues {Curve {keys [<at 4 0 RCIM_Cubic>, {arriveTangent=1.8076, interpMode=RCIM_Cubic, leaveTangent=1.8076, tangentMode=RCTM_User, time=5.3, value=1}]}}}
     CurveIndex=$Emitter.Age
     Scale Curve=360
    )
    Kill Radius:NiagaraFloat=&FloatFromCurve(FloatCurve=DataInterface() {propertyValues {Curve {keys [<at 2.5 0>,<at 3.2 5>]}}}, CurveIndex=$Emitter.Age)
    Kill Within Radius=true
    Module.Use Falloff=true
    UseAttractorOffset=false
   )
   &SolveForcesAndVelocity(Clamp Velocity=false, Limit Acceleration=false)
   &ScaleSpriteSizeBySpeed(
    Max Scale Factor=<1 20>
    Min Scale Factor=<0.5 0.1>
    Velocity Threshold:NiagaraFloat=&FloatFromCurve(FloatCurve=DataInterface() {propertyValues {Curve {keys [<at 0 6>,<at 0.05 0.9>,<at 1 0.05>]}}}, Scale Curve=1000)
   )
   &ScaleSpriteSize(Scale Factor=&MakeVector2D(
     X=&FloatFromCurve(
      FloatCurve=DataInterface() {propertyValues {Curve {keys [<at 0 0.5>,<at 0.1 1>,<at 1 0.5>]}}}
      Scale Curve=&FN_FrontEndFloatModifier(Gameplay=1, Front End=0.4)
     )
     Y=&FloatFromCurve(
      FloatCurve=DataInterface() {propertyValues {Curve {keys [<at 0 1>,<at 1 0>]}}}
      Scale Curve=&FN_FrontEndFloatModifier(Gameplay=1, Front End=0.4)
     )
    ))
  ]
  rendererClasses [&NiagaraSpriteRendererProperties]
  renderers [{rendererClass=&NiagaraSpriteRendererProperties}]
 }
 Sunlight {
  emitterSpawnScript [/SetVariables(Emitter.SortOrder=1, Emitter.SortOrder2=2) {isSetParameters=true}]
  emitterUpdateScript [
   &EmitterState(
    Inactive Response=::NewEnumerator1
    Life Cycle Mode=::NewEnumerator1
    Loop Behavior=::NewEnumerator1
    Loop Delay=0.1
    Loop Duration=8.5
    UseLoopDelay=true
   )
   &SpawnBurst_Instantaneous(Use Spawn Probability=false, UseLoopCountLimit=false)
   &SpawnBurst_Instantaneous(
    Spawn Group=1
    Spawn Time=&RandomRangeFloat(Minimum=0.05, Maximum=0.075)
    Use Spawn Probability=false
    UseLoopCountLimit=false
   ) {module=SpawnBurst_Instantaneous002}
  ]
  particleSpawnScript [
   /SetVariables(Particles.VisibilityTag=&SpawnGroup()) {isSetParameters=true}
   &InitializeParticle(
    Color=$System.SunshineColor
    Lifetime=5.35
    Mass Mode=::NewEnumerator1
    Mesh Scale Mode=::NewEnumerator3
    Mesh Uniform Scale=&FloatFromCurve(
     FloatCurve=DataInterface() {propertyValues {Curve {keys [{tangentMode=RCTM_User, time=0, value=1}, {tangentMode=RCTM_User, time=1, value=2}]}}}
     CurveIndex=&ReturnNormalizedExecIndex()
     Scale Curve=5
    )
    Sprite Size=<170 170>
    Sprite Size Mode=::NewEnumerator4
    UsePositionOffset=false
    Write Material Random=true
   ) {module=InitializeParticle001}
   &CameraOffset(Camera Offset Amount=10)
   &SpriteFacingAndAlignment(Sprite Alignment=<0 0 1>, WriteAlignment=true, WriteFacing=false)
   &InitialMeshOrientation(
    Mesh Orientation Mode=::NewEnumerator2
    Rotation=&SelectVectorFromArray(
     Direct Array Index=&SpawnGroup()
     Vector Selection Array=DataInterface() {propertyValues {FloatData [<-0.25 0 0>, <0.5 0 0>]}}
    )
    Rotation Coordinate Space=::NewEnumerator2
    Use Rotation Vector=true
   )
  ]
  particleUpdateScript [
   &ParticleState()
   &SolveForcesAndVelocity(Clamp Velocity=false, Limit Acceleration=false)
   &ScaleColor(
    Scale Alpha=&FloatFromCurve(
     FloatCurve=DataInterface() {propertyValues {Curve {keys [
         {arriveTangent=4.0997, interpMode=RCIM_Cubic, leaveTangent=4.0997, tangentMode=RCTM_User, time=0, value=0}
         {interpMode=RCIM_Cubic, tangentMode=RCTM_User, time=0.3, value=1}
         {arriveTangent=-1.6432, interpMode=RCIM_Cubic, leaveTangent=-1.6432, time=0.6, value=1}
         {arriveTangent=-3.5577, tangentMode=RCTM_User, time=0.9086, value=0}
        ]}}}
     Scale Curve=0.3
    )
    Scale RGB=&MakeVectorFromLinearColorRGB(Color=&FNTimeOfDayColor(Night Color=<rgba 0.225 0.905 1.5 1>, Morning Color=<rgba 1 1 1 1>, Day Color=<rgba 1 1 1 1>, Evening Color=<rgba 1 1 1 1>))
    ScaleA=true
    ScaleRGB=true
   )
   &ScaleSpriteSize(Scale Factor=&Multiply_Vector2D_ByFloat(
     Float=&FloatFromCurve(
      FloatCurve=DataInterface() {propertyValues {Curve {keys [
          {arriveTangent=26.1988, interpMode=RCIM_Cubic, leaveTangent=26.1988, tangentMode=RCTM_User, time=0, value=0}
          {arriveTangent=0.4194, interpMode=RCIM_Cubic, leaveTangent=0.4194, tangentMode=RCTM_User, time=0.0664, value=1}
          {interpMode=RCIM_Cubic, tangentMode=RCTM_User, time=0.5, value=1}
          {arriveTangent=-1.5738, interpMode=RCIM_Cubic, leaveTangent=-1.5738, tangentMode=RCTM_User, time=1, value=0.8}
         ]}}}
      Scale Curve=1.25
     )
     Vector2D=<1 1>
    ))
   /SetVariables(Particles.Corona_Size:Vector2f=&Multiply_Vector2D_ByFloat(Float=0.5, Vector2D=$Particles.SpriteSize)) {isSetParameters=true}
   &ScaleSpriteSize(Uniform Curve Sprite Scale=DataInterface() {propertyValues {Curve {keys [
        {arriveTangent=8.4722, interpMode=RCIM_Cubic, leaveTangent=8.4722, tangentMode=RCTM_User, time=0, value=0}
        {interpMode=RCIM_Cubic, tangentMode=RCTM_User, time=0.3, value=1}
        {arriveTangent=0.1764, interpMode=RCIM_Cubic, leaveTangent=0.1764, time=0.7668, value=1}
        {arriveTangent=-4.2873, interpMode=RCIM_Cubic, leaveTangent=-4.2873, time=0.8668, value=1.1}
        <at 1 0 RCIM_Cubic>
       ]}}}) {enabled=false, module=ScaleSpriteSize001}
  ]
  rendererClasses [&NiagaraSpriteRendererProperties]
  renderers [
   {rendererClass=&NiagaraSpriteRendererProperties}
   {rendererClass=&NiagaraSpriteRendererProperties, rendererIndex=1}
  ]
 }
 Wave {
  emitterSpawnScript [/SetVariables() {isSetParameters=true}]
  emitterUpdateScript [
   &EmitterState(Life Cycle Mode=::NewEnumerator1, Loop Behavior=::NewEnumerator1, Loop Duration=1, UseLoopDelay=true)
   &SpawnBurst_Instantaneous(Spawn Count=1, Use Spawn Probability=false)
  ]
  particleSpawnScript [
   &InitializeParticle(
    Color=$System.WaveCol
    Lifetime=0.6
    Mesh Scale Mode=::NewEnumerator3
    Mesh Uniform Scale=0.4
    UsePositionOffset=false
    Write Material Random=true
   )
   &CameraOffset(Camera Offset Amount=100)
   &InitialMeshOrientation(Mesh Orientation Mode=::NewEnumerator2, Orientation Vector=<0 0 1>, Use Rotation Vector=false)
   /SetVariables(Particles.SphereColor:LinearColor=$System.ColorWaveCol, Particles.SphereScale=0.1) {isSetParameters=true}
  ]
  particleUpdateScript [
   &ParticleState()
   &DynamicMaterialParameters(
    Index 0 Param 1=&FloatFromCurve(FloatCurve=DataInterface() {propertyValues {Curve {keys [{arriveTangent=-2.1015, interpMode=RCIM_Cubic, leaveTangent=-2.1015, tangentMode=RCTM_User, time=0, value=0.5}, <at 1 -0.15>]}}})
    Index 0 Param 2=&FloatFromCurve(FloatCurve=DataInterface() {propertyValues {Curve {keys [{arriveTangent=-0.3464, interpMode=RCIM_Cubic, leaveTangent=-0.3464, tangentMode=RCTM_User, time=0, value=1}, <at 1 0.9>]}}})
    Index 0 Param 3=&FloatFromCurve(FloatCurve=DataInterface() {propertyValues {Curve {keys [<at 0 0>,<at 0.15 1>,<at 1 1.2>]}}})
    Module.Param0WriteEnabled=true
    Module.Param1WriteEnabled=true
    Module.Param2WriteEnabled=true
    Module.Param3WriteEnabled=false
   )
   &ScaleColor(
    Scale Alpha=&FloatFromCurve(FloatCurve=DataInterface() {propertyValues {Curve {keys [<at 0 1>,<at 1 0>]}}})
    Scale RGB=&VectorFromFloat(Value=&FloatFromCurve(FloatCurve=DataInterface() {propertyValues {Curve {keys [<at 0 10>,<at 0.15 3>,<at 1 1>]}}}))
    ScaleA=true
    ScaleRGB=true
   ) {module=ScaleColor001}
   &ScaleMeshSize(Scale Factor=&VectorFromFloat(Value=&FloatFromCurve(
      FloatCurve=DataInterface() {propertyValues {Curve {keys [<at 0 0>,<at 0.1 1.4>,<at 1 2>]}}}
      Scale Curve=&Multiply_Float(A=1, B=1)
     )))
   /SetVariables(
    Particles.SphereColor:LinearColor=&MultiplyLinearColor(
     Linear Color=$System.ColorWaveCol
     Scale Factor=&MakeFloat4FromVectorAndFloat(
      "Vector (RGB)"=<1 1 1>
      "Float (Alpha)"=&FloatFromCurve(
       FloatCurve=DataInterface() {propertyValues {Curve {keys [
           <at 0 1>
           {interpMode=RCIM_Cubic, tangentMode=RCTM_User, time=0.05, value=1}
           {interpMode=RCIM_Cubic, tangentMode=RCTM_User, time=0.5, value=0}
          ]}}}
       Scale Curve=10
      )
     )
    )
    Particles.SphereDynamicParam:Vector4f=&MakeVector4(X=1, Y=0, Z=1, W=1)
    Particles.SphereScale:NiagaraFloat=&FloatFromCurve(
     FloatCurve=DataInterface() {propertyValues {Curve {keys [
         {arriveTangent=4.1785, interpMode=RCIM_Cubic, leaveTangent=4.1785, tangentMode=RCTM_User, time=0, value=0.05}
         {arriveTangent=1.7974, interpMode=RCIM_Cubic, leaveTangent=1.7974, tangentMode=RCTM_User, time=0.15, value=0.6}
         {arriveTangent=-0.0405, interpMode=RCIM_Cubic, leaveTangent=-0.0405, tangentMode=RCTM_User, time=0.5, value=1}
        ]}}}
     Scale Curve=0.5
    )
    Particles.SphereScalevector:Vector3f=&VectorFromFloat(Value=$Particles.SphereScale)
   ) {isSetParameters=true}
  ]
  rendererClasses [&NiagaraMeshRendererProperties]
  renderers [
   {rendererClass=&NiagaraMeshRendererProperties, rendererIndex=0}
   {rendererClass=&NiagaraMeshRendererProperties, rendererIndex=1}
  ]
 }
}
summary {
 systemName=NS_Emote_RealCrown_Energy
 emitters [
  {emitterName=SlowWaves, rendererClasses [&NiagaraMeshRendererProperties]}
  {emitterName=Spark, rendererClasses [&NiagaraSpriteRendererProperties]}
  {emitterName=Sunlight, rendererClasses [&NiagaraSpriteRendererProperties]}
  {emitterName=RankIcons, rendererClasses [&NiagaraSpriteRendererProperties]}
  {
   emitterName=ColorWaveBlackhole001
   rendererClasses [&NiagaraSpriteRendererProperties, &NiagaraLightRendererProperties]
  }
  {emitterName=Wave, rendererClasses [&NiagaraMeshRendererProperties]}
 ]
 userVariables [
  User.bIsFrontEnd {type=NiagaraBool}
  User.bIsFrontEndPreview {type=NiagaraBool}
  User.InputDigits {type=NiagaraInt32, defaultValue=129}
  User.Rank_Champion {type=NiagaraInt32, defaultValue=2}
  User.Rank_Champion_Reload {type=NiagaraInt32, defaultValue=5}
  User.Rank_Elite {type=NiagaraInt32, defaultValue=1}
  User.Rank_Elite_Reload {type=NiagaraInt32, defaultValue=4}
  User.Rank_Unreal {type=NiagaraInt32, defaultValue=3}
  User.Rank_Unreal_Reload {type=NiagaraInt32, defaultValue=6}
 ]
}
