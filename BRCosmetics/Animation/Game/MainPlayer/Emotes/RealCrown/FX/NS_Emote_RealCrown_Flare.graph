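# Niagara system for the RealCrown emote's flare: a colored wave, socket-placed glow flare, two bling sprites and a spark burst.
# Emitters share system color variables; rank user variables and array-select inputs suggest rank-based tinting, though that logic is not shown.
# The wave and flare write light attributes, and sparks stretch with speed under drag and curl noise.
# Time-of-day and front-end inputs tune brightness, size and camera offset; DirectionalBurst is disabled.
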NiagaraSystem /BRCosmetics/Animation/Game/MainPlayer/Emotes/RealCrown/FX/NS_Emote_RealCrown_Flare
dependencies {
 usedDataInterfaces [
  &NiagaraDataInterfaceCurve
  &NiagaraDataInterfaceColorCurve
  &NiagaraDataInterfaceArrayColor
  &NiagaraDataInterfaceArrayFloat
 ]
 usedDynamicInputs [
  &MakeCustomIntFromBool
  &SetBoolByFloatComparison
  &MakeFloatFromInt
  &Add_Integer
  &SelectLinearColorFromArray
  &MakeIntFromFloat
  &SelectFloatFromArray
 ]
 usedModules [
  &SystemState
  &self:SystemSpawnScript.SystemScriptSource.SystemScriptGraph.NiagaraNodeAssignment_2.SetVariables_#1
  &self:SystemSpawnScript.SystemScriptSource.SystemScriptGraph.NiagaraNodeAssignment_3.SetVariables_#2
  &self:SystemSpawnScript.SystemScriptSource.SystemScriptGraph.NiagaraNodeAssignment_0.SetVariables_#3
 ]
}
emitters {
 Color_wave {
  emitterSpawnScript [/SetVariables(Emitter.SortOrder=3) {isSetParameters=true}]
  emitterUpdateScript [
   &EmitterState(Inactive Response=::NewEnumerator1, Life Cycle Mode=::NewEnumerator1, Loop Behavior=::NewEnumerator1, Loop Duration=3.09, UseLoopDelay=false)
   &SpawnBurst_Instantaneous(Use Spawn Probability=false, UseLoopCountLimit=false)
  ]
  particleSpawnScript [
   &InitializeParticle(
    Color=<rgba 3 3 3 2>
    Lifetime=1.0811
    Mesh Scale Mode=::NewEnumerator3
    Mesh Uniform Scale=-3.5
    Position Offset=$System.Position
    Sprite Rotation Angle=90
    Sprite Rotation Mode=::NewEnumerator0
    Sprite Size Mode=::NewEnumerator3
    Uniform Sprite Size=50
    UsePositionOffset=true
    Write Material Random=true
   )
   &SpriteFacingAndAlignment(WriteAlignment=false, WriteFacing=true) {enabled=false}
  ]
  particleUpdateScript [
   &ParticleState() {module=ParticleState001}
   &ScaleSpriteSize(Uniform Curve Sprite Scale=DataInterface() {propertyValues {Curve {keys [
        {arriveTangent=8.4722, interpMode=RCIM_Cubic, leaveTangent=8.4722, tangentMode=RCTM_User, time=0, value=0}
        {interpMode=RCIM_Cubic, tangentMode=RCTM_User, time=0.1152, value=1}
        {arriveTangent=0.1331, interpMode=RCIM_Cubic, leaveTangent=0.1331, time=0.7668, value=1}
        <at 0.8668 1.1 RCIM_Cubic>
        <at 1 1 RCIM_Cubic>
       ]}}}) {module=ScaleSpriteSize001}
   &ScaleSpriteSize(
    Uniform Curve Scale=1.2
    Uniform Curve Sprite Scale=DataInterface() {propertyValues {Curve {keys [
        {arriveTangent=6.5767, interpMode=RCIM_Cubic, leaveTangent=6.5767, tangentMode=RCTM_User, time=0, value=0}
        {arriveTangent=0.7192, interpMode=RCIM_Cubic, leaveTangent=0.7192, tangentMode=RCTM_User, time=0.2305, value=0.7836}
        {tangentMode=RCTM_User, time=1, value=1}
       ]}}}
   ) {enabled=false}
   &DynamicMaterialParameters(
    Index 0 Param 1=1
    Index 0 Param 2=&FloatFromCurve(FloatCurve=DataInterface() {propertyValues {Curve {keys [
         {interpMode=RCIM_Cubic, tangentMode=RCTM_User, time=0, value=1}
         {arriveTangent=-1, interpMode=RCIM_Cubic, leaveTangent=-1, time=0.6, value=0.7}
         {interpMode=RCIM_Cubic, leaveTangent=-3.3753, time=1, value=0}
        ]}}})
    Index 0 Param 3=1
    Index 0 Param 4=1
    Module.Param0WriteEnabled=true
    Module.Param1WriteEnabled=true
    Module.Param2WriteEnabled=true
    Module.Param3WriteEnabled=true
   )
   &ScaleColor(
    Scale Alpha=&FloatFromCurve(
     FloatCurve=DataInterface() {propertyValues {Curve {keys [
         {interpMode=RCIM_Cubic, tangentMode=RCTM_User, time=0, value=0}
         {interpMode=RCIM_Cubic, tangentMode=RCTM_User, time=0.5, value=1}
         {tangentMode=RCTM_User, time=1, value=0}
        ]}}}
     Scale Curve=&Multiply_Float(A=2, B=&FNTimeOfDayScalar(Night Scale=0.5, Morning Scale=0.8, Evening Scale=0.8))
    )
    Scale RGB=&MakeVectorFromLinearColorRGB(Color=&MultiplyLinearColor(
      Linear Color=&ColorFromCurve(ColorCurve=DataInterface() {propertyValues {
         AlphaCurve {keys []}
         BlueCurve {keys [<at 0.1259 2 RCIM_Cubic>, <at 0.2277 5>, {interpMode=RCIM_Cubic, leaveTangent=-0.7175, time=0.4478, value=2}]}
         GreenCurve {keys [<at 0.1259 0.1681 RCIM_Cubic>, <at 0.2277 0.2344>, {interpMode=RCIM_Cubic, leaveTangent=-4.4353, time=0.4478, value=0.1399}]}
         RedCurve {keys [<at 0.1259 0.5589 RCIM_Cubic>, <at 0.2277 1.1257>, {interpMode=RCIM_Cubic, leaveTangent=2.2176, time=0.4478, value=0.5865}]}
        }})
      Scale Factor=<1 1 1 1>
     ))
    ScaleA=true
    ScaleRGB=true
   )
   /SetVariables(Particles.Color02:LinearColor=&MakeLinearColorFromVectorAndFloat(
     "Vector (RGB)"=&MakeVectorFromLinearColorRGB(Color=&MultiplyLinearColor(Linear Color=$System.ColorWaveCol, Scale Factor=<2 2 2 1>))
     "Float (Alpha)"=&FloatFromCurve(
      FloatCurve=DataInterface() {propertyValues {Curve {keys [
          {interpMode=RCIM_Cubic, tangentMode=RCTM_User, time=0, value=1}
          {interpMode=RCIM_Cubic, tangentMode=RCTM_User, time=0.5, value=1}
          {arriveTangent=-2, interpMode=RCIM_Cubic, leaveTangent=-2, time=0.6842, value=1}
          {tangentMode=RCTM_User, time=1, value=0}
         ]}}}
      Scale Curve=0.1
     )
    )) {isSetParameters=true}
   &FN_Light_Attributes(
    Color=&MultiplyLinearColor(
     Linear Color=$System.ColorWaveCol
     Scale Factor=&MakeVector4(
      X=1
      Y=1
      Z=1
      W=&FloatFromCurve(FloatCurve=DataInterface() {propertyValues {Curve {keys [
           {interpMode=RCIM_Cubic, tangentMode=RCTM_User, time=0, value=0}
           {interpMode=RCIM_Cubic, tangentMode=RCTM_User, time=0.151, value=1}
           {arriveTangent=-1.1779, interpMode=RCIM_Cubic, leaveTangent=-1.1779, time=0.6, value=1}
           {tangentMode=RCTM_User, time=1, value=0}
          ]}}})
     )
    )
    Create Color Attribute=true
    Create Position Attribute=true
    Enable Particle Light Enable=true
   )
  ]
  rendererClasses [&NiagaraSpriteRendererProperties]
  renderers [{rendererClass=&NiagaraSpriteRendererProperties}]
 }
 DirectionalBurst {
  bEnabled=false
  emitterUpdateScript [
   &EmitterState(Life Cycle Mode=::NewEnumerator1, Loop Behavior=::NewEnumerator1, Loop Duration=6, UseLoopDelay=false)
   &SpawnRate(Use Spawn Probability=false) {enabled=false}
   &SpawnBurst_Instantaneous(Spawn Time=2.84, Use Spawn Probability=false, UseLoopCountLimit=false) {module=SpawnBurst_Instantaneous001}
  ]
  particleSpawnScript [&InitializeParticle(
    Color=$System.SparkCol
    Lifetime=&RandomRangeFloat(Minimum=0.55, Maximum=0.55)
    Mass Max=2
    Mass Mode=::NewEnumerator3
    Sprite Rotation Angle=&RandomRangeFloat(Minimum=-0.2, Maximum=0.2)
    Sprite Rotation Mode=::NewEnumerator0
    Sprite Size Mode=::NewEnumerator3
    Uniform Sprite Size=&RandomRangeFloat(Minimum=40, Maximum=40)
    UsePositionOffset=true
    Write Material Random=true
   )]
  particleUpdateScript [
   &ParticleState(Kill Particles When Lifetime Has Elapsed=false)
   &SpriteRotationRate(Rotation Rate=&FloatFromCurve(FloatCurve=DataInterface() {propertyValues {Curve {keys [<at 0 1>,<at 1 0>]}}}, Scale Curve=100))
   &ScaleColor(
    Scale Alpha=&FloatFromCurve(FloatCurve=DataInterface() {propertyValues {Curve {keys [<at -0.0046 1.0113 RCIM_Cubic>,<at 1.0011 0.002 RCIM_Cubic>]}}})
    Scale RGB=&VectorFromFloat(Value=&FloatFromCurve(FloatCurve=DataInterface() {propertyValues {Curve {keys [{interpMode=RCIM_Cubic, tangentMode=RCTM_User, time=0, value=1}, {interpMode=RCIM_Cubic, tangentMode=RCTM_User, time=1, value=0}]}}}))
    ScaleA=true
    ScaleRGB=true
   )
   &SolveForcesAndVelocity(Clamp Velocity=false, Limit Acceleration=false)
  ]
  rendererClasses [&NiagaraSpriteRendererProperties]
  renderers [{rendererClass=&NiagaraSpriteRendererProperties}]
 }
 Flare001 {
  emitterUpdateScript [
   &EmitterState(Life Cycle Mode=::NewEnumerator1, Loop Behavior=::NewEnumerator1, Loop Duration=3, UseLoopDelay=false)
   &SpawnBurst_Instantaneous(Spawn Count=1, Spawn Probability=0.5, Use Spawn Probability=false)
  ]
  particleSpawnScript [
   &InitializeParticle(
    Color=&Multiply_LinearColorByFloat(Linear Color=$System.RingColor, Float=0.5)
    Lifetime=0.7568
    Sprite Rotation Mode=::NewEnumerator2
    Sprite Size Mode=::NewEnumerator3
    Uniform Sprite Size=8
    UsePositionOffset=false
    Write Material Random=true
   )
   &SkeletalMeshLocation(
    MaskSpawnGroup=false
    Mesh Sampling Type=::NewEnumerator5
    Position Sampling=::NewEnumerator2
    Skeletal Mesh=$User.SK_mesh
    Socket Sampling Mode=::NewEnumerator5
   )
   &DynamicMaterialParameters(
    Index 0 Param 1=&RandomRangeFloat(Minimum=0.4, Maximum=0.6)
    Index 0 Param 2=1
    Index 0 Param 3=1
    Index 0 Param 4=1
    Module.Param0WriteEnabled=true
    Module.Param1WriteEnabled=true
    Module.Param2WriteEnabled=true
    Module.Param3WriteEnabled=true
   )
   /SetVariables(Particles.Dark_Glow_CameraOffset=&MakeCustomFloatFromBool(Bool=$User.bIsFrontEndPreview, True Float=1, False Float=50)) {isSetParameters=true}
   &CameraOffset(Camera Offset Amount=15)
   &SystemLocation(MaskSpawnGroup=false, Offset=<-10 0 0>) {enabled=false}
   /SetVariables(Particles.Position=$System.ProjectilePosition) {enabled=false, isSetParameters=true}
  ]
  particleUpdateScript [
   /SetVariables(Particles.Dark_Glow=&MakeLinearColorFromVectorAndFloat(
     "Vector (RGB)"=&MakeVectorFromLinearColorRGB(Color=<rgba 0.326 0.1507 0.141 1>)
     "Float (Alpha)"=&Multiply_Float(
      A=&Multiply_Float(A=&FNTimeOfDayScalar(Night Scale=0.5, Morning Scale=0.3, Evening Scale=0.4), B=1.6)
      B=&FloatFromCurve(FloatCurve=DataInterface() {propertyValues {Curve {keys [<at 0 0>,<at 0.2 1>,<at 0.7 0.6>,<at 1 0>]}}}, Scale Curve=0.2)
     )
    )) {isSetParameters=true}
   /SetVariables(Particles.Dark_Glow_Size=&Multiply_Vector2(
     A=&Vector2DFromFloat(Value=50)
     B=&Vector2DFromFloat(Value=&FloatFromCurve(FloatCurve=DataInterface() {propertyValues {Curve {keys [<at 0 1.5 RCIM_Cubic>,<at 1 0.5>]}}}))
    )) {isSetParameters=true}
   &ParticleState()
   &SpriteRotationRate(Rotation Rate=100)
   &ScaleColor(
    Scale Alpha=&Multiply_Float(
     A=&FNTimeOfDayScalar(Night Scale=0.7, Morning Scale=1, Evening Scale=1)
     B=&FloatFromCurve(FloatCurve=DataInterface() {propertyValues {Curve {keys [
          <at 0 0>
          <at 0.2265 1>
          {arriveTangent=-3.5143, interpMode=RCIM_Cubic, leaveTangent=-3.5143, tangentMode=RCTM_User, time=0.3534, value=1}
          <at 0.8696 0>
         ]}}})
    )
    Scale RGB=<2 2 2>
    ScaleA=true
    ScaleRGB=true
   )
   &ScaleSpriteSize(Scale Factor=&Vector2DFromFloat(Value=&FloatFromCurve(FloatCurve=DataInterface() {propertyValues {Curve {keys [
          {arriveTangent=0.2438, interpMode=RCIM_Cubic, leaveTangent=0.2438, tangentMode=RCTM_User, time=0, value=0.3}
          {arriveTangent=8.1387, interpMode=RCIM_Cubic, leaveTangent=8.1387, tangentMode=RCTM_User, time=0.2514, value=2.5}
         ]}}})))
   &FN_Light_Attributes(
    Color=<rgba 1 0.7652 0.2031 10>
    Create Color Attribute=true
    Create Exponent Attribute=true
    Create Radius Attribute=true
    Exponent=3
    Radius=50
    Scale Color=&FloatFromCurve(FloatCurve=DataInterface() {propertyValues {Curve {keys [
         {arriveTangent=-2.6654, interpMode=RCIM_Cubic, leaveTangent=-2.6654, tangentMode=RCTM_User, time=0, value=1}
         {arriveTangent=-0.18, interpMode=RCIM_Cubic, leaveTangent=-0.18, tangentMode=RCTM_User, time=1, value=0}
        ]}}})
   )
  ]
  rendererClasses [&NiagaraSpriteRendererProperties]
  renderers [
   {rendererClass=&NiagaraSpriteRendererProperties}
   {rendererClass=&NiagaraSpriteRendererProperties, rendererIndex=1}
  ]
 }
 NE_Topper_Bling {
  emitterUpdateScript [
   &EmitterState(Life Cycle Mode=::NewEnumerator1, Loop Behavior=::NewEnumerator1, Loop Duration=6, UseLoopDelay=true)
   &SpawnBurst_Instantaneous(Spawn Count=1, Use Spawn Probability=false)
  ]
  particleSpawnScript [
   &InitializeParticle(
    Color=$System.SparkCol
    Lifetime=0.973
    Sprite Rotation Mode=::NewEnumerator1
    Sprite Size Mode=::NewEnumerator3
    Sprite UV Mode=::NewEnumerator4
    Uniform Sprite Size=35
    UsePositionOffset=false
    Write Material Random=true
   ) {module=InitializeParticle001}
   &CameraOffset(Camera Offset Amount=10)
  ]
  particleUpdateScript [
   &ParticleState()
   &SpriteRotationRate(Rotation Rate=&FloatFromCurve(FloatCurve=DataInterface() {propertyValues {Curve {keys [<at 0 1>,<at 1 0>]}}}, Scale Curve=350)) {enabled=false}
   &SolveForcesAndVelocity(Clamp Velocity=false, Limit Acceleration=false)
   &ScaleColor(
    Scale Alpha=&FloatFromCurve(FloatCurve=DataInterface() {propertyValues {Curve {keys [<at -0.0092 0.0144>,<at 0.0967 0.9767>,<at 1.0011 0.002>]}}})
    ScaleA=true
    ScaleRGB=true
   )
   &ScaleSpriteSize(Scale Factor=&Vector2DFromFloat(Value=&FloatFromCurve(FloatCurve=DataInterface() {propertyValues {Curve {keys [
          <at -0.0046 0 RCIM_Cubic>
          {
           arriveTangent=0.0048
           arriveTangentWeight=0.1184
           interpMode=RCIM_Cubic
           leaveTangent=0.0054
           leaveTangentWeight=0.5412
           tangentMode=RCTM_Break
           tangentWeightMode=RCTWM_WeightedBoth
           time=0.1214
           value=1
          }
          <at 1.0011 0.002 RCIM_Cubic>
         ]}}})))
  ]
  rendererClasses [&NiagaraSpriteRendererProperties]
  renderers [{rendererClass=&NiagaraSpriteRendererProperties}]
 }
 NE_Topper_Bling001 {
  emitterUpdateScript [
   &EmitterState(Life Cycle Mode=::NewEnumerator1, Loop Behavior=::NewEnumerator1, Loop Duration=6, UseLoopDelay=true)
   &SpawnBurst_Instantaneous(Spawn Count=1, Use Spawn Probability=false)
  ]
  particleSpawnScript [
   &InitializeParticle(
    Color=$System.Flare2Col
    Lifetime=0.973
    Sprite Rotation Mode=::NewEnumerator1
    Sprite Size Mode=::NewEnumerator3
    Sprite UV Mode=::NewEnumerator4
    Uniform Sprite Size=28
    UsePositionOffset=false
    Write Material Random=true
   ) {module=InitializeParticle001}
   &CameraOffset(Camera Offset Amount=10)
  ]
  particleUpdateScript [
   &ParticleState()
   &SpriteRotationRate(Rotation Rate=&FloatFromCurve(FloatCurve=DataInterface() {propertyValues {Curve {keys [<at 0 1>,<at 1 0>]}}}, Scale Curve=-60))
   &SolveForcesAndVelocity(Clamp Velocity=false, Limit Acceleration=false)
   &ScaleColor(
    Scale Alpha=&FloatFromCurve(FloatCurve=DataInterface() {propertyValues {Curve {keys [<at -0.0092 0.0144>,<at 0.0967 0.9767>,<at 1.0011 0.002>]}}})
    ScaleA=true
    ScaleRGB=true
   )
   &ScaleSpriteSize(Scale Factor=&Vector2DFromFloat(Value=&FloatFromCurve(FloatCurve=DataInterface() {propertyValues {Curve {keys [
          <at -0.0046 0 RCIM_Cubic>
          {
           arriveTangent=0.0048
           arriveTangentWeight=0.1184
           interpMode=RCIM_Cubic
           leaveTangent=0.0054
           leaveTangentWeight=0.5412
           tangentMode=RCTM_Break
           tangentWeightMode=RCTWM_WeightedBoth
           time=0.1214
           value=1
          }
          <at 1.0011 0.002 RCIM_Cubic>
         ]}}})))
  ]
  rendererClasses [&NiagaraSpriteRendererProperties]
  renderers [{rendererClass=&NiagaraSpriteRendererProperties}]
 }
 Spark {
  emitterUpdateScript [
   &EmitterState(Life Cycle Mode=::NewEnumerator1, Loop Behavior=::NewEnumerator1, Loop Delay=0.05, Loop Duration=1, UseLoopDelay=true)
   &SpawnBurst_Instantaneous(Spawn Count=10, Use Spawn Probability=false)
  ]
  particleSpawnScript [
   &InitializeParticle(
    Color=$System.SparkCol
    Lifetime Max=0.8108
    Lifetime Min=1.3514
    Lifetime Mode=::NewEnumerator1
    Sprite Size Max=<2.5 6>
    Sprite Size Min=<2 3>
    Sprite Size Mode=::NewEnumerator2
    UsePositionOffset=false
    Write Material Random=true
   )
   &AddVelocityFromPoint(
    Module.Use Falloff Distance=false
    Velocity Strength=&Multiply_Float(
     A=&RandomRangeFloat(Minimum=30, Maximum=100)
     B=&FN_FrontEndFloatModifier(Gameplay=1, Front End=0.5)
    )
   )
   &ShapeLocation(MaskSpawnGroup=false, Override Seed=false, Sphere Radius=1, UseOwnerScale=true)
  ]
  particleUpdateScript [
   &ParticleState()
   &ScaleColor(
    Linear Color Curve=DataInterface() {propertyValues {
      AlphaCurve {keys [<at 0.0074 1>,<at 0.8634 1>,<at 0.8905 1>]}
      BlueCurve {keys [<at 0 15>]}
      GreenCurve {keys [<at 0 0>]}
      RedCurve {keys [<at 0 5.5313>]}
     }}
    Scale Mode=::NewEnumerator2
   ) {enabled=false}
   &ScaleColor(
    Scale Alpha=&FNTimeOfDayScalar(Night Scale=1, Morning Scale=1, Evening Scale=1)
    Scale RGB=&VectorFromFloat(Value=10)
    ScaleA=true
    ScaleRGB=true
   ) {module=ScaleColor001}
   &Drag(Drag=&RandomRangeFloat(Minimum=4, Maximum=7), Use Linear Drag=true, Use Rotational Drag=false)
   &CurlNoiseForce(
    Noise Strength=&Multiply_Float(
     A=&RandomRangeFloat(Minimum=15, Maximum=50)
     B=&FloatFromCurve(FloatCurve=DataInterface() {propertyValues {Curve {keys [<at 0 1>,<at 1 0.2>]}}}, CurveIndex=$Emitter.Age)
    )
    Pan Noise=true
   )
   &GravityForce(Gravity=&RandomRangeVector(Minimum=<0 0 20>, Maximum=<0 0 -20>))
   &PointAttractionForce(
    Attraction Radius=200
    Attraction Strength=&FloatFromCurve(
     FloatCurve=DataInterface() {propertyValues {Curve {keys [<at 1 0 RCIM_Cubic>, {arriveTangent=2.474, interpMode=RCIM_Cubic, leaveTangent=2.474, tangentMode=RCTM_User, time=2.5, value=1}]}}}
     CurveIndex=$Emitter.Age
     Scale Curve=360
    )
    Kill Radius:NiagaraFloat=&FloatFromCurve(FloatCurve=DataInterface() {propertyValues {Curve {keys [<at 2 0>,<at 2.7 5>]}}}, CurveIndex=$Emitter.Age)
    Kill Within Radius=true
    Module.Use Falloff=true
    UseAttractorOffset=false
   ) {enabled=false}
   &SolveForcesAndVelocity(Clamp Velocity=false, Limit Acceleration=false)
   &ScaleSpriteSizeBySpeed(
    Max Scale Factor=<1 20>
    Min Scale Factor=<0.5 0.1>
    Velocity Threshold:NiagaraFloat=&FloatFromCurve(FloatCurve=DataInterface() {propertyValues {Curve {keys [<at 0 6>,<at 0.05 0.9>,<at 1 0.05>]}}}, Scale Curve=1000)
   )
   &ScaleSpriteSize(Scale Factor=&MakeVector2D(
     X=&FloatFromCurve(
      FloatCurve=DataInterface() {propertyValues {Curve {keys [<at 0 0.5>,<at 0.1 1>,<at 1 0.5>]}}}
      Scale Curve=&FN_FrontEndFloatModifier(Gameplay=1, Front End=0.4)
     )
     Y=&FloatFromCurve(
      FloatCurve=DataInterface() {propertyValues {Curve {keys [<at 0 1>,<at 1 0>]}}}
      Scale Curve=&FN_FrontEndFloatModifier(Gameplay=1, Front End=0.4)
     )
    ))
  ]
  rendererClasses [&NiagaraSpriteRendererProperties]
  renderers [{rendererClass=&NiagaraSpriteRendererProperties}]
 }
}
summary {
 systemName=NS_Emote_RealCrown_Flare
 emitters [
  {bEnabled=false, emitterName=DirectionalBurst, rendererClasses [&NiagaraSpriteRendererProperties]}
  {emitterName=NE_Topper_Bling, rendererClasses [&NiagaraSpriteRendererProperties]}
  {emitterName=Flare001, rendererClasses [&NiagaraSpriteRendererProperties]}
  {emitterName=NE_Topper_Bling001, rendererClasses [&NiagaraSpriteRendererProperties]}
  {emitterName=Spark, rendererClasses [&NiagaraSpriteRendererProperties]}
  {emitterName=Color_wave, rendererClasses [&NiagaraSpriteRendererProperties]}
 ]
 userVariables [
  User.bIsFrontEnd {type=NiagaraBool}
  User.bIsFrontEndPreview {type=NiagaraBool}
  User.GlobalScale {type=NiagaraFloat, defaultValue=396.1002, description=""}
  User.Rank_Champion {type=NiagaraInt32, defaultValue=888}
  User.Rank_Champion_Reload {type=NiagaraInt32, defaultValue=555}
  User.Rank_Elite {type=NiagaraInt32, defaultValue=777}
  User.Rank_Elite_Reload {type=NiagaraInt32, defaultValue=444}
  User.Rank_Unreal {type=NiagaraInt32, defaultValue=999}
  User.Rank_Unreal_Reload {type=NiagaraInt32, defaultValue=666}
  User.SK_mesh {
   type=NiagaraDataInterfaceSkeletalMesh
   defaultValue {
    struct=NiagaraExt_VariableValue_DataInterface
    value {dataInterface=&self:NiagaraDataInterfaceSkeletalMesh_0}
   }
  }
 ]
}
