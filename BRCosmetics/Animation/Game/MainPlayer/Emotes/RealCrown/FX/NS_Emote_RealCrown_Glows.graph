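# Niagara system providing glow effects for the RealCrown player emote, with sprite, light and ribbon renderers.
# The Glows emitter bursts grouped particles with dynamically driven color and velocity, then updates their position from skeletal mesh bones, applying camera offset and color-over-life scaling.
# Its dependencies also include spawn-rate ribbons, point attraction, dynamic material parameters, array-based color selection and system-level variable assignments.
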NiagaraSystem /BRCosmetics/Animation/Game/MainPlayer/Emotes/RealCrown/FX/NS_Emote_RealCrown_Glows
dependencies {
 usedDataInterfaces [
  {refPath=&NiagaraDataInterfaceColorCurve}
  {refPath=&NiagaraDataInterfaceArrayColor}
  {refPath=&NiagaraDataInterfaceCurve}
  {refPath=&NiagaraDataInterfaceSkeletalMesh}
 ]
 usedDynamicInputs [
  {refPath=&SimulationPosition}
  {refPath=&SelectLinearColorFromArray}
  {refPath=&RandomRangeVector}
  {refPath=&FloatFromCurve}
  {refPath=&MakeCustomFloatFromBool}
  {refPath=&SetBoolByFloatComparison}
  {refPath=&Add_Float}
  {refPath=&Abs_Float}
  {refPath=&MakeFloatFromVector}
  {refPath=&ClampFloat}
 ]
 usedModules [
  {refPath=&EmitterState}
  {refPath=&SpawnBurst_Instantaneous}
  {refPath=&FN_SetSpawnGroup}
  {refPath=&InitializeParticle}
  {refPath=&ParticleState}
  {refPath=&SkeletalMeshLocation}
  {refPath=&CameraOffset}
  {refPath=&ScaleColor}
  {refPath=&FN_Light_v2}
  {refPath=&AddVelocity}
  {refPath=&SolveForcesAndVelocity}
  {refPath=&SpawnRate}
  {refPath=&InitializeRibbon}
  {refPath=&ScaleRibbonWidth}
  {refPath=&DynamicMaterialParameters}
  {refPath=&PointAttractionForce}
  {refPath=&SystemState}
  {refPath=&self:SystemSpawnScript.SystemScriptSource.SystemScriptGraph.NiagaraNodeAssignment_0.SetVariables_#1}
  {refPath=&self:SystemSpawnScript.SystemScriptSource.SystemScriptGraph.NiagaraNodeAssignment_1.SetVariables_#2}
  {refPath=&self:SystemSpawnScript.SystemScriptSource.SystemScriptGraph.NiagaraNodeAssignment_2.SetVariables_#3}
 ]
 usedRenderers [
  {refPath=&NiagaraSpriteRendererProperties}
  {refPath=&NiagaraLightRendererProperties}
  {refPath=&NiagaraRibbonRendererProperties}
 ]
}
emitters {
 Glows {
  emitterSpawnScript {scriptName=EmitterSpawnScript, modules []}
  emitterUpdateScript {
   scriptName=EmitterUpdateScript
   modules [
    {
     bIsSetParametersModule=false
     enabled=true
     moduleName=EmitterState
     inputs [
      Life Cycle Mode {
       bIsDynamic=false
       bIsEditable=true
       bIsStaticSwitch=true
       bIsVisible=true
       type {classStructOrEnum {refPath=/Niagara/Enums/ENiagaraEmitterLifeCycleMode}}
      }
      Inactive Response {
       bIsDynamic=false
       bIsEditable=true
       bIsStaticSwitch=true
       bIsVisible=true
       type {classStructOrEnum {refPath=/Niagara/Enums/ENiagaraInactiveMode}}
      }
      Loop Behavior {
       bIsDynamic=false
       bIsEditable=true
       bIsStaticSwitch=true
       bIsVisible=true
       type {classStructOrEnum {refPath=/Niagara/Enums/ENiagara_EmitterStateOptions}}
      }
      Loop Duration Mode {
       bIsDynamic=false
       bIsEditable=true
       bIsStaticSwitch=true
       bIsVisible=true
       type {classStructOrEnum {refPath=/Niagara/Enums/ENiagara_InfiniteLoopDuration}}
      }
      Loop Count {
       bIsDynamic=false
       bIsEditable=false
       bIsStaticSwitch=false
       bIsVisible=false
       type {classStructOrEnum {refPath=/Script/Niagara.NiagaraInt32}}
      }
      Loop Duration {bIsDynamic=false, bIsEditable=true, bIsStaticSwitch=false, bIsVisible=true, type {classStructOrEnum {refPath=/Script/Niagara.NiagaraFloat}}}
      Recalculate Duration Each Loop {bIsDynamic=false, bIsEditable=false, bIsStaticSwitch=false, bIsVisible=false, type {classStructOrEnum {refPath=/Script/Niagara.NiagaraBool}}}
      UseLoopDelay {bIsDynamic=false, bIsEditable=true, bIsStaticSwitch=false, bIsVisible=true, type {classStructOrEnum {refPath=&NiagaraBool}}}
      Loop Delay {bIsDynamic=false, bIsEditable=false, bIsStaticSwitch=false, bIsVisible=true, type {classStructOrEnum {refPath=&NiagaraFloat}}}
      Delay First Loop Only {bIsDynamic=false, bIsEditable=false, bIsStaticSwitch=false, bIsVisible=false, type {classStructOrEnum {refPath=&NiagaraBool}}}
      Scalability Mode {
       bIsDynamic=false
       bIsEditable=true
       bIsStaticSwitch=true
       bIsVisible=true
       type {classStructOrEnum {refPath=/Niagara/Enums/ENiagaraEmitterScalabilityMode_Limited}}
      }
      Enable Distance Culling {bIsDynamic=false, bIsEditable=false, bIsStaticSwitch=true, bIsVisible=false, type {classStructOrEnum {refPath=&NiagaraBool}}}
      MinDistance {
       bIsDynamic=false
       bIsEditable=false
       bIsStaticSwitch=false
       bIsVisible=false
       type {classStructOrEnum {refPath=&NiagaraFloat}}
      }
      Min Distance Response {
       bIsDynamic=false
       bIsEditable=false
       bIsStaticSwitch=false
       bIsVisible=false
       type {classStructOrEnum {refPath=/Script/Niagara.ENiagaraExecutionStateManagement}}
      }
      MaxDistance {
       bIsDynamic=false
       bIsEditable=false
       bIsStaticSwitch=false
       bIsVisible=false
       type {classStructOrEnum {refPath=&NiagaraFloat}}
      }
      Max Distance Response {
       bIsDynamic=false
       bIsEditable=false
       bIsStaticSwitch=false
       bIsVisible=false
       type {classStructOrEnum {refPath=&ENiagaraExecutionStateManagement}}
      }
      Scale Spawn Count {bIsDynamic=false, bIsEditable=false, bIsStaticSwitch=true, bIsVisible=false, type {classStructOrEnum {refPath=&NiagaraBool}}}
      Spawn Count Scale {
       bIsDynamic=false
       bIsEditable=false
       bIsStaticSwitch=false
       bIsVisible=false
       type {classStructOrEnum {refPath=&NiagaraFloat}}
      }
      Spawn Count Scale By Distance {
       bIsDynamic=false
       bIsEditable=false
       bIsStaticSwitch=false
       bIsVisible=false
       type {classStructOrEnum {refPath=&NiagaraDataInterfaceCurve}}
      }
      Enable Visibility Culling {bIsDynamic=false, bIsEditable=false, bIsStaticSwitch=true, bIsVisible=false, type {classStructOrEnum {refPath=&NiagaraBool}}}
      Visibility Cull Response {
       bIsDynamic=false
       bIsEditable=false
       bIsStaticSwitch=false
       bIsVisible=false
       type {classStructOrEnum {refPath=&ENiagaraExecutionStateManagement}}
      }
      Visibility Cull Delay {
       bIsDynamic=false
       bIsEditable=false
       bIsStaticSwitch=false
       bIsVisible=false
       type {classStructOrEnum {refPath=&NiagaraFloat}}
      }
      Reset Age On Awaken {bIsDynamic=false, bIsEditable=false, bIsStaticSwitch=false, bIsVisible=false, type {classStructOrEnum {refPath=&NiagaraBool}}}
     ]
     moduleScript {refPath=&EmitterState}
    }
    {
     bIsSetParametersModule=false
     enabled=true
     moduleName=SpawnBurst_Instantaneous
     inputs [
      Spawn Count {bIsDynamic=false, bIsEditable=true, bIsStaticSwitch=false, bIsVisible=true, type {classStructOrEnum {refPath=&NiagaraInt32}}}
      Spawn Time {bIsDynamic=false, bIsEditable=true, bIsStaticSwitch=false, bIsVisible=true, type {classStructOrEnum {refPath=&NiagaraFloat}}}
      Use Spawn Probability {bIsDynamic=false, bIsEditable=true, bIsStaticSwitch=false, bIsVisible=true, type {classStructOrEnum {refPath=&NiagaraBool}}}
      Spawn Probability {bIsDynamic=false, bIsEditable=true, bIsStaticSwitch=false, bIsVisible=true, type {classStructOrEnum {refPath=&NiagaraFloat}}}
      Spawn Group {bIsDynamic=false, bIsEditable=true, bIsStaticSwitch=false, bIsVisible=true, type {classStructOrEnum {refPath=&NiagaraInt32}}}
      Age {bIsDynamic=false, bIsEditable=true, bIsStaticSwitch=false, bIsVisible=true, type {classStructOrEnum {refPath=&NiagaraFloat}}}
     ]
     moduleScript {refPath=&SpawnBurst_Instantaneous}
    }
   ]
  }
  particleSpawnScript {
   scriptName=ParticleSpawnScript
   modules [
    {
     bIsSetParametersModule=false
     enabled=true
     moduleName=FN_SetSpawnGroup
     inputs [
      Number of Spawn Groups {bIsDynamic=false, bIsEditable=true, bIsStaticSwitch=false, bIsVisible=true, type {classStructOrEnum {refPath=&NiagaraInt32}}}
      Particles Per Group {
       bIsDynamic=false
       bIsEditable=false
       bIsStaticSwitch=false
       bIsVisible=false
       type {classStructOrEnum {refPath=&NiagaraInt32}}
      }
      Pattern {
       bIsDynamic=false
       bIsEditable=true
       bIsStaticSwitch=true
       bIsVisible=true
       type {classStructOrEnum {refPath=/Game/Effects/Niagara/Enum/EniagaraPartitionWaveForm.ENiagaraPartitionWaveForm}}
      }
      Normalize Spawn Group {bIsDynamic=false, bIsEditable=true, bIsStaticSwitch=true, bIsVisible=true, type {classStructOrEnum {refPath=&NiagaraBool}}}
      Inclusivity {bIsDynamic=false, bIsEditable=false, bIsStaticSwitch=true, bIsVisible=false, type {classStructOrEnum {refPath=&NiagaraBool}}}
      Set Ribbon ID and Ribbon Link Order {bIsDynamic=false, bIsEditable=true, bIsStaticSwitch=true, bIsVisible=true, type {classStructOrEnum {refPath=&NiagaraBool}}}
     ]
     moduleScript {refPath=&FN_SetSpawnGroup}
    }
    {
     bIsSetParametersModule=false
     enabled=true
     moduleName=InitializeParticle001
     inputs [
      Lifetime Mode {
       bIsDynamic=false
       bIsEditable=true
       bIsStaticSwitch=true
       bIsVisible=true
       type {classStructOrEnum {refPath=/Niagara/Enums/ENiagara_LifetimeMode}}
      }
      Lifetime Min {
       bIsDynamic=false
       bIsEditable=false
       bIsStaticSwitch=false
       bIsVisible=false
       type {classStructOrEnum {refPath=&NiagaraFloat}}
      }
      Lifetime {bIsDynamic=false, bIsEditable=true, bIsStaticSwitch=false, bIsVisible=true, type {classStructOrEnum {refPath=&NiagaraFloat}}}
      Lifetime Max {
       bIsDynamic=false
       bIsEditable=false
       bIsStaticSwitch=false
       bIsVisible=false
       type {classStructOrEnum {refPath=&NiagaraFloat}}
      }
      Lifetime Randomness Mode {
       bIsDynamic=false
       bIsEditable=false
       bIsStaticSwitch=true
       bIsVisible=false
       type {classStructOrEnum {refPath=/Niagara/Enums/ENiagaraRandomnessMode}}
      }
      Color Mode {
       bIsDynamic=false
       bIsEditable=true
       bIsStaticSwitch=true
       bIsVisible=true
       type {classStructOrEnum {refPath=/Niagara/Enums/ENiagara_ColorInitializationMode}}
      }
      Color Minimum {
       bIsDynamic=false
       bIsEditable=false
       bIsStaticSwitch=false
       bIsVisible=false
       type {classStructOrEnum {refPath=/Script/CoreUObject.LinearColor}}
      }
      Color {
       bIsDynamic=true
       bIsEditable=true
       bIsStaticSwitch=false
       bIsVisible=true
       type {classStructOrEnum {refPath=&LinearColor}}
      }
      AdjustHue {bIsDynamic=false, bIsEditable=false, bIsStaticSwitch=false, bIsVisible=false, type {classStructOrEnum {refPath=&NiagaraBool}}}
      Hue Shift Range {
       bIsDynamic=false
       bIsEditable=false
       bIsStaticSwitch=false
       bIsVisible=false
       type {classStructOrEnum {refPath=/Script/CoreUObject.Vector2f}}
      }
      Color Maximum {
       bIsDynamic=false
       bIsEditable=false
       bIsStaticSwitch=false
       bIsVisible=false
       type {classStructOrEnum {refPath=&LinearColor}}
      }
      AdjustSaturation {bIsDynamic=false, bIsEditable=false, bIsStaticSwitch=false, bIsVisible=false, type {classStructOrEnum {refPath=&NiagaraBool}}}
      Saturation Range {
       bIsDynamic=false
       bIsEditable=false
       bIsStaticSwitch=false
       bIsVisible=false
       type {classStructOrEnum {refPath=&Vector2f}}
      }
      AdjustValue {bIsDynamic=false, bIsEditable=false, bIsStaticSwitch=false, bIsVisible=false, type {classStructOrEnum {refPath=&NiagaraBool}}}
      Value Range {
       bIsDynamic=false
       bIsEditable=false
       bIsStaticSwitch=false
       bIsVisible=false
       type {classStructOrEnum {refPath=&Vector2f}}
      }
      AdjustAlpha {bIsDynamic=false, bIsEditable=false, bIsStaticSwitch=false, bIsVisible=false, type {classStructOrEnum {refPath=&NiagaraBool}}}
      Alpha Scale Range {
       bIsDynamic=false
       bIsEditable=false
       bIsStaticSwitch=false
       bIsVisible=false
       type {classStructOrEnum {refPath=&Vector2f}}
      }
      Color Channel Mode {
       bIsDynamic=false
       bIsEditable=false
       bIsStaticSwitch=true
       bIsVisible=false
       type {classStructOrEnum {refPath=/Niagara/Enums/ENiagaraChannelCorrelation}}
      }
      Color Randomness Mode {
       bIsDynamic=false
       bIsEditable=false
       bIsStaticSwitch=true
       bIsVisible=false
       type {classStructOrEnum {refPath=&ENiagaraRandomnessMode}}
      }
      Position Mode {
       bIsDynamic=false
       bIsEditable=true
       bIsStaticSwitch=true
       bIsVisible=true
       type {classStructOrEnum {refPath=/Niagara/Enums/ENiagara_PositionInitializationMode}}
      }
      Position {
       bIsDynamic=true
       bIsEditable=false
       bIsStaticSwitch=false
       bIsVisible=false
       type {classStructOrEnum {refPath=/Script/Niagara.NiagaraPosition}}
      }
      UsePositionOffset {bIsDynamic=false, bIsEditable=true, bIsStaticSwitch=false, bIsVisible=true, type {classStructOrEnum {refPath=&NiagaraBool}}}
      Position Offset {bIsDynamic=false, bIsEditable=false, bIsStaticSwitch=false, bIsVisible=true, type {classStructOrEnum {refPath=/Script/CoreUObject.Vector3f}}}
      Position Offset Coordinate Space {
       bIsDynamic=false
       bIsEditable=false
       bIsStaticSwitch=false
       bIsVisible=false
       type {classStructOrEnum {refPath=/Script/Niagara.ENiagaraCoordinateSpace}}
      }
      Mass Mode {
       bIsDynamic=false
       bIsEditable=true
       bIsStaticSwitch=true
       bIsVisible=true
       type {classStructOrEnum {refPath=/Niagara/Enums/ENiagara_MassInitializationMode}}
      }
      Mass Min {
       bIsDynamic=false
       bIsEditable=false
       bIsStaticSwitch=false
       bIsVisible=false
       type {classStructOrEnum {refPath=&NiagaraFloat}}
      }
      Mass {bIsDynamic=false, bIsEditable=true, bIsStaticSwitch=false, bIsVisible=true, type {classStructOrEnum {refPath=&NiagaraFloat}}}
      Mass Max {
       bIsDynamic=false
       bIsEditable=false
       bIsStaticSwitch=false
       bIsVisible=false
       type {classStructOrEnum {refPath=&NiagaraFloat}}
      }
      Mass Randomness Mode {
       bIsDynamic=false
       bIsEditable=false
       bIsStaticSwitch=true
       bIsVisible=false
       type {classStructOrEnum {refPath=&ENiagaraRandomnessMode}}
      }
      Write Material Random {bIsDynamic=false, bIsEditable=true, bIsStaticSwitch=false, bIsVisible=true, type {classStructOrEnum {refPath=&NiagaraBool}}}
      Material Random {bIsDynamic=false, bIsEditable=true, bIsStaticSwitch=false, bIsVisible=true, type {classStructOrEnum {refPath=&NiagaraFloat}}}
      Sprite Size Mode {
       bIsDynamic=false
       bIsEditable=true
       bIsStaticSwitch=true
       bIsVisible=true
       type {classStructOrEnum {refPath=/Niagara/Enums/ENiagara_SizeScaleMode}}
      }
      Sprite Size {
       bIsDynamic=false
       bIsEditable=false
       bIsStaticSwitch=false
       bIsVisible=false
       type {classStructOrEnum {refPath=&Vector2f}}
      }
      Uniform Sprite Size {bIsDynamic=false, bIsEditable=true, bIsStaticSwitch=false, bIsVisible=true, type {classStructOrEnum {refPath=&NiagaraFloat}}}
      Uniform Sprite Size Min {
       bIsDynamic=false
       bIsEditable=false
       bIsStaticSwitch=false
       bIsVisible=false
       type {classStructOrEnum {refPath=&NiagaraFloat}}
      }
      Sprite Size Min {
       bIsDynamic=false
       bIsEditable=false
       bIsStaticSwitch=false
       bIsVisible=false
       type {classStructOrEnum {refPath=&Vector2f}}
      }
      Sprite Size Max {
       bIsDynamic=false
       bIsEditable=false
       bIsStaticSwitch=false
       bIsVisible=false
       type {classStructOrEnum {refPath=&Vector2f}}
      }
      Uniform Sprite Size Max {
       bIsDynamic=false
       bIsEditable=false
       bIsStaticSwitch=false
       bIsVisible=false
       type {classStructOrEnum {refPath=&NiagaraFloat}}
      }
      Sprite Size Randomness Mode {
       bIsDynamic=false
       bIsEditable=false
       bIsStaticSwitch=true
       bIsVisible=false
       type {classStructOrEnum {refPath=&ENiagaraRandomnessMode}}
      }
      Sprite Rotation Mode {
       bIsDynamic=false
       bIsEditable=true
       bIsStaticSwitch=true
       bIsVisible=true
       type {classStructOrEnum {refPath=/Niagara/Enums/ENiagara_SpriteRotationMode}}
      }
      Sprite Rotation Angle Min {
       bIsDynamic=false
       bIsEditable=false
       bIsStaticSwitch=false
       bIsVisible=false
       type {classStructOrEnum {refPath=&NiagaraFloat}}
      }
      Sprite Rotation Angle {
       bIsDynamic=false
       bIsEditable=false
       bIsStaticSwitch=false
       bIsVisible=false
       type {classStructOrEnum {refPath=&NiagaraFloat}}
      }
      Sprite Rotation Angle Max {
       bIsDynamic=false
       bIsEditable=false
       bIsStaticSwitch=false
       bIsVisible=false
       type {classStructOrEnum {refPath=&NiagaraFloat}}
      }
      Sprite Rotation Randomness Mode {
       bIsDynamic=false
       bIsEditable=false
       bIsStaticSwitch=true
       bIsVisible=false
       type {classStructOrEnum {refPath=&ENiagaraRandomnessMode}}
      }
      Sprite UV Mode {
       bIsDynamic=false
       bIsEditable=true
       bIsStaticSwitch=true
       bIsVisible=true
       type {classStructOrEnum {refPath=/Niagara/Enums/ENiagara_UVFlippingMode}}
      }
      Sprite UV Randomness Mode {
       bIsDynamic=false
       bIsEditable=false
       bIsStaticSwitch=true
       bIsVisible=false
       type {classStructOrEnum {refPath=&ENiagaraRandomnessMode}}
      }
      Sprite UV Scale {
       bIsDynamic=false
       bIsEditable=false
       bIsStaticSwitch=false
       bIsVisible=false
       type {classStructOrEnum {refPath=&Vector2f}}
      }
      Mesh Scale Mode {
       bIsDynamic=false
       bIsEditable=true
       bIsStaticSwitch=true
       bIsVisible=true
       type {classStructOrEnum {refPath=&ENiagara_SizeScaleMode}}
      }
      Mesh Uniform Scale Min {
       bIsDynamic=false
       bIsEditable=false
       bIsStaticSwitch=false
       bIsVisible=false
       type {classStructOrEnum {refPath=&NiagaraFloat}}
      }
      Mesh Uniform Scale {
       bIsDynamic=false
       bIsEditable=false
       bIsStaticSwitch=false
       bIsVisible=false
       type {classStructOrEnum {refPath=&NiagaraFloat}}
      }
      Mesh Scale Min {
       bIsDynamic=false
       bIsEditable=false
       bIsStaticSwitch=false
       bIsVisible=false
       type {classStructOrEnum {refPath=&Vector3f}}
      }
      Mesh Scale {
       bIsDynamic=false
       bIsEditable=false
       bIsStaticSwitch=false
       bIsVisible=false
       type {classStructOrEnum {refPath=&Vector3f}}
      }
      Mesh Scale Max {
       bIsDynamic=false
       bIsEditable=false
       bIsStaticSwitch=false
       bIsVisible=false
       type {classStructOrEnum {refPath=&Vector3f}}
      }
      Mesh Uniform Scale Max {
       bIsDynamic=false
       bIsEditable=false
       bIsStaticSwitch=false
       bIsVisible=false
       type {classStructOrEnum {refPath=&NiagaraFloat}}
      }
      Mesh Scale Randomness Mode {
       bIsDynamic=false
       bIsEditable=false
       bIsStaticSwitch=true
       bIsVisible=false
       type {classStructOrEnum {refPath=&ENiagaraRandomnessMode}}
      }
      Mesh Renderer Array Visibility Mode {
       bIsDynamic=false
       bIsEditable=true
       bIsStaticSwitch=true
       bIsVisible=true
       type {classStructOrEnum {refPath=/Niagara/Enums/Ribbons/ENiagara_UnsetDirectSetRandom}}
      }
      Mesh Renderer Mesh Index {
       bIsDynamic=false
       bIsEditable=false
       bIsStaticSwitch=false
       bIsVisible=false
       type {classStructOrEnum {refPath=&NiagaraInt32}}
      }
      Wrap Index By Mesh Count {bIsDynamic=false, bIsEditable=false, bIsStaticSwitch=true, bIsVisible=false, type {classStructOrEnum {refPath=&NiagaraBool}}}
      Mesh Renderer Info {
       bIsDynamic=false
       bIsEditable=false
       bIsStaticSwitch=false
       bIsVisible=false
       type {classStructOrEnum {refPath=&NiagaraDataInterfaceMeshRendererInfo}}
      }
      Mesh Renderer Random Seed {
       bIsDynamic=false
       bIsEditable=false
       bIsStaticSwitch=false
       bIsVisible=false
       type {classStructOrEnum {refPath=&NiagaraInt32}}
      }
      Ribbon Width Mode {
       bIsDynamic=false
       bIsEditable=true
       bIsStaticSwitch=true
       bIsVisible=true
       type {classStructOrEnum {refPath=/Niagara/Enums/Ribbons/ENiagara_UnsetDirectSet}}
      }
      Ribbon Width {
       bIsDynamic=false
       bIsEditable=false
       bIsStaticSwitch=false
       bIsVisible=false
       type {classStructOrEnum {refPath=&NiagaraFloat}}
      }
      Ribbon Facing Mode {
       bIsDynamic=false
       bIsEditable=true
       bIsStaticSwitch=true
       bIsVisible=true
       type {classStructOrEnum {refPath=&ENiagara_UnsetDirectSet}}
      }
      Ribbon Facing Vector {
       bIsDynamic=false
       bIsEditable=false
       bIsStaticSwitch=false
       bIsVisible=false
       type {classStructOrEnum {refPath=&Vector3f}}
      }
      Ribbon Twist Mode {
       bIsDynamic=false
       bIsEditable=true
       bIsStaticSwitch=true
       bIsVisible=true
       type {classStructOrEnum {refPath=&ENiagara_UnsetDirectSet}}
      }
      "Ribbon Twist " {
       bIsDynamic=false
       bIsEditable=false
       bIsStaticSwitch=false
       bIsVisible=false
       type {classStructOrEnum {refPath=&NiagaraFloat}}
      }
     ]
     moduleScript {refPath=&InitializeParticle}
    }
    {
     bIsSetParametersModule=false
     enabled=true
     moduleName=AddVelocity
     inputs [
      Velocity Mode {
       bIsDynamic=false
       bIsEditable=true
       bIsStaticSwitch=true
       bIsVisible=true
       type {classStructOrEnum {refPath=/Niagara/Enums/Utility/ENiagara_VelocityMode}}
      }
      Velocity {bIsDynamic=true, bIsEditable=true, bIsStaticSwitch=false, bIsVisible=true, type {classStructOrEnum {refPath=&Vector3f}}}
      Velocity Speed {
       bIsDynamic=false
       bIsEditable=false
       bIsStaticSwitch=false
       bIsVisible=false
       type {classStructOrEnum {refPath=&NiagaraFloat}}
      }
      Velocity Speed Scale {bIsDynamic=false, bIsEditable=true, bIsStaticSwitch=false, bIsVisible=true, type {classStructOrEnum {refPath=&NiagaraFloat}}}
      Constrain To Radius {
       bIsDynamic=false
       bIsEditable=false
       bIsStaticSwitch=false
       bIsVisible=false
       type {classStructOrEnum {refPath=&NiagaraFloat}}
      }
      Distribution Along Cone Axis {
       bIsDynamic=false
       bIsEditable=false
       bIsStaticSwitch=false
       bIsVisible=false
       type {classStructOrEnum {refPath=&NiagaraFloat}}
      }
      Radius Falloff Near / Far {
       bIsDynamic=false
       bIsEditable=false
       bIsStaticSwitch=false
       bIsVisible=false
       type {classStructOrEnum {refPath=&Vector2f}}
      }
      Radius Falloff Exponent {
       bIsDynamic=false
       bIsEditable=false
       bIsStaticSwitch=false
       bIsVisible=false
       type {classStructOrEnum {refPath=&NiagaraFloat}}
      }
      Speed Falloff From Cone Axis {
       bIsDynamic=false
       bIsEditable=false
       bIsStaticSwitch=false
       bIsVisible=false
       type {classStructOrEnum {refPath=&NiagaraFloat}}
      }
      Invert Velocity Falloff {bIsDynamic=false, bIsEditable=false, bIsStaticSwitch=false, bIsVisible=false, type {classStructOrEnum {refPath=&NiagaraBool}}}
      Override Seed {bIsDynamic=false, bIsEditable=false, bIsStaticSwitch=false, bIsVisible=false, type {classStructOrEnum {refPath=&NiagaraBool}}}
      Random Seed {
       bIsDynamic=false
       bIsEditable=false
       bIsStaticSwitch=false
       bIsVisible=false
       type {classStructOrEnum {refPath=&NiagaraInt32}}
      }
      Fixed Random Seed {bIsDynamic=false, bIsEditable=false, bIsStaticSwitch=true, bIsVisible=false, type {classStructOrEnum {refPath=&NiagaraBool}}}
      Randomness Mode {
       bIsDynamic=false
       bIsEditable=false
       bIsStaticSwitch=true
       bIsVisible=false
       type {classStructOrEnum {refPath=&ENiagaraRandomnessMode}}
      }
      Cone Axis {
       bIsDynamic=false
       bIsEditable=false
       bIsStaticSwitch=false
       bIsVisible=false
       type {classStructOrEnum {refPath=&Vector3f}}
      }
      Cone Angle {
       bIsDynamic=false
       bIsEditable=false
       bIsStaticSwitch=false
       bIsVisible=false
       type {classStructOrEnum {refPath=&NiagaraFloat}}
      }
      Inner Cone Angle {
       bIsDynamic=false
       bIsEditable=false
       bIsStaticSwitch=false
       bIsVisible=false
       type {classStructOrEnum {refPath=&NiagaraFloat}}
      }
      Cone Angle Mode {
       bIsDynamic=false
       bIsEditable=false
       bIsStaticSwitch=true
       bIsVisible=false
       type {classStructOrEnum {refPath=/Niagara/Enums/Angles/ENiagara_AngleInput}}
      }
      Rotation Mode {
       bIsDynamic=false
       bIsEditable=true
       bIsStaticSwitch=true
       bIsVisible=true
       type {classStructOrEnum {refPath=/Niagara/Enums/Transforms/ENiagara_RotationMode}}
      }
      Rotation Axis {
       bIsDynamic=false
       bIsEditable=false
       bIsStaticSwitch=false
       bIsVisible=false
       type {classStructOrEnum {refPath=&Vector3f}}
      }
      Yaw / Pitch / Roll {
       bIsDynamic=false
       bIsEditable=false
       bIsStaticSwitch=false
       bIsVisible=false
       type {classStructOrEnum {refPath=&Vector3f}}
      }
      Rotation Quaternion {bIsDynamic=false, bIsEditable=false, bIsStaticSwitch=false, bIsVisible=false, type {classStructOrEnum {refPath=/Script/CoreUObject.Quat4f}}}
      Rotation Matrix {
       bIsDynamic=false
       bIsEditable=false
       bIsStaticSwitch=false
       bIsVisible=false
       type {classStructOrEnum {refPath=/Script/Niagara.NiagaraMatrix}}
      }
      Rotation Angle {
       bIsDynamic=false
       bIsEditable=false
       bIsStaticSwitch=false
       bIsVisible=false
       type {classStructOrEnum {refPath=&NiagaraFloat}}
      }
      Invert Rotation Quaternion {bIsDynamic=false, bIsEditable=false, bIsStaticSwitch=false, bIsVisible=false, type {classStructOrEnum {refPath=&NiagaraBool}}}
      Additional Quaternion Rotation {bIsDynamic=false, bIsEditable=false, bIsStaticSwitch=false, bIsVisible=false, type {classStructOrEnum {refPath=&NiagaraBool}}}
      Additional Yaw / Pitch / Roll {
       bIsDynamic=false
       bIsEditable=false
       bIsStaticSwitch=false
       bIsVisible=false
       type {classStructOrEnum {refPath=&Vector3f}}
      }
      Rotation Angle Type {
       bIsDynamic=false
       bIsEditable=false
       bIsStaticSwitch=true
       bIsVisible=false
       type {classStructOrEnum {refPath=&ENiagara_AngleInput}}
      }
      Rotation Coordinate Space {
       bIsDynamic=false
       bIsEditable=true
       bIsStaticSwitch=false
       bIsVisible=true
       type {classStructOrEnum {refPath=&ENiagaraCoordinateSpace}}
      }
      Velocity Origin {
       bIsDynamic=false
       bIsEditable=false
       bIsStaticSwitch=false
       bIsVisible=false
       type {classStructOrEnum {refPath=&NiagaraPosition}}
      }
      Origin Offset {
       bIsDynamic=false
       bIsEditable=false
       bIsStaticSwitch=false
       bIsVisible=false
       type {classStructOrEnum {refPath=&Vector3f}}
      }
      Origin Offset Coordinate Space {
       bIsDynamic=false
       bIsEditable=false
       bIsStaticSwitch=false
       bIsVisible=false
       type {classStructOrEnum {refPath=&ENiagaraCoordinateSpace}}
      }
      Default Position {
       bIsDynamic=false
       bIsEditable=false
       bIsStaticSwitch=false
       bIsVisible=false
       type {classStructOrEnum {refPath=&NiagaraPosition}}
      }
     ]
     moduleScript {refPath=&AddVelocity}
    }
   ]
  }
  particleUpdateScript {
   scriptName=ParticleUpdateScript
   modules [
    {
     bIsSetParametersModule=false
     enabled=true
     moduleName=SolveForcesAndVelocity
     inputs [
      Clamp Velocity {bIsDynamic=false, bIsEditable=true, bIsStaticSwitch=false, bIsVisible=true, type {classStructOrEnum {refPath=&NiagaraBool}}}
      Speed Limit {bIsDynamic=false, bIsEditable=false, bIsStaticSwitch=false, bIsVisible=true, type {classStructOrEnum {refPath=&NiagaraFloat}}}
      Limit Acceleration {bIsDynamic=false, bIsEditable=true, bIsStaticSwitch=false, bIsVisible=true, type {classStructOrEnum {refPath=&NiagaraBool}}}
      Acceleration Limit {bIsDynamic=false, bIsEditable=false, bIsStaticSwitch=false, bIsVisible=true, type {classStructOrEnum {refPath=&NiagaraFloat}}}
      Manually Enable Rotational Solver {bIsDynamic=false, bIsEditable=true, bIsStaticSwitch=false, bIsVisible=true, type {classStructOrEnum {refPath=&NiagaraBool}}}
      Write to Intrinsic Properties {bIsDynamic=false, bIsEditable=true, bIsStaticSwitch=false, bIsVisible=true, type {classStructOrEnum {refPath=&NiagaraBool}}}
      Write to Presolve Properties {bIsDynamic=false, bIsEditable=true, bIsStaticSwitch=true, bIsVisible=true, type {classStructOrEnum {refPath=&NiagaraBool}}}
      DeltaTime {bIsDynamic=false, bIsEditable=true, bIsStaticSwitch=false, bIsVisible=true, type {classStructOrEnum {refPath=&NiagaraFloat}}}
      Mass {bIsDynamic=false, bIsEditable=true, bIsStaticSwitch=false, bIsVisible=true, type {classStructOrEnum {refPath=&NiagaraFloat}}}
      Position {
       bIsDynamic=false
       bIsEditable=true
       bIsStaticSwitch=false
       bIsVisible=true
       type {classStructOrEnum {refPath=&NiagaraPosition}}
      }
      Velocity {bIsDynamic=false, bIsEditable=true, bIsStaticSwitch=false, bIsVisible=true, type {classStructOrEnum {refPath=&Vector3f}}}
      PreviousVelocity {bIsDynamic=false, bIsEditable=true, bIsStaticSwitch=false, bIsVisible=true, type {classStructOrEnum {refPath=&Vector3f}}}
      Force {bIsDynamic=false, bIsEditable=true, bIsStaticSwitch=false, bIsVisible=true, type {classStructOrEnum {refPath=&Vector3f}}}
      Mesh Orientation {bIsDynamic=false, bIsEditable=false, bIsStaticSwitch=false, bIsVisible=false, type {classStructOrEnum {refPath=&Quat4f}}}
      Rotational Velocity {
       bIsDynamic=false
       bIsEditable=false
       bIsStaticSwitch=false
       bIsVisible=false
       type {classStructOrEnum {refPath=&Vector3f}}
      }
      Rotational Force {
       bIsDynamic=false
       bIsEditable=false
       bIsStaticSwitch=false
       bIsVisible=false
       type {classStructOrEnum {refPath=&Vector3f}}
      }
      Rotational Inertia {
       bIsDynamic=false
       bIsEditable=false
       bIsStaticSwitch=false
       bIsVisible=false
       type {classStructOrEnum {refPath=&NiagaraFloat}}
      }
     ]
     moduleScript {refPath=&SolveForcesAndVelocity}
    }
    {
     bIsSetParametersModule=false
     enabled=true
     moduleName=ParticleState
     inputs [
      Kill Particles When Lifetime Has Elapsed {bIsDynamic=false, bIsEditable=true, bIsStaticSwitch=true, bIsVisible=true, type {classStructOrEnum {refPath=&NiagaraBool}}}
      Loop Particles Lifetime {bIsDynamic=false, bIsEditable=false, bIsStaticSwitch=false, bIsVisible=false, type {classStructOrEnum {refPath=&NiagaraBool}}}
      Let Infinitely Lived Particles Die When Emitter Deactivates {bIsDynamic=false, bIsEditable=false, bIsStaticSwitch=false, bIsVisible=false, type {classStructOrEnum {refPath=&NiagaraBool}}}
      Lifetime {bIsDynamic=false, bIsEditable=true, bIsStaticSwitch=false, bIsVisible=true, type {classStructOrEnum {refPath=&NiagaraFloat}}}
      DeltaTime {bIsDynamic=false, bIsEditable=true, bIsStaticSwitch=false, bIsVisible=true, type {classStructOrEnum {refPath=&NiagaraFloat}}}
     ]
     moduleScript {refPath=&ParticleState}
    }
    {
     bIsSetParametersModule=false
     enabled=true
     moduleName=SkeletalMeshLocation
     inputs [
      Skeletal Mesh {
       bIsDynamic=false
       bIsEditable=true
       bIsStaticSwitch=false
       bIsVisible=true
       type {classStructOrEnum {refPath=&NiagaraDataInterfaceSkeletalMesh}}
      }
      Mesh Sampling Type {
       bIsDynamic=false
       bIsEditable=true
       bIsStaticSwitch=true
       bIsVisible=true
       type {classStructOrEnum {refPath=/Niagara/Enums/ENiagaraSkelSamplingModeFull}}
      }
      Bone / Socket Sampling Mode {
       bIsDynamic=false
       bIsEditable=false
       bIsStaticSwitch=true
       bIsVisible=false
       type {classStructOrEnum {refPath=/Niagara/Enums/ENiagaraBoneSocketSamplingMode}}
      }
      Bone Sampling Mode {
       bIsDynamic=false
       bIsEditable=true
       bIsStaticSwitch=true
       bIsVisible=true
       type {classStructOrEnum {refPath=/Niagara/Enums/ENiagaraBoneSamplingMode}}
      }
      Socket Sampling Mode {
       bIsDynamic=false
       bIsEditable=false
       bIsStaticSwitch=true
       bIsVisible=false
       type {classStructOrEnum {refPath=/Niagara/Enums/ENiagaraSocketSamplingMode}}
      }
      Triangle Sampling Mode {
       bIsDynamic=false
       bIsEditable=false
       bIsStaticSwitch=true
       bIsVisible=false
       type {classStructOrEnum {refPath=/Niagara/Enums/ENiagaraTriangleSamplingMode}}
      }
      Vertex Sampling Mode {
       bIsDynamic=false
       bIsEditable=false
       bIsStaticSwitch=true
       bIsVisible=false
       type {classStructOrEnum {refPath=/Niagara/Enums/ENiagaraVertexSamplingMode}}
      }
      Bone / Socket Index {bIsDynamic=false, bIsEditable=true, bIsStaticSwitch=false, bIsVisible=true, type {classStructOrEnum {refPath=&NiagaraInt32}}}
      Triangle ID {
       bIsDynamic=false
       bIsEditable=false
       bIsStaticSwitch=false
       bIsVisible=false
       type {classStructOrEnum {refPath=&NiagaraInt32}}
      }
      Vertex ID {
       bIsDynamic=false
       bIsEditable=false
       bIsStaticSwitch=false
       bIsVisible=false
       type {classStructOrEnum {refPath=&NiagaraInt32}}
      }
      FixedBarycentric {bIsDynamic=false, bIsEditable=false, bIsStaticSwitch=false, bIsVisible=false, type {classStructOrEnum {refPath=&NiagaraBool}}}
      Triangle Coordinate {
       bIsDynamic=false
       bIsEditable=false
       bIsStaticSwitch=false
       bIsVisible=false
       type {classStructOrEnum {refPath=&Vector3f}}
      }
      Sampled UV Index {
       bIsDynamic=false
       bIsEditable=false
       bIsStaticSwitch=false
       bIsVisible=false
       type {classStructOrEnum {refPath=&NiagaraInt32}}
      }
      Sampled Position Coordinate Space {
       bIsDynamic=false
       bIsEditable=true
       bIsStaticSwitch=true
       bIsVisible=true
       type {classStructOrEnum {refPath=&ENiagaraCoordinateSpace}}
      }
      Sampled Velocity Coordinate Space {
       bIsDynamic=false
       bIsEditable=true
       bIsStaticSwitch=true
       bIsVisible=true
       type {classStructOrEnum {refPath=&ENiagaraCoordinateSpace}}
      }
      Sampled Orientation Coordinate Space {
       bIsDynamic=false
       bIsEditable=true
       bIsStaticSwitch=true
       bIsVisible=true
       type {classStructOrEnum {refPath=&ENiagaraCoordinateSpace}}
      }
      Generate New Random Samples Every Frame {bIsDynamic=false, bIsEditable=false, bIsStaticSwitch=true, bIsVisible=false, type {classStructOrEnum {refPath=&NiagaraBool}}}
      Write Sampled Mesh Attrributes to Particle Attributes {bIsDynamic=false, bIsEditable=true, bIsStaticSwitch=true, bIsVisible=true, type {classStructOrEnum {refPath=&NiagaraBool}}}
      Randomness Mode {
       bIsDynamic=false
       bIsEditable=false
       bIsStaticSwitch=false
       bIsVisible=false
       type {classStructOrEnum {refPath=&ENiagaraRandomnessMode}}
      }
      Random Seed {
       bIsDynamic=false
       bIsEditable=false
       bIsStaticSwitch=false
       bIsVisible=false
       type {classStructOrEnum {refPath=&NiagaraInt32}}
      }
      Fixed Random Seed {bIsDynamic=false, bIsEditable=false, bIsStaticSwitch=false, bIsVisible=false, type {classStructOrEnum {refPath=&NiagaraBool}}}
      Apply To Particle Position {bIsDynamic=false, bIsEditable=true, bIsStaticSwitch=false, bIsVisible=true, type {classStructOrEnum {refPath=&NiagaraBool}}}
      Sampled Position Offset {bIsDynamic=false, bIsEditable=true, bIsStaticSwitch=false, bIsVisible=true, type {classStructOrEnum {refPath=&Vector3f}}}
      OffsetAlongNormal {bIsDynamic=false, bIsEditable=false, bIsStaticSwitch=false, bIsVisible=false, type {classStructOrEnum {refPath=&NiagaraBool}}}
      Offset Position Along Sampled Normal {
       bIsDynamic=false
       bIsEditable=false
       bIsStaticSwitch=false
       bIsVisible=false
       type {classStructOrEnum {refPath=&NiagaraFloat}}
      }
      Sampled Position Offset Coordinate Space {
       bIsDynamic=false
       bIsEditable=true
       bIsStaticSwitch=true
       bIsVisible=true
       type {classStructOrEnum {refPath=/Niagara/Enums/ENiagaraMeshTransforms}}
      }
      Inherit Velocity {bIsDynamic=false, bIsEditable=true, bIsStaticSwitch=false, bIsVisible=true, type {classStructOrEnum {refPath=&NiagaraBool}}}
      "Inherit Velocity (Scale)" {bIsDynamic=false, bIsEditable=false, bIsStaticSwitch=false, bIsVisible=true, type {classStructOrEnum {refPath=&NiagaraFloat}}}
      VelocitySpeedLimit {bIsDynamic=false, bIsEditable=false, bIsStaticSwitch=false, bIsVisible=false, type {classStructOrEnum {refPath=&NiagaraBool}}}
      Inherited Velocity Speed Limit {
       bIsDynamic=false
       bIsEditable=false
       bIsStaticSwitch=false
       bIsVisible=false
       type {classStructOrEnum {refPath=&NiagaraFloat}}
      }
      Apply Sampled Orientation to Mesh Particles {bIsDynamic=false, bIsEditable=true, bIsStaticSwitch=true, bIsVisible=true, type {classStructOrEnum {refPath=&NiagaraBool}}}
      Interpolate Mesh Sampling {bIsDynamic=false, bIsEditable=true, bIsStaticSwitch=true, bIsVisible=true, type {classStructOrEnum {refPath=&NiagaraBool}}}
      MaskSpawnGroup {bIsDynamic=false, bIsEditable=true, bIsStaticSwitch=false, bIsVisible=true, type {classStructOrEnum {refPath=&NiagaraBool}}}
      Spawn Group Mask {bIsDynamic=false, bIsEditable=false, bIsStaticSwitch=false, bIsVisible=true, type {classStructOrEnum {refPath=&NiagaraInt32}}}
      Use Sampled Mesh Attribute as Ribbon ID {bIsDynamic=false, bIsEditable=true, bIsStaticSwitch=false, bIsVisible=true, type {classStructOrEnum {refPath=&NiagaraBool}}}
     ]
     moduleScript {refPath=&SkeletalMeshLocation}
    }
    {
     bIsSetParametersModule=false
     enabled=true
     moduleName=CameraOffset
     inputs [Camera Offset Amount {bIsDynamic=false, bIsEditable=true, bIsStaticSwitch=false, bIsVisible=true, type {classStructOrEnum {refPath=&NiagaraFloat}}}]
     moduleScript {refPath=&CameraOffset}
    }
    {
     bIsSetParametersModule=false
     enabled=true
     moduleName=ScaleColor
     inputs [
      Scale Mode {
       bIsDynamic=false
       bIsEditable=true
       bIsStaticSwitch=true
       bIsVisible=true
       type {classStructOrEnum {refPath=/Niagara/Enums/ENiagaraScaleColorMode}}
      }
      Linear Color Curve {
       bIsDynamic=false
       bIsEditable=true
       bIsStaticSwitch=false
       bIsVisible=true
       type {classStructOrEnum {refPath=&NiagaraDataInterfaceColorCurve}}
      }
      ScaleRGB {bIsDynamic=false, bIsEditable=false, bIsStaticSwitch=false, bIsVisible=false, type {classStructOrEnum {refPath=&NiagaraBool}}}
      Scale RGB {
       bIsDynamic=false
       bIsEditable=false
       bIsStaticSwitch=false
       bIsVisible=false
       type {classStructOrEnum {refPath=&Vector3f}}
      }
      ScaleRGBA {bIsDynamic=false, bIsEditable=false, bIsStaticSwitch=false, bIsVisible=false, type {classStructOrEnum {refPath=&NiagaraBool}}}
      Scale RGBA {
       bIsDynamic=false
       bIsEditable=false
       bIsStaticSwitch=false
       bIsVisible=false
       type {classStructOrEnum {refPath=/Script/CoreUObject.Vector4f}}
      }
      Curve Index {bIsDynamic=false, bIsEditable=true, bIsStaticSwitch=false, bIsVisible=true, type {classStructOrEnum {refPath=&NiagaraFloat}}}
      ScaleA {bIsDynamic=false, bIsEditable=false, bIsStaticSwitch=false, bIsVisible=false, type {classStructOrEnum {refPath=&NiagaraBool}}}
      Scale Alpha {
       bIsDynamic=false
       bIsEditable=false
       bIsStaticSwitch=false
       bIsVisible=false
       type {classStructOrEnum {refPath=&NiagaraFloat}}
      }
      Color Value To Scale {
       bIsDynamic=false
       bIsEditable=true
       bIsStaticSwitch=false
       bIsVisible=true
       type {classStructOrEnum {refPath=&LinearColor}}
      }
     ]
     moduleScript {refPath=&ScaleColor}
    }
    {
     bIsSetParametersModule=false
     enabled=true
     moduleName=FN_Light_v2
     inputs [
      LightEnabled {bIsDynamic=false, bIsEditable=true, bIsStaticSwitch=false, bIsVisible=true, type {classStructOrEnum {refPath=&NiagaraBool}}}
      Scale Color {bIsDynamic=false, bIsEditable=true, bIsStaticSwitch=false, bIsVisible=true, type {classStructOrEnum {refPath=&NiagaraFloat}}}
      Light Exponent {bIsDynamic=false, bIsEditable=true, bIsStaticSwitch=false, bIsVisible=true, type {classStructOrEnum {refPath=&NiagaraFloat}}}
      Position Offset {bIsDynamic=false, bIsEditable=true, bIsStaticSwitch=false, bIsVisible=true, type {classStructOrEnum {refPath=&Vector3f}}}
      Light Radius {bIsDynamic=false, bIsEditable=true, bIsStaticSwitch=false, bIsVisible=true, type {classStructOrEnum {refPath=&NiagaraFloat}}}
     ]
     moduleScript {refPath=&FN_Light_v2}
    }
    {
     bIsSetParametersModule=false
     enabled=true
     moduleName=ScaleColor001
     inputs [
      Scale Mode {
       bIsDynamic=false
       bIsEditable=true
       bIsStaticSwitch=true
       bIsVisible=true
       type {classStructOrEnum {refPath=&ENiagaraScaleColorMode}}
      }
      Linear Color Curve {
       bIsDynamic=false
       bIsEditable=false
       bIsStaticSwitch=false
       bIsVisible=false
       type {classStructOrEnum {refPath=&NiagaraDataInterfaceColorCurve}}
      }
      ScaleRGB {bIsDynamic=false, bIsEditable=true, bIsStaticSwitch=false, bIsVisible=true, type {classStructOrEnum {refPath=&NiagaraBool}}}
      Scale RGB {bIsDynamic=false, bIsEditable=true, bIsStaticSwitch=false, bIsVisible=true, type {classStructOrEnum {refPath=&Vector3f}}}
      ScaleRGBA {bIsDynamic=false, bIsEditable=false, bIsStaticSwitch=false, bIsVisible=false, type {classStructOrEnum {refPath=&NiagaraBool}}}
      Scale RGBA {
       bIsDynamic=false
       bIsEditable=false
       bIsStaticSwitch=false
       bIsVisible=false
       type {classStructOrEnum {refPath=&Vector4f}}
      }
      Curve Index {
       bIsDynamic=false
       bIsEditable=false
       bIsStaticSwitch=false
       bIsVisible=false
       type {classStructOrEnum {refPath=&NiagaraFloat}}
      }
      ScaleA {bIsDynamic=false, bIsEditable=true, bIsStaticSwitch=false, bIsVisible=true, type {classStructOrEnum {refPath=&NiagaraBool}}}
      Scale Alpha {bIsDynamic=false, bIsEditable=true, bIsStaticSwitch=false, bIsVisible=true, type {classStructOrEnum {refPath=&NiagaraFloat}}}
      Color Value To Scale {
       bIsDynamic=false
       bIsEditable=true
       bIsStaticSwitch=false
       bIsVisible=true
       type {classStructOrEnum {refPath=&LinearColor}}
      }
     ]
     moduleScript {refPath=&ScaleColor}
    }
   ]
  }
  rendererClasses [{refPath=&NiagaraSpriteRendererProperties}]
  renderers [{rendererIndex=0, rendererClass {refPath=&NiagaraSpriteRendererProperties}}]
 }
 HandGlow {
  bEnabled=false
  emitterSpawnScript {scriptName=EmitterSpawnScript, modules []}
  emitterUpdateScript {
   scriptName=EmitterUpdateScript
   modules [
    {
     bIsSetParametersModule=false
     enabled=true
     moduleName=EmitterState
     inputs [
      Life Cycle Mode {
       bIsDynamic=false
       bIsEditable=true
       bIsStaticSwitch=true
       bIsVisible=true
       type {classStructOrEnum {refPath=&ENiagaraEmitterLifeCycleMode}}
      }
      Inactive Response {
       bIsDynamic=false
       bIsEditable=true
       bIsStaticSwitch=true
       bIsVisible=true
       type {classStructOrEnum {refPath=&ENiagaraInactiveMode}}
      }
      Loop Behavior {
       bIsDynamic=false
       bIsEditable=true
       bIsStaticSwitch=true
       bIsVisible=true
       type {classStructOrEnum {refPath=&ENiagara_EmitterStateOptions}}
      }
      Loop Duration Mode {
       bIsDynamic=false
       bIsEditable=true
       bIsStaticSwitch=true
       bIsVisible=true
       type {classStructOrEnum {refPath=&ENiagara_InfiniteLoopDuration}}
      }
      Loop Count {
       bIsDynamic=false
       bIsEditable=false
       bIsStaticSwitch=false
       bIsVisible=false
       type {classStructOrEnum {refPath=&NiagaraInt32}}
      }
      Loop Duration {
       bIsDynamic=false
       bIsEditable=false
       bIsStaticSwitch=false
       bIsVisible=false
       type {classStructOrEnum {refPath=&NiagaraFloat}}
      }
      Recalculate Duration Each Loop {bIsDynamic=false, bIsEditable=false, bIsStaticSwitch=false, bIsVisible=false, type {classStructOrEnum {refPath=&NiagaraBool}}}
      UseLoopDelay {bIsDynamic=false, bIsEditable=true, bIsStaticSwitch=false, bIsVisible=true, type {classStructOrEnum {refPath=&NiagaraBool}}}
      Loop Delay {bIsDynamic=false, bIsEditable=false, bIsStaticSwitch=false, bIsVisible=true, type {classStructOrEnum {refPath=&NiagaraFloat}}}
      Delay First Loop Only {bIsDynamic=false, bIsEditable=false, bIsStaticSwitch=false, bIsVisible=false, type {classStructOrEnum {refPath=&NiagaraBool}}}
      Scalability Mode {
       bIsDynamic=false
       bIsEditable=true
       bIsStaticSwitch=true
       bIsVisible=true
       type {classStructOrEnum {refPath=&ENiagaraEmitterScalabilityMode_Limited}}
      }
      Enable Distance Culling {bIsDynamic=false, bIsEditable=false, bIsStaticSwitch=true, bIsVisible=false, type {classStructOrEnum {refPath=&NiagaraBool}}}
      MinDistance {
       bIsDynamic=false
       bIsEditable=false
       bIsStaticSwitch=false
       bIsVisible=false
       type {classStructOrEnum {refPath=&NiagaraFloat}}
      }
      Min Distance Response {
       bIsDynamic=false
       bIsEditable=false
       bIsStaticSwitch=false
       bIsVisible=false
       type {classStructOrEnum {refPath=&ENiagaraExecutionStateManagement}}
      }
      MaxDistance {
       bIsDynamic=false
       bIsEditable=false
       bIsStaticSwitch=false
       bIsVisible=false
       type {classStructOrEnum {refPath=&NiagaraFloat}}
      }
      Max Distance Response {
       bIsDynamic=false
       bIsEditable=false
       bIsStaticSwitch=false
       bIsVisible=false
       type {classStructOrEnum {refPath=&ENiagaraExecutionStateManagement}}
      }
      Scale Spawn Count {bIsDynamic=false, bIsEditable=false, bIsStaticSwitch=true, bIsVisible=false, type {classStructOrEnum {refPath=&NiagaraBool}}}
      Spawn Count Scale {
       bIsDynamic=false
       bIsEditable=false
       bIsStaticSwitch=false
       bIsVisible=false
       type {classStructOrEnum {refPath=&NiagaraFloat}}
      }
      Spawn Count Scale By Distance {
       bIsDynamic=false
       bIsEditable=false
       bIsStaticSwitch=false
       bIsVisible=false
       type {classStructOrEnum {refPath=&NiagaraDataInterfaceCurve}}
      }
      Enable Visibility Culling {bIsDynamic=false, bIsEditable=false, bIsStaticSwitch=true, bIsVisible=false, type {classStructOrEnum {refPath=&NiagaraBool}}}
      Visibility Cull Response {
       bIsDynamic=false
       bIsEditable=false
       bIsStaticSwitch=false
       bIsVisible=false
       type {classStructOrEnum {refPath=&ENiagaraExecutionStateManagement}}
      }
      Visibility Cull Delay {
       bIsDynamic=false
       bIsEditable=false
       bIsStaticSwitch=false
       bIsVisible=false
       type {classStructOrEnum {refPath=&NiagaraFloat}}
      }
      Reset Age On Awaken {bIsDynamic=false, bIsEditable=false, bIsStaticSwitch=false, bIsVisible=false, type {classStructOrEnum {refPath=&NiagaraBool}}}
     ]
     moduleScript {refPath=&EmitterState}
    }
    {
     bIsSetParametersModule=false
     enabled=true
     moduleName=SpawnBurst_Instantaneous
     inputs [
      Spawn Count {bIsDynamic=false, bIsEditable=true, bIsStaticSwitch=false, bIsVisible=true, type {classStructOrEnum {refPath=&NiagaraInt32}}}
      Spawn Time {bIsDynamic=false, bIsEditable=true, bIsStaticSwitch=false, bIsVisible=true, type {classStructOrEnum {refPath=&NiagaraFloat}}}
      Use Spawn Probability {bIsDynamic=false, bIsEditable=true, bIsStaticSwitch=false, bIsVisible=true, type {classStructOrEnum {refPath=&NiagaraBool}}}
      Spawn Probability {bIsDynamic=false, bIsEditable=true, bIsStaticSwitch=false, bIsVisible=true, type {classStructOrEnum {refPath=&NiagaraFloat}}}
      Spawn Group {bIsDynamic=false, bIsEditable=true, bIsStaticSwitch=false, bIsVisible=true, type {classStructOrEnum {refPath=&NiagaraInt32}}}
      Age {bIsDynamic=false, bIsEditable=true, bIsStaticSwitch=false, bIsVisible=true, type {classStructOrEnum {refPath=&NiagaraFloat}}}
     ]
     moduleScript {refPath=&SpawnBurst_Instantaneous}
    }
   ]
  }
  particleSpawnScript {
   scriptName=ParticleSpawnScript
   modules [
    {
     bIsSetParametersModule=false
     enabled=true
     moduleName=FN_SetSpawnGroup
     inputs [
      Number of Spawn Groups {bIsDynamic=false, bIsEditable=true, bIsStaticSwitch=false, bIsVisible=true, type {classStructOrEnum {refPath=&NiagaraInt32}}}
      Particles Per Group {
       bIsDynamic=false
       bIsEditable=false
       bIsStaticSwitch=false
       bIsVisible=false
       type {classStructOrEnum {refPath=&NiagaraInt32}}
      }
      Pattern {
       bIsDynamic=false
       bIsEditable=true
       bIsStaticSwitch=true
       bIsVisible=true
       type {classStructOrEnum {refPath=&ENiagaraPartitionWaveForm}}
      }
      Normalize Spawn Group {bIsDynamic=false, bIsEditable=true, bIsStaticSwitch=true, bIsVisible=true, type {classStructOrEnum {refPath=&NiagaraBool}}}
      Inclusivity {bIsDynamic=false, bIsEditable=false, bIsStaticSwitch=true, bIsVisible=false, type {classStructOrEnum {refPath=&NiagaraBool}}}
      Set Ribbon ID and Ribbon Link Order {bIsDynamic=false, bIsEditable=true, bIsStaticSwitch=true, bIsVisible=true, type {classStructOrEnum {refPath=&NiagaraBool}}}
     ]
     moduleScript {refPath=&FN_SetSpawnGroup}
    }
    {
     bIsSetParametersModule=false
     enabled=true
     moduleName=InitializeParticle001
     inputs [
      Lifetime Mode {
       bIsDynamic=false
       bIsEditable=true
       bIsStaticSwitch=true
       bIsVisible=true
       type {classStructOrEnum {refPath=&ENiagara_LifetimeMode}}
      }
      Lifetime Min {
       bIsDynamic=false
       bIsEditable=false
       bIsStaticSwitch=false
       bIsVisible=false
       type {classStructOrEnum {refPath=&NiagaraFloat}}
      }
      Lifetime {bIsDynamic=false, bIsEditable=true, bIsStaticSwitch=false, bIsVisible=true, type {classStructOrEnum {refPath=&NiagaraFloat}}}
      Lifetime Max {
       bIsDynamic=false
       bIsEditable=false
       bIsStaticSwitch=false
       bIsVisible=false
       type {classStructOrEnum {refPath=&NiagaraFloat}}
      }
      Lifetime Randomness Mode {
       bIsDynamic=false
       bIsEditable=false
       bIsStaticSwitch=true
       bIsVisible=false
       type {classStructOrEnum {refPath=&ENiagaraRandomnessMode}}
      }
      Color Mode {
       bIsDynamic=false
       bIsEditable=true
       bIsStaticSwitch=true
       bIsVisible=true
       type {classStructOrEnum {refPath=&ENiagara_ColorInitializationMode}}
      }
      Color Minimum {
       bIsDynamic=false
       bIsEditable=false
       bIsStaticSwitch=false
       bIsVisible=false
       type {classStructOrEnum {refPath=&LinearColor}}
      }
      Color {
       bIsDynamic=false
       bIsEditable=true
       bIsStaticSwitch=false
       bIsVisible=true
       type {classStructOrEnum {refPath=&LinearColor}}
      }
      AdjustHue {bIsDynamic=false, bIsEditable=false, bIsStaticSwitch=false, bIsVisible=false, type {classStructOrEnum {refPath=&NiagaraBool}}}
      Hue Shift Range {
       bIsDynamic=false
       bIsEditable=false
       bIsStaticSwitch=false
       bIsVisible=false
       type {classStructOrEnum {refPath=&Vector2f}}
      }
      Color Maximum {
       bIsDynamic=false
       bIsEditable=false
       bIsStaticSwitch=false
       bIsVisible=false
       type {classStructOrEnum {refPath=&LinearColor}}
      }
      AdjustSaturation {bIsDynamic=false, bIsEditable=false, bIsStaticSwitch=false, bIsVisible=false, type {classStructOrEnum {refPath=&NiagaraBool}}}
      Saturation Range {
       bIsDynamic=false
       bIsEditable=false
       bIsStaticSwitch=false
       bIsVisible=false
       type {classStructOrEnum {refPath=&Vector2f}}
      }
      AdjustValue {bIsDynamic=false, bIsEditable=false, bIsStaticSwitch=false, bIsVisible=false, type {classStructOrEnum {refPath=&NiagaraBool}}}
      Value Range {
       bIsDynamic=false
       bIsEditable=false
       bIsStaticSwitch=false
       bIsVisible=false
       type {classStructOrEnum {refPath=&Vector2f}}
      }
      AdjustAlpha {bIsDynamic=false, bIsEditable=false, bIsStaticSwitch=false, bIsVisible=false, type {classStructOrEnum {refPath=&NiagaraBool}}}
      Alpha Scale Range {
       bIsDynamic=false
       bIsEditable=false
       bIsStaticSwitch=false
       bIsVisible=false
       type {classStructOrEnum {refPath=&Vector2f}}
      }
      Color Channel Mode {
       bIsDynamic=false
       bIsEditable=false
       bIsStaticSwitch=true
       bIsVisible=false
       type {classStructOrEnum {refPath=&ENiagaraChannelCorrelation}}
      }
      Color Randomness Mode {
       bIsDynamic=false
       bIsEditable=false
       bIsStaticSwitch=true
       bIsVisible=false
       type {classStructOrEnum {refPath=&ENiagaraRandomnessMode}}
      }
      Position Mode {
       bIsDynamic=false
       bIsEditable=true
       bIsStaticSwitch=true
       bIsVisible=true
       type {classStructOrEnum {refPath=&ENiagara_PositionInitializationMode}}
      }
      Position {
       bIsDynamic=true
       bIsEditable=false
       bIsStaticSwitch=false
       bIsVisible=false
       type {classStructOrEnum {refPath=&NiagaraPosition}}
      }
      UsePositionOffset {bIsDynamic=false, bIsEditable=true, bIsStaticSwitch=false, bIsVisible=true, type {classStructOrEnum {refPath=&NiagaraBool}}}
      Position Offset {bIsDynamic=false, bIsEditable=false, bIsStaticSwitch=false, bIsVisible=true, type {classStructOrEnum {refPath=&Vector3f}}}
      Position Offset Coordinate Space {
       bIsDynamic=false
       bIsEditable=false
       bIsStaticSwitch=false
       bIsVisible=false
       type {classStructOrEnum {refPath=&ENiagaraCoordinateSpace}}
      }
      Mass Mode {
       bIsDynamic=false
       bIsEditable=true
       bIsStaticSwitch=true
       bIsVisible=true
       type {classStructOrEnum {refPath=&ENiagara_MassInitializationMode}}
      }
      Mass Min {
       bIsDynamic=false
       bIsEditable=false
       bIsStaticSwitch=false
       bIsVisible=false
       type {classStructOrEnum {refPath=&NiagaraFloat}}
      }
      Mass {bIsDynamic=false, bIsEditable=true, bIsStaticSwitch=false, bIsVisible=true, type {classStructOrEnum {refPath=&NiagaraFloat}}}
      Mass Max {
       bIsDynamic=false
       bIsEditable=false
       bIsStaticSwitch=false
       bIsVisible=false
       type {classStructOrEnum {refPath=&NiagaraFloat}}
      }
      Mass Randomness Mode {
       bIsDynamic=false
       bIsEditable=false
       bIsStaticSwitch=true
       bIsVisible=false
       type {classStructOrEnum {refPath=&ENiagaraRandomnessMode}}
      }
      Write Material Random {bIsDynamic=false, bIsEditable=true, bIsStaticSwitch=false, bIsVisible=true, type {classStructOrEnum {refPath=&NiagaraBool}}}
      Material Random {bIsDynamic=false, bIsEditable=true, bIsStaticSwitch=false, bIsVisible=true, type {classStructOrEnum {refPath=&NiagaraFloat}}}
      Sprite Size Mode {
       bIsDynamic=false
       bIsEditable=true
       bIsStaticSwitch=true
       bIsVisible=true
       type {classStructOrEnum {refPath=&ENiagara_SizeScaleMode}}
      }
      Sprite Size {
       bIsDynamic=false
       bIsEditable=false
       bIsStaticSwitch=false
       bIsVisible=false
       type {classStructOrEnum {refPath=&Vector2f}}
      }
      Uniform Sprite Size {bIsDynamic=false, bIsEditable=true, bIsStaticSwitch=false, bIsVisible=true, type {classStructOrEnum {refPath=&NiagaraFloat}}}
      Uniform Sprite Size Min {
       bIsDynamic=false
       bIsEditable=false
       bIsStaticSwitch=false
       bIsVisible=false
       type {classStructOrEnum {refPath=&NiagaraFloat}}
      }
      Sprite Size Min {
       bIsDynamic=false
       bIsEditable=false
       bIsStaticSwitch=false
       bIsVisible=false
       type {classStructOrEnum {refPath=&Vector2f}}
      }
      Sprite Size Max {
       bIsDynamic=false
       bIsEditable=false
       bIsStaticSwitch=false
       bIsVisible=false
       type {classStructOrEnum {refPath=&Vector2f}}
      }
      Uniform Sprite Size Max {
       bIsDynamic=false
       bIsEditable=false
       bIsStaticSwitch=false
       bIsVisible=false
       type {classStructOrEnum {refPath=&NiagaraFloat}}
      }
      Sprite Size Randomness Mode {
       bIsDynamic=false
       bIsEditable=false
       bIsStaticSwitch=true
       bIsVisible=false
       type {classStructOrEnum {refPath=&ENiagaraRandomnessMode}}
      }
      Sprite Rotation Mode {
       bIsDynamic=false
       bIsEditable=true
       bIsStaticSwitch=true
       bIsVisible=true
       type {classStructOrEnum {refPath=&ENiagara_SpriteRotationMode}}
      }
      Sprite Rotation Angle Min {
       bIsDynamic=false
       bIsEditable=false
       bIsStaticSwitch=false
       bIsVisible=false
       type {classStructOrEnum {refPath=&NiagaraFloat}}
      }
      Sprite Rotation Angle {
       bIsDynamic=false
       bIsEditable=false
       bIsStaticSwitch=false
       bIsVisible=false
       type {classStructOrEnum {refPath=&NiagaraFloat}}
      }
      Sprite Rotation Angle Max {
       bIsDynamic=false
       bIsEditable=false
       bIsStaticSwitch=false
       bIsVisible=false
       type {classStructOrEnum {refPath=&NiagaraFloat}}
      }
      Sprite Rotation Randomness Mode {
       bIsDynamic=false
       bIsEditable=false
       bIsStaticSwitch=true
       bIsVisible=false
       type {classStructOrEnum {refPath=&ENiagaraRandomnessMode}}
      }
      Sprite UV Mode {
       bIsDynamic=false
       bIsEditable=true
       bIsStaticSwitch=true
       bIsVisible=true
       type {classStructOrEnum {refPath=&ENiagara_UVFlippingMode}}
      }
      Sprite UV Randomness Mode {
       bIsDynamic=false
       bIsEditable=false
       bIsStaticSwitch=true
       bIsVisible=false
       type {classStructOrEnum {refPath=&ENiagaraRandomnessMode}}
      }
      Sprite UV Scale {
       bIsDynamic=false
       bIsEditable=false
       bIsStaticSwitch=false
       bIsVisible=false
       type {classStructOrEnum {refPath=&Vector2f}}
      }
      Mesh Scale Mode {
       bIsDynamic=false
       bIsEditable=true
       bIsStaticSwitch=true
       bIsVisible=true
       type {classStructOrEnum {refPath=&ENiagara_SizeScaleMode}}
      }
      Mesh Uniform Scale Min {
       bIsDynamic=false
       bIsEditable=false
       bIsStaticSwitch=false
       bIsVisible=false
       type {classStructOrEnum {refPath=&NiagaraFloat}}
      }
      Mesh Uniform Scale {
       bIsDynamic=false
       bIsEditable=false
       bIsStaticSwitch=false
       bIsVisible=false
       type {classStructOrEnum {refPath=&NiagaraFloat}}
      }
      Mesh Scale Min {
       bIsDynamic=false
       bIsEditable=false
       bIsStaticSwitch=false
       bIsVisible=false
       type {classStructOrEnum {refPath=&Vector3f}}
      }
      Mesh Scale {
       bIsDynamic=false
       bIsEditable=false
       bIsStaticSwitch=false
       bIsVisible=false
       type {classStructOrEnum {refPath=&Vector3f}}
      }
      Mesh Scale Max {
       bIsDynamic=false
       bIsEditable=false
       bIsStaticSwitch=false
       bIsVisible=false
       type {classStructOrEnum {refPath=&Vector3f}}
      }
      Mesh Uniform Scale Max {
       bIsDynamic=false
       bIsEditable=false
       bIsStaticSwitch=false
       bIsVisible=false
       type {classStructOrEnum {refPath=&NiagaraFloat}}
      }
      Mesh Scale Randomness Mode {
       bIsDynamic=false
       bIsEditable=false
       bIsStaticSwitch=true
       bIsVisible=false
       type {classStructOrEnum {refPath=&ENiagaraRandomnessMode}}
      }
      Mesh Renderer Array Visibility Mode {
       bIsDynamic=false
       bIsEditable=true
       bIsStaticSwitch=true
       bIsVisible=true
       type {classStructOrEnum {refPath=&ENiagara_UnsetDirectSetRandom}}
      }
      Mesh Renderer Mesh Index {
       bIsDynamic=false
       bIsEditable=false
       bIsStaticSwitch=false
       bIsVisible=false
       type {classStructOrEnum {refPath=&NiagaraInt32}}
      }
      Wrap Index By Mesh Count {bIsDynamic=false, bIsEditable=false, bIsStaticSwitch=true, bIsVisible=false, type {classStructOrEnum {refPath=&NiagaraBool}}}
      Mesh Renderer Info {
       bIsDynamic=false
       bIsEditable=false
       bIsStaticSwitch=false
       bIsVisible=false
       type {classStructOrEnum {refPath=&NiagaraDataInterfaceMeshRendererInfo}}
      }
      Mesh Renderer Random Seed {
       bIsDynamic=false
       bIsEditable=false
       bIsStaticSwitch=false
       bIsVisible=false
       type {classStructOrEnum {refPath=&NiagaraInt32}}
      }
      Ribbon Width Mode {
       bIsDynamic=false
       bIsEditable=true
       bIsStaticSwitch=true
       bIsVisible=true
       type {classStructOrEnum {refPath=&ENiagara_UnsetDirectSet}}
      }
      Ribbon Width {
       bIsDynamic=false
       bIsEditable=false
       bIsStaticSwitch=false
       bIsVisible=false
       type {classStructOrEnum {refPath=&NiagaraFloat}}
      }
      Ribbon Facing Mode {
       bIsDynamic=false
       bIsEditable=true
       bIsStaticSwitch=true
       bIsVisible=true
       type {classStructOrEnum {refPath=&ENiagara_UnsetDirectSet}}
      }
      Ribbon Facing Vector {
       bIsDynamic=false
       bIsEditable=false
       bIsStaticSwitch=false
       bIsVisible=false
       type {classStructOrEnum {refPath=&Vector3f}}
      }
      Ribbon Twist Mode {
       bIsDynamic=false
       bIsEditable=true
       bIsStaticSwitch=true
       bIsVisible=true
       type {classStructOrEnum {refPath=&ENiagara_UnsetDirectSet}}
      }
      "Ribbon Twist " {
       bIsDynamic=false
       bIsEditable=false
       bIsStaticSwitch=false
       bIsVisible=false
       type {classStructOrEnum {refPath=&NiagaraFloat}}
      }
     ]
     moduleScript {refPath=&InitializeParticle}
    }
   ]
  }
  particleUpdateScript {
   scriptName=ParticleUpdateScript
   modules [
    {
     bIsSetParametersModule=false
     enabled=true
     moduleName=ParticleState
     inputs [
      Kill Particles When Lifetime Has Elapsed {bIsDynamic=false, bIsEditable=true, bIsStaticSwitch=true, bIsVisible=true, type {classStructOrEnum {refPath=&NiagaraBool}}}
      Loop Particles Lifetime {bIsDynamic=false, bIsEditable=false, bIsStaticSwitch=false, bIsVisible=false, type {classStructOrEnum {refPath=&NiagaraBool}}}
      Let Infinitely Lived Particles Die When Emitter Deactivates {bIsDynamic=false, bIsEditable=false, bIsStaticSwitch=false, bIsVisible=false, type {classStructOrEnum {refPath=&NiagaraBool}}}
      Lifetime {bIsDynamic=false, bIsEditable=true, bIsStaticSwitch=false, bIsVisible=true, type {classStructOrEnum {refPath=&NiagaraFloat}}}
      DeltaTime {bIsDynamic=false, bIsEditable=true, bIsStaticSwitch=false, bIsVisible=true, type {classStructOrEnum {refPath=&NiagaraFloat}}}
     ]
     moduleScript {refPath=&ParticleState}
    }
    {
     bIsSetParametersModule=false
     enabled=true
     moduleName=SkeletalMeshLocation
     inputs [
      Skeletal Mesh {
       bIsDynamic=false
       bIsEditable=true
       bIsStaticSwitch=false
       bIsVisible=true
       type {classStructOrEnum {refPath=&NiagaraDataInterfaceSkeletalMesh}}
      }
      Mesh Sampling Type {
       bIsDynamic=false
       bIsEditable=true
       bIsStaticSwitch=true
       bIsVisible=true
       type {classStructOrEnum {refPath=&ENiagaraSkelSamplingModeFull}}
      }
      Bone / Socket Sampling Mode {
       bIsDynamic=false
       bIsEditable=false
       bIsStaticSwitch=true
       bIsVisible=false
       type {classStructOrEnum {refPath=&ENiagaraBoneSocketSamplingMode}}
      }
      Bone Sampling Mode {
       bIsDynamic=false
       bIsEditable=true
       bIsStaticSwitch=true
       bIsVisible=true
       type {classStructOrEnum {refPath=&ENiagaraBoneSamplingMode}}
      }
      Socket Sampling Mode {
       bIsDynamic=false
       bIsEditable=false
       bIsStaticSwitch=true
       bIsVisible=false
       type {classStructOrEnum {refPath=&ENiagaraSocketSamplingMode}}
      }
      Triangle Sampling Mode {
       bIsDynamic=false
       bIsEditable=false
       bIsStaticSwitch=true
       bIsVisible=false
       type {classStructOrEnum {refPath=&ENiagaraTriangleSamplingMode}}
      }
      Vertex Sampling Mode {
       bIsDynamic=false
       bIsEditable=false
       bIsStaticSwitch=true
       bIsVisible=false
       type {classStructOrEnum {refPath=&ENiagaraVertexSamplingMode}}
      }
      Bone / Socket Index {bIsDynamic=false, bIsEditable=true, bIsStaticSwitch=false, bIsVisible=true, type {classStructOrEnum {refPath=&NiagaraInt32}}}
      Triangle ID {
       bIsDynamic=false
       bIsEditable=false
       bIsStaticSwitch=false
       bIsVisible=false
       type {classStructOrEnum {refPath=&NiagaraInt32}}
      }
      Vertex ID {
       bIsDynamic=false
       bIsEditable=false
       bIsStaticSwitch=false
       bIsVisible=false
       type {classStructOrEnum {refPath=&NiagaraInt32}}
      }
      FixedBarycentric {bIsDynamic=false, bIsEditable=false, bIsStaticSwitch=false, bIsVisible=false, type {classStructOrEnum {refPath=&NiagaraBool}}}
      Triangle Coordinate {
       bIsDynamic=false
       bIsEditable=false
       bIsStaticSwitch=false
       bIsVisible=false
       type {classStructOrEnum {refPath=&Vector3f}}
      }
      Sampled UV Index {
       bIsDynamic=false
       bIsEditable=false
       bIsStaticSwitch=false
       bIsVisible=false
       type {classStructOrEnum {refPath=&NiagaraInt32}}
      }
      Sampled Position Coordinate Space {
       bIsDynamic=false
       bIsEditable=true
       bIsStaticSwitch=true
       bIsVisible=true
       type {classStructOrEnum {refPath=&ENiagaraCoordinateSpace}}
      }
      Sampled Velocity Coordinate Space {
       bIsDynamic=false
       bIsEditable=true
       bIsStaticSwitch=true
       bIsVisible=true
       type {classStructOrEnum {refPath=&ENiagaraCoordinateSpace}}
      }
      Sampled Orientation Coordinate Space {
       bIsDynamic=false
       bIsEditable=true
       bIsStaticSwitch=true
       bIsVisible=true
       type {classStructOrEnum {refPath=&ENiagaraCoordinateSpace}}
      }
      Generate New Random Samples Every Frame {bIsDynamic=false, bIsEditable=false, bIsStaticSwitch=true, bIsVisible=false, type {classStructOrEnum {refPath=&NiagaraBool}}}
      Write Sampled Mesh Attrributes to Particle Attributes {bIsDynamic=false, bIsEditable=true, bIsStaticSwitch=true, bIsVisible=true, type {classStructOrEnum {refPath=&NiagaraBool}}}
      Randomness Mode {
       bIsDynamic=false
       bIsEditable=false
       bIsStaticSwitch=false
       bIsVisible=false
       type {classStructOrEnum {refPath=&ENiagaraRandomnessMode}}
      }
      Random Seed {
       bIsDynamic=false
       bIsEditable=false
       bIsStaticSwitch=false
       bIsVisible=false
       type {classStructOrEnum {refPath=&NiagaraInt32}}
      }
      Fixed Random Seed {bIsDynamic=false, bIsEditable=false, bIsStaticSwitch=false, bIsVisible=false, type {classStructOrEnum {refPath=&NiagaraBool}}}
      Apply To Particle Position {bIsDynamic=false, bIsEditable=true, bIsStaticSwitch=false, bIsVisible=true, type {classStructOrEnum {refPath=&NiagaraBool}}}
      Sampled Position Offset {bIsDynamic=false, bIsEditable=true, bIsStaticSwitch=false, bIsVisible=true, type {classStructOrEnum {refPath=&Vector3f}}}
      OffsetAlongNormal {bIsDynamic=false, bIsEditable=false, bIsStaticSwitch=false, bIsVisible=false, type {classStructOrEnum {refPath=&NiagaraBool}}}
      Offset Position Along Sampled Normal {
       bIsDynamic=false
       bIsEditable=false
       bIsStaticSwitch=false
       bIsVisible=false
       type {classStructOrEnum {refPath=&NiagaraFloat}}
      }
      Sampled Position Offset Coordinate Space {
       bIsDynamic=false
       bIsEditable=true
       bIsStaticSwitch=true
       bIsVisible=true
       type {classStructOrEnum {refPath=&ENiagaraMeshTransforms}}
      }
      Inherit Velocity {bIsDynamic=false, bIsEditable=true, bIsStaticSwitch=false, bIsVisible=true, type {classStructOrEnum {refPath=&NiagaraBool}}}
      "Inherit Velocity (Scale)" {bIsDynamic=false, bIsEditable=false, bIsStaticSwitch=false, bIsVisible=true, type {classStructOrEnum {refPath=&NiagaraFloat}}}
      VelocitySpeedLimit {bIsDynamic=false, bIsEditable=false, bIsStaticSwitch=false, bIsVisible=false, type {classStructOrEnum {refPath=&NiagaraBool}}}
      Inherited Velocity Speed Limit {
       bIsDynamic=false
       bIsEditable=false
       bIsStaticSwitch=false
       bIsVisible=false
       type {classStructOrEnum {refPath=&NiagaraFloat}}
      }
      Apply Sampled Orientation to Mesh Particles {bIsDynamic=false, bIsEditable=true, bIsStaticSwitch=true, bIsVisible=true, type {classStructOrEnum {refPath=&NiagaraBool}}}
      Interpolate Mesh Sampling {bIsDynamic=false, bIsEditable=true, bIsStaticSwitch=true, bIsVisible=true, type {classStructOrEnum {refPath=&NiagaraBool}}}
      MaskSpawnGroup {bIsDynamic=false, bIsEditable=true, bIsStaticSwitch=false, bIsVisible=true, type {classStructOrEnum {refPath=&NiagaraBool}}}
      Spawn Group Mask {bIsDynamic=false, bIsEditable=false, bIsStaticSwitch=false, bIsVisible=true, type {classStructOrEnum {refPath=&NiagaraInt32}}}
      Use Sampled Mesh Attribute as Ribbon ID {bIsDynamic=false, bIsEditable=true, bIsStaticSwitch=false, bIsVisible=true, type {classStructOrEnum {refPath=&NiagaraBool}}}
     ]
     moduleScript {refPath=&SkeletalMeshLocation}
    }
    {
     bIsSetParametersModule=false
     enabled=true
     moduleName=CameraOffset
     inputs [Camera Offset Amount {bIsDynamic=false, bIsEditable=true, bIsStaticSwitch=false, bIsVisible=true, type {classStructOrEnum {refPath=&NiagaraFloat}}}]
     moduleScript {refPath=&CameraOffset}
    }
    {
     bIsSetParametersModule=false
     enabled=true
     moduleName=ScaleColor
     inputs [
      Scale Mode {
       bIsDynamic=false
       bIsEditable=true
       bIsStaticSwitch=true
       bIsVisible=true
       type {classStructOrEnum {refPath=&ENiagaraScaleColorMode}}
      }
      Linear Color Curve {
       bIsDynamic=false
       bIsEditable=true
       bIsStaticSwitch=false
       bIsVisible=true
       type {classStructOrEnum {refPath=&NiagaraDataInterfaceColorCurve}}
      }
      ScaleRGB {bIsDynamic=false, bIsEditable=false, bIsStaticSwitch=false, bIsVisible=false, type {classStructOrEnum {refPath=&NiagaraBool}}}
      Scale RGB {
       bIsDynamic=false
       bIsEditable=false
       bIsStaticSwitch=false
       bIsVisible=false
       type {classStructOrEnum {refPath=&Vector3f}}
      }
      ScaleRGBA {bIsDynamic=false, bIsEditable=false, bIsStaticSwitch=false, bIsVisible=false, type {classStructOrEnum {refPath=&NiagaraBool}}}
      Scale RGBA {
       bIsDynamic=false
       bIsEditable=false
       bIsStaticSwitch=false
       bIsVisible=false
       type {classStructOrEnum {refPath=&Vector4f}}
      }
      Curve Index {bIsDynamic=false, bIsEditable=true, bIsStaticSwitch=false, bIsVisible=true, type {classStructOrEnum {refPath=&NiagaraFloat}}}
      ScaleA {bIsDynamic=false, bIsEditable=false, bIsStaticSwitch=false, bIsVisible=false, type {classStructOrEnum {refPath=&NiagaraBool}}}
      Scale Alpha {
       bIsDynamic=false
       bIsEditable=false
       bIsStaticSwitch=false
       bIsVisible=false
       type {classStructOrEnum {refPath=&NiagaraFloat}}
      }
      Color Value To Scale {
       bIsDynamic=false
       bIsEditable=true
       bIsStaticSwitch=false
       bIsVisible=true
       type {classStructOrEnum {refPath=&LinearColor}}
      }
     ]
     moduleScript {refPath=&ScaleColor}
    }
    {
     bIsSetParametersModule=false
     enabled=true
     moduleName=FN_Light_v2
     inputs [
      LightEnabled {bIsDynamic=false, bIsEditable=true, bIsStaticSwitch=false, bIsVisible=true, type {classStructOrEnum {refPath=&NiagaraBool}}}
      Scale Color {bIsDynamic=false, bIsEditable=true, bIsStaticSwitch=false, bIsVisible=true, type {classStructOrEnum {refPath=&NiagaraFloat}}}
      Light Exponent {bIsDynamic=false, bIsEditable=true, bIsStaticSwitch=false, bIsVisible=true, type {classStructOrEnum {refPath=&NiagaraFloat}}}
      Position Offset {bIsDynamic=false, bIsEditable=true, bIsStaticSwitch=false, bIsVisible=true, type {classStructOrEnum {refPath=&Vector3f}}}
      Light Radius {bIsDynamic=false, bIsEditable=true, bIsStaticSwitch=false, bIsVisible=true, type {classStructOrEnum {refPath=&NiagaraFloat}}}
     ]
     moduleScript {refPath=&FN_Light_v2}
    }
    {
     bIsSetParametersModule=false
     enabled=true
     moduleName=ScaleColor001
     inputs [
      Scale Mode {
       bIsDynamic=false
       bIsEditable=true
       bIsStaticSwitch=true
       bIsVisible=true
       type {classStructOrEnum {refPath=&ENiagaraScaleColorMode}}
      }
      Linear Color Curve {
       bIsDynamic=false
       bIsEditable=false
       bIsStaticSwitch=false
       bIsVisible=false
       type {classStructOrEnum {refPath=&NiagaraDataInterfaceColorCurve}}
      }
      ScaleRGB {bIsDynamic=false, bIsEditable=true, bIsStaticSwitch=false, bIsVisible=true, type {classStructOrEnum {refPath=&NiagaraBool}}}
      Scale RGB {bIsDynamic=false, bIsEditable=true, bIsStaticSwitch=false, bIsVisible=true, type {classStructOrEnum {refPath=&Vector3f}}}
      ScaleRGBA {bIsDynamic=false, bIsEditable=false, bIsStaticSwitch=false, bIsVisible=false, type {classStructOrEnum {refPath=&NiagaraBool}}}
      Scale RGBA {
       bIsDynamic=false
       bIsEditable=false
       bIsStaticSwitch=false
       bIsVisible=false
       type {classStructOrEnum {refPath=&Vector4f}}
      }
      Curve Index {
       bIsDynamic=false
       bIsEditable=false
       bIsStaticSwitch=false
       bIsVisible=false
       type {classStructOrEnum {refPath=&NiagaraFloat}}
      }
      ScaleA {bIsDynamic=false, bIsEditable=true, bIsStaticSwitch=false, bIsVisible=true, type {classStructOrEnum {refPath=&NiagaraBool}}}
      Scale Alpha {bIsDynamic=false, bIsEditable=true, bIsStaticSwitch=false, bIsVisible=true, type {classStructOrEnum {refPath=&NiagaraFloat}}}
      Color Value To Scale {
       bIsDynamic=false
       bIsEditable=true
       bIsStaticSwitch=false
       bIsVisible=true
       type {classStructOrEnum {refPath=&LinearColor}}
      }
     ]
     moduleScript {refPath=&ScaleColor}
    }
   ]
  }
  rendererClasses [{refPath=&NiagaraSpriteRendererProperties}, {refPath=&NiagaraLightRendererProperties}]
  renderers [
   {rendererIndex=0, rendererClass {refPath=&NiagaraSpriteRendererProperties}}
   {rendererIndex=1, rendererClass {refPath=&NiagaraLightRendererProperties}}
   {rendererIndex=2, rendererClass {refPath=&NiagaraSpriteRendererProperties}}
  ]
 }
 HandGlow001 {
  bEnabled=false
  emitterSpawnScript {scriptName=EmitterSpawnScript, modules []}
  emitterUpdateScript {
   scriptName=EmitterUpdateScript
   modules [
    {
     bIsSetParametersModule=false
     enabled=true
     moduleName=EmitterState
     inputs [
      Life Cycle Mode {
       bIsDynamic=false
       bIsEditable=true
       bIsStaticSwitch=true
       bIsVisible=true
       type {classStructOrEnum {refPath=&ENiagaraEmitterLifeCycleMode}}
      }
      Inactive Response {
       bIsDynamic=false
       bIsEditable=true
       bIsStaticSwitch=true
       bIsVisible=true
       type {classStructOrEnum {refPath=&ENiagaraInactiveMode}}
      }
      Loop Behavior {
       bIsDynamic=false
       bIsEditable=true
       bIsStaticSwitch=true
       bIsVisible=true
       type {classStructOrEnum {refPath=&ENiagara_EmitterStateOptions}}
      }
      Loop Duration Mode {
       bIsDynamic=false
       bIsEditable=true
       bIsStaticSwitch=true
       bIsVisible=true
       type {classStructOrEnum {refPath=&ENiagara_InfiniteLoopDuration}}
      }
      Loop Count {
       bIsDynamic=false
       bIsEditable=false
       bIsStaticSwitch=false
       bIsVisible=false
       type {classStructOrEnum {refPath=&NiagaraInt32}}
      }
      Loop Duration {
       bIsDynamic=false
       bIsEditable=false
       bIsStaticSwitch=false
       bIsVisible=false
       type {classStructOrEnum {refPath=&NiagaraFloat}}
      }
      Recalculate Duration Each Loop {bIsDynamic=false, bIsEditable=false, bIsStaticSwitch=false, bIsVisible=false, type {classStructOrEnum {refPath=&NiagaraBool}}}
      UseLoopDelay {bIsDynamic=false, bIsEditable=true, bIsStaticSwitch=false, bIsVisible=true, type {classStructOrEnum {refPath=&NiagaraBool}}}
      Loop Delay {bIsDynamic=false, bIsEditable=false, bIsStaticSwitch=false, bIsVisible=true, type {classStructOrEnum {refPath=&NiagaraFloat}}}
      Delay First Loop Only {bIsDynamic=false, bIsEditable=false, bIsStaticSwitch=false, bIsVisible=false, type {classStructOrEnum {refPath=&NiagaraBool}}}
      Scalability Mode {
       bIsDynamic=false
       bIsEditable=true
       bIsStaticSwitch=true
       bIsVisible=true
       type {classStructOrEnum {refPath=&ENiagaraEmitterScalabilityMode_Limited}}
      }
      Enable Distance Culling {bIsDynamic=false, bIsEditable=false, bIsStaticSwitch=true, bIsVisible=false, type {classStructOrEnum {refPath=&NiagaraBool}}}
      MinDistance {
       bIsDynamic=false
       bIsEditable=false
       bIsStaticSwitch=false
       bIsVisible=false
       type {classStructOrEnum {refPath=&NiagaraFloat}}
      }
      Min Distance Response {
       bIsDynamic=false
       bIsEditable=false
       bIsStaticSwitch=false
       bIsVisible=false
       type {classStructOrEnum {refPath=&ENiagaraExecutionStateManagement}}
      }
      MaxDistance {
       bIsDynamic=false
       bIsEditable=false
       bIsStaticSwitch=false
       bIsVisible=false
       type {classStructOrEnum {refPath=&NiagaraFloat}}
      }
      Max Distance Response {
       bIsDynamic=false
       bIsEditable=false
       bIsStaticSwitch=false
       bIsVisible=false
       type {classStructOrEnum {refPath=&ENiagaraExecutionStateManagement}}
      }
      Scale Spawn Count {bIsDynamic=false, bIsEditable=false, bIsStaticSwitch=true, bIsVisible=false, type {classStructOrEnum {refPath=&NiagaraBool}}}
      Spawn Count Scale {
       bIsDynamic=false
       bIsEditable=false
       bIsStaticSwitch=false
       bIsVisible=false
       type {classStructOrEnum {refPath=&NiagaraFloat}}
      }
      Spawn Count Scale By Distance {
       bIsDynamic=false
       bIsEditable=false
       bIsStaticSwitch=false
       bIsVisible=false
       type {classStructOrEnum {refPath=&NiagaraDataInterfaceCurve}}
      }
      Enable Visibility Culling {bIsDynamic=false, bIsEditable=false, bIsStaticSwitch=true, bIsVisible=false, type {classStructOrEnum {refPath=&NiagaraBool}}}
      Visibility Cull Response {
       bIsDynamic=false
       bIsEditable=false
       bIsStaticSwitch=false
       bIsVisible=false
       type {classStructOrEnum {refPath=&ENiagaraExecutionStateManagement}}
      }
      Visibility Cull Delay {
       bIsDynamic=false
       bIsEditable=false
       bIsStaticSwitch=false
       bIsVisible=false
       type {classStructOrEnum {refPath=&NiagaraFloat}}
      }
      Reset Age On Awaken {bIsDynamic=false, bIsEditable=false, bIsStaticSwitch=false, bIsVisible=false, type {classStructOrEnum {refPath=&NiagaraBool}}}
     ]
     moduleScript {refPath=&EmitterState}
    }
    {
     bIsSetParametersModule=false
     enabled=true
     moduleName=SpawnBurst_Instantaneous
     inputs [
      Spawn Count {bIsDynamic=false, bIsEditable=true, bIsStaticSwitch=false, bIsVisible=true, type {classStructOrEnum {refPath=&NiagaraInt32}}}
      Spawn Time {bIsDynamic=false, bIsEditable=true, bIsStaticSwitch=false, bIsVisible=true, type {classStructOrEnum {refPath=&NiagaraFloat}}}
      Use Spawn Probability {bIsDynamic=false, bIsEditable=true, bIsStaticSwitch=false, bIsVisible=true, type {classStructOrEnum {refPath=&NiagaraBool}}}
      Spawn Probability {bIsDynamic=false, bIsEditable=true, bIsStaticSwitch=false, bIsVisible=true, type {classStructOrEnum {refPath=&NiagaraFloat}}}
      Spawn Group {bIsDynamic=false, bIsEditable=true, bIsStaticSwitch=false, bIsVisible=true, type {classStructOrEnum {refPath=&NiagaraInt32}}}
      Age {bIsDynamic=false, bIsEditable=true, bIsStaticSwitch=false, bIsVisible=true, type {classStructOrEnum {refPath=&NiagaraFloat}}}
     ]
     moduleScript {refPath=&SpawnBurst_Instantaneous}
    }
   ]
  }
  particleSpawnScript {
   scriptName=ParticleSpawnScript
   modules [
    {
     bIsSetParametersModule=false
     enabled=true
     moduleName=FN_SetSpawnGroup
     inputs [
      Number of Spawn Groups {bIsDynamic=false, bIsEditable=true, bIsStaticSwitch=false, bIsVisible=true, type {classStructOrEnum {refPath=&NiagaraInt32}}}
      Particles Per Group {
       bIsDynamic=false
       bIsEditable=false
       bIsStaticSwitch=false
       bIsVisible=false
       type {classStructOrEnum {refPath=&NiagaraInt32}}
      }
      Pattern {
       bIsDynamic=false
       bIsEditable=true
       bIsStaticSwitch=true
       bIsVisible=true
       type {classStructOrEnum {refPath=&ENiagaraPartitionWaveForm}}
      }
      Normalize Spawn Group {bIsDynamic=false, bIsEditable=true, bIsStaticSwitch=true, bIsVisible=true, type {classStructOrEnum {refPath=&NiagaraBool}}}
      Inclusivity {bIsDynamic=false, bIsEditable=false, bIsStaticSwitch=true, bIsVisible=false, type {classStructOrEnum {refPath=&NiagaraBool}}}
      Set Ribbon ID and Ribbon Link Order {bIsDynamic=false, bIsEditable=true, bIsStaticSwitch=true, bIsVisible=true, type {classStructOrEnum {refPath=&NiagaraBool}}}
     ]
     moduleScript {refPath=&FN_SetSpawnGroup}
    }
    {
     bIsSetParametersModule=false
     enabled=true
     moduleName=InitializeParticle001
     inputs [
      Lifetime Mode {
       bIsDynamic=false
       bIsEditable=true
       bIsStaticSwitch=true
       bIsVisible=true
       type {classStructOrEnum {refPath=&ENiagara_LifetimeMode}}
      }
      Lifetime Min {
       bIsDynamic=false
       bIsEditable=false
       bIsStaticSwitch=false
       bIsVisible=false
       type {classStructOrEnum {refPath=&NiagaraFloat}}
      }
      Lifetime {bIsDynamic=false, bIsEditable=true, bIsStaticSwitch=false, bIsVisible=true, type {classStructOrEnum {refPath=&NiagaraFloat}}}
      Lifetime Max {
       bIsDynamic=false
       bIsEditable=false
       bIsStaticSwitch=false
       bIsVisible=false
       type {classStructOrEnum {refPath=&NiagaraFloat}}
      }
      Lifetime Randomness Mode {
       bIsDynamic=false
       bIsEditable=false
       bIsStaticSwitch=true
       bIsVisible=false
       type {classStructOrEnum {refPath=&ENiagaraRandomnessMode}}
      }
      Color Mode {
       bIsDynamic=false
       bIsEditable=true
       bIsStaticSwitch=true
       bIsVisible=true
       type {classStructOrEnum {refPath=&ENiagara_ColorInitializationMode}}
      }
      Color Minimum {
       bIsDynamic=false
       bIsEditable=false
       bIsStaticSwitch=false
       bIsVisible=false
       type {classStructOrEnum {refPath=&LinearColor}}
      }
      Color {
       bIsDynamic=false
       bIsEditable=true
       bIsStaticSwitch=false
       bIsVisible=true
       type {classStructOrEnum {refPath=&LinearColor}}
      }
      AdjustHue {bIsDynamic=false, bIsEditable=false, bIsStaticSwitch=false, bIsVisible=false, type {classStructOrEnum {refPath=&NiagaraBool}}}
      Hue Shift Range {
       bIsDynamic=false
       bIsEditable=false
       bIsStaticSwitch=false
       bIsVisible=false
       type {classStructOrEnum {refPath=&Vector2f}}
      }
      Color Maximum {
       bIsDynamic=false
       bIsEditable=false
       bIsStaticSwitch=false
       bIsVisible=false
       type {classStructOrEnum {refPath=&LinearColor}}
      }
      AdjustSaturation {bIsDynamic=false, bIsEditable=false, bIsStaticSwitch=false, bIsVisible=false, type {classStructOrEnum {refPath=&NiagaraBool}}}
      Saturation Range {
       bIsDynamic=false
       bIsEditable=false
       bIsStaticSwitch=false
       bIsVisible=false
       type {classStructOrEnum {refPath=&Vector2f}}
      }
      AdjustValue {bIsDynamic=false, bIsEditable=false, bIsStaticSwitch=false, bIsVisible=false, type {classStructOrEnum {refPath=&NiagaraBool}}}
      Value Range {
       bIsDynamic=false
       bIsEditable=false
       bIsStaticSwitch=false
       bIsVisible=false
       type {classStructOrEnum {refPath=&Vector2f}}
      }
      AdjustAlpha {bIsDynamic=false, bIsEditable=false, bIsStaticSwitch=false, bIsVisible=false, type {classStructOrEnum {refPath=&NiagaraBool}}}
      Alpha Scale Range {
       bIsDynamic=false
       bIsEditable=false
       bIsStaticSwitch=false
       bIsVisible=false
       type {classStructOrEnum {refPath=&Vector2f}}
      }
      Color Channel Mode {
       bIsDynamic=false
       bIsEditable=false
       bIsStaticSwitch=true
       bIsVisible=false
       type {classStructOrEnum {refPath=&ENiagaraChannelCorrelation}}
      }
      Color Randomness Mode {
       bIsDynamic=false
       bIsEditable=false
       bIsStaticSwitch=true
       bIsVisible=false
       type {classStructOrEnum {refPath=&ENiagaraRandomnessMode}}
      }
      Position Mode {
       bIsDynamic=false
       bIsEditable=true
       bIsStaticSwitch=true
       bIsVisible=true
       type {classStructOrEnum {refPath=&ENiagara_PositionInitializationMode}}
      }
      Position {
       bIsDynamic=true
       bIsEditable=false
       bIsStaticSwitch=false
       bIsVisible=false
       type {classStructOrEnum {refPath=&NiagaraPosition}}
      }
      UsePositionOffset {bIsDynamic=false, bIsEditable=true, bIsStaticSwitch=false, bIsVisible=true, type {classStructOrEnum {refPath=&NiagaraBool}}}
      Position Offset {bIsDynamic=false, bIsEditable=false, bIsStaticSwitch=false, bIsVisible=true, type {classStructOrEnum {refPath=&Vector3f}}}
      Position Offset Coordinate Space {
       bIsDynamic=false
       bIsEditable=false
       bIsStaticSwitch=false
       bIsVisible=false
       type {classStructOrEnum {refPath=&ENiagaraCoordinateSpace}}
      }
      Mass Mode {
       bIsDynamic=false
       bIsEditable=true
       bIsStaticSwitch=true
       bIsVisible=true
       type {classStructOrEnum {refPath=&ENiagara_MassInitializationMode}}
      }
      Mass Min {
       bIsDynamic=false
       bIsEditable=false
       bIsStaticSwitch=false
       bIsVisible=false
       type {classStructOrEnum {refPath=&NiagaraFloat}}
      }
      Mass {bIsDynamic=false, bIsEditable=true, bIsStaticSwitch=false, bIsVisible=true, type {classStructOrEnum {refPath=&NiagaraFloat}}}
      Mass Max {
       bIsDynamic=false
       bIsEditable=false
       bIsStaticSwitch=false
       bIsVisible=false
       type {classStructOrEnum {refPath=&NiagaraFloat}}
      }
      Mass Randomness Mode {
       bIsDynamic=false
       bIsEditable=false
       bIsStaticSwitch=true
       bIsVisible=false
       type {classStructOrEnum {refPath=&ENiagaraRandomnessMode}}
      }
      Write Material Random {bIsDynamic=false, bIsEditable=true, bIsStaticSwitch=false, bIsVisible=true, type {classStructOrEnum {refPath=&NiagaraBool}}}
      Material Random {bIsDynamic=false, bIsEditable=true, bIsStaticSwitch=false, bIsVisible=true, type {classStructOrEnum {refPath=&NiagaraFloat}}}
      Sprite Size Mode {
       bIsDynamic=false
       bIsEditable=true
       bIsStaticSwitch=true
       bIsVisible=true
       type {classStructOrEnum {refPath=&ENiagara_SizeScaleMode}}
      }
      Sprite Size {
       bIsDynamic=false
       bIsEditable=false
       bIsStaticSwitch=false
       bIsVisible=false
       type {classStructOrEnum {refPath=&Vector2f}}
      }
      Uniform Sprite Size {bIsDynamic=false, bIsEditable=true, bIsStaticSwitch=false, bIsVisible=true, type {classStructOrEnum {refPath=&NiagaraFloat}}}
      Uniform Sprite Size Min {
       bIsDynamic=false
       bIsEditable=false
       bIsStaticSwitch=false
       bIsVisible=false
       type {classStructOrEnum {refPath=&NiagaraFloat}}
      }
      Sprite Size Min {
       bIsDynamic=false
       bIsEditable=false
       bIsStaticSwitch=false
       bIsVisible=false
       type {classStructOrEnum {refPath=&Vector2f}}
      }
      Sprite Size Max {
       bIsDynamic=false
       bIsEditable=false
       bIsStaticSwitch=false
       bIsVisible=false
       type {classStructOrEnum {refPath=&Vector2f}}
      }
      Uniform Sprite Size Max {
       bIsDynamic=false
       bIsEditable=false
       bIsStaticSwitch=false
       bIsVisible=false
       type {classStructOrEnum {refPath=&NiagaraFloat}}
      }
      Sprite Size Randomness Mode {
       bIsDynamic=false
       bIsEditable=false
       bIsStaticSwitch=true
       bIsVisible=false
       type {classStructOrEnum {refPath=&ENiagaraRandomnessMode}}
      }
      Sprite Rotation Mode {
       bIsDynamic=false
       bIsEditable=true
       bIsStaticSwitch=true
       bIsVisible=true
       type {classStructOrEnum {refPath=&ENiagara_SpriteRotationMode}}
      }
      Sprite Rotation Angle Min {
       bIsDynamic=false
       bIsEditable=false
       bIsStaticSwitch=false
       bIsVisible=false
       type {classStructOrEnum {refPath=&NiagaraFloat}}
      }
      Sprite Rotation Angle {
       bIsDynamic=false
       bIsEditable=false
       bIsStaticSwitch=false
       bIsVisible=false
       type {classStructOrEnum {refPath=&NiagaraFloat}}
      }
      Sprite Rotation Angle Max {
       bIsDynamic=false
       bIsEditable=false
       bIsStaticSwitch=false
       bIsVisible=false
       type {classStructOrEnum {refPath=&NiagaraFloat}}
      }
      Sprite Rotation Randomness Mode {
       bIsDynamic=false
       bIsEditable=false
       bIsStaticSwitch=true
       bIsVisible=false
       type {classStructOrEnum {refPath=&ENiagaraRandomnessMode}}
      }
      Sprite UV Mode {
       bIsDynamic=false
       bIsEditable=true
       bIsStaticSwitch=true
       bIsVisible=true
       type {classStructOrEnum {refPath=&ENiagara_UVFlippingMode}}
      }
      Sprite UV Randomness Mode {
       bIsDynamic=false
       bIsEditable=false
       bIsStaticSwitch=true
       bIsVisible=false
       type {classStructOrEnum {refPath=&ENiagaraRandomnessMode}}
      }
      Sprite UV Scale {
       bIsDynamic=false
       bIsEditable=false
       bIsStaticSwitch=false
       bIsVisible=false
       type {classStructOrEnum {refPath=&Vector2f}}
      }
      Mesh Scale Mode {
       bIsDynamic=false
       bIsEditable=true
       bIsStaticSwitch=true
       bIsVisible=true
       type {classStructOrEnum {refPath=&ENiagara_SizeScaleMode}}
      }
      Mesh Uniform Scale Min {
       bIsDynamic=false
       bIsEditable=false
       bIsStaticSwitch=false
       bIsVisible=false
       type {classStructOrEnum {refPath=&NiagaraFloat}}
      }
      Mesh Uniform Scale {
       bIsDynamic=false
       bIsEditable=false
       bIsStaticSwitch=false
       bIsVisible=false
       type {classStructOrEnum {refPath=&NiagaraFloat}}
      }
      Mesh Scale Min {
       bIsDynamic=false
       bIsEditable=false
       bIsStaticSwitch=false
       bIsVisible=false
       type {classStructOrEnum {refPath=&Vector3f}}
      }
      Mesh Scale {
       bIsDynamic=false
       bIsEditable=false
       bIsStaticSwitch=false
       bIsVisible=false
       type {classStructOrEnum {refPath=&Vector3f}}
      }
      Mesh Scale Max {
       bIsDynamic=false
       bIsEditable=false
       bIsStaticSwitch=false
       bIsVisible=false
       type {classStructOrEnum {refPath=&Vector3f}}
      }
      Mesh Uniform Scale Max {
       bIsDynamic=false
       bIsEditable=false
       bIsStaticSwitch=false
       bIsVisible=false
       type {classStructOrEnum {refPath=&NiagaraFloat}}
      }
      Mesh Scale Randomness Mode {
       bIsDynamic=false
       bIsEditable=false
       bIsStaticSwitch=true
       bIsVisible=false
       type {classStructOrEnum {refPath=&ENiagaraRandomnessMode}}
      }
      Mesh Renderer Array Visibility Mode {
       bIsDynamic=false
       bIsEditable=true
       bIsStaticSwitch=true
       bIsVisible=true
       type {classStructOrEnum {refPath=&ENiagara_UnsetDirectSetRandom}}
      }
      Mesh Renderer Mesh Index {
       bIsDynamic=false
       bIsEditable=false
       bIsStaticSwitch=false
       bIsVisible=false
       type {classStructOrEnum {refPath=&NiagaraInt32}}
      }
      Wrap Index By Mesh Count {bIsDynamic=false, bIsEditable=false, bIsStaticSwitch=true, bIsVisible=false, type {classStructOrEnum {refPath=&NiagaraBool}}}
      Mesh Renderer Info {
       bIsDynamic=false
       bIsEditable=false
       bIsStaticSwitch=false
       bIsVisible=false
       type {classStructOrEnum {refPath=&NiagaraDataInterfaceMeshRendererInfo}}
      }
      Mesh Renderer Random Seed {
       bIsDynamic=false
       bIsEditable=false
       bIsStaticSwitch=false
       bIsVisible=false
       type {classStructOrEnum {refPath=&NiagaraInt32}}
      }
      Ribbon Width Mode {
       bIsDynamic=false
       bIsEditable=true
       bIsStaticSwitch=true
       bIsVisible=true
       type {classStructOrEnum {refPath=&ENiagara_UnsetDirectSet}}
      }
      Ribbon Width {
       bIsDynamic=false
       bIsEditable=false
       bIsStaticSwitch=false
       bIsVisible=false
       type {classStructOrEnum {refPath=&NiagaraFloat}}
      }
      Ribbon Facing Mode {
       bIsDynamic=false
       bIsEditable=true
       bIsStaticSwitch=true
       bIsVisible=true
       type {classStructOrEnum {refPath=&ENiagara_UnsetDirectSet}}
      }
      Ribbon Facing Vector {
       bIsDynamic=false
       bIsEditable=false
       bIsStaticSwitch=false
       bIsVisible=false
       type {classStructOrEnum {refPath=&Vector3f}}
      }
      Ribbon Twist Mode {
       bIsDynamic=false
       bIsEditable=true
       bIsStaticSwitch=true
       bIsVisible=true
       type {classStructOrEnum {refPath=&ENiagara_UnsetDirectSet}}
      }
      "Ribbon Twist " {
       bIsDynamic=false
       bIsEditable=false
       bIsStaticSwitch=false
       bIsVisible=false
       type {classStructOrEnum {refPath=&NiagaraFloat}}
      }
     ]
     moduleScript {refPath=&InitializeParticle}
    }
   ]
  }
  particleUpdateScript {
   scriptName=ParticleUpdateScript
   modules [
    {
     bIsSetParametersModule=false
     enabled=true
     moduleName=ParticleState
     inputs [
      Kill Particles When Lifetime Has Elapsed {bIsDynamic=false, bIsEditable=true, bIsStaticSwitch=true, bIsVisible=true, type {classStructOrEnum {refPath=&NiagaraBool}}}
      Loop Particles Lifetime {bIsDynamic=false, bIsEditable=false, bIsStaticSwitch=false, bIsVisible=false, type {classStructOrEnum {refPath=&NiagaraBool}}}
      Let Infinitely Lived Particles Die When Emitter Deactivates {bIsDynamic=false, bIsEditable=false, bIsStaticSwitch=false, bIsVisible=false, type {classStructOrEnum {refPath=&NiagaraBool}}}
      Lifetime {bIsDynamic=false, bIsEditable=true, bIsStaticSwitch=false, bIsVisible=true, type {classStructOrEnum {refPath=&NiagaraFloat}}}
      DeltaTime {bIsDynamic=false, bIsEditable=true, bIsStaticSwitch=false, bIsVisible=true, type {classStructOrEnum {refPath=&NiagaraFloat}}}
     ]
     moduleScript {refPath=&ParticleState}
    }
    {
     bIsSetParametersModule=false
     enabled=true
     moduleName=SkeletalMeshLocation
     inputs [
      Skeletal Mesh {
       bIsDynamic=false
       bIsEditable=true
       bIsStaticSwitch=false
       bIsVisible=true
       type {classStructOrEnum {refPath=&NiagaraDataInterfaceSkeletalMesh}}
      }
      Mesh Sampling Type {
       bIsDynamic=false
       bIsEditable=true
       bIsStaticSwitch=true
       bIsVisible=true
       type {classStructOrEnum {refPath=&ENiagaraSkelSamplingModeFull}}
      }
      Bone / Socket Sampling Mode {
       bIsDynamic=false
       bIsEditable=false
       bIsStaticSwitch=true
       bIsVisible=false
       type {classStructOrEnum {refPath=&ENiagaraBoneSocketSamplingMode}}
      }
      Bone Sampling Mode {
       bIsDynamic=false
       bIsEditable=true
       bIsStaticSwitch=true
       bIsVisible=true
       type {classStructOrEnum {refPath=&ENiagaraBoneSamplingMode}}
      }
      Socket Sampling Mode {
       bIsDynamic=false
       bIsEditable=false
       bIsStaticSwitch=true
       bIsVisible=false
       type {classStructOrEnum {refPath=&ENiagaraSocketSamplingMode}}
      }
      Triangle Sampling Mode {
       bIsDynamic=false
       bIsEditable=false
       bIsStaticSwitch=true
       bIsVisible=false
       type {classStructOrEnum {refPath=&ENiagaraTriangleSamplingMode}}
      }
      Vertex Sampling Mode {
       bIsDynamic=false
       bIsEditable=false
       bIsStaticSwitch=true
       bIsVisible=false
       type {classStructOrEnum {refPath=&ENiagaraVertexSamplingMode}}
      }
      Bone / Socket Index {bIsDynamic=false, bIsEditable=true, bIsStaticSwitch=false, bIsVisible=true, type {classStructOrEnum {refPath=&NiagaraInt32}}}
      Triangle ID {
       bIsDynamic=false
       bIsEditable=false
       bIsStaticSwitch=false
       bIsVisible=false
       type {classStructOrEnum {refPath=&NiagaraInt32}}
      }
      Vertex ID {
       bIsDynamic=false
       bIsEditable=false
       bIsStaticSwitch=false
       bIsVisible=false
       type {classStructOrEnum {refPath=&NiagaraInt32}}
      }
      FixedBarycentric {bIsDynamic=false, bIsEditable=false, bIsStaticSwitch=false, bIsVisible=false, type {classStructOrEnum {refPath=&NiagaraBool}}}
      Triangle Coordinate {
       bIsDynamic=false
       bIsEditable=false
       bIsStaticSwitch=false
       bIsVisible=false
       type {classStructOrEnum {refPath=&Vector3f}}
      }
      Sampled UV Index {
       bIsDynamic=false
       bIsEditable=false
       bIsStaticSwitch=false
       bIsVisible=false
       type {classStructOrEnum {refPath=&NiagaraInt32}}
      }
      Sampled Position Coordinate Space {
       bIsDynamic=false
       bIsEditable=true
       bIsStaticSwitch=true
       bIsVisible=true
       type {classStructOrEnum {refPath=&ENiagaraCoordinateSpace}}
      }
      Sampled Velocity Coordinate Space {
       bIsDynamic=false
       bIsEditable=true
       bIsStaticSwitch=true
       bIsVisible=true
       type {classStructOrEnum {refPath=&ENiagaraCoordinateSpace}}
      }
      Sampled Orientation Coordinate Space {
       bIsDynamic=false
       bIsEditable=true
       bIsStaticSwitch=true
       bIsVisible=true
       type {classStructOrEnum {refPath=&ENiagaraCoordinateSpace}}
      }
      Generate New Random Samples Every Frame {bIsDynamic=false, bIsEditable=false, bIsStaticSwitch=true, bIsVisible=false, type {classStructOrEnum {refPath=&NiagaraBool}}}
      Write Sampled Mesh Attrributes to Particle Attributes {bIsDynamic=false, bIsEditable=true, bIsStaticSwitch=true, bIsVisible=true, type {classStructOrEnum {refPath=&NiagaraBool}}}
      Randomness Mode {
       bIsDynamic=false
       bIsEditable=false
       bIsStaticSwitch=false
       bIsVisible=false
       type {classStructOrEnum {refPath=&ENiagaraRandomnessMode}}
      }
      Random Seed {
       bIsDynamic=false
       bIsEditable=false
       bIsStaticSwitch=false
       bIsVisible=false
       type {classStructOrEnum {refPath=&NiagaraInt32}}
      }
      Fixed Random Seed {bIsDynamic=false, bIsEditable=false, bIsStaticSwitch=false, bIsVisible=false, type {classStructOrEnum {refPath=&NiagaraBool}}}
      Apply To Particle Position {bIsDynamic=false, bIsEditable=true, bIsStaticSwitch=false, bIsVisible=true, type {classStructOrEnum {refPath=&NiagaraBool}}}
      Sampled Position Offset {bIsDynamic=false, bIsEditable=true, bIsStaticSwitch=false, bIsVisible=true, type {classStructOrEnum {refPath=&Vector3f}}}
      OffsetAlongNormal {bIsDynamic=false, bIsEditable=false, bIsStaticSwitch=false, bIsVisible=false, type {classStructOrEnum {refPath=&NiagaraBool}}}
      Offset Position Along Sampled Normal {
       bIsDynamic=false
       bIsEditable=false
       bIsStaticSwitch=false
       bIsVisible=false
       type {classStructOrEnum {refPath=&NiagaraFloat}}
      }
      Sampled Position Offset Coordinate Space {
       bIsDynamic=false
       bIsEditable=true
       bIsStaticSwitch=true
       bIsVisible=true
       type {classStructOrEnum {refPath=&ENiagaraMeshTransforms}}
      }
      Inherit Velocity {bIsDynamic=false, bIsEditable=true, bIsStaticSwitch=false, bIsVisible=true, type {classStructOrEnum {refPath=&NiagaraBool}}}
      "Inherit Velocity (Scale)" {bIsDynamic=false, bIsEditable=false, bIsStaticSwitch=false, bIsVisible=true, type {classStructOrEnum {refPath=&NiagaraFloat}}}
      VelocitySpeedLimit {bIsDynamic=false, bIsEditable=false, bIsStaticSwitch=false, bIsVisible=false, type {classStructOrEnum {refPath=&NiagaraBool}}}
      Inherited Velocity Speed Limit {
       bIsDynamic=false
       bIsEditable=false
       bIsStaticSwitch=false
       bIsVisible=false
       type {classStructOrEnum {refPath=&NiagaraFloat}}
      }
      Apply Sampled Orientation to Mesh Particles {bIsDynamic=false, bIsEditable=true, bIsStaticSwitch=true, bIsVisible=true, type {classStructOrEnum {refPath=&NiagaraBool}}}
      Interpolate Mesh Sampling {bIsDynamic=false, bIsEditable=true, bIsStaticSwitch=true, bIsVisible=true, type {classStructOrEnum {refPath=&NiagaraBool}}}
      MaskSpawnGroup {bIsDynamic=false, bIsEditable=true, bIsStaticSwitch=false, bIsVisible=true, type {classStructOrEnum {refPath=&NiagaraBool}}}
      Spawn Group Mask {bIsDynamic=false, bIsEditable=false, bIsStaticSwitch=false, bIsVisible=true, type {classStructOrEnum {refPath=&NiagaraInt32}}}
      Use Sampled Mesh Attribute as Ribbon ID {bIsDynamic=false, bIsEditable=true, bIsStaticSwitch=false, bIsVisible=true, type {classStructOrEnum {refPath=&NiagaraBool}}}
     ]
     moduleScript {refPath=&SkeletalMeshLocation}
    }
    {
     bIsSetParametersModule=false
     enabled=true
     moduleName=CameraOffset
     inputs [Camera Offset Amount {bIsDynamic=false, bIsEditable=true, bIsStaticSwitch=false, bIsVisible=true, type {classStructOrEnum {refPath=&NiagaraFloat}}}]
     moduleScript {refPath=&CameraOffset}
    }
    {
     bIsSetParametersModule=false
     enabled=true
     moduleName=ScaleColor
     inputs [
      Scale Mode {
       bIsDynamic=false
       bIsEditable=true
       bIsStaticSwitch=true
       bIsVisible=true
       type {classStructOrEnum {refPath=&ENiagaraScaleColorMode}}
      }
      Linear Color Curve {
       bIsDynamic=false
       bIsEditable=true
       bIsStaticSwitch=false
       bIsVisible=true
       type {classStructOrEnum {refPath=&NiagaraDataInterfaceColorCurve}}
      }
      ScaleRGB {bIsDynamic=false, bIsEditable=false, bIsStaticSwitch=false, bIsVisible=false, type {classStructOrEnum {refPath=&NiagaraBool}}}
      Scale RGB {
       bIsDynamic=false
       bIsEditable=false
       bIsStaticSwitch=false
       bIsVisible=false
       type {classStructOrEnum {refPath=&Vector3f}}
      }
      ScaleRGBA {bIsDynamic=false, bIsEditable=false, bIsStaticSwitch=false, bIsVisible=false, type {classStructOrEnum {refPath=&NiagaraBool}}}
      Scale RGBA {
       bIsDynamic=false
       bIsEditable=false
       bIsStaticSwitch=false
       bIsVisible=false
       type {classStructOrEnum {refPath=&Vector4f}}
      }
      Curve Index {bIsDynamic=false, bIsEditable=true, bIsStaticSwitch=false, bIsVisible=true, type {classStructOrEnum {refPath=&NiagaraFloat}}}
      ScaleA {bIsDynamic=false, bIsEditable=false, bIsStaticSwitch=false, bIsVisible=false, type {classStructOrEnum {refPath=&NiagaraBool}}}
      Scale Alpha {
       bIsDynamic=false
       bIsEditable=false
       bIsStaticSwitch=false
       bIsVisible=false
       type {classStructOrEnum {refPath=&NiagaraFloat}}
      }
      Color Value To Scale {
       bIsDynamic=false
       bIsEditable=true
       bIsStaticSwitch=false
       bIsVisible=true
       type {classStructOrEnum {refPath=&LinearColor}}
      }
     ]
     moduleScript {refPath=&ScaleColor}
    }
    {
     bIsSetParametersModule=false
     enabled=true
     moduleName=FN_Light_v2
     inputs [
      LightEnabled {bIsDynamic=false, bIsEditable=true, bIsStaticSwitch=false, bIsVisible=true, type {classStructOrEnum {refPath=&NiagaraBool}}}
      Scale Color {bIsDynamic=false, bIsEditable=true, bIsStaticSwitch=false, bIsVisible=true, type {classStructOrEnum {refPath=&NiagaraFloat}}}
      Light Exponent {bIsDynamic=false, bIsEditable=true, bIsStaticSwitch=false, bIsVisible=true, type {classStructOrEnum {refPath=&NiagaraFloat}}}
      Position Offset {bIsDynamic=false, bIsEditable=true, bIsStaticSwitch=false, bIsVisible=true, type {classStructOrEnum {refPath=&Vector3f}}}
      Light Radius {bIsDynamic=false, bIsEditable=true, bIsStaticSwitch=false, bIsVisible=true, type {classStructOrEnum {refPath=&NiagaraFloat}}}
     ]
     moduleScript {refPath=&FN_Light_v2}
    }
    {
     bIsSetParametersModule=false
     enabled=true
     moduleName=ScaleColor001
     inputs [
      Scale Mode {
       bIsDynamic=false
       bIsEditable=true
       bIsStaticSwitch=true
       bIsVisible=true
       type {classStructOrEnum {refPath=&ENiagaraScaleColorMode}}
      }
      Linear Color Curve {
       bIsDynamic=false
       bIsEditable=false
       bIsStaticSwitch=false
       bIsVisible=false
       type {classStructOrEnum {refPath=&NiagaraDataInterfaceColorCurve}}
      }
      ScaleRGB {bIsDynamic=false, bIsEditable=true, bIsStaticSwitch=false, bIsVisible=true, type {classStructOrEnum {refPath=&NiagaraBool}}}
      Scale RGB {bIsDynamic=false, bIsEditable=true, bIsStaticSwitch=false, bIsVisible=true, type {classStructOrEnum {refPath=&Vector3f}}}
      ScaleRGBA {bIsDynamic=false, bIsEditable=false, bIsStaticSwitch=false, bIsVisible=false, type {classStructOrEnum {refPath=&NiagaraBool}}}
      Scale RGBA {
       bIsDynamic=false
       bIsEditable=false
       bIsStaticSwitch=false
       bIsVisible=false
       type {classStructOrEnum {refPath=&Vector4f}}
      }
      Curve Index {
       bIsDynamic=false
       bIsEditable=false
       bIsStaticSwitch=false
       bIsVisible=false
       type {classStructOrEnum {refPath=&NiagaraFloat}}
      }
      ScaleA {bIsDynamic=false, bIsEditable=true, bIsStaticSwitch=false, bIsVisible=true, type {classStructOrEnum {refPath=&NiagaraBool}}}
      Scale Alpha {bIsDynamic=false, bIsEditable=true, bIsStaticSwitch=false, bIsVisible=true, type {classStructOrEnum {refPath=&NiagaraFloat}}}
      Color Value To Scale {
       bIsDynamic=false
       bIsEditable=true
       bIsStaticSwitch=false
       bIsVisible=true
       type {classStructOrEnum {refPath=&LinearColor}}
      }
     ]
     moduleScript {refPath=&ScaleColor}
    }
   ]
  }
  rendererClasses [{refPath=&NiagaraSpriteRendererProperties}, {refPath=&NiagaraLightRendererProperties}]
  renderers [
   {rendererIndex=0, rendererClass {refPath=&NiagaraSpriteRendererProperties}}
   {rendererIndex=1, rendererClass {refPath=&NiagaraLightRendererProperties}}
   {rendererIndex=2, rendererClass {refPath=&NiagaraSpriteRendererProperties}}
  ]
 }
 HandGlow002 {
  bEnabled=false
  emitterSpawnScript {scriptName=EmitterSpawnScript, modules []}
  emitterUpdateScript {
   scriptName=EmitterUpdateScript
   modules [
    {
     bIsSetParametersModule=false
     enabled=true
     moduleName=EmitterState
     inputs [
      Life Cycle Mode {
       bIsDynamic=false
       bIsEditable=true
       bIsStaticSwitch=true
       bIsVisible=true
       type {classStructOrEnum {refPath=&ENiagaraEmitterLifeCycleMode}}
      }
      Inactive Response {
       bIsDynamic=false
       bIsEditable=true
       bIsStaticSwitch=true
       bIsVisible=true
       type {classStructOrEnum {refPath=&ENiagaraInactiveMode}}
      }
      Loop Behavior {
       bIsDynamic=false
       bIsEditable=true
       bIsStaticSwitch=true
       bIsVisible=true
       type {classStructOrEnum {refPath=&ENiagara_EmitterStateOptions}}
      }
      Loop Duration Mode {
       bIsDynamic=false
       bIsEditable=true
       bIsStaticSwitch=true
       bIsVisible=true
       type {classStructOrEnum {refPath=&ENiagara_InfiniteLoopDuration}}
      }
      Loop Count {
       bIsDynamic=false
       bIsEditable=false
       bIsStaticSwitch=false
       bIsVisible=false
       type {classStructOrEnum {refPath=&NiagaraInt32}}
      }
      Loop Duration {
       bIsDynamic=false
       bIsEditable=false
       bIsStaticSwitch=false
       bIsVisible=false
       type {classStructOrEnum {refPath=&NiagaraFloat}}
      }
      Recalculate Duration Each Loop {bIsDynamic=false, bIsEditable=false, bIsStaticSwitch=false, bIsVisible=false, type {classStructOrEnum {refPath=&NiagaraBool}}}
      UseLoopDelay {bIsDynamic=false, bIsEditable=true, bIsStaticSwitch=false, bIsVisible=true, type {classStructOrEnum {refPath=&NiagaraBool}}}
      Loop Delay {bIsDynamic=false, bIsEditable=false, bIsStaticSwitch=false, bIsVisible=true, type {classStructOrEnum {refPath=&NiagaraFloat}}}
      Delay First Loop Only {bIsDynamic=false, bIsEditable=false, bIsStaticSwitch=false, bIsVisible=false, type {classStructOrEnum {refPath=&NiagaraBool}}}
      Scalability Mode {
       bIsDynamic=false
       bIsEditable=true
       bIsStaticSwitch=true
       bIsVisible=true
       type {classStructOrEnum {refPath=&ENiagaraEmitterScalabilityMode_Limited}}
      }
      Enable Distance Culling {bIsDynamic=false, bIsEditable=false, bIsStaticSwitch=true, bIsVisible=false, type {classStructOrEnum {refPath=&NiagaraBool}}}
      MinDistance {
       bIsDynamic=false
       bIsEditable=false
       bIsStaticSwitch=false
       bIsVisible=false
       type {classStructOrEnum {refPath=&NiagaraFloat}}
      }
      Min Distance Response {
       bIsDynamic=false
       bIsEditable=false
       bIsStaticSwitch=false
       bIsVisible=false
       type {classStructOrEnum {refPath=&ENiagaraExecutionStateManagement}}
      }
      MaxDistance {
       bIsDynamic=false
       bIsEditable=false
       bIsStaticSwitch=false
       bIsVisible=false
       type {classStructOrEnum {refPath=&NiagaraFloat}}
      }
      Max Distance Response {
       bIsDynamic=false
       bIsEditable=false
       bIsStaticSwitch=false
       bIsVisible=false
       type {classStructOrEnum {refPath=&ENiagaraExecutionStateManagement}}
      }
      Scale Spawn Count {bIsDynamic=false, bIsEditable=false, bIsStaticSwitch=true, bIsVisible=false, type {classStructOrEnum {refPath=&NiagaraBool}}}
      Spawn Count Scale {
       bIsDynamic=false
       bIsEditable=false
       bIsStaticSwitch=false
       bIsVisible=false
       type {classStructOrEnum {refPath=&NiagaraFloat}}
      }
      Spawn Count Scale By Distance {
       bIsDynamic=false
       bIsEditable=false
       bIsStaticSwitch=false
       bIsVisible=false
       type {classStructOrEnum {refPath=&NiagaraDataInterfaceCurve}}
      }
      Enable Visibility Culling {bIsDynamic=false, bIsEditable=false, bIsStaticSwitch=true, bIsVisible=false, type {classStructOrEnum {refPath=&NiagaraBool}}}
      Visibility Cull Response {
       bIsDynamic=false
       bIsEditable=false
       bIsStaticSwitch=false
       bIsVisible=false
       type {classStructOrEnum {refPath=&ENiagaraExecutionStateManagement}}
      }
      Visibility Cull Delay {
       bIsDynamic=false
       bIsEditable=false
       bIsStaticSwitch=false
       bIsVisible=false
       type {classStructOrEnum {refPath=&NiagaraFloat}}
      }
      Reset Age On Awaken {bIsDynamic=false, bIsEditable=false, bIsStaticSwitch=false, bIsVisible=false, type {classStructOrEnum {refPath=&NiagaraBool}}}
     ]
     moduleScript {refPath=&EmitterState}
    }
    {
     bIsSetParametersModule=false
     enabled=true
     moduleName=SpawnBurst_Instantaneous
     inputs [
      Spawn Count {bIsDynamic=false, bIsEditable=true, bIsStaticSwitch=false, bIsVisible=true, type {classStructOrEnum {refPath=&NiagaraInt32}}}
      Spawn Time {bIsDynamic=false, bIsEditable=true, bIsStaticSwitch=false, bIsVisible=true, type {classStructOrEnum {refPath=&NiagaraFloat}}}
      Use Spawn Probability {bIsDynamic=false, bIsEditable=true, bIsStaticSwitch=false, bIsVisible=true, type {classStructOrEnum {refPath=&NiagaraBool}}}
      Spawn Probability {bIsDynamic=false, bIsEditable=true, bIsStaticSwitch=false, bIsVisible=true, type {classStructOrEnum {refPath=&NiagaraFloat}}}
      Spawn Group {bIsDynamic=false, bIsEditable=true, bIsStaticSwitch=false, bIsVisible=true, type {classStructOrEnum {refPath=&NiagaraInt32}}}
      Age {bIsDynamic=false, bIsEditable=true, bIsStaticSwitch=false, bIsVisible=true, type {classStructOrEnum {refPath=&NiagaraFloat}}}
     ]
     moduleScript {refPath=&SpawnBurst_Instantaneous}
    }
   ]
  }
  particleSpawnScript {
   scriptName=ParticleSpawnScript
   modules [
    {
     bIsSetParametersModule=false
     enabled=true
     moduleName=FN_SetSpawnGroup
     inputs [
      Number of Spawn Groups {bIsDynamic=false, bIsEditable=true, bIsStaticSwitch=false, bIsVisible=true, type {classStructOrEnum {refPath=&NiagaraInt32}}}
      Particles Per Group {
       bIsDynamic=false
       bIsEditable=false
       bIsStaticSwitch=false
       bIsVisible=false
       type {classStructOrEnum {refPath=&NiagaraInt32}}
      }
      Pattern {
       bIsDynamic=false
       bIsEditable=true
       bIsStaticSwitch=true
       bIsVisible=true
       type {classStructOrEnum {refPath=&ENiagaraPartitionWaveForm}}
      }
      Normalize Spawn Group {bIsDynamic=false, bIsEditable=true, bIsStaticSwitch=true, bIsVisible=true, type {classStructOrEnum {refPath=&NiagaraBool}}}
      Inclusivity {bIsDynamic=false, bIsEditable=false, bIsStaticSwitch=true, bIsVisible=false, type {classStructOrEnum {refPath=&NiagaraBool}}}
      Set Ribbon ID and Ribbon Link Order {bIsDynamic=false, bIsEditable=true, bIsStaticSwitch=true, bIsVisible=true, type {classStructOrEnum {refPath=&NiagaraBool}}}
     ]
     moduleScript {refPath=&FN_SetSpawnGroup}
    }
    {
     bIsSetParametersModule=false
     enabled=true
     moduleName=InitializeParticle001
     inputs [
      Lifetime Mode {
       bIsDynamic=false
       bIsEditable=true
       bIsStaticSwitch=true
       bIsVisible=true
       type {classStructOrEnum {refPath=&ENiagara_LifetimeMode}}
      }
      Lifetime Min {
       bIsDynamic=false
       bIsEditable=false
       bIsStaticSwitch=false
       bIsVisible=false
       type {classStructOrEnum {refPath=&NiagaraFloat}}
      }
      Lifetime {bIsDynamic=false, bIsEditable=true, bIsStaticSwitch=false, bIsVisible=true, type {classStructOrEnum {refPath=&NiagaraFloat}}}
      Lifetime Max {
       bIsDynamic=false
       bIsEditable=false
       bIsStaticSwitch=false
       bIsVisible=false
       type {classStructOrEnum {refPath=&NiagaraFloat}}
      }
      Lifetime Randomness Mode {
       bIsDynamic=false
       bIsEditable=false
       bIsStaticSwitch=true
       bIsVisible=false
       type {classStructOrEnum {refPath=&ENiagaraRandomnessMode}}
      }
      Color Mode {
       bIsDynamic=false
       bIsEditable=true
       bIsStaticSwitch=true
       bIsVisible=true
       type {classStructOrEnum {refPath=&ENiagara_ColorInitializationMode}}
      }
      Color Minimum {
       bIsDynamic=false
       bIsEditable=false
       bIsStaticSwitch=false
       bIsVisible=false
       type {classStructOrEnum {refPath=&LinearColor}}
      }
      Color {
       bIsDynamic=false
       bIsEditable=true
       bIsStaticSwitch=false
       bIsVisible=true
       type {classStructOrEnum {refPath=&LinearColor}}
      }
      AdjustHue {bIsDynamic=false, bIsEditable=false, bIsStaticSwitch=false, bIsVisible=false, type {classStructOrEnum {refPath=&NiagaraBool}}}
      Hue Shift Range {
       bIsDynamic=false
       bIsEditable=false
       bIsStaticSwitch=false
       bIsVisible=false
       type {classStructOrEnum {refPath=&Vector2f}}
      }
      Color Maximum {
       bIsDynamic=false
       bIsEditable=false
       bIsStaticSwitch=false
       bIsVisible=false
       type {classStructOrEnum {refPath=&LinearColor}}
      }
      AdjustSaturation {bIsDynamic=false, bIsEditable=false, bIsStaticSwitch=false, bIsVisible=false, type {classStructOrEnum {refPath=&NiagaraBool}}}
      Saturation Range {
       bIsDynamic=false
       bIsEditable=false
       bIsStaticSwitch=false
       bIsVisible=false
       type {classStructOrEnum {refPath=&Vector2f}}
      }
      AdjustValue {bIsDynamic=false, bIsEditable=false, bIsStaticSwitch=false, bIsVisible=false, type {classStructOrEnum {refPath=&NiagaraBool}}}
      Value Range {
       bIsDynamic=false
       bIsEditable=false
       bIsStaticSwitch=false
       bIsVisible=false
       type {classStructOrEnum {refPath=&Vector2f}}
      }
      AdjustAlpha {bIsDynamic=false, bIsEditable=false, bIsStaticSwitch=false, bIsVisible=false, type {classStructOrEnum {refPath=&NiagaraBool}}}
      Alpha Scale Range {
       bIsDynamic=false
       bIsEditable=false
       bIsStaticSwitch=false
       bIsVisible=false
       type {classStructOrEnum {refPath=&Vector2f}}
      }
      Color Channel Mode {
       bIsDynamic=false
       bIsEditable=false
       bIsStaticSwitch=true
       bIsVisible=false
       type {classStructOrEnum {refPath=&ENiagaraChannelCorrelation}}
      }
      Color Randomness Mode {
       bIsDynamic=false
       bIsEditable=false
       bIsStaticSwitch=true
       bIsVisible=false
       type {classStructOrEnum {refPath=&ENiagaraRandomnessMode}}
      }
      Position Mode {
       bIsDynamic=false
       bIsEditable=true
       bIsStaticSwitch=true
       bIsVisible=true
       type {classStructOrEnum {refPath=&ENiagara_PositionInitializationMode}}
      }
      Position {
       bIsDynamic=true
       bIsEditable=false
       bIsStaticSwitch=false
       bIsVisible=false
       type {classStructOrEnum {refPath=&NiagaraPosition}}
      }
      UsePositionOffset {bIsDynamic=false, bIsEditable=true, bIsStaticSwitch=false, bIsVisible=true, type {classStructOrEnum {refPath=&NiagaraBool}}}
      Position Offset {bIsDynamic=false, bIsEditable=false, bIsStaticSwitch=false, bIsVisible=true, type {classStructOrEnum {refPath=&Vector3f}}}
      Position Offset Coordinate Space {
       bIsDynamic=false
       bIsEditable=false
       bIsStaticSwitch=false
       bIsVisible=false
       type {classStructOrEnum {refPath=&ENiagaraCoordinateSpace}}
      }
      Mass Mode {
       bIsDynamic=false
       bIsEditable=true
       bIsStaticSwitch=true
       bIsVisible=true
       type {classStructOrEnum {refPath=&ENiagara_MassInitializationMode}}
      }
      Mass Min {
       bIsDynamic=false
       bIsEditable=false
       bIsStaticSwitch=false
       bIsVisible=false
       type {classStructOrEnum {refPath=&NiagaraFloat}}
      }
      Mass {bIsDynamic=false, bIsEditable=true, bIsStaticSwitch=false, bIsVisible=true, type {classStructOrEnum {refPath=&NiagaraFloat}}}
      Mass Max {
       bIsDynamic=false
       bIsEditable=false
       bIsStaticSwitch=false
       bIsVisible=false
       type {classStructOrEnum {refPath=&NiagaraFloat}}
      }
      Mass Randomness Mode {
       bIsDynamic=false
       bIsEditable=false
       bIsStaticSwitch=true
       bIsVisible=false
       type {classStructOrEnum {refPath=&ENiagaraRandomnessMode}}
      }
      Write Material Random {bIsDynamic=false, bIsEditable=true, bIsStaticSwitch=false, bIsVisible=true, type {classStructOrEnum {refPath=&NiagaraBool}}}
      Material Random {bIsDynamic=false, bIsEditable=true, bIsStaticSwitch=false, bIsVisible=true, type {classStructOrEnum {refPath=&NiagaraFloat}}}
      Sprite Size Mode {
       bIsDynamic=false
       bIsEditable=true
       bIsStaticSwitch=true
       bIsVisible=true
       type {classStructOrEnum {refPath=&ENiagara_SizeScaleMode}}
      }
      Sprite Size {
       bIsDynamic=false
       bIsEditable=false
       bIsStaticSwitch=false
       bIsVisible=false
       type {classStructOrEnum {refPath=&Vector2f}}
      }
      Uniform Sprite Size {bIsDynamic=false, bIsEditable=true, bIsStaticSwitch=false, bIsVisible=true, type {classStructOrEnum {refPath=&NiagaraFloat}}}
      Uniform Sprite Size Min {
       bIsDynamic=false
       bIsEditable=false
       bIsStaticSwitch=false
       bIsVisible=false
       type {classStructOrEnum {refPath=&NiagaraFloat}}
      }
      Sprite Size Min {
       bIsDynamic=false
       bIsEditable=false
       bIsStaticSwitch=false
       bIsVisible=false
       type {classStructOrEnum {refPath=&Vector2f}}
      }
      Sprite Size Max {
       bIsDynamic=false
       bIsEditable=false
       bIsStaticSwitch=false
       bIsVisible=false
       type {classStructOrEnum {refPath=&Vector2f}}
      }
      Uniform Sprite Size Max {
       bIsDynamic=false
       bIsEditable=false
       bIsStaticSwitch=false
       bIsVisible=false
       type {classStructOrEnum {refPath=&NiagaraFloat}}
      }
      Sprite Size Randomness Mode {
       bIsDynamic=false
       bIsEditable=false
       bIsStaticSwitch=true
       bIsVisible=false
       type {classStructOrEnum {refPath=&ENiagaraRandomnessMode}}
      }
      Sprite Rotation Mode {
       bIsDynamic=false
       bIsEditable=true
       bIsStaticSwitch=true
       bIsVisible=true
       type {classStructOrEnum {refPath=&ENiagara_SpriteRotationMode}}
      }
      Sprite Rotation Angle Min {
       bIsDynamic=false
       bIsEditable=false
       bIsStaticSwitch=false
       bIsVisible=false
       type {classStructOrEnum {refPath=&NiagaraFloat}}
      }
      Sprite Rotation Angle {
       bIsDynamic=false
       bIsEditable=false
       bIsStaticSwitch=false
       bIsVisible=false
       type {classStructOrEnum {refPath=&NiagaraFloat}}
      }
      Sprite Rotation Angle Max {
       bIsDynamic=false
       bIsEditable=false
       bIsStaticSwitch=false
       bIsVisible=false
       type {classStructOrEnum {refPath=&NiagaraFloat}}
      }
      Sprite Rotation Randomness Mode {
       bIsDynamic=false
       bIsEditable=false
       bIsStaticSwitch=true
       bIsVisible=false
       type {classStructOrEnum {refPath=&ENiagaraRandomnessMode}}
      }
      Sprite UV Mode {
       bIsDynamic=false
       bIsEditable=true
       bIsStaticSwitch=true
       bIsVisible=true
       type {classStructOrEnum {refPath=&ENiagara_UVFlippingMode}}
      }
      Sprite UV Randomness Mode {
       bIsDynamic=false
       bIsEditable=false
       bIsStaticSwitch=true
       bIsVisible=false
       type {classStructOrEnum {refPath=&ENiagaraRandomnessMode}}
      }
      Sprite UV Scale {
       bIsDynamic=false
       bIsEditable=false
       bIsStaticSwitch=false
       bIsVisible=false
       type {classStructOrEnum {refPath=&Vector2f}}
      }
      Mesh Scale Mode {
       bIsDynamic=false
       bIsEditable=true
       bIsStaticSwitch=true
       bIsVisible=true
       type {classStructOrEnum {refPath=&ENiagara_SizeScaleMode}}
      }
      Mesh Uniform Scale Min {
       bIsDynamic=false
       bIsEditable=false
       bIsStaticSwitch=false
       bIsVisible=false
       type {classStructOrEnum {refPath=&NiagaraFloat}}
      }
      Mesh Uniform Scale {
       bIsDynamic=false
       bIsEditable=false
       bIsStaticSwitch=false
       bIsVisible=false
       type {classStructOrEnum {refPath=&NiagaraFloat}}
      }
      Mesh Scale Min {
       bIsDynamic=false
       bIsEditable=false
       bIsStaticSwitch=false
       bIsVisible=false
       type {classStructOrEnum {refPath=&Vector3f}}
      }
      Mesh Scale {
       bIsDynamic=false
       bIsEditable=false
       bIsStaticSwitch=false
       bIsVisible=false
       type {classStructOrEnum {refPath=&Vector3f}}
      }
      Mesh Scale Max {
       bIsDynamic=false
       bIsEditable=false
       bIsStaticSwitch=false
       bIsVisible=false
       type {classStructOrEnum {refPath=&Vector3f}}
      }
      Mesh Uniform Scale Max {
       bIsDynamic=false
       bIsEditable=false
       bIsStaticSwitch=false
       bIsVisible=false
       type {classStructOrEnum {refPath=&NiagaraFloat}}
      }
      Mesh Scale Randomness Mode {
       bIsDynamic=false
       bIsEditable=false
       bIsStaticSwitch=true
       bIsVisible=false
       type {classStructOrEnum {refPath=&ENiagaraRandomnessMode}}
      }
      Mesh Renderer Array Visibility Mode {
       bIsDynamic=false
       bIsEditable=true
       bIsStaticSwitch=true
       bIsVisible=true
       type {classStructOrEnum {refPath=&ENiagara_UnsetDirectSetRandom}}
      }
      Mesh Renderer Mesh Index {
       bIsDynamic=false
       bIsEditable=false
       bIsStaticSwitch=false
       bIsVisible=false
       type {classStructOrEnum {refPath=&NiagaraInt32}}
      }
      Wrap Index By Mesh Count {bIsDynamic=false, bIsEditable=false, bIsStaticSwitch=true, bIsVisible=false, type {classStructOrEnum {refPath=&NiagaraBool}}}
      Mesh Renderer Info {
       bIsDynamic=false
       bIsEditable=false
       bIsStaticSwitch=false
       bIsVisible=false
       type {classStructOrEnum {refPath=&NiagaraDataInterfaceMeshRendererInfo}}
      }
      Mesh Renderer Random Seed {
       bIsDynamic=false
       bIsEditable=false
       bIsStaticSwitch=false
       bIsVisible=false
       type {classStructOrEnum {refPath=&NiagaraInt32}}
      }
      Ribbon Width Mode {
       bIsDynamic=false
       bIsEditable=true
       bIsStaticSwitch=true
       bIsVisible=true
       type {classStructOrEnum {refPath=&ENiagara_UnsetDirectSet}}
      }
      Ribbon Width {
       bIsDynamic=false
       bIsEditable=false
       bIsStaticSwitch=false
       bIsVisible=false
       type {classStructOrEnum {refPath=&NiagaraFloat}}
      }
      Ribbon Facing Mode {
       bIsDynamic=false
       bIsEditable=true
       bIsStaticSwitch=true
       bIsVisible=true
       type {classStructOrEnum {refPath=&ENiagara_UnsetDirectSet}}
      }
      Ribbon Facing Vector {
       bIsDynamic=false
       bIsEditable=false
       bIsStaticSwitch=false
       bIsVisible=false
       type {classStructOrEnum {refPath=&Vector3f}}
      }
      Ribbon Twist Mode {
       bIsDynamic=false
       bIsEditable=true
       bIsStaticSwitch=true
       bIsVisible=true
       type {classStructOrEnum {refPath=&ENiagara_UnsetDirectSet}}
      }
      "Ribbon Twist " {
       bIsDynamic=false
       bIsEditable=false
       bIsStaticSwitch=false
       bIsVisible=false
       type {classStructOrEnum {refPath=&NiagaraFloat}}
      }
     ]
     moduleScript {refPath=&InitializeParticle}
    }
   ]
  }
  particleUpdateScript {
   scriptName=ParticleUpdateScript
   modules [
    {
     bIsSetParametersModule=false
     enabled=true
     moduleName=ParticleState
     inputs [
      Kill Particles When Lifetime Has Elapsed {bIsDynamic=false, bIsEditable=true, bIsStaticSwitch=true, bIsVisible=true, type {classStructOrEnum {refPath=&NiagaraBool}}}
      Loop Particles Lifetime {bIsDynamic=false, bIsEditable=false, bIsStaticSwitch=false, bIsVisible=false, type {classStructOrEnum {refPath=&NiagaraBool}}}
      Let Infinitely Lived Particles Die When Emitter Deactivates {bIsDynamic=false, bIsEditable=false, bIsStaticSwitch=false, bIsVisible=false, type {classStructOrEnum {refPath=&NiagaraBool}}}
      Lifetime {bIsDynamic=false, bIsEditable=true, bIsStaticSwitch=false, bIsVisible=true, type {classStructOrEnum {refPath=&NiagaraFloat}}}
      DeltaTime {bIsDynamic=false, bIsEditable=true, bIsStaticSwitch=false, bIsVisible=true, type {classStructOrEnum {refPath=&NiagaraFloat}}}
     ]
     moduleScript {refPath=&ParticleState}
    }
    {
     bIsSetParametersModule=false
     enabled=true
     moduleName=SkeletalMeshLocation
     inputs [
      Skeletal Mesh {
       bIsDynamic=false
       bIsEditable=true
       bIsStaticSwitch=false
       bIsVisible=true
       type {classStructOrEnum {refPath=&NiagaraDataInterfaceSkeletalMesh}}
      }
      Mesh Sampling Type {
       bIsDynamic=false
       bIsEditable=true
       bIsStaticSwitch=true
       bIsVisible=true
       type {classStructOrEnum {refPath=&ENiagaraSkelSamplingModeFull}}
      }
      Bone / Socket Sampling Mode {
       bIsDynamic=false
       bIsEditable=false
       bIsStaticSwitch=true
       bIsVisible=false
       type {classStructOrEnum {refPath=&ENiagaraBoneSocketSamplingMode}}
      }
      Bone Sampling Mode {
       bIsDynamic=false
       bIsEditable=true
       bIsStaticSwitch=true
       bIsVisible=true
       type {classStructOrEnum {refPath=&ENiagaraBoneSamplingMode}}
      }
      Socket Sampling Mode {
       bIsDynamic=false
       bIsEditable=false
       bIsStaticSwitch=true
       bIsVisible=false
       type {classStructOrEnum {refPath=&ENiagaraSocketSamplingMode}}
      }
      Triangle Sampling Mode {
       bIsDynamic=false
       bIsEditable=false
       bIsStaticSwitch=true
       bIsVisible=false
       type {classStructOrEnum {refPath=&ENiagaraTriangleSamplingMode}}
      }
      Vertex Sampling Mode {
       bIsDynamic=false
       bIsEditable=false
       bIsStaticSwitch=true
       bIsVisible=false
       type {classStructOrEnum {refPath=&ENiagaraVertexSamplingMode}}
      }
      Bone / Socket Index {bIsDynamic=false, bIsEditable=true, bIsStaticSwitch=false, bIsVisible=true, type {classStructOrEnum {refPath=&NiagaraInt32}}}
      Triangle ID {
       bIsDynamic=false
       bIsEditable=false
       bIsStaticSwitch=false
       bIsVisible=false
       type {classStructOrEnum {refPath=&NiagaraInt32}}
      }
      Vertex ID {
       bIsDynamic=false
       bIsEditable=false
       bIsStaticSwitch=false
       bIsVisible=false
       type {classStructOrEnum {refPath=&NiagaraInt32}}
      }
      FixedBarycentric {bIsDynamic=false, bIsEditable=false, bIsStaticSwitch=false, bIsVisible=false, type {classStructOrEnum {refPath=&NiagaraBool}}}
      Triangle Coordinate {
       bIsDynamic=false
       bIsEditable=false
       bIsStaticSwitch=false
       bIsVisible=false
       type {classStructOrEnum {refPath=&Vector3f}}
      }
      Sampled UV Index {
       bIsDynamic=false
       bIsEditable=false
       bIsStaticSwitch=false
       bIsVisible=false
       type {classStructOrEnum {refPath=&NiagaraInt32}}
      }
      Sampled Position Coordinate Space {
       bIsDynamic=false
       bIsEditable=true
       bIsStaticSwitch=true
       bIsVisible=true
       type {classStructOrEnum {refPath=&ENiagaraCoordinateSpace}}
      }
      Sampled Velocity Coordinate Space {
       bIsDynamic=false
       bIsEditable=true
       bIsStaticSwitch=true
       bIsVisible=true
       type {classStructOrEnum {refPath=&ENiagaraCoordinateSpace}}
      }
      Sampled Orientation Coordinate Space {
       bIsDynamic=false
       bIsEditable=true
       bIsStaticSwitch=true
       bIsVisible=true
       type {classStructOrEnum {refPath=&ENiagaraCoordinateSpace}}
      }
      Generate New Random Samples Every Frame {bIsDynamic=false, bIsEditable=false, bIsStaticSwitch=true, bIsVisible=false, type {classStructOrEnum {refPath=&NiagaraBool}}}
      Write Sampled Mesh Attrributes to Particle Attributes {bIsDynamic=false, bIsEditable=true, bIsStaticSwitch=true, bIsVisible=true, type {classStructOrEnum {refPath=&NiagaraBool}}}
      Randomness Mode {
       bIsDynamic=false
       bIsEditable=false
       bIsStaticSwitch=false
       bIsVisible=false
       type {classStructOrEnum {refPath=&ENiagaraRandomnessMode}}
      }
      Random Seed {
       bIsDynamic=false
       bIsEditable=false
       bIsStaticSwitch=false
       bIsVisible=false
       type {classStructOrEnum {refPath=&NiagaraInt32}}
      }
      Fixed Random Seed {bIsDynamic=false, bIsEditable=false, bIsStaticSwitch=false, bIsVisible=false, type {classStructOrEnum {refPath=&NiagaraBool}}}
      Apply To Particle Position {bIsDynamic=false, bIsEditable=true, bIsStaticSwitch=false, bIsVisible=true, type {classStructOrEnum {refPath=&NiagaraBool}}}
      Sampled Position Offset {bIsDynamic=false, bIsEditable=true, bIsStaticSwitch=false, bIsVisible=true, type {classStructOrEnum {refPath=&Vector3f}}}
      OffsetAlongNormal {bIsDynamic=false, bIsEditable=false, bIsStaticSwitch=false, bIsVisible=false, type {classStructOrEnum {refPath=&NiagaraBool}}}
      Offset Position Along Sampled Normal {
       bIsDynamic=false
       bIsEditable=false
       bIsStaticSwitch=false
       bIsVisible=false
       type {classStructOrEnum {refPath=&NiagaraFloat}}
      }
      Sampled Position Offset Coordinate Space {
       bIsDynamic=false
       bIsEditable=true
       bIsStaticSwitch=true
       bIsVisible=true
       type {classStructOrEnum {refPath=&ENiagaraMeshTransforms}}
      }
      Inherit Velocity {bIsDynamic=false, bIsEditable=true, bIsStaticSwitch=false, bIsVisible=true, type {classStructOrEnum {refPath=&NiagaraBool}}}
      "Inherit Velocity (Scale)" {bIsDynamic=false, bIsEditable=false, bIsStaticSwitch=false, bIsVisible=true, type {classStructOrEnum {refPath=&NiagaraFloat}}}
      VelocitySpeedLimit {bIsDynamic=false, bIsEditable=false, bIsStaticSwitch=false, bIsVisible=false, type {classStructOrEnum {refPath=&NiagaraBool}}}
      Inherited Velocity Speed Limit {
       bIsDynamic=false
       bIsEditable=false
       bIsStaticSwitch=false
       bIsVisible=false
       type {classStructOrEnum {refPath=&NiagaraFloat}}
      }
      Apply Sampled Orientation to Mesh Particles {bIsDynamic=false, bIsEditable=true, bIsStaticSwitch=true, bIsVisible=true, type {classStructOrEnum {refPath=&NiagaraBool}}}
      Interpolate Mesh Sampling {bIsDynamic=false, bIsEditable=true, bIsStaticSwitch=true, bIsVisible=true, type {classStructOrEnum {refPath=&NiagaraBool}}}
      MaskSpawnGroup {bIsDynamic=false, bIsEditable=true, bIsStaticSwitch=false, bIsVisible=true, type {classStructOrEnum {refPath=&NiagaraBool}}}
      Spawn Group Mask {bIsDynamic=false, bIsEditable=false, bIsStaticSwitch=false, bIsVisible=true, type {classStructOrEnum {refPath=&NiagaraInt32}}}
      Use Sampled Mesh Attribute as Ribbon ID {bIsDynamic=false, bIsEditable=true, bIsStaticSwitch=false, bIsVisible=true, type {classStructOrEnum {refPath=&NiagaraBool}}}
     ]
     moduleScript {refPath=&SkeletalMeshLocation}
    }
    {
     bIsSetParametersModule=false
     enabled=true
     moduleName=CameraOffset
     inputs [Camera Offset Amount {bIsDynamic=false, bIsEditable=true, bIsStaticSwitch=false, bIsVisible=true, type {classStructOrEnum {refPath=&NiagaraFloat}}}]
     moduleScript {refPath=&CameraOffset}
    }
    {
     bIsSetParametersModule=false
     enabled=true
     moduleName=ScaleColor
     inputs [
      Scale Mode {
       bIsDynamic=false
       bIsEditable=true
       bIsStaticSwitch=true
       bIsVisible=true
       type {classStructOrEnum {refPath=&ENiagaraScaleColorMode}}
      }
      Linear Color Curve {
       bIsDynamic=false
       bIsEditable=true
       bIsStaticSwitch=false
       bIsVisible=true
       type {classStructOrEnum {refPath=&NiagaraDataInterfaceColorCurve}}
      }
      ScaleRGB {bIsDynamic=false, bIsEditable=false, bIsStaticSwitch=false, bIsVisible=false, type {classStructOrEnum {refPath=&NiagaraBool}}}
      Scale RGB {
       bIsDynamic=false
       bIsEditable=false
       bIsStaticSwitch=false
       bIsVisible=false
       type {classStructOrEnum {refPath=&Vector3f}}
      }
      ScaleRGBA {bIsDynamic=false, bIsEditable=false, bIsStaticSwitch=false, bIsVisible=false, type {classStructOrEnum {refPath=&NiagaraBool}}}
      Scale RGBA {
       bIsDynamic=false
       bIsEditable=false
       bIsStaticSwitch=false
       bIsVisible=false
       type {classStructOrEnum {refPath=&Vector4f}}
      }
      Curve Index {bIsDynamic=false, bIsEditable=true, bIsStaticSwitch=false, bIsVisible=true, type {classStructOrEnum {refPath=&NiagaraFloat}}}
      ScaleA {bIsDynamic=false, bIsEditable=false, bIsStaticSwitch=false, bIsVisible=false, type {classStructOrEnum {refPath=&NiagaraBool}}}
      Scale Alpha {
       bIsDynamic=false
       bIsEditable=false
       bIsStaticSwitch=false
       bIsVisible=false
       type {classStructOrEnum {refPath=&NiagaraFloat}}
      }
      Color Value To Scale {
       bIsDynamic=false
       bIsEditable=true
       bIsStaticSwitch=false
       bIsVisible=true
       type {classStructOrEnum {refPath=&LinearColor}}
      }
     ]
     moduleScript {refPath=&ScaleColor}
    }
    {
     bIsSetParametersModule=false
     enabled=true
     moduleName=FN_Light_v2
     inputs [
      LightEnabled {bIsDynamic=false, bIsEditable=true, bIsStaticSwitch=false, bIsVisible=true, type {classStructOrEnum {refPath=&NiagaraBool}}}
      Scale Color {bIsDynamic=false, bIsEditable=true, bIsStaticSwitch=false, bIsVisible=true, type {classStructOrEnum {refPath=&NiagaraFloat}}}
      Light Exponent {bIsDynamic=false, bIsEditable=true, bIsStaticSwitch=false, bIsVisible=true, type {classStructOrEnum {refPath=&NiagaraFloat}}}
      Position Offset {bIsDynamic=false, bIsEditable=true, bIsStaticSwitch=false, bIsVisible=true, type {classStructOrEnum {refPath=&Vector3f}}}
      Light Radius {bIsDynamic=false, bIsEditable=true, bIsStaticSwitch=false, bIsVisible=true, type {classStructOrEnum {refPath=&NiagaraFloat}}}
     ]
     moduleScript {refPath=&FN_Light_v2}
    }
    {
     bIsSetParametersModule=false
     enabled=true
     moduleName=ScaleColor001
     inputs [
      Scale Mode {
       bIsDynamic=false
       bIsEditable=true
       bIsStaticSwitch=true
       bIsVisible=true
       type {classStructOrEnum {refPath=&ENiagaraScaleColorMode}}
      }
      Linear Color Curve {
       bIsDynamic=false
       bIsEditable=false
       bIsStaticSwitch=false
       bIsVisible=false
       type {classStructOrEnum {refPath=&NiagaraDataInterfaceColorCurve}}
      }
      ScaleRGB {bIsDynamic=false, bIsEditable=true, bIsStaticSwitch=false, bIsVisible=true, type {classStructOrEnum {refPath=&NiagaraBool}}}
      Scale RGB {bIsDynamic=false, bIsEditable=true, bIsStaticSwitch=false, bIsVisible=true, type {classStructOrEnum {refPath=&Vector3f}}}
      ScaleRGBA {bIsDynamic=false, bIsEditable=false, bIsStaticSwitch=false, bIsVisible=false, type {classStructOrEnum {refPath=&NiagaraBool}}}
      Scale RGBA {
       bIsDynamic=false
       bIsEditable=false
       bIsStaticSwitch=false
       bIsVisible=false
       type {classStructOrEnum {refPath=&Vector4f}}
      }
      Curve Index {
       bIsDynamic=false
       bIsEditable=false
       bIsStaticSwitch=false
       bIsVisible=false
       type {classStructOrEnum {refPath=&NiagaraFloat}}
      }
      ScaleA {bIsDynamic=false, bIsEditable=true, bIsStaticSwitch=false, bIsVisible=true, type {classStructOrEnum {refPath=&NiagaraBool}}}
      Scale Alpha {bIsDynamic=false, bIsEditable=true, bIsStaticSwitch=false, bIsVisible=true, type {classStructOrEnum {refPath=&NiagaraFloat}}}
      Color Value To Scale {
       bIsDynamic=false
       bIsEditable=true
       bIsStaticSwitch=false
       bIsVisible=true
       type {classStructOrEnum {refPath=&LinearColor}}
      }
     ]
     moduleScript {refPath=&ScaleColor}
    }
   ]
  }
  rendererClasses [{refPath=&NiagaraSpriteRendererProperties}, {refPath=&NiagaraLightRendererProperties}]
  renderers [
   {rendererIndex=0, rendererClass {refPath=&NiagaraSpriteRendererProperties}}
   {rendererIndex=1, rendererClass {refPath=&NiagaraLightRendererProperties}}
  ]
 }
 HandGlow003 {
  bEnabled=false
  emitterSpawnScript {scriptName=EmitterSpawnScript, modules []}
  emitterUpdateScript {
   scriptName=EmitterUpdateScript
   modules [
    {
     bIsSetParametersModule=false
     enabled=true
     moduleName=EmitterState
     inputs [
      Life Cycle Mode {
       bIsDynamic=false
       bIsEditable=true
       bIsStaticSwitch=true
       bIsVisible=true
       type {classStructOrEnum {refPath=&ENiagaraEmitterLifeCycleMode}}
      }
      Inactive Response {
       bIsDynamic=false
       bIsEditable=true
       bIsStaticSwitch=true
       bIsVisible=true
       type {classStructOrEnum {refPath=&ENiagaraInactiveMode}}
      }
      Loop Behavior {
       bIsDynamic=false
       bIsEditable=true
       bIsStaticSwitch=true
       bIsVisible=true
       type {classStructOrEnum {refPath=&ENiagara_EmitterStateOptions}}
      }
      Loop Duration Mode {
       bIsDynamic=false
       bIsEditable=true
       bIsStaticSwitch=true
       bIsVisible=true
       type {classStructOrEnum {refPath=&ENiagara_InfiniteLoopDuration}}
      }
      Loop Count {
       bIsDynamic=false
       bIsEditable=false
       bIsStaticSwitch=false
       bIsVisible=false
       type {classStructOrEnum {refPath=&NiagaraInt32}}
      }
      Loop Duration {
       bIsDynamic=false
       bIsEditable=false
       bIsStaticSwitch=false
       bIsVisible=false
       type {classStructOrEnum {refPath=&NiagaraFloat}}
      }
      Recalculate Duration Each Loop {bIsDynamic=false, bIsEditable=false, bIsStaticSwitch=false, bIsVisible=false, type {classStructOrEnum {refPath=&NiagaraBool}}}
      UseLoopDelay {bIsDynamic=false, bIsEditable=true, bIsStaticSwitch=false, bIsVisible=true, type {classStructOrEnum {refPath=&NiagaraBool}}}
      Loop Delay {bIsDynamic=false, bIsEditable=false, bIsStaticSwitch=false, bIsVisible=true, type {classStructOrEnum {refPath=&NiagaraFloat}}}
      Delay First Loop Only {bIsDynamic=false, bIsEditable=false, bIsStaticSwitch=false, bIsVisible=false, type {classStructOrEnum {refPath=&NiagaraBool}}}
      Scalability Mode {
       bIsDynamic=false
       bIsEditable=true
       bIsStaticSwitch=true
       bIsVisible=true
       type {classStructOrEnum {refPath=&ENiagaraEmitterScalabilityMode_Limited}}
      }
      Enable Distance Culling {bIsDynamic=false, bIsEditable=false, bIsStaticSwitch=true, bIsVisible=false, type {classStructOrEnum {refPath=&NiagaraBool}}}
      MinDistance {
       bIsDynamic=false
       bIsEditable=false
       bIsStaticSwitch=false
       bIsVisible=false
       type {classStructOrEnum {refPath=&NiagaraFloat}}
      }
      Min Distance Response {
       bIsDynamic=false
       bIsEditable=false
       bIsStaticSwitch=false
       bIsVisible=false
       type {classStructOrEnum {refPath=&ENiagaraExecutionStateManagement}}
      }
      MaxDistance {
       bIsDynamic=false
       bIsEditable=false
       bIsStaticSwitch=false
       bIsVisible=false
       type {classStructOrEnum {refPath=&NiagaraFloat}}
      }
      Max Distance Response {
       bIsDynamic=false
       bIsEditable=false
       bIsStaticSwitch=false
       bIsVisible=false
       type {classStructOrEnum {refPath=&ENiagaraExecutionStateManagement}}
      }
      Scale Spawn Count {bIsDynamic=false, bIsEditable=false, bIsStaticSwitch=true, bIsVisible=false, type {classStructOrEnum {refPath=&NiagaraBool}}}
      Spawn Count Scale {
       bIsDynamic=false
       bIsEditable=false
       bIsStaticSwitch=false
       bIsVisible=false
       type {classStructOrEnum {refPath=&NiagaraFloat}}
      }
      Spawn Count Scale By Distance {
       bIsDynamic=false
       bIsEditable=false
       bIsStaticSwitch=false
       bIsVisible=false
       type {classStructOrEnum {refPath=&NiagaraDataInterfaceCurve}}
      }
      Enable Visibility Culling {bIsDynamic=false, bIsEditable=false, bIsStaticSwitch=true, bIsVisible=false, type {classStructOrEnum {refPath=&NiagaraBool}}}
      Visibility Cull Response {
       bIsDynamic=false
       bIsEditable=false
       bIsStaticSwitch=false
       bIsVisible=false
       type {classStructOrEnum {refPath=&ENiagaraExecutionStateManagement}}
      }
      Visibility Cull Delay {
       bIsDynamic=false
       bIsEditable=false
       bIsStaticSwitch=false
       bIsVisible=false
       type {classStructOrEnum {refPath=&NiagaraFloat}}
      }
      Reset Age On Awaken {bIsDynamic=false, bIsEditable=false, bIsStaticSwitch=false, bIsVisible=false, type {classStructOrEnum {refPath=&NiagaraBool}}}
     ]
     moduleScript {refPath=&EmitterState}
    }
    {
     bIsSetParametersModule=false
     enabled=true
     moduleName=SpawnBurst_Instantaneous
     inputs [
      Spawn Count {bIsDynamic=false, bIsEditable=true, bIsStaticSwitch=false, bIsVisible=true, type {classStructOrEnum {refPath=&NiagaraInt32}}}
      Spawn Time {bIsDynamic=false, bIsEditable=true, bIsStaticSwitch=false, bIsVisible=true, type {classStructOrEnum {refPath=&NiagaraFloat}}}
      Use Spawn Probability {bIsDynamic=false, bIsEditable=true, bIsStaticSwitch=false, bIsVisible=true, type {classStructOrEnum {refPath=&NiagaraBool}}}
      Spawn Probability {bIsDynamic=false, bIsEditable=true, bIsStaticSwitch=false, bIsVisible=true, type {classStructOrEnum {refPath=&NiagaraFloat}}}
      Spawn Group {bIsDynamic=false, bIsEditable=true, bIsStaticSwitch=false, bIsVisible=true, type {classStructOrEnum {refPath=&NiagaraInt32}}}
      Age {bIsDynamic=false, bIsEditable=true, bIsStaticSwitch=false, bIsVisible=true, type {classStructOrEnum {refPath=&NiagaraFloat}}}
     ]
     moduleScript {refPath=&SpawnBurst_Instantaneous}
    }
   ]
  }
  particleSpawnScript {
   scriptName=ParticleSpawnScript
   modules [
    {
     bIsSetParametersModule=false
     enabled=true
     moduleName=FN_SetSpawnGroup
     inputs [
      Number of Spawn Groups {bIsDynamic=false, bIsEditable=true, bIsStaticSwitch=false, bIsVisible=true, type {classStructOrEnum {refPath=&NiagaraInt32}}}
      Particles Per Group {
       bIsDynamic=false
       bIsEditable=false
       bIsStaticSwitch=false
       bIsVisible=false
       type {classStructOrEnum {refPath=&NiagaraInt32}}
      }
      Pattern {
       bIsDynamic=false
       bIsEditable=true
       bIsStaticSwitch=true
       bIsVisible=true
       type {classStructOrEnum {refPath=&ENiagaraPartitionWaveForm}}
      }
      Normalize Spawn Group {bIsDynamic=false, bIsEditable=true, bIsStaticSwitch=true, bIsVisible=true, type {classStructOrEnum {refPath=&NiagaraBool}}}
      Inclusivity {bIsDynamic=false, bIsEditable=false, bIsStaticSwitch=true, bIsVisible=false, type {classStructOrEnum {refPath=&NiagaraBool}}}
      Set Ribbon ID and Ribbon Link Order {bIsDynamic=false, bIsEditable=true, bIsStaticSwitch=true, bIsVisible=true, type {classStructOrEnum {refPath=&NiagaraBool}}}
     ]
     moduleScript {refPath=&FN_SetSpawnGroup}
    }
    {
     bIsSetParametersModule=false
     enabled=true
     moduleName=InitializeParticle001
     inputs [
      Lifetime Mode {
       bIsDynamic=false
       bIsEditable=true
       bIsStaticSwitch=true
       bIsVisible=true
       type {classStructOrEnum {refPath=&ENiagara_LifetimeMode}}
      }
      Lifetime Min {
       bIsDynamic=false
       bIsEditable=false
       bIsStaticSwitch=false
       bIsVisible=false
       type {classStructOrEnum {refPath=&NiagaraFloat}}
      }
      Lifetime {bIsDynamic=false, bIsEditable=true, bIsStaticSwitch=false, bIsVisible=true, type {classStructOrEnum {refPath=&NiagaraFloat}}}
      Lifetime Max {
       bIsDynamic=false
       bIsEditable=false
       bIsStaticSwitch=false
       bIsVisible=false
       type {classStructOrEnum {refPath=&NiagaraFloat}}
      }
      Lifetime Randomness Mode {
       bIsDynamic=false
       bIsEditable=false
       bIsStaticSwitch=true
       bIsVisible=false
       type {classStructOrEnum {refPath=&ENiagaraRandomnessMode}}
      }
      Color Mode {
       bIsDynamic=false
       bIsEditable=true
       bIsStaticSwitch=true
       bIsVisible=true
       type {classStructOrEnum {refPath=&ENiagara_ColorInitializationMode}}
      }
      Color Minimum {
       bIsDynamic=false
       bIsEditable=false
       bIsStaticSwitch=false
       bIsVisible=false
       type {classStructOrEnum {refPath=&LinearColor}}
      }
      Color {
       bIsDynamic=false
       bIsEditable=true
       bIsStaticSwitch=false
       bIsVisible=true
       type {classStructOrEnum {refPath=&LinearColor}}
      }
      AdjustHue {bIsDynamic=false, bIsEditable=false, bIsStaticSwitch=false, bIsVisible=false, type {classStructOrEnum {refPath=&NiagaraBool}}}
      Hue Shift Range {
       bIsDynamic=false
       bIsEditable=false
       bIsStaticSwitch=false
       bIsVisible=false
       type {classStructOrEnum {refPath=&Vector2f}}
      }
      Color Maximum {
       bIsDynamic=false
       bIsEditable=false
       bIsStaticSwitch=false
       bIsVisible=false
       type {classStructOrEnum {refPath=&LinearColor}}
      }
      AdjustSaturation {bIsDynamic=false, bIsEditable=false, bIsStaticSwitch=false, bIsVisible=false, type {classStructOrEnum {refPath=&NiagaraBool}}}
      Saturation Range {
       bIsDynamic=false
       bIsEditable=false
       bIsStaticSwitch=false
       bIsVisible=false
       type {classStructOrEnum {refPath=&Vector2f}}
      }
      AdjustValue {bIsDynamic=false, bIsEditable=false, bIsStaticSwitch=false, bIsVisible=false, type {classStructOrEnum {refPath=&NiagaraBool}}}
      Value Range {
       bIsDynamic=false
       bIsEditable=false
       bIsStaticSwitch=false
       bIsVisible=false
       type {classStructOrEnum {refPath=&Vector2f}}
      }
      AdjustAlpha {bIsDynamic=false, bIsEditable=false, bIsStaticSwitch=false, bIsVisible=false, type {classStructOrEnum {refPath=&NiagaraBool}}}
      Alpha Scale Range {
       bIsDynamic=false
       bIsEditable=false
       bIsStaticSwitch=false
       bIsVisible=false
       type {classStructOrEnum {refPath=&Vector2f}}
      }
      Color Channel Mode {
       bIsDynamic=false
       bIsEditable=false
       bIsStaticSwitch=true
       bIsVisible=false
       type {classStructOrEnum {refPath=&ENiagaraChannelCorrelation}}
      }
      Color Randomness Mode {
       bIsDynamic=false
       bIsEditable=false
       bIsStaticSwitch=true
       bIsVisible=false
       type {classStructOrEnum {refPath=&ENiagaraRandomnessMode}}
      }
      Position Mode {
       bIsDynamic=false
       bIsEditable=true
       bIsStaticSwitch=true
       bIsVisible=true
       type {classStructOrEnum {refPath=&ENiagara_PositionInitializationMode}}
      }
      Position {
       bIsDynamic=true
       bIsEditable=false
       bIsStaticSwitch=false
       bIsVisible=false
       type {classStructOrEnum {refPath=&NiagaraPosition}}
      }
      UsePositionOffset {bIsDynamic=false, bIsEditable=true, bIsStaticSwitch=false, bIsVisible=true, type {classStructOrEnum {refPath=&NiagaraBool}}}
      Position Offset {bIsDynamic=false, bIsEditable=false, bIsStaticSwitch=false, bIsVisible=true, type {classStructOrEnum {refPath=&Vector3f}}}
      Position Offset Coordinate Space {
       bIsDynamic=false
       bIsEditable=false
       bIsStaticSwitch=false
       bIsVisible=false
       type {classStructOrEnum {refPath=&ENiagaraCoordinateSpace}}
      }
      Mass Mode {
       bIsDynamic=false
       bIsEditable=true
       bIsStaticSwitch=true
       bIsVisible=true
       type {classStructOrEnum {refPath=&ENiagara_MassInitializationMode}}
      }
      Mass Min {
       bIsDynamic=false
       bIsEditable=false
       bIsStaticSwitch=false
       bIsVisible=false
       type {classStructOrEnum {refPath=&NiagaraFloat}}
      }
      Mass {bIsDynamic=false, bIsEditable=true, bIsStaticSwitch=false, bIsVisible=true, type {classStructOrEnum {refPath=&NiagaraFloat}}}
      Mass Max {
       bIsDynamic=false
       bIsEditable=false
       bIsStaticSwitch=false
       bIsVisible=false
       type {classStructOrEnum {refPath=&NiagaraFloat}}
      }
      Mass Randomness Mode {
       bIsDynamic=false
       bIsEditable=false
       bIsStaticSwitch=true
       bIsVisible=false
       type {classStructOrEnum {refPath=&ENiagaraRandomnessMode}}
      }
      Write Material Random {bIsDynamic=false, bIsEditable=true, bIsStaticSwitch=false, bIsVisible=true, type {classStructOrEnum {refPath=&NiagaraBool}}}
      Material Random {bIsDynamic=false, bIsEditable=true, bIsStaticSwitch=false, bIsVisible=true, type {classStructOrEnum {refPath=&NiagaraFloat}}}
      Sprite Size Mode {
       bIsDynamic=false
       bIsEditable=true
       bIsStaticSwitch=true
       bIsVisible=true
       type {classStructOrEnum {refPath=&ENiagara_SizeScaleMode}}
      }
      Sprite Size {
       bIsDynamic=false
       bIsEditable=false
       bIsStaticSwitch=false
       bIsVisible=false
       type {classStructOrEnum {refPath=&Vector2f}}
      }
      Uniform Sprite Size {bIsDynamic=false, bIsEditable=true, bIsStaticSwitch=false, bIsVisible=true, type {classStructOrEnum {refPath=&NiagaraFloat}}}
      Uniform Sprite Size Min {
       bIsDynamic=false
       bIsEditable=false
       bIsStaticSwitch=false
       bIsVisible=false
       type {classStructOrEnum {refPath=&NiagaraFloat}}
      }
      Sprite Size Min {
       bIsDynamic=false
       bIsEditable=false
       bIsStaticSwitch=false
       bIsVisible=false
       type {classStructOrEnum {refPath=&Vector2f}}
      }
      Sprite Size Max {
       bIsDynamic=false
       bIsEditable=false
       bIsStaticSwitch=false
       bIsVisible=false
       type {classStructOrEnum {refPath=&Vector2f}}
      }
      Uniform Sprite Size Max {
       bIsDynamic=false
       bIsEditable=false
       bIsStaticSwitch=false
       bIsVisible=false
       type {classStructOrEnum {refPath=&NiagaraFloat}}
      }
      Sprite Size Randomness Mode {
       bIsDynamic=false
       bIsEditable=false
       bIsStaticSwitch=true
       bIsVisible=false
       type {classStructOrEnum {refPath=&ENiagaraRandomnessMode}}
      }
      Sprite Rotation Mode {
       bIsDynamic=false
       bIsEditable=true
       bIsStaticSwitch=true
       bIsVisible=true
       type {classStructOrEnum {refPath=&ENiagara_SpriteRotationMode}}
      }
      Sprite Rotation Angle Min {
       bIsDynamic=false
       bIsEditable=false
       bIsStaticSwitch=false
       bIsVisible=false
       type {classStructOrEnum {refPath=&NiagaraFloat}}
      }
      Sprite Rotation Angle {
       bIsDynamic=false
       bIsEditable=false
       bIsStaticSwitch=false
       bIsVisible=false
       type {classStructOrEnum {refPath=&NiagaraFloat}}
      }
      Sprite Rotation Angle Max {
       bIsDynamic=false
       bIsEditable=false
       bIsStaticSwitch=false
       bIsVisible=false
       type {classStructOrEnum {refPath=&NiagaraFloat}}
      }
      Sprite Rotation Randomness Mode {
       bIsDynamic=false
       bIsEditable=false
       bIsStaticSwitch=true
       bIsVisible=false
       type {classStructOrEnum {refPath=&ENiagaraRandomnessMode}}
      }
      Sprite UV Mode {
       bIsDynamic=false
       bIsEditable=true
       bIsStaticSwitch=true
       bIsVisible=true
       type {classStructOrEnum {refPath=&ENiagara_UVFlippingMode}}
      }
      Sprite UV Randomness Mode {
       bIsDynamic=false
       bIsEditable=false
       bIsStaticSwitch=true
       bIsVisible=false
       type {classStructOrEnum {refPath=&ENiagaraRandomnessMode}}
      }
      Sprite UV Scale {
       bIsDynamic=false
       bIsEditable=false
       bIsStaticSwitch=false
       bIsVisible=false
       type {classStructOrEnum {refPath=&Vector2f}}
      }
      Mesh Scale Mode {
       bIsDynamic=false
       bIsEditable=true
       bIsStaticSwitch=true
       bIsVisible=true
       type {classStructOrEnum {refPath=&ENiagara_SizeScaleMode}}
      }
      Mesh Uniform Scale Min {
       bIsDynamic=false
       bIsEditable=false
       bIsStaticSwitch=false
       bIsVisible=false
       type {classStructOrEnum {refPath=&NiagaraFloat}}
      }
      Mesh Uniform Scale {
       bIsDynamic=false
       bIsEditable=false
       bIsStaticSwitch=false
       bIsVisible=false
       type {classStructOrEnum {refPath=&NiagaraFloat}}
      }
      Mesh Scale Min {
       bIsDynamic=false
       bIsEditable=false
       bIsStaticSwitch=false
       bIsVisible=false
       type {classStructOrEnum {refPath=&Vector3f}}
      }
      Mesh Scale {
       bIsDynamic=false
       bIsEditable=false
       bIsStaticSwitch=false
       bIsVisible=false
       type {classStructOrEnum {refPath=&Vector3f}}
      }
      Mesh Scale Max {
       bIsDynamic=false
       bIsEditable=false
       bIsStaticSwitch=false
       bIsVisible=false
       type {classStructOrEnum {refPath=&Vector3f}}
      }
      Mesh Uniform Scale Max {
       bIsDynamic=false
       bIsEditable=false
       bIsStaticSwitch=false
       bIsVisible=false
       type {classStructOrEnum {refPath=&NiagaraFloat}}
      }
      Mesh Scale Randomness Mode {
       bIsDynamic=false
       bIsEditable=false
       bIsStaticSwitch=true
       bIsVisible=false
       type {classStructOrEnum {refPath=&ENiagaraRandomnessMode}}
      }
      Mesh Renderer Array Visibility Mode {
       bIsDynamic=false
       bIsEditable=true
       bIsStaticSwitch=true
       bIsVisible=true
       type {classStructOrEnum {refPath=&ENiagara_UnsetDirectSetRandom}}
      }
      Mesh Renderer Mesh Index {
       bIsDynamic=false
       bIsEditable=false
       bIsStaticSwitch=false
       bIsVisible=false
       type {classStructOrEnum {refPath=&NiagaraInt32}}
      }
      Wrap Index By Mesh Count {bIsDynamic=false, bIsEditable=false, bIsStaticSwitch=true, bIsVisible=false, type {classStructOrEnum {refPath=&NiagaraBool}}}
      Mesh Renderer Info {
       bIsDynamic=false
       bIsEditable=false
       bIsStaticSwitch=false
       bIsVisible=false
       type {classStructOrEnum {refPath=&NiagaraDataInterfaceMeshRendererInfo}}
      }
      Mesh Renderer Random Seed {
       bIsDynamic=false
       bIsEditable=false
       bIsStaticSwitch=false
       bIsVisible=false
       type {classStructOrEnum {refPath=&NiagaraInt32}}
      }
      Ribbon Width Mode {
       bIsDynamic=false
       bIsEditable=true
       bIsStaticSwitch=true
       bIsVisible=true
       type {classStructOrEnum {refPath=&ENiagara_UnsetDirectSet}}
      }
      Ribbon Width {
       bIsDynamic=false
       bIsEditable=false
       bIsStaticSwitch=false
       bIsVisible=false
       type {classStructOrEnum {refPath=&NiagaraFloat}}
      }
      Ribbon Facing Mode {
       bIsDynamic=false
       bIsEditable=true
       bIsStaticSwitch=true
       bIsVisible=true
       type {classStructOrEnum {refPath=&ENiagara_UnsetDirectSet}}
      }
      Ribbon Facing Vector {
       bIsDynamic=false
       bIsEditable=false
       bIsStaticSwitch=false
       bIsVisible=false
       type {classStructOrEnum {refPath=&Vector3f}}
      }
      Ribbon Twist Mode {
       bIsDynamic=false
       bIsEditable=true
       bIsStaticSwitch=true
       bIsVisible=true
       type {classStructOrEnum {refPath=&ENiagara_UnsetDirectSet}}
      }
      "Ribbon Twist " {
       bIsDynamic=false
       bIsEditable=false
       bIsStaticSwitch=false
       bIsVisible=false
       type {classStructOrEnum {refPath=&NiagaraFloat}}
      }
     ]
     moduleScript {refPath=&InitializeParticle}
    }
   ]
  }
  particleUpdateScript {
   scriptName=ParticleUpdateScript
   modules [
    {
     bIsSetParametersModule=false
     enabled=true
     moduleName=ParticleState
     inputs [
      Kill Particles When Lifetime Has Elapsed {bIsDynamic=false, bIsEditable=true, bIsStaticSwitch=true, bIsVisible=true, type {classStructOrEnum {refPath=&NiagaraBool}}}
      Loop Particles Lifetime {bIsDynamic=false, bIsEditable=false, bIsStaticSwitch=false, bIsVisible=false, type {classStructOrEnum {refPath=&NiagaraBool}}}
      Let Infinitely Lived Particles Die When Emitter Deactivates {bIsDynamic=false, bIsEditable=false, bIsStaticSwitch=false, bIsVisible=false, type {classStructOrEnum {refPath=&NiagaraBool}}}
      Lifetime {bIsDynamic=false, bIsEditable=true, bIsStaticSwitch=false, bIsVisible=true, type {classStructOrEnum {refPath=&NiagaraFloat}}}
      DeltaTime {bIsDynamic=false, bIsEditable=true, bIsStaticSwitch=false, bIsVisible=true, type {classStructOrEnum {refPath=&NiagaraFloat}}}
     ]
     moduleScript {refPath=&ParticleState}
    }
    {
     bIsSetParametersModule=false
     enabled=true
     moduleName=SkeletalMeshLocation
     inputs [
      Skeletal Mesh {
       bIsDynamic=false
       bIsEditable=true
       bIsStaticSwitch=false
       bIsVisible=true
       type {classStructOrEnum {refPath=&NiagaraDataInterfaceSkeletalMesh}}
      }
      Mesh Sampling Type {
       bIsDynamic=false
       bIsEditable=true
       bIsStaticSwitch=true
       bIsVisible=true
       type {classStructOrEnum {refPath=&ENiagaraSkelSamplingModeFull}}
      }
      Bone / Socket Sampling Mode {
       bIsDynamic=false
       bIsEditable=false
       bIsStaticSwitch=true
       bIsVisible=false
       type {classStructOrEnum {refPath=&ENiagaraBoneSocketSamplingMode}}
      }
      Bone Sampling Mode {
       bIsDynamic=false
       bIsEditable=true
       bIsStaticSwitch=true
       bIsVisible=true
       type {classStructOrEnum {refPath=&ENiagaraBoneSamplingMode}}
      }
      Socket Sampling Mode {
       bIsDynamic=false
       bIsEditable=false
       bIsStaticSwitch=true
       bIsVisible=false
       type {classStructOrEnum {refPath=&ENiagaraSocketSamplingMode}}
      }
      Triangle Sampling Mode {
       bIsDynamic=false
       bIsEditable=false
       bIsStaticSwitch=true
       bIsVisible=false
       type {classStructOrEnum {refPath=&ENiagaraTriangleSamplingMode}}
      }
      Vertex Sampling Mode {
       bIsDynamic=false
       bIsEditable=false
       bIsStaticSwitch=true
       bIsVisible=false
       type {classStructOrEnum {refPath=&ENiagaraVertexSamplingMode}}
      }
      Bone / Socket Index {bIsDynamic=false, bIsEditable=true, bIsStaticSwitch=false, bIsVisible=true, type {classStructOrEnum {refPath=&NiagaraInt32}}}
      Triangle ID {
       bIsDynamic=false
       bIsEditable=false
       bIsStaticSwitch=false
       bIsVisible=false
       type {classStructOrEnum {refPath=&NiagaraInt32}}
      }
      Vertex ID {
       bIsDynamic=false
       bIsEditable=false
       bIsStaticSwitch=false
       bIsVisible=false
       type {classStructOrEnum {refPath=&NiagaraInt32}}
      }
      FixedBarycentric {bIsDynamic=false, bIsEditable=false, bIsStaticSwitch=false, bIsVisible=false, type {classStructOrEnum {refPath=&NiagaraBool}}}
      Triangle Coordinate {
       bIsDynamic=false
       bIsEditable=false
       bIsStaticSwitch=false
       bIsVisible=false
       type {classStructOrEnum {refPath=&Vector3f}}
      }
      Sampled UV Index {
       bIsDynamic=false
       bIsEditable=false
       bIsStaticSwitch=false
       bIsVisible=false
       type {classStructOrEnum {refPath=&NiagaraInt32}}
      }
      Sampled Position Coordinate Space {
       bIsDynamic=false
       bIsEditable=true
       bIsStaticSwitch=true
       bIsVisible=true
       type {classStructOrEnum {refPath=&ENiagaraCoordinateSpace}}
      }
      Sampled Velocity Coordinate Space {
       bIsDynamic=false
       bIsEditable=true
       bIsStaticSwitch=true
       bIsVisible=true
       type {classStructOrEnum {refPath=&ENiagaraCoordinateSpace}}
      }
      Sampled Orientation Coordinate Space {
       bIsDynamic=false
       bIsEditable=true
       bIsStaticSwitch=true
       bIsVisible=true
       type {classStructOrEnum {refPath=&ENiagaraCoordinateSpace}}
      }
      Generate New Random Samples Every Frame {bIsDynamic=false, bIsEditable=false, bIsStaticSwitch=true, bIsVisible=false, type {classStructOrEnum {refPath=&NiagaraBool}}}
      Write Sampled Mesh Attrributes to Particle Attributes {bIsDynamic=false, bIsEditable=true, bIsStaticSwitch=true, bIsVisible=true, type {classStructOrEnum {refPath=&NiagaraBool}}}
      Randomness Mode {
       bIsDynamic=false
       bIsEditable=false
       bIsStaticSwitch=false
       bIsVisible=false
       type {classStructOrEnum {refPath=&ENiagaraRandomnessMode}}
      }
      Random Seed {
       bIsDynamic=false
       bIsEditable=false
       bIsStaticSwitch=false
       bIsVisible=false
       type {classStructOrEnum {refPath=&NiagaraInt32}}
      }
      Fixed Random Seed {bIsDynamic=false, bIsEditable=false, bIsStaticSwitch=false, bIsVisible=false, type {classStructOrEnum {refPath=&NiagaraBool}}}
      Apply To Particle Position {bIsDynamic=false, bIsEditable=true, bIsStaticSwitch=false, bIsVisible=true, type {classStructOrEnum {refPath=&NiagaraBool}}}
      Sampled Position Offset {bIsDynamic=false, bIsEditable=true, bIsStaticSwitch=false, bIsVisible=true, type {classStructOrEnum {refPath=&Vector3f}}}
      OffsetAlongNormal {bIsDynamic=false, bIsEditable=false, bIsStaticSwitch=false, bIsVisible=false, type {classStructOrEnum {refPath=&NiagaraBool}}}
      Offset Position Along Sampled Normal {
       bIsDynamic=false
       bIsEditable=false
       bIsStaticSwitch=false
       bIsVisible=false
       type {classStructOrEnum {refPath=&NiagaraFloat}}
      }
      Sampled Position Offset Coordinate Space {
       bIsDynamic=false
       bIsEditable=true
       bIsStaticSwitch=true
       bIsVisible=true
       type {classStructOrEnum {refPath=&ENiagaraMeshTransforms}}
      }
      Inherit Velocity {bIsDynamic=false, bIsEditable=true, bIsStaticSwitch=false, bIsVisible=true, type {classStructOrEnum {refPath=&NiagaraBool}}}
      "Inherit Velocity (Scale)" {bIsDynamic=false, bIsEditable=false, bIsStaticSwitch=false, bIsVisible=true, type {classStructOrEnum {refPath=&NiagaraFloat}}}
      VelocitySpeedLimit {bIsDynamic=false, bIsEditable=false, bIsStaticSwitch=false, bIsVisible=false, type {classStructOrEnum {refPath=&NiagaraBool}}}
      Inherited Velocity Speed Limit {
       bIsDynamic=false
       bIsEditable=false
       bIsStaticSwitch=false
       bIsVisible=false
       type {classStructOrEnum {refPath=&NiagaraFloat}}
      }
      Apply Sampled Orientation to Mesh Particles {bIsDynamic=false, bIsEditable=true, bIsStaticSwitch=true, bIsVisible=true, type {classStructOrEnum {refPath=&NiagaraBool}}}
      Interpolate Mesh Sampling {bIsDynamic=false, bIsEditable=true, bIsStaticSwitch=true, bIsVisible=true, type {classStructOrEnum {refPath=&NiagaraBool}}}
      MaskSpawnGroup {bIsDynamic=false, bIsEditable=true, bIsStaticSwitch=false, bIsVisible=true, type {classStructOrEnum {refPath=&NiagaraBool}}}
      Spawn Group Mask {bIsDynamic=false, bIsEditable=false, bIsStaticSwitch=false, bIsVisible=true, type {classStructOrEnum {refPath=&NiagaraInt32}}}
      Use Sampled Mesh Attribute as Ribbon ID {bIsDynamic=false, bIsEditable=true, bIsStaticSwitch=false, bIsVisible=true, type {classStructOrEnum {refPath=&NiagaraBool}}}
     ]
     moduleScript {refPath=&SkeletalMeshLocation}
    }
    {
     bIsSetParametersModule=false
     enabled=true
     moduleName=CameraOffset
     inputs [Camera Offset Amount {bIsDynamic=false, bIsEditable=true, bIsStaticSwitch=false, bIsVisible=true, type {classStructOrEnum {refPath=&NiagaraFloat}}}]
     moduleScript {refPath=&CameraOffset}
    }
    {
     bIsSetParametersModule=false
     enabled=true
     moduleName=ScaleColor
     inputs [
      Scale Mode {
       bIsDynamic=false
       bIsEditable=true
       bIsStaticSwitch=true
       bIsVisible=true
       type {classStructOrEnum {refPath=&ENiagaraScaleColorMode}}
      }
      Linear Color Curve {
       bIsDynamic=false
       bIsEditable=true
       bIsStaticSwitch=false
       bIsVisible=true
       type {classStructOrEnum {refPath=&NiagaraDataInterfaceColorCurve}}
      }
      ScaleRGB {bIsDynamic=false, bIsEditable=false, bIsStaticSwitch=false, bIsVisible=false, type {classStructOrEnum {refPath=&NiagaraBool}}}
      Scale RGB {
       bIsDynamic=false
       bIsEditable=false
       bIsStaticSwitch=false
       bIsVisible=false
       type {classStructOrEnum {refPath=&Vector3f}}
      }
      ScaleRGBA {bIsDynamic=false, bIsEditable=false, bIsStaticSwitch=false, bIsVisible=false, type {classStructOrEnum {refPath=&NiagaraBool}}}
      Scale RGBA {
       bIsDynamic=false
       bIsEditable=false
       bIsStaticSwitch=false
       bIsVisible=false
       type {classStructOrEnum {refPath=&Vector4f}}
      }
      Curve Index {bIsDynamic=false, bIsEditable=true, bIsStaticSwitch=false, bIsVisible=true, type {classStructOrEnum {refPath=&NiagaraFloat}}}
      ScaleA {bIsDynamic=false, bIsEditable=false, bIsStaticSwitch=false, bIsVisible=false, type {classStructOrEnum {refPath=&NiagaraBool}}}
      Scale Alpha {
       bIsDynamic=false
       bIsEditable=false
       bIsStaticSwitch=false
       bIsVisible=false
       type {classStructOrEnum {refPath=&NiagaraFloat}}
      }
      Color Value To Scale {
       bIsDynamic=false
       bIsEditable=true
       bIsStaticSwitch=false
       bIsVisible=true
       type {classStructOrEnum {refPath=&LinearColor}}
      }
     ]
     moduleScript {refPath=&ScaleColor}
    }
    {
     bIsSetParametersModule=false
     enabled=true
     moduleName=FN_Light_v2
     inputs [
      LightEnabled {bIsDynamic=false, bIsEditable=true, bIsStaticSwitch=false, bIsVisible=true, type {classStructOrEnum {refPath=&NiagaraBool}}}
      Scale Color {bIsDynamic=false, bIsEditable=true, bIsStaticSwitch=false, bIsVisible=true, type {classStructOrEnum {refPath=&NiagaraFloat}}}
      Light Exponent {bIsDynamic=false, bIsEditable=true, bIsStaticSwitch=false, bIsVisible=true, type {classStructOrEnum {refPath=&NiagaraFloat}}}
      Position Offset {bIsDynamic=false, bIsEditable=true, bIsStaticSwitch=false, bIsVisible=true, type {classStructOrEnum {refPath=&Vector3f}}}
      Light Radius {bIsDynamic=false, bIsEditable=true, bIsStaticSwitch=false, bIsVisible=true, type {classStructOrEnum {refPath=&NiagaraFloat}}}
     ]
     moduleScript {refPath=&FN_Light_v2}
    }
    {
     bIsSetParametersModule=false
     enabled=true
     moduleName=ScaleColor001
     inputs [
      Scale Mode {
       bIsDynamic=false
       bIsEditable=true
       bIsStaticSwitch=true
       bIsVisible=true
       type {classStructOrEnum {refPath=&ENiagaraScaleColorMode}}
      }
      Linear Color Curve {
       bIsDynamic=false
       bIsEditable=false
       bIsStaticSwitch=false
       bIsVisible=false
       type {classStructOrEnum {refPath=&NiagaraDataInterfaceColorCurve}}
      }
      ScaleRGB {bIsDynamic=false, bIsEditable=true, bIsStaticSwitch=false, bIsVisible=true, type {classStructOrEnum {refPath=&NiagaraBool}}}
      Scale RGB {bIsDynamic=false, bIsEditable=true, bIsStaticSwitch=false, bIsVisible=true, type {classStructOrEnum {refPath=&Vector3f}}}
      ScaleRGBA {bIsDynamic=false, bIsEditable=false, bIsStaticSwitch=false, bIsVisible=false, type {classStructOrEnum {refPath=&NiagaraBool}}}
      Scale RGBA {
       bIsDynamic=false
       bIsEditable=false
       bIsStaticSwitch=false
       bIsVisible=false
       type {classStructOrEnum {refPath=&Vector4f}}
      }
      Curve Index {
       bIsDynamic=false
       bIsEditable=false
       bIsStaticSwitch=false
       bIsVisible=false
       type {classStructOrEnum {refPath=&NiagaraFloat}}
      }
      ScaleA {bIsDynamic=false, bIsEditable=true, bIsStaticSwitch=false, bIsVisible=true, type {classStructOrEnum {refPath=&NiagaraBool}}}
      Scale Alpha {bIsDynamic=false, bIsEditable=true, bIsStaticSwitch=false, bIsVisible=true, type {classStructOrEnum {refPath=&NiagaraFloat}}}
      Color Value To Scale {
       bIsDynamic=false
       bIsEditable=true
       bIsStaticSwitch=false
       bIsVisible=true
       type {classStructOrEnum {refPath=&LinearColor}}
      }
     ]
     moduleScript {refPath=&ScaleColor}
    }
   ]
  }
  rendererClasses [{refPath=&NiagaraSpriteRendererProperties}, {refPath=&NiagaraLightRendererProperties}]
  renderers [
   {rendererIndex=0, rendererClass {refPath=&NiagaraSpriteRendererProperties}}
   {rendererIndex=1, rendererClass {refPath=&NiagaraLightRendererProperties}}
  ]
 }
 HandGlow004 {
  bEnabled=false
  emitterSpawnScript {scriptName=EmitterSpawnScript, modules []}
  emitterUpdateScript {
   scriptName=EmitterUpdateScript
   modules [
    {
     bIsSetParametersModule=false
     enabled=true
     moduleName=EmitterState
     inputs [
      Life Cycle Mode {
       bIsDynamic=false
       bIsEditable=true
       bIsStaticSwitch=true
       bIsVisible=true
       type {classStructOrEnum {refPath=&ENiagaraEmitterLifeCycleMode}}
      }
      Inactive Response {
       bIsDynamic=false
       bIsEditable=true
       bIsStaticSwitch=true
       bIsVisible=true
       type {classStructOrEnum {refPath=&ENiagaraInactiveMode}}
      }
      Loop Behavior {
       bIsDynamic=false
       bIsEditable=true
       bIsStaticSwitch=true
       bIsVisible=true
       type {classStructOrEnum {refPath=&ENiagara_EmitterStateOptions}}
      }
      Loop Duration Mode {
       bIsDynamic=false
       bIsEditable=true
       bIsStaticSwitch=true
       bIsVisible=true
       type {classStructOrEnum {refPath=&ENiagara_InfiniteLoopDuration}}
      }
      Loop Count {
       bIsDynamic=false
       bIsEditable=false
       bIsStaticSwitch=false
       bIsVisible=false
       type {classStructOrEnum {refPath=&NiagaraInt32}}
      }
      Loop Duration {
       bIsDynamic=false
       bIsEditable=false
       bIsStaticSwitch=false
       bIsVisible=false
       type {classStructOrEnum {refPath=&NiagaraFloat}}
      }
      Recalculate Duration Each Loop {bIsDynamic=false, bIsEditable=false, bIsStaticSwitch=false, bIsVisible=false, type {classStructOrEnum {refPath=&NiagaraBool}}}
      UseLoopDelay {bIsDynamic=false, bIsEditable=true, bIsStaticSwitch=false, bIsVisible=true, type {classStructOrEnum {refPath=&NiagaraBool}}}
      Loop Delay {bIsDynamic=false, bIsEditable=false, bIsStaticSwitch=false, bIsVisible=true, type {classStructOrEnum {refPath=&NiagaraFloat}}}
      Delay First Loop Only {bIsDynamic=false, bIsEditable=false, bIsStaticSwitch=false, bIsVisible=false, type {classStructOrEnum {refPath=&NiagaraBool}}}
      Scalability Mode {
       bIsDynamic=false
       bIsEditable=true
       bIsStaticSwitch=true
       bIsVisible=true
       type {classStructOrEnum {refPath=&ENiagaraEmitterScalabilityMode_Limited}}
      }
      Enable Distance Culling {bIsDynamic=false, bIsEditable=false, bIsStaticSwitch=true, bIsVisible=false, type {classStructOrEnum {refPath=&NiagaraBool}}}
      MinDistance {
       bIsDynamic=false
       bIsEditable=false
       bIsStaticSwitch=false
       bIsVisible=false
       type {classStructOrEnum {refPath=&NiagaraFloat}}
      }
      Min Distance Response {
       bIsDynamic=false
       bIsEditable=false
       bIsStaticSwitch=false
       bIsVisible=false
       type {classStructOrEnum {refPath=&ENiagaraExecutionStateManagement}}
      }
      MaxDistance {
       bIsDynamic=false
       bIsEditable=false
       bIsStaticSwitch=false
       bIsVisible=false
       type {classStructOrEnum {refPath=&NiagaraFloat}}
      }
      Max Distance Response {
       bIsDynamic=false
       bIsEditable=false
       bIsStaticSwitch=false
       bIsVisible=false
       type {classStructOrEnum {refPath=&ENiagaraExecutionStateManagement}}
      }
      Scale Spawn Count {bIsDynamic=false, bIsEditable=false, bIsStaticSwitch=true, bIsVisible=false, type {classStructOrEnum {refPath=&NiagaraBool}}}
      Spawn Count Scale {
       bIsDynamic=false
       bIsEditable=false
       bIsStaticSwitch=false
       bIsVisible=false
       type {classStructOrEnum {refPath=&NiagaraFloat}}
      }
      Spawn Count Scale By Distance {
       bIsDynamic=false
       bIsEditable=false
       bIsStaticSwitch=false
       bIsVisible=false
       type {classStructOrEnum {refPath=&NiagaraDataInterfaceCurve}}
      }
      Enable Visibility Culling {bIsDynamic=false, bIsEditable=false, bIsStaticSwitch=true, bIsVisible=false, type {classStructOrEnum {refPath=&NiagaraBool}}}
      Visibility Cull Response {
       bIsDynamic=false
       bIsEditable=false
       bIsStaticSwitch=false
       bIsVisible=false
       type {classStructOrEnum {refPath=&ENiagaraExecutionStateManagement}}
      }
      Visibility Cull Delay {
       bIsDynamic=false
       bIsEditable=false
       bIsStaticSwitch=false
       bIsVisible=false
       type {classStructOrEnum {refPath=&NiagaraFloat}}
      }
      Reset Age On Awaken {bIsDynamic=false, bIsEditable=false, bIsStaticSwitch=false, bIsVisible=false, type {classStructOrEnum {refPath=&NiagaraBool}}}
     ]
     moduleScript {refPath=&EmitterState}
    }
    {
     bIsSetParametersModule=false
     enabled=true
     moduleName=SpawnBurst_Instantaneous
     inputs [
      Spawn Count {bIsDynamic=false, bIsEditable=true, bIsStaticSwitch=false, bIsVisible=true, type {classStructOrEnum {refPath=&NiagaraInt32}}}
      Spawn Time {bIsDynamic=false, bIsEditable=true, bIsStaticSwitch=false, bIsVisible=true, type {classStructOrEnum {refPath=&NiagaraFloat}}}
      Use Spawn Probability {bIsDynamic=false, bIsEditable=true, bIsStaticSwitch=false, bIsVisible=true, type {classStructOrEnum {refPath=&NiagaraBool}}}
      Spawn Probability {bIsDynamic=false, bIsEditable=true, bIsStaticSwitch=false, bIsVisible=true, type {classStructOrEnum {refPath=&NiagaraFloat}}}
      Spawn Group {bIsDynamic=false, bIsEditable=true, bIsStaticSwitch=false, bIsVisible=true, type {classStructOrEnum {refPath=&NiagaraInt32}}}
      Age {bIsDynamic=false, bIsEditable=true, bIsStaticSwitch=false, bIsVisible=true, type {classStructOrEnum {refPath=&NiagaraFloat}}}
     ]
     moduleScript {refPath=&SpawnBurst_Instantaneous}
    }
   ]
  }
  particleSpawnScript {
   scriptName=ParticleSpawnScript
   modules [
    {
     bIsSetParametersModule=false
     enabled=true
     moduleName=FN_SetSpawnGroup
     inputs [
      Number of Spawn Groups {bIsDynamic=false, bIsEditable=true, bIsStaticSwitch=false, bIsVisible=true, type {classStructOrEnum {refPath=&NiagaraInt32}}}
      Particles Per Group {
       bIsDynamic=false
       bIsEditable=false
       bIsStaticSwitch=false
       bIsVisible=false
       type {classStructOrEnum {refPath=&NiagaraInt32}}
      }
      Pattern {
       bIsDynamic=false
       bIsEditable=true
       bIsStaticSwitch=true
       bIsVisible=true
       type {classStructOrEnum {refPath=&ENiagaraPartitionWaveForm}}
      }
      Normalize Spawn Group {bIsDynamic=false, bIsEditable=true, bIsStaticSwitch=true, bIsVisible=true, type {classStructOrEnum {refPath=&NiagaraBool}}}
      Inclusivity {bIsDynamic=false, bIsEditable=false, bIsStaticSwitch=true, bIsVisible=false, type {classStructOrEnum {refPath=&NiagaraBool}}}
      Set Ribbon ID and Ribbon Link Order {bIsDynamic=false, bIsEditable=true, bIsStaticSwitch=true, bIsVisible=true, type {classStructOrEnum {refPath=&NiagaraBool}}}
     ]
     moduleScript {refPath=&FN_SetSpawnGroup}
    }
    {
     bIsSetParametersModule=false
     enabled=true
     moduleName=InitializeParticle001
     inputs [
      Lifetime Mode {
       bIsDynamic=false
       bIsEditable=true
       bIsStaticSwitch=true
       bIsVisible=true
       type {classStructOrEnum {refPath=&ENiagara_LifetimeMode}}
      }
      Lifetime Min {
       bIsDynamic=false
       bIsEditable=false
       bIsStaticSwitch=false
       bIsVisible=false
       type {classStructOrEnum {refPath=&NiagaraFloat}}
      }
      Lifetime {bIsDynamic=false, bIsEditable=true, bIsStaticSwitch=false, bIsVisible=true, type {classStructOrEnum {refPath=&NiagaraFloat}}}
      Lifetime Max {
       bIsDynamic=false
       bIsEditable=false
       bIsStaticSwitch=false
       bIsVisible=false
       type {classStructOrEnum {refPath=&NiagaraFloat}}
      }
      Lifetime Randomness Mode {
       bIsDynamic=false
       bIsEditable=false
       bIsStaticSwitch=true
       bIsVisible=false
       type {classStructOrEnum {refPath=&ENiagaraRandomnessMode}}
      }
      Color Mode {
       bIsDynamic=false
       bIsEditable=true
       bIsStaticSwitch=true
       bIsVisible=true
       type {classStructOrEnum {refPath=&ENiagara_ColorInitializationMode}}
      }
      Color Minimum {
       bIsDynamic=false
       bIsEditable=false
       bIsStaticSwitch=false
       bIsVisible=false
       type {classStructOrEnum {refPath=&LinearColor}}
      }
      Color {
       bIsDynamic=false
       bIsEditable=true
       bIsStaticSwitch=false
       bIsVisible=true
       type {classStructOrEnum {refPath=&LinearColor}}
      }
      AdjustHue {bIsDynamic=false, bIsEditable=false, bIsStaticSwitch=false, bIsVisible=false, type {classStructOrEnum {refPath=&NiagaraBool}}}
      Hue Shift Range {
       bIsDynamic=false
       bIsEditable=false
       bIsStaticSwitch=false
       bIsVisible=false
       type {classStructOrEnum {refPath=&Vector2f}}
      }
      Color Maximum {
       bIsDynamic=false
       bIsEditable=false
       bIsStaticSwitch=false
       bIsVisible=false
       type {classStructOrEnum {refPath=&LinearColor}}
      }
      AdjustSaturation {bIsDynamic=false, bIsEditable=false, bIsStaticSwitch=false, bIsVisible=false, type {classStructOrEnum {refPath=&NiagaraBool}}}
      Saturation Range {
       bIsDynamic=false
       bIsEditable=false
       bIsStaticSwitch=false
       bIsVisible=false
       type {classStructOrEnum {refPath=&Vector2f}}
      }
      AdjustValue {bIsDynamic=false, bIsEditable=false, bIsStaticSwitch=false, bIsVisible=false, type {classStructOrEnum {refPath=&NiagaraBool}}}
      Value Range {
       bIsDynamic=false
       bIsEditable=false
       bIsStaticSwitch=false
       bIsVisible=false
       type {classStructOrEnum {refPath=&Vector2f}}
      }
      AdjustAlpha {bIsDynamic=false, bIsEditable=false, bIsStaticSwitch=false, bIsVisible=false, type {classStructOrEnum {refPath=&NiagaraBool}}}
      Alpha Scale Range {
       bIsDynamic=false
       bIsEditable=false
       bIsStaticSwitch=false
       bIsVisible=false
       type {classStructOrEnum {refPath=&Vector2f}}
      }
      Color Channel Mode {
       bIsDynamic=false
       bIsEditable=false
       bIsStaticSwitch=true
       bIsVisible=false
       type {classStructOrEnum {refPath=&ENiagaraChannelCorrelation}}
      }
      Color Randomness Mode {
       bIsDynamic=false
       bIsEditable=false
       bIsStaticSwitch=true
       bIsVisible=false
       type {classStructOrEnum {refPath=&ENiagaraRandomnessMode}}
      }
      Position Mode {
       bIsDynamic=false
       bIsEditable=true
       bIsStaticSwitch=true
       bIsVisible=true
       type {classStructOrEnum {refPath=&ENiagara_PositionInitializationMode}}
      }
      Position {
       bIsDynamic=true
       bIsEditable=false
       bIsStaticSwitch=false
       bIsVisible=false
       type {classStructOrEnum {refPath=&NiagaraPosition}}
      }
      UsePositionOffset {bIsDynamic=false, bIsEditable=true, bIsStaticSwitch=false, bIsVisible=true, type {classStructOrEnum {refPath=&NiagaraBool}}}
      Position Offset {bIsDynamic=false, bIsEditable=false, bIsStaticSwitch=false, bIsVisible=true, type {classStructOrEnum {refPath=&Vector3f}}}
      Position Offset Coordinate Space {
       bIsDynamic=false
       bIsEditable=false
       bIsStaticSwitch=false
       bIsVisible=false
       type {classStructOrEnum {refPath=&ENiagaraCoordinateSpace}}
      }
      Mass Mode {
       bIsDynamic=false
       bIsEditable=true
       bIsStaticSwitch=true
       bIsVisible=true
       type {classStructOrEnum {refPath=&ENiagara_MassInitializationMode}}
      }
      Mass Min {
       bIsDynamic=false
       bIsEditable=false
       bIsStaticSwitch=false
       bIsVisible=false
       type {classStructOrEnum {refPath=&NiagaraFloat}}
      }
      Mass {bIsDynamic=false, bIsEditable=true, bIsStaticSwitch=false, bIsVisible=true, type {classStructOrEnum {refPath=&NiagaraFloat}}}
      Mass Max {
       bIsDynamic=false
       bIsEditable=false
       bIsStaticSwitch=false
       bIsVisible=false
       type {classStructOrEnum {refPath=&NiagaraFloat}}
      }
      Mass Randomness Mode {
       bIsDynamic=false
       bIsEditable=false
       bIsStaticSwitch=true
       bIsVisible=false
       type {classStructOrEnum {refPath=&ENiagaraRandomnessMode}}
      }
      Write Material Random {bIsDynamic=false, bIsEditable=true, bIsStaticSwitch=false, bIsVisible=true, type {classStructOrEnum {refPath=&NiagaraBool}}}
      Material Random {bIsDynamic=false, bIsEditable=true, bIsStaticSwitch=false, bIsVisible=true, type {classStructOrEnum {refPath=&NiagaraFloat}}}
      Sprite Size Mode {
       bIsDynamic=false
       bIsEditable=true
       bIsStaticSwitch=true
       bIsVisible=true
       type {classStructOrEnum {refPath=&ENiagara_SizeScaleMode}}
      }
      Sprite Size {
       bIsDynamic=false
       bIsEditable=false
       bIsStaticSwitch=false
       bIsVisible=false
       type {classStructOrEnum {refPath=&Vector2f}}
      }
      Uniform Sprite Size {bIsDynamic=false, bIsEditable=true, bIsStaticSwitch=false, bIsVisible=true, type {classStructOrEnum {refPath=&NiagaraFloat}}}
      Uniform Sprite Size Min {
       bIsDynamic=false
       bIsEditable=false
       bIsStaticSwitch=false
       bIsVisible=false
       type {classStructOrEnum {refPath=&NiagaraFloat}}
      }
      Sprite Size Min {
       bIsDynamic=false
       bIsEditable=false
       bIsStaticSwitch=false
       bIsVisible=false
       type {classStructOrEnum {refPath=&Vector2f}}
      }
      Sprite Size Max {
       bIsDynamic=false
       bIsEditable=false
       bIsStaticSwitch=false
       bIsVisible=false
       type {classStructOrEnum {refPath=&Vector2f}}
      }
      Uniform Sprite Size Max {
       bIsDynamic=false
       bIsEditable=false
       bIsStaticSwitch=false
       bIsVisible=false
       type {classStructOrEnum {refPath=&NiagaraFloat}}
      }
      Sprite Size Randomness Mode {
       bIsDynamic=false
       bIsEditable=false
       bIsStaticSwitch=true
       bIsVisible=false
       type {classStructOrEnum {refPath=&ENiagaraRandomnessMode}}
      }
      Sprite Rotation Mode {
       bIsDynamic=false
       bIsEditable=true
       bIsStaticSwitch=true
       bIsVisible=true
       type {classStructOrEnum {refPath=&ENiagara_SpriteRotationMode}}
      }
      Sprite Rotation Angle Min {
       bIsDynamic=false
       bIsEditable=false
       bIsStaticSwitch=false
       bIsVisible=false
       type {classStructOrEnum {refPath=&NiagaraFloat}}
      }
      Sprite Rotation Angle {
       bIsDynamic=false
       bIsEditable=false
       bIsStaticSwitch=false
       bIsVisible=false
       type {classStructOrEnum {refPath=&NiagaraFloat}}
      }
      Sprite Rotation Angle Max {
       bIsDynamic=false
       bIsEditable=false
       bIsStaticSwitch=false
       bIsVisible=false
       type {classStructOrEnum {refPath=&NiagaraFloat}}
      }
      Sprite Rotation Randomness Mode {
       bIsDynamic=false
       bIsEditable=false
       bIsStaticSwitch=true
       bIsVisible=false
       type {classStructOrEnum {refPath=&ENiagaraRandomnessMode}}
      }
      Sprite UV Mode {
       bIsDynamic=false
       bIsEditable=true
       bIsStaticSwitch=true
       bIsVisible=true
       type {classStructOrEnum {refPath=&ENiagara_UVFlippingMode}}
      }
      Sprite UV Randomness Mode {
       bIsDynamic=false
       bIsEditable=false
       bIsStaticSwitch=true
       bIsVisible=false
       type {classStructOrEnum {refPath=&ENiagaraRandomnessMode}}
      }
      Sprite UV Scale {
       bIsDynamic=false
       bIsEditable=false
       bIsStaticSwitch=false
       bIsVisible=false
       type {classStructOrEnum {refPath=&Vector2f}}
      }
      Mesh Scale Mode {
       bIsDynamic=false
       bIsEditable=true
       bIsStaticSwitch=true
       bIsVisible=true
       type {classStructOrEnum {refPath=&ENiagara_SizeScaleMode}}
      }
      Mesh Uniform Scale Min {
       bIsDynamic=false
       bIsEditable=false
       bIsStaticSwitch=false
       bIsVisible=false
       type {classStructOrEnum {refPath=&NiagaraFloat}}
      }
      Mesh Uniform Scale {
       bIsDynamic=false
       bIsEditable=false
       bIsStaticSwitch=false
       bIsVisible=false
       type {classStructOrEnum {refPath=&NiagaraFloat}}
      }
      Mesh Scale Min {
       bIsDynamic=false
       bIsEditable=false
       bIsStaticSwitch=false
       bIsVisible=false
       type {classStructOrEnum {refPath=&Vector3f}}
      }
      Mesh Scale {
       bIsDynamic=false
       bIsEditable=false
       bIsStaticSwitch=false
       bIsVisible=false
       type {classStructOrEnum {refPath=&Vector3f}}
      }
      Mesh Scale Max {
       bIsDynamic=false
       bIsEditable=false
       bIsStaticSwitch=false
       bIsVisible=false
       type {classStructOrEnum {refPath=&Vector3f}}
      }
      Mesh Uniform Scale Max {
       bIsDynamic=false
       bIsEditable=false
       bIsStaticSwitch=false
       bIsVisible=false
       type {classStructOrEnum {refPath=&NiagaraFloat}}
      }
      Mesh Scale Randomness Mode {
       bIsDynamic=false
       bIsEditable=false
       bIsStaticSwitch=true
       bIsVisible=false
       type {classStructOrEnum {refPath=&ENiagaraRandomnessMode}}
      }
      Mesh Renderer Array Visibility Mode {
       bIsDynamic=false
       bIsEditable=true
       bIsStaticSwitch=true
       bIsVisible=true
       type {classStructOrEnum {refPath=&ENiagara_UnsetDirectSetRandom}}
      }
      Mesh Renderer Mesh Index {
       bIsDynamic=false
       bIsEditable=false
       bIsStaticSwitch=false
       bIsVisible=false
       type {classStructOrEnum {refPath=&NiagaraInt32}}
      }
      Wrap Index By Mesh Count {bIsDynamic=false, bIsEditable=false, bIsStaticSwitch=true, bIsVisible=false, type {classStructOrEnum {refPath=&NiagaraBool}}}
      Mesh Renderer Info {
       bIsDynamic=false
       bIsEditable=false
       bIsStaticSwitch=false
       bIsVisible=false
       type {classStructOrEnum {refPath=&NiagaraDataInterfaceMeshRendererInfo}}
      }
      Mesh Renderer Random Seed {
       bIsDynamic=false
       bIsEditable=false
       bIsStaticSwitch=false
       bIsVisible=false
       type {classStructOrEnum {refPath=&NiagaraInt32}}
      }
      Ribbon Width Mode {
       bIsDynamic=false
       bIsEditable=true
       bIsStaticSwitch=true
       bIsVisible=true
       type {classStructOrEnum {refPath=&ENiagara_UnsetDirectSet}}
      }
      Ribbon Width {
       bIsDynamic=false
       bIsEditable=false
       bIsStaticSwitch=false
       bIsVisible=false
       type {classStructOrEnum {refPath=&NiagaraFloat}}
      }
      Ribbon Facing Mode {
       bIsDynamic=false
       bIsEditable=true
       bIsStaticSwitch=true
       bIsVisible=true
       type {classStructOrEnum {refPath=&ENiagara_UnsetDirectSet}}
      }
      Ribbon Facing Vector {
       bIsDynamic=false
       bIsEditable=false
       bIsStaticSwitch=false
       bIsVisible=false
       type {classStructOrEnum {refPath=&Vector3f}}
      }
      Ribbon Twist Mode {
       bIsDynamic=false
       bIsEditable=true
       bIsStaticSwitch=true
       bIsVisible=true
       type {classStructOrEnum {refPath=&ENiagara_UnsetDirectSet}}
      }
      "Ribbon Twist " {
       bIsDynamic=false
       bIsEditable=false
       bIsStaticSwitch=false
       bIsVisible=false
       type {classStructOrEnum {refPath=&NiagaraFloat}}
      }
     ]
     moduleScript {refPath=&InitializeParticle}
    }
   ]
  }
  particleUpdateScript {
   scriptName=ParticleUpdateScript
   modules [
    {
     bIsSetParametersModule=false
     enabled=true
     moduleName=ParticleState
     inputs [
      Kill Particles When Lifetime Has Elapsed {bIsDynamic=false, bIsEditable=true, bIsStaticSwitch=true, bIsVisible=true, type {classStructOrEnum {refPath=&NiagaraBool}}}
      Loop Particles Lifetime {bIsDynamic=false, bIsEditable=false, bIsStaticSwitch=false, bIsVisible=false, type {classStructOrEnum {refPath=&NiagaraBool}}}
      Let Infinitely Lived Particles Die When Emitter Deactivates {bIsDynamic=false, bIsEditable=false, bIsStaticSwitch=false, bIsVisible=false, type {classStructOrEnum {refPath=&NiagaraBool}}}
      Lifetime {bIsDynamic=false, bIsEditable=true, bIsStaticSwitch=false, bIsVisible=true, type {classStructOrEnum {refPath=&NiagaraFloat}}}
      DeltaTime {bIsDynamic=false, bIsEditable=true, bIsStaticSwitch=false, bIsVisible=true, type {classStructOrEnum {refPath=&NiagaraFloat}}}
     ]
     moduleScript {refPath=&ParticleState}
    }
    {
     bIsSetParametersModule=false
     enabled=true
     moduleName=SkeletalMeshLocation
     inputs [
      Skeletal Mesh {
       bIsDynamic=false
       bIsEditable=true
       bIsStaticSwitch=false
       bIsVisible=true
       type {classStructOrEnum {refPath=&NiagaraDataInterfaceSkeletalMesh}}
      }
      Mesh Sampling Type {
       bIsDynamic=false
       bIsEditable=true
       bIsStaticSwitch=true
       bIsVisible=true
       type {classStructOrEnum {refPath=&ENiagaraSkelSamplingModeFull}}
      }
      Bone / Socket Sampling Mode {
       bIsDynamic=false
       bIsEditable=false
       bIsStaticSwitch=true
       bIsVisible=false
       type {classStructOrEnum {refPath=&ENiagaraBoneSocketSamplingMode}}
      }
      Bone Sampling Mode {
       bIsDynamic=false
       bIsEditable=true
       bIsStaticSwitch=true
       bIsVisible=true
       type {classStructOrEnum {refPath=&ENiagaraBoneSamplingMode}}
      }
      Socket Sampling Mode {
       bIsDynamic=false
       bIsEditable=false
       bIsStaticSwitch=true
       bIsVisible=false
       type {classStructOrEnum {refPath=&ENiagaraSocketSamplingMode}}
      }
      Triangle Sampling Mode {
       bIsDynamic=false
       bIsEditable=false
       bIsStaticSwitch=true
       bIsVisible=false
       type {classStructOrEnum {refPath=&ENiagaraTriangleSamplingMode}}
      }
      Vertex Sampling Mode {
       bIsDynamic=false
       bIsEditable=false
       bIsStaticSwitch=true
       bIsVisible=false
       type {classStructOrEnum {refPath=&ENiagaraVertexSamplingMode}}
      }
      Bone / Socket Index {bIsDynamic=false, bIsEditable=true, bIsStaticSwitch=false, bIsVisible=true, type {classStructOrEnum {refPath=&NiagaraInt32}}}
      Triangle ID {
       bIsDynamic=false
       bIsEditable=false
       bIsStaticSwitch=false
       bIsVisible=false
       type {classStructOrEnum {refPath=&NiagaraInt32}}
      }
      Vertex ID {
       bIsDynamic=false
       bIsEditable=false
       bIsStaticSwitch=false
       bIsVisible=false
       type {classStructOrEnum {refPath=&NiagaraInt32}}
      }
      FixedBarycentric {bIsDynamic=false, bIsEditable=false, bIsStaticSwitch=false, bIsVisible=false, type {classStructOrEnum {refPath=&NiagaraBool}}}
      Triangle Coordinate {
       bIsDynamic=false
       bIsEditable=false
       bIsStaticSwitch=false
       bIsVisible=false
       type {classStructOrEnum {refPath=&Vector3f}}
      }
      Sampled UV Index {
       bIsDynamic=false
       bIsEditable=false
       bIsStaticSwitch=false
       bIsVisible=false
       type {classStructOrEnum {refPath=&NiagaraInt32}}
      }
      Sampled Position Coordinate Space {
       bIsDynamic=false
       bIsEditable=true
       bIsStaticSwitch=true
       bIsVisible=true
       type {classStructOrEnum {refPath=&ENiagaraCoordinateSpace}}
      }
      Sampled Velocity Coordinate Space {
       bIsDynamic=false
       bIsEditable=true
       bIsStaticSwitch=true
       bIsVisible=true
       type {classStructOrEnum {refPath=&ENiagaraCoordinateSpace}}
      }
      Sampled Orientation Coordinate Space {
       bIsDynamic=false
       bIsEditable=true
       bIsStaticSwitch=true
       bIsVisible=true
       type {classStructOrEnum {refPath=&ENiagaraCoordinateSpace}}
      }
      Generate New Random Samples Every Frame {bIsDynamic=false, bIsEditable=false, bIsStaticSwitch=true, bIsVisible=false, type {classStructOrEnum {refPath=&NiagaraBool}}}
      Write Sampled Mesh Attrributes to Particle Attributes {bIsDynamic=false, bIsEditable=true, bIsStaticSwitch=true, bIsVisible=true, type {classStructOrEnum {refPath=&NiagaraBool}}}
      Randomness Mode {
       bIsDynamic=false
       bIsEditable=false
       bIsStaticSwitch=false
       bIsVisible=false
       type {classStructOrEnum {refPath=&ENiagaraRandomnessMode}}
      }
      Random Seed {
       bIsDynamic=false
       bIsEditable=false
       bIsStaticSwitch=false
       bIsVisible=false
       type {classStructOrEnum {refPath=&NiagaraInt32}}
      }
      Fixed Random Seed {bIsDynamic=false, bIsEditable=false, bIsStaticSwitch=false, bIsVisible=false, type {classStructOrEnum {refPath=&NiagaraBool}}}
      Apply To Particle Position {bIsDynamic=false, bIsEditable=true, bIsStaticSwitch=false, bIsVisible=true, type {classStructOrEnum {refPath=&NiagaraBool}}}
      Sampled Position Offset {bIsDynamic=false, bIsEditable=true, bIsStaticSwitch=false, bIsVisible=true, type {classStructOrEnum {refPath=&Vector3f}}}
      OffsetAlongNormal {bIsDynamic=false, bIsEditable=false, bIsStaticSwitch=false, bIsVisible=false, type {classStructOrEnum {refPath=&NiagaraBool}}}
      Offset Position Along Sampled Normal {
       bIsDynamic=false
       bIsEditable=false
       bIsStaticSwitch=false
       bIsVisible=false
       type {classStructOrEnum {refPath=&NiagaraFloat}}
      }
      Sampled Position Offset Coordinate Space {
       bIsDynamic=false
       bIsEditable=true
       bIsStaticSwitch=true
       bIsVisible=true
       type {classStructOrEnum {refPath=&ENiagaraMeshTransforms}}
      }
      Inherit Velocity {bIsDynamic=false, bIsEditable=true, bIsStaticSwitch=false, bIsVisible=true, type {classStructOrEnum {refPath=&NiagaraBool}}}
      "Inherit Velocity (Scale)" {bIsDynamic=false, bIsEditable=false, bIsStaticSwitch=false, bIsVisible=true, type {classStructOrEnum {refPath=&NiagaraFloat}}}
      VelocitySpeedLimit {bIsDynamic=false, bIsEditable=false, bIsStaticSwitch=false, bIsVisible=false, type {classStructOrEnum {refPath=&NiagaraBool}}}
      Inherited Velocity Speed Limit {
       bIsDynamic=false
       bIsEditable=false
       bIsStaticSwitch=false
       bIsVisible=false
       type {classStructOrEnum {refPath=&NiagaraFloat}}
      }
      Apply Sampled Orientation to Mesh Particles {bIsDynamic=false, bIsEditable=true, bIsStaticSwitch=true, bIsVisible=true, type {classStructOrEnum {refPath=&NiagaraBool}}}
      Interpolate Mesh Sampling {bIsDynamic=false, bIsEditable=true, bIsStaticSwitch=true, bIsVisible=true, type {classStructOrEnum {refPath=&NiagaraBool}}}
      MaskSpawnGroup {bIsDynamic=false, bIsEditable=true, bIsStaticSwitch=false, bIsVisible=true, type {classStructOrEnum {refPath=&NiagaraBool}}}
      Spawn Group Mask {bIsDynamic=false, bIsEditable=false, bIsStaticSwitch=false, bIsVisible=true, type {classStructOrEnum {refPath=&NiagaraInt32}}}
      Use Sampled Mesh Attribute as Ribbon ID {bIsDynamic=false, bIsEditable=true, bIsStaticSwitch=false, bIsVisible=true, type {classStructOrEnum {refPath=&NiagaraBool}}}
     ]
     moduleScript {refPath=&SkeletalMeshLocation}
    }
    {
     bIsSetParametersModule=false
     enabled=true
     moduleName=CameraOffset
     inputs [Camera Offset Amount {bIsDynamic=false, bIsEditable=true, bIsStaticSwitch=false, bIsVisible=true, type {classStructOrEnum {refPath=&NiagaraFloat}}}]
     moduleScript {refPath=&CameraOffset}
    }
    {
     bIsSetParametersModule=false
     enabled=true
     moduleName=ScaleColor
     inputs [
      Scale Mode {
       bIsDynamic=false
       bIsEditable=true
       bIsStaticSwitch=true
       bIsVisible=true
       type {classStructOrEnum {refPath=&ENiagaraScaleColorMode}}
      }
      Linear Color Curve {
       bIsDynamic=false
       bIsEditable=true
       bIsStaticSwitch=false
       bIsVisible=true
       type {classStructOrEnum {refPath=&NiagaraDataInterfaceColorCurve}}
      }
      ScaleRGB {bIsDynamic=false, bIsEditable=false, bIsStaticSwitch=false, bIsVisible=false, type {classStructOrEnum {refPath=&NiagaraBool}}}
      Scale RGB {
       bIsDynamic=false
       bIsEditable=false
       bIsStaticSwitch=false
       bIsVisible=false
       type {classStructOrEnum {refPath=&Vector3f}}
      }
      ScaleRGBA {bIsDynamic=false, bIsEditable=false, bIsStaticSwitch=false, bIsVisible=false, type {classStructOrEnum {refPath=&NiagaraBool}}}
      Scale RGBA {
       bIsDynamic=false
       bIsEditable=false
       bIsStaticSwitch=false
       bIsVisible=false
       type {classStructOrEnum {refPath=&Vector4f}}
      }
      Curve Index {bIsDynamic=false, bIsEditable=true, bIsStaticSwitch=false, bIsVisible=true, type {classStructOrEnum {refPath=&NiagaraFloat}}}
      ScaleA {bIsDynamic=false, bIsEditable=false, bIsStaticSwitch=false, bIsVisible=false, type {classStructOrEnum {refPath=&NiagaraBool}}}
      Scale Alpha {
       bIsDynamic=false
       bIsEditable=false
       bIsStaticSwitch=false
       bIsVisible=false
       type {classStructOrEnum {refPath=&NiagaraFloat}}
      }
      Color Value To Scale {
       bIsDynamic=false
       bIsEditable=true
       bIsStaticSwitch=false
       bIsVisible=true
       type {classStructOrEnum {refPath=&LinearColor}}
      }
     ]
     moduleScript {refPath=&ScaleColor}
    }
    {
     bIsSetParametersModule=false
     enabled=true
     moduleName=FN_Light_v2
     inputs [
      LightEnabled {bIsDynamic=false, bIsEditable=true, bIsStaticSwitch=false, bIsVisible=true, type {classStructOrEnum {refPath=&NiagaraBool}}}
      Scale Color {bIsDynamic=false, bIsEditable=true, bIsStaticSwitch=false, bIsVisible=true, type {classStructOrEnum {refPath=&NiagaraFloat}}}
      Light Exponent {bIsDynamic=false, bIsEditable=true, bIsStaticSwitch=false, bIsVisible=true, type {classStructOrEnum {refPath=&NiagaraFloat}}}
      Position Offset {bIsDynamic=false, bIsEditable=true, bIsStaticSwitch=false, bIsVisible=true, type {classStructOrEnum {refPath=&Vector3f}}}
      Light Radius {bIsDynamic=false, bIsEditable=true, bIsStaticSwitch=false, bIsVisible=true, type {classStructOrEnum {refPath=&NiagaraFloat}}}
     ]
     moduleScript {refPath=&FN_Light_v2}
    }
    {
     bIsSetParametersModule=false
     enabled=true
     moduleName=ScaleColor001
     inputs [
      Scale Mode {
       bIsDynamic=false
       bIsEditable=true
       bIsStaticSwitch=true
       bIsVisible=true
       type {classStructOrEnum {refPath=&ENiagaraScaleColorMode}}
      }
      Linear Color Curve {
       bIsDynamic=false
       bIsEditable=false
       bIsStaticSwitch=false
       bIsVisible=false
       type {classStructOrEnum {refPath=&NiagaraDataInterfaceColorCurve}}
      }
      ScaleRGB {bIsDynamic=false, bIsEditable=true, bIsStaticSwitch=false, bIsVisible=true, type {classStructOrEnum {refPath=&NiagaraBool}}}
      Scale RGB {bIsDynamic=false, bIsEditable=true, bIsStaticSwitch=false, bIsVisible=true, type {classStructOrEnum {refPath=&Vector3f}}}
      ScaleRGBA {bIsDynamic=false, bIsEditable=false, bIsStaticSwitch=false, bIsVisible=false, type {classStructOrEnum {refPath=&NiagaraBool}}}
      Scale RGBA {
       bIsDynamic=false
       bIsEditable=false
       bIsStaticSwitch=false
       bIsVisible=false
       type {classStructOrEnum {refPath=&Vector4f}}
      }
      Curve Index {
       bIsDynamic=false
       bIsEditable=false
       bIsStaticSwitch=false
       bIsVisible=false
       type {classStructOrEnum {refPath=&NiagaraFloat}}
      }
      ScaleA {bIsDynamic=false, bIsEditable=true, bIsStaticSwitch=false, bIsVisible=true, type {classStructOrEnum {refPath=&NiagaraBool}}}
      Scale Alpha {bIsDynamic=false, bIsEditable=true, bIsStaticSwitch=false, bIsVisible=true, type {classStructOrEnum {refPath=&NiagaraFloat}}}
      Color Value To Scale {
       bIsDynamic=false
       bIsEditable=true
       bIsStaticSwitch=false
       bIsVisible=true
       type {classStructOrEnum {refPath=&LinearColor}}
      }
     ]
     moduleScript {refPath=&ScaleColor}
    }
   ]
  }
  rendererClasses [{refPath=&NiagaraSpriteRendererProperties}, {refPath=&NiagaraLightRendererProperties}]
  renderers [
   {rendererIndex=0, rendererClass {refPath=&NiagaraSpriteRendererProperties}}
   {rendererIndex=1, rendererClass {refPath=&NiagaraLightRendererProperties}}
  ]
 }
 Trails {
  emitterSpawnScript {scriptName=EmitterSpawnScript, modules []}
  emitterUpdateScript {
   scriptName=EmitterUpdateScript
   modules [
    {
     bIsSetParametersModule=false
     enabled=true
     moduleName=EmitterState
     inputs [
      Life Cycle Mode {
       bIsDynamic=false
       bIsEditable=true
       bIsStaticSwitch=true
       bIsVisible=true
       type {classStructOrEnum {refPath=&ENiagaraEmitterLifeCycleMode}}
      }
      Inactive Response {
       bIsDynamic=false
       bIsEditable=true
       bIsStaticSwitch=true
       bIsVisible=true
       type {classStructOrEnum {refPath=&ENiagaraInactiveMode}}
      }
      Loop Behavior {
       bIsDynamic=false
       bIsEditable=true
       bIsStaticSwitch=true
       bIsVisible=true
       type {classStructOrEnum {refPath=&ENiagara_EmitterStateOptions}}
      }
      Loop Duration Mode {
       bIsDynamic=false
       bIsEditable=true
       bIsStaticSwitch=true
       bIsVisible=true
       type {classStructOrEnum {refPath=&ENiagara_InfiniteLoopDuration}}
      }
      Loop Count {
       bIsDynamic=false
       bIsEditable=false
       bIsStaticSwitch=false
       bIsVisible=false
       type {classStructOrEnum {refPath=&NiagaraInt32}}
      }
      Loop Duration {bIsDynamic=false, bIsEditable=true, bIsStaticSwitch=false, bIsVisible=true, type {classStructOrEnum {refPath=&NiagaraFloat}}}
      Recalculate Duration Each Loop {bIsDynamic=false, bIsEditable=false, bIsStaticSwitch=false, bIsVisible=false, type {classStructOrEnum {refPath=&NiagaraBool}}}
      UseLoopDelay {bIsDynamic=false, bIsEditable=true, bIsStaticSwitch=false, bIsVisible=true, type {classStructOrEnum {refPath=&NiagaraBool}}}
      Loop Delay {bIsDynamic=false, bIsEditable=true, bIsStaticSwitch=false, bIsVisible=true, type {classStructOrEnum {refPath=&NiagaraFloat}}}
      Delay First Loop Only {bIsDynamic=false, bIsEditable=false, bIsStaticSwitch=false, bIsVisible=false, type {classStructOrEnum {refPath=&NiagaraBool}}}
      Scalability Mode {
       bIsDynamic=false
       bIsEditable=true
       bIsStaticSwitch=true
       bIsVisible=true
       type {classStructOrEnum {refPath=&ENiagaraEmitterScalabilityMode_Limited}}
      }
      Enable Distance Culling {bIsDynamic=false, bIsEditable=false, bIsStaticSwitch=true, bIsVisible=false, type {classStructOrEnum {refPath=&NiagaraBool}}}
      MinDistance {
       bIsDynamic=false
       bIsEditable=false
       bIsStaticSwitch=false
       bIsVisible=false
       type {classStructOrEnum {refPath=&NiagaraFloat}}
      }
      Min Distance Response {
       bIsDynamic=false
       bIsEditable=false
       bIsStaticSwitch=false
       bIsVisible=false
       type {classStructOrEnum {refPath=&ENiagaraExecutionStateManagement}}
      }
      MaxDistance {
       bIsDynamic=false
       bIsEditable=false
       bIsStaticSwitch=false
       bIsVisible=false
       type {classStructOrEnum {refPath=&NiagaraFloat}}
      }
      Max Distance Response {
       bIsDynamic=false
       bIsEditable=false
       bIsStaticSwitch=false
       bIsVisible=false
       type {classStructOrEnum {refPath=&ENiagaraExecutionStateManagement}}
      }
      Scale Spawn Count {bIsDynamic=false, bIsEditable=false, bIsStaticSwitch=true, bIsVisible=false, type {classStructOrEnum {refPath=&NiagaraBool}}}
      Spawn Count Scale {
       bIsDynamic=false
       bIsEditable=false
       bIsStaticSwitch=false
       bIsVisible=false
       type {classStructOrEnum {refPath=&NiagaraFloat}}
      }
      Spawn Count Scale By Distance {
       bIsDynamic=false
       bIsEditable=false
       bIsStaticSwitch=false
       bIsVisible=false
       type {classStructOrEnum {refPath=&NiagaraDataInterfaceCurve}}
      }
      Enable Visibility Culling {bIsDynamic=false, bIsEditable=false, bIsStaticSwitch=true, bIsVisible=false, type {classStructOrEnum {refPath=&NiagaraBool}}}
      Visibility Cull Response {
       bIsDynamic=false
       bIsEditable=false
       bIsStaticSwitch=false
       bIsVisible=false
       type {classStructOrEnum {refPath=&ENiagaraExecutionStateManagement}}
      }
      Visibility Cull Delay {
       bIsDynamic=false
       bIsEditable=false
       bIsStaticSwitch=false
       bIsVisible=false
       type {classStructOrEnum {refPath=&NiagaraFloat}}
      }
      Reset Age On Awaken {bIsDynamic=false, bIsEditable=false, bIsStaticSwitch=false, bIsVisible=false, type {classStructOrEnum {refPath=&NiagaraBool}}}
     ]
     moduleScript {refPath=&EmitterState}
    }
    {
     bIsSetParametersModule=false
     enabled=true
     moduleName=SpawnRate001
     inputs [
      SpawnRate {bIsDynamic=false, bIsEditable=true, bIsStaticSwitch=false, bIsVisible=true, type {classStructOrEnum {refPath=&NiagaraFloat}}}
      Use Spawn Probability {bIsDynamic=false, bIsEditable=true, bIsStaticSwitch=false, bIsVisible=true, type {classStructOrEnum {refPath=&NiagaraBool}}}
      Spawn Probability {bIsDynamic=false, bIsEditable=false, bIsStaticSwitch=false, bIsVisible=true, type {classStructOrEnum {refPath=&NiagaraFloat}}}
      SpawnGroup {bIsDynamic=false, bIsEditable=true, bIsStaticSwitch=false, bIsVisible=true, type {classStructOrEnum {refPath=&NiagaraInt32}}}
     ]
     moduleScript {refPath=&SpawnRate}
    }
   ]
  }
  particleSpawnScript {
   scriptName=ParticleSpawnScript
   modules [
    {
     bIsSetParametersModule=false
     enabled=true
     moduleName=FN_SetSpawnGroup
     inputs [
      Number of Spawn Groups {bIsDynamic=false, bIsEditable=true, bIsStaticSwitch=false, bIsVisible=true, type {classStructOrEnum {refPath=&NiagaraInt32}}}
      Particles Per Group {
       bIsDynamic=false
       bIsEditable=false
       bIsStaticSwitch=false
       bIsVisible=false
       type {classStructOrEnum {refPath=&NiagaraInt32}}
      }
      Pattern {
       bIsDynamic=false
       bIsEditable=true
       bIsStaticSwitch=true
       bIsVisible=true
       type {classStructOrEnum {refPath=&ENiagaraPartitionWaveForm}}
      }
      Normalize Spawn Group {bIsDynamic=false, bIsEditable=true, bIsStaticSwitch=true, bIsVisible=true, type {classStructOrEnum {refPath=&NiagaraBool}}}
      Inclusivity {bIsDynamic=false, bIsEditable=false, bIsStaticSwitch=true, bIsVisible=false, type {classStructOrEnum {refPath=&NiagaraBool}}}
      Set Ribbon ID and Ribbon Link Order {bIsDynamic=false, bIsEditable=true, bIsStaticSwitch=true, bIsVisible=true, type {classStructOrEnum {refPath=&NiagaraBool}}}
     ]
     moduleScript {refPath=&FN_SetSpawnGroup}
    }
    {
     bIsSetParametersModule=false
     enabled=true
     moduleName=InitializeRibbon
     inputs [
      Lifetime Mode {
       bIsDynamic=false
       bIsEditable=true
       bIsStaticSwitch=true
       bIsVisible=true
       type {classStructOrEnum {refPath=&ENiagara_LifetimeMode}}
      }
      Lifetime Min {
       bIsDynamic=false
       bIsEditable=false
       bIsStaticSwitch=false
       bIsVisible=false
       type {classStructOrEnum {refPath=&NiagaraFloat}}
      }
      Lifetime {bIsDynamic=false, bIsEditable=true, bIsStaticSwitch=false, bIsVisible=true, type {classStructOrEnum {refPath=&NiagaraFloat}}}
      Lifetime Max {
       bIsDynamic=false
       bIsEditable=false
       bIsStaticSwitch=false
       bIsVisible=false
       type {classStructOrEnum {refPath=&NiagaraFloat}}
      }
      Lifetime Randomness Mode {
       bIsDynamic=false
       bIsEditable=false
       bIsStaticSwitch=true
       bIsVisible=false
       type {classStructOrEnum {refPath=&ENiagaraRandomnessMode}}
      }
      Color Mode {
       bIsDynamic=false
       bIsEditable=true
       bIsStaticSwitch=true
       bIsVisible=true
       type {classStructOrEnum {refPath=&ENiagara_ColorInitializationMode}}
      }
      Color {
       bIsDynamic=true
       bIsEditable=true
       bIsStaticSwitch=false
       bIsVisible=true
       type {classStructOrEnum {refPath=&LinearColor}}
      }
      Color Minimum {
       bIsDynamic=false
       bIsEditable=false
       bIsStaticSwitch=false
       bIsVisible=false
       type {classStructOrEnum {refPath=&LinearColor}}
      }
      Color Maximum {
       bIsDynamic=false
       bIsEditable=false
       bIsStaticSwitch=false
       bIsVisible=false
       type {classStructOrEnum {refPath=&LinearColor}}
      }
      AdjustHue {bIsDynamic=false, bIsEditable=false, bIsStaticSwitch=false, bIsVisible=false, type {classStructOrEnum {refPath=&NiagaraBool}}}
      Hue Shift Range {
       bIsDynamic=false
       bIsEditable=false
       bIsStaticSwitch=false
       bIsVisible=false
       type {classStructOrEnum {refPath=&Vector2f}}
      }
      AdjustSaturation {bIsDynamic=false, bIsEditable=false, bIsStaticSwitch=false, bIsVisible=false, type {classStructOrEnum {refPath=&NiagaraBool}}}
      Saturation Range {
       bIsDynamic=false
       bIsEditable=false
       bIsStaticSwitch=false
       bIsVisible=false
       type {classStructOrEnum {refPath=&Vector2f}}
      }
      AdjustValue {bIsDynamic=false, bIsEditable=false, bIsStaticSwitch=false, bIsVisible=false, type {classStructOrEnum {refPath=&NiagaraBool}}}
      Value Range {
       bIsDynamic=false
       bIsEditable=false
       bIsStaticSwitch=false
       bIsVisible=false
       type {classStructOrEnum {refPath=&Vector2f}}
      }
      AdjustAlpha {bIsDynamic=false, bIsEditable=false, bIsStaticSwitch=false, bIsVisible=false, type {classStructOrEnum {refPath=&NiagaraBool}}}
      Alpha Scale Range {
       bIsDynamic=false
       bIsEditable=false
       bIsStaticSwitch=false
       bIsVisible=false
       type {classStructOrEnum {refPath=&Vector2f}}
      }
      Color Channel Mode {
       bIsDynamic=false
       bIsEditable=false
       bIsStaticSwitch=true
       bIsVisible=false
       type {classStructOrEnum {refPath=&ENiagaraChannelCorrelation}}
      }
      Color Randomness Mode {
       bIsDynamic=false
       bIsEditable=false
       bIsStaticSwitch=true
       bIsVisible=false
       type {classStructOrEnum {refPath=&ENiagaraRandomnessMode}}
      }
      Position Mode {
       bIsDynamic=false
       bIsEditable=true
       bIsStaticSwitch=true
       bIsVisible=true
       type {classStructOrEnum {refPath=&ENiagara_PositionInitializationMode}}
      }
      Position {
       bIsDynamic=false
       bIsEditable=false
       bIsStaticSwitch=false
       bIsVisible=false
       type {classStructOrEnum {refPath=&NiagaraPosition}}
      }
      UsePositionOffset {bIsDynamic=false, bIsEditable=true, bIsStaticSwitch=false, bIsVisible=true, type {classStructOrEnum {refPath=&NiagaraBool}}}
      Position Offset {bIsDynamic=false, bIsEditable=false, bIsStaticSwitch=false, bIsVisible=true, type {classStructOrEnum {refPath=&Vector3f}}}
      Position Offset Coordinate Space {
       bIsDynamic=false
       bIsEditable=false
       bIsStaticSwitch=false
       bIsVisible=false
       type {classStructOrEnum {refPath=&ENiagaraCoordinateSpace}}
      }
      Mass Mode {
       bIsDynamic=false
       bIsEditable=true
       bIsStaticSwitch=true
       bIsVisible=true
       type {classStructOrEnum {refPath=&ENiagara_MassInitializationMode}}
      }
      Mass Min {
       bIsDynamic=false
       bIsEditable=false
       bIsStaticSwitch=false
       bIsVisible=false
       type {classStructOrEnum {refPath=&NiagaraFloat}}
      }
      Mass {
       bIsDynamic=false
       bIsEditable=false
       bIsStaticSwitch=false
       bIsVisible=false
       type {classStructOrEnum {refPath=&NiagaraFloat}}
      }
      Mass Max {
       bIsDynamic=false
       bIsEditable=false
       bIsStaticSwitch=false
       bIsVisible=false
       type {classStructOrEnum {refPath=&NiagaraFloat}}
      }
      Mass Randomness Mode {
       bIsDynamic=false
       bIsEditable=false
       bIsStaticSwitch=true
       bIsVisible=false
       type {classStructOrEnum {refPath=&ENiagaraRandomnessMode}}
      }
      Write Material Random {bIsDynamic=false, bIsEditable=true, bIsStaticSwitch=false, bIsVisible=true, type {classStructOrEnum {refPath=&NiagaraBool}}}
      Material Random {bIsDynamic=false, bIsEditable=true, bIsStaticSwitch=false, bIsVisible=true, type {classStructOrEnum {refPath=&NiagaraFloat}}}
      Ribbon Width Mode {
       bIsDynamic=false
       bIsEditable=true
       bIsStaticSwitch=true
       bIsVisible=true
       type {classStructOrEnum {refPath=&ENiagara_UnsetDirectSet}}
      }
      Ribbon Width {bIsDynamic=false, bIsEditable=true, bIsStaticSwitch=false, bIsVisible=true, type {classStructOrEnum {refPath=&NiagaraFloat}}}
      Ribbon Twist Mode {
       bIsDynamic=false
       bIsEditable=true
       bIsStaticSwitch=true
       bIsVisible=true
       type {classStructOrEnum {refPath=&ENiagara_UnsetDirectSet}}
      }
      "Ribbon Twist " {
       bIsDynamic=false
       bIsEditable=false
       bIsStaticSwitch=false
       bIsVisible=false
       type {classStructOrEnum {refPath=&NiagaraFloat}}
      }
      Ribbon Facing Mode {
       bIsDynamic=false
       bIsEditable=true
       bIsStaticSwitch=true
       bIsVisible=true
       type {classStructOrEnum {refPath=&ENiagara_UnsetDirectSet}}
      }
      Ribbon Facing Vector {bIsDynamic=false, bIsEditable=true, bIsStaticSwitch=false, bIsVisible=true, type {classStructOrEnum {refPath=&Vector3f}}}
     ]
     moduleScript {refPath=&InitializeRibbon}
    }
    {
     bIsSetParametersModule=false
     enabled=true
     moduleName=SkeletalMeshLocation
     inputs [
      Skeletal Mesh {
       bIsDynamic=false
       bIsEditable=true
       bIsStaticSwitch=false
       bIsVisible=true
       type {classStructOrEnum {refPath=&NiagaraDataInterfaceSkeletalMesh}}
      }
      Mesh Sampling Type {
       bIsDynamic=false
       bIsEditable=true
       bIsStaticSwitch=true
       bIsVisible=true
       type {classStructOrEnum {refPath=&ENiagaraSkelSamplingModeFull}}
      }
      Vertex Sampling Mode {
       bIsDynamic=false
       bIsEditable=false
       bIsStaticSwitch=true
       bIsVisible=false
       type {classStructOrEnum {refPath=&ENiagaraVertexSamplingMode}}
      }
      Socket Sampling Mode {
       bIsDynamic=false
       bIsEditable=false
       bIsStaticSwitch=true
       bIsVisible=false
       type {classStructOrEnum {refPath=&ENiagaraSocketSamplingMode}}
      }
      Bone Sampling Mode {
       bIsDynamic=false
       bIsEditable=true
       bIsStaticSwitch=true
       bIsVisible=true
       type {classStructOrEnum {refPath=&ENiagaraBoneSamplingMode}}
      }
      Bone / Socket Sampling Mode {
       bIsDynamic=false
       bIsEditable=false
       bIsStaticSwitch=true
       bIsVisible=false
       type {classStructOrEnum {refPath=&ENiagaraBoneSocketSamplingMode}}
      }
      Triangle Sampling Mode {
       bIsDynamic=false
       bIsEditable=false
       bIsStaticSwitch=true
       bIsVisible=false
       type {classStructOrEnum {refPath=&ENiagaraTriangleSamplingMode}}
      }
      Bone / Socket Index {bIsDynamic=false, bIsEditable=true, bIsStaticSwitch=false, bIsVisible=true, type {classStructOrEnum {refPath=&NiagaraInt32}}}
      Random Evaluation {
       bIsDynamic=false
       bIsEditable=false
       bIsStaticSwitch=true
       bIsVisible=false
       type {classStructOrEnum {refPath=/Niagara/Enums/ENiagaraRandomnessEvaluation}}
      }
      Vertex ID {
       bIsDynamic=false
       bIsEditable=false
       bIsStaticSwitch=false
       bIsVisible=false
       type {classStructOrEnum {refPath=&NiagaraInt32}}
      }
      Randomness Mode {
       bIsDynamic=false
       bIsEditable=false
       bIsStaticSwitch=true
       bIsVisible=false
       type {classStructOrEnum {refPath=&ENiagaraRandomnessMode}}
      }
      Triangle ID {
       bIsDynamic=false
       bIsEditable=false
       bIsStaticSwitch=false
       bIsVisible=false
       type {classStructOrEnum {refPath=&NiagaraInt32}}
      }
      FixedBarycentric {bIsDynamic=false, bIsEditable=false, bIsStaticSwitch=false, bIsVisible=false, type {classStructOrEnum {refPath=&NiagaraBool}}}
      Triangle Coordinate {
       bIsDynamic=false
       bIsEditable=false
       bIsStaticSwitch=false
       bIsVisible=false
       type {classStructOrEnum {refPath=&Vector3f}}
      }
      Sampled UV Index {
       bIsDynamic=false
       bIsEditable=false
       bIsStaticSwitch=false
       bIsVisible=false
       type {classStructOrEnum {refPath=&NiagaraInt32}}
      }
      OverrideRandomSeed {bIsDynamic=false, bIsEditable=false, bIsStaticSwitch=false, bIsVisible=false, type {classStructOrEnum {refPath=&NiagaraBool}}}
      Random Seed {
       bIsDynamic=false
       bIsEditable=false
       bIsStaticSwitch=false
       bIsVisible=false
       type {classStructOrEnum {refPath=&NiagaraInt32}}
      }
      Fixed Random Seed {bIsDynamic=false, bIsEditable=false, bIsStaticSwitch=true, bIsVisible=false, type {classStructOrEnum {refPath=&NiagaraBool}}}
      Position Sampling {
       bIsDynamic=false
       bIsEditable=true
       bIsStaticSwitch=true
       bIsVisible=true
       type {classStructOrEnum {refPath=/Niagara/Enums/ENiagaraSkelMeshPositionSamplingMode}}
      }
      Sampled Position Coordinate Space {
       bIsDynamic=false
       bIsEditable=true
       bIsStaticSwitch=true
       bIsVisible=true
       type {classStructOrEnum {refPath=&ENiagaraCoordinateSpace}}
      }
      Sampled Position Offset {bIsDynamic=false, bIsEditable=true, bIsStaticSwitch=false, bIsVisible=true, type {classStructOrEnum {refPath=&Vector3f}}}
      Sampled Position Offset Coordinate Space {
       bIsDynamic=false
       bIsEditable=true
       bIsStaticSwitch=true
       bIsVisible=true
       type {classStructOrEnum {refPath=/Niagara/Enums/ENiagaraSkelMeshTransforms}}
      }
      OffsetAlongNormal {bIsDynamic=false, bIsEditable=false, bIsStaticSwitch=false, bIsVisible=false, type {classStructOrEnum {refPath=&NiagaraBool}}}
      Offset Position Along Sampled Normal {
       bIsDynamic=false
       bIsEditable=false
       bIsStaticSwitch=false
       bIsVisible=false
       type {classStructOrEnum {refPath=&NiagaraFloat}}
      }
      Write To Particle Position {bIsDynamic=false, bIsEditable=true, bIsStaticSwitch=false, bIsVisible=true, type {classStructOrEnum {refPath=&NiagaraBool}}}
      Source Movement Scale {
       bIsDynamic=false
       bIsEditable=false
       bIsStaticSwitch=false
       bIsVisible=false
       type {classStructOrEnum {refPath=&Vector3f}}
      }
      Initialize Particle Position To Sampled Position {bIsDynamic=false, bIsEditable=false, bIsStaticSwitch=false, bIsVisible=false, type {classStructOrEnum {refPath=&NiagaraBool}}}
      Velocity Sampling {
       bIsDynamic=false
       bIsEditable=true
       bIsStaticSwitch=true
       bIsVisible=true
       type {classStructOrEnum {refPath=/Niagara/Enums/ENiagara_AttributeSamplingApplyOutput}}
      }
      Sampled Velocity Coordinate Space {
       bIsDynamic=false
       bIsEditable=true
       bIsStaticSwitch=true
       bIsVisible=true
       type {classStructOrEnum {refPath=&ENiagaraCoordinateSpace}}
      }
      "Inherit Velocity (Scale)" {
       bIsDynamic=false
       bIsEditable=false
       bIsStaticSwitch=false
       bIsVisible=false
       type {classStructOrEnum {refPath=&NiagaraFloat}}
      }
      VelocitySpeedLimit {bIsDynamic=false, bIsEditable=false, bIsStaticSwitch=false, bIsVisible=false, type {classStructOrEnum {refPath=&NiagaraBool}}}
      Inherited Velocity Speed Limit {
       bIsDynamic=false
       bIsEditable=false
       bIsStaticSwitch=false
       bIsVisible=false
       type {classStructOrEnum {refPath=&NiagaraFloat}}
      }
      Orientation Sampling {
       bIsDynamic=false
       bIsEditable=true
       bIsStaticSwitch=true
       bIsVisible=true
       type {classStructOrEnum {refPath=&ENiagara_AttributeSamplingApplyOutput}}
      }
      Sampled Orientation Coordinate Space {
       bIsDynamic=false
       bIsEditable=true
       bIsStaticSwitch=true
       bIsVisible=true
       type {classStructOrEnum {refPath=&ENiagaraCoordinateSpace}}
      }
      Use Sampled Mesh Attribute as Ribbon ID {bIsDynamic=false, bIsEditable=true, bIsStaticSwitch=false, bIsVisible=true, type {classStructOrEnum {refPath=&NiagaraBool}}}
      Sampled Scale Coordinate Space {
       bIsDynamic=false
       bIsEditable=true
       bIsStaticSwitch=true
       bIsVisible=true
       type {classStructOrEnum {refPath=&ENiagaraCoordinateSpace}}
      }
      Interpolate Mesh Sampling {bIsDynamic=false, bIsEditable=true, bIsStaticSwitch=true, bIsVisible=true, type {classStructOrEnum {refPath=&NiagaraBool}}}
      "Interpolate Mesh Sampling (Update)" {bIsDynamic=false, bIsEditable=false, bIsStaticSwitch=true, bIsVisible=false, type {classStructOrEnum {refPath=&NiagaraBool}}}
      MaskSpawnGroup {bIsDynamic=false, bIsEditable=true, bIsStaticSwitch=false, bIsVisible=true, type {classStructOrEnum {refPath=&NiagaraBool}}}
      Spawn Group Mask {bIsDynamic=false, bIsEditable=false, bIsStaticSwitch=false, bIsVisible=true, type {classStructOrEnum {refPath=&NiagaraInt32}}}
     ]
     moduleScript {refPath=&SkeletalMeshLocation}
    }
   ]
  }
  particleUpdateScript {
   scriptName=ParticleUpdateScript
   modules [
    {
     bIsSetParametersModule=false
     enabled=true
     moduleName=ParticleState
     inputs [
      Kill Particles When Lifetime Has Elapsed {bIsDynamic=false, bIsEditable=true, bIsStaticSwitch=true, bIsVisible=true, type {classStructOrEnum {refPath=&NiagaraBool}}}
      Loop Particles Lifetime {bIsDynamic=false, bIsEditable=false, bIsStaticSwitch=false, bIsVisible=false, type {classStructOrEnum {refPath=&NiagaraBool}}}
      Let Infinitely Lived Particles Die When Emitter Deactivates {bIsDynamic=false, bIsEditable=false, bIsStaticSwitch=false, bIsVisible=false, type {classStructOrEnum {refPath=&NiagaraBool}}}
      Lifetime {bIsDynamic=false, bIsEditable=true, bIsStaticSwitch=false, bIsVisible=true, type {classStructOrEnum {refPath=&NiagaraFloat}}}
      DeltaTime {bIsDynamic=false, bIsEditable=true, bIsStaticSwitch=false, bIsVisible=true, type {classStructOrEnum {refPath=&NiagaraFloat}}}
     ]
     moduleScript {refPath=&ParticleState}
    }
    {
     bIsSetParametersModule=false
     enabled=true
     moduleName=ScaleColor
     inputs [
      Scale Mode {
       bIsDynamic=false
       bIsEditable=true
       bIsStaticSwitch=true
       bIsVisible=true
       type {classStructOrEnum {refPath=&ENiagaraScaleColorMode}}
      }
      Linear Color Curve {
       bIsDynamic=false
       bIsEditable=true
       bIsStaticSwitch=false
       bIsVisible=true
       type {classStructOrEnum {refPath=&NiagaraDataInterfaceColorCurve}}
      }
      ScaleRGB {bIsDynamic=false, bIsEditable=false, bIsStaticSwitch=false, bIsVisible=false, type {classStructOrEnum {refPath=&NiagaraBool}}}
      Scale RGB {
       bIsDynamic=false
       bIsEditable=false
       bIsStaticSwitch=false
       bIsVisible=false
       type {classStructOrEnum {refPath=&Vector3f}}
      }
      ScaleRGBA {bIsDynamic=false, bIsEditable=false, bIsStaticSwitch=false, bIsVisible=false, type {classStructOrEnum {refPath=&NiagaraBool}}}
      Scale RGBA {
       bIsDynamic=false
       bIsEditable=false
       bIsStaticSwitch=false
       bIsVisible=false
       type {classStructOrEnum {refPath=&Vector4f}}
      }
      Curve Index {bIsDynamic=false, bIsEditable=true, bIsStaticSwitch=false, bIsVisible=true, type {classStructOrEnum {refPath=&NiagaraFloat}}}
      ScaleA {bIsDynamic=false, bIsEditable=false, bIsStaticSwitch=false, bIsVisible=false, type {classStructOrEnum {refPath=&NiagaraBool}}}
      Scale Alpha {
       bIsDynamic=false
       bIsEditable=false
       bIsStaticSwitch=false
       bIsVisible=false
       type {classStructOrEnum {refPath=&NiagaraFloat}}
      }
      Color Value To Scale {
       bIsDynamic=false
       bIsEditable=true
       bIsStaticSwitch=false
       bIsVisible=true
       type {classStructOrEnum {refPath=&LinearColor}}
      }
     ]
     moduleScript {refPath=&ScaleColor}
    }
    {
     bIsSetParametersModule=false
     enabled=true
     moduleName=ScaleRibbonWidth
     inputs [Ribbon Width Scale {bIsDynamic=true, bIsEditable=true, bIsStaticSwitch=false, bIsVisible=true, type {classStructOrEnum {refPath=&NiagaraFloat}}}]
     moduleScript {refPath=&ScaleRibbonWidth}
    }
    {
     bIsSetParametersModule=false
     enabled=false
     moduleName=ScaleRibbonWidth001
     inputs [Ribbon Width Scale {bIsDynamic=true, bIsEditable=true, bIsStaticSwitch=false, bIsVisible=true, type {classStructOrEnum {refPath=&NiagaraFloat}}}]
     moduleScript {refPath=&ScaleRibbonWidth}
    }
    {
     bIsSetParametersModule=false
     enabled=false
     moduleName=DynamicMaterialParameters001
     inputs [
      Write Parameter Index 0 {bIsDynamic=false, bIsEditable=true, bIsStaticSwitch=true, bIsVisible=true, type {classStructOrEnum {refPath=&NiagaraBool}}}
      Module.Param0WriteEnabled {bIsDynamic=false, bIsEditable=true, bIsStaticSwitch=false, bIsVisible=true, type {classStructOrEnum {refPath=&NiagaraBool}}}
      Index 0 Param 1 {bIsDynamic=false, bIsEditable=true, bIsStaticSwitch=false, bIsVisible=true, type {classStructOrEnum {refPath=&NiagaraFloat}}}
      Module.Param1WriteEnabled {bIsDynamic=false, bIsEditable=true, bIsStaticSwitch=false, bIsVisible=true, type {classStructOrEnum {refPath=&NiagaraBool}}}
      Index 0 Param 2 {bIsDynamic=false, bIsEditable=true, bIsStaticSwitch=false, bIsVisible=true, type {classStructOrEnum {refPath=&NiagaraFloat}}}
      Module.Param2WriteEnabled {bIsDynamic=false, bIsEditable=true, bIsStaticSwitch=false, bIsVisible=true, type {classStructOrEnum {refPath=&NiagaraBool}}}
      Index 0 Param 3 {bIsDynamic=false, bIsEditable=true, bIsStaticSwitch=false, bIsVisible=true, type {classStructOrEnum {refPath=&NiagaraFloat}}}
      Module.Param3WriteEnabled {bIsDynamic=false, bIsEditable=true, bIsStaticSwitch=false, bIsVisible=true, type {classStructOrEnum {refPath=&NiagaraBool}}}
      Index 0 Param 4 {bIsDynamic=false, bIsEditable=false, bIsStaticSwitch=false, bIsVisible=true, type {classStructOrEnum {refPath=&NiagaraFloat}}}
      Write Parameter Index 1 {bIsDynamic=false, bIsEditable=true, bIsStaticSwitch=true, bIsVisible=true, type {classStructOrEnum {refPath=&NiagaraBool}}}
      Index 1 Param 1 {
       bIsDynamic=false
       bIsEditable=false
       bIsStaticSwitch=false
       bIsVisible=false
       type {classStructOrEnum {refPath=&NiagaraFloat}}
      }
      Index 1 Param 2 {
       bIsDynamic=false
       bIsEditable=false
       bIsStaticSwitch=false
       bIsVisible=false
       type {classStructOrEnum {refPath=&NiagaraFloat}}
      }
      Index 1 Param 3 {
       bIsDynamic=false
       bIsEditable=false
       bIsStaticSwitch=false
       bIsVisible=false
       type {classStructOrEnum {refPath=&NiagaraFloat}}
      }
      Index 1 Param 4 {
       bIsDynamic=false
       bIsEditable=false
       bIsStaticSwitch=false
       bIsVisible=false
       type {classStructOrEnum {refPath=&NiagaraFloat}}
      }
      Write Parameter Index 2 {bIsDynamic=false, bIsEditable=true, bIsStaticSwitch=true, bIsVisible=true, type {classStructOrEnum {refPath=&NiagaraBool}}}
      Index 2 Param 1 {
       bIsDynamic=false
       bIsEditable=false
       bIsStaticSwitch=false
       bIsVisible=false
       type {classStructOrEnum {refPath=&NiagaraFloat}}
      }
      Index 2 Param 2 {
       bIsDynamic=false
       bIsEditable=false
       bIsStaticSwitch=false
       bIsVisible=false
       type {classStructOrEnum {refPath=&NiagaraFloat}}
      }
      Index 2 Param 3 {
       bIsDynamic=false
       bIsEditable=false
       bIsStaticSwitch=false
       bIsVisible=false
       type {classStructOrEnum {refPath=&NiagaraFloat}}
      }
      Index 2 Param 4 {
       bIsDynamic=false
       bIsEditable=false
       bIsStaticSwitch=false
       bIsVisible=false
       type {classStructOrEnum {refPath=&NiagaraFloat}}
      }
      Write Parameter Index 3 {bIsDynamic=false, bIsEditable=true, bIsStaticSwitch=true, bIsVisible=true, type {classStructOrEnum {refPath=&NiagaraBool}}}
      Index 3 Param 1 {
       bIsDynamic=false
       bIsEditable=false
       bIsStaticSwitch=false
       bIsVisible=false
       type {classStructOrEnum {refPath=&NiagaraFloat}}
      }
      Index 3 Param 2 {
       bIsDynamic=false
       bIsEditable=false
       bIsStaticSwitch=false
       bIsVisible=false
       type {classStructOrEnum {refPath=&NiagaraFloat}}
      }
      Index 3 Param 3 {
       bIsDynamic=false
       bIsEditable=false
       bIsStaticSwitch=false
       bIsVisible=false
       type {classStructOrEnum {refPath=&NiagaraFloat}}
      }
      Index 3 Param 4 {
       bIsDynamic=false
       bIsEditable=false
       bIsStaticSwitch=false
       bIsVisible=false
       type {classStructOrEnum {refPath=&NiagaraFloat}}
      }
     ]
     moduleScript {refPath=&DynamicMaterialParameters}
    }
    {
     bIsSetParametersModule=false
     enabled=false
     moduleName=CameraOffset
     inputs [Camera Offset Amount {bIsDynamic=false, bIsEditable=true, bIsStaticSwitch=false, bIsVisible=true, type {classStructOrEnum {refPath=&NiagaraFloat}}}]
     moduleScript {refPath=&CameraOffset}
    }
    {
     bIsSetParametersModule=false
     enabled=false
     moduleName=SkeletalMeshLocation001
     inputs [
      Skeletal Mesh {
       bIsDynamic=false
       bIsEditable=true
       bIsStaticSwitch=false
       bIsVisible=true
       type {classStructOrEnum {refPath=&NiagaraDataInterfaceSkeletalMesh}}
      }
      Mesh Sampling Type {
       bIsDynamic=false
       bIsEditable=true
       bIsStaticSwitch=true
       bIsVisible=true
       type {classStructOrEnum {refPath=&ENiagaraSkelSamplingModeFull}}
      }
      Vertex Sampling Mode {
       bIsDynamic=false
       bIsEditable=false
       bIsStaticSwitch=true
       bIsVisible=false
       type {classStructOrEnum {refPath=&ENiagaraVertexSamplingMode}}
      }
      Socket Sampling Mode {
       bIsDynamic=false
       bIsEditable=false
       bIsStaticSwitch=true
       bIsVisible=false
       type {classStructOrEnum {refPath=&ENiagaraSocketSamplingMode}}
      }
      Bone Sampling Mode {
       bIsDynamic=false
       bIsEditable=true
       bIsStaticSwitch=true
       bIsVisible=true
       type {classStructOrEnum {refPath=&ENiagaraBoneSamplingMode}}
      }
      Bone / Socket Sampling Mode {
       bIsDynamic=false
       bIsEditable=false
       bIsStaticSwitch=true
       bIsVisible=false
       type {classStructOrEnum {refPath=&ENiagaraBoneSocketSamplingMode}}
      }
      Triangle Sampling Mode {
       bIsDynamic=false
       bIsEditable=false
       bIsStaticSwitch=true
       bIsVisible=false
       type {classStructOrEnum {refPath=&ENiagaraTriangleSamplingMode}}
      }
      Bone / Socket Index {bIsDynamic=false, bIsEditable=true, bIsStaticSwitch=false, bIsVisible=true, type {classStructOrEnum {refPath=&NiagaraInt32}}}
      Random Evaluation {
       bIsDynamic=false
       bIsEditable=false
       bIsStaticSwitch=true
       bIsVisible=false
       type {classStructOrEnum {refPath=&ENiagaraRandomnessEvaluation}}
      }
      Vertex ID {
       bIsDynamic=false
       bIsEditable=false
       bIsStaticSwitch=false
       bIsVisible=false
       type {classStructOrEnum {refPath=&NiagaraInt32}}
      }
      Randomness Mode {
       bIsDynamic=false
       bIsEditable=false
       bIsStaticSwitch=true
       bIsVisible=false
       type {classStructOrEnum {refPath=&ENiagaraRandomnessMode}}
      }
      Triangle ID {
       bIsDynamic=false
       bIsEditable=false
       bIsStaticSwitch=false
       bIsVisible=false
       type {classStructOrEnum {refPath=&NiagaraInt32}}
      }
      FixedBarycentric {bIsDynamic=false, bIsEditable=false, bIsStaticSwitch=false, bIsVisible=false, type {classStructOrEnum {refPath=&NiagaraBool}}}
      Triangle Coordinate {
       bIsDynamic=false
       bIsEditable=false
       bIsStaticSwitch=false
       bIsVisible=false
       type {classStructOrEnum {refPath=&Vector3f}}
      }
      Sampled UV Index {
       bIsDynamic=false
       bIsEditable=false
       bIsStaticSwitch=false
       bIsVisible=false
       type {classStructOrEnum {refPath=&NiagaraInt32}}
      }
      OverrideRandomSeed {bIsDynamic=false, bIsEditable=false, bIsStaticSwitch=false, bIsVisible=false, type {classStructOrEnum {refPath=&NiagaraBool}}}
      Random Seed {
       bIsDynamic=false
       bIsEditable=false
       bIsStaticSwitch=false
       bIsVisible=false
       type {classStructOrEnum {refPath=&NiagaraInt32}}
      }
      Fixed Random Seed {bIsDynamic=false, bIsEditable=false, bIsStaticSwitch=true, bIsVisible=false, type {classStructOrEnum {refPath=&NiagaraBool}}}
      Position Sampling {
       bIsDynamic=false
       bIsEditable=true
       bIsStaticSwitch=true
       bIsVisible=true
       type {classStructOrEnum {refPath=&ENiagaraSkelMeshPositionSamplingMode}}
      }
      Sampled Position Coordinate Space {
       bIsDynamic=false
       bIsEditable=true
       bIsStaticSwitch=true
       bIsVisible=true
       type {classStructOrEnum {refPath=&ENiagaraCoordinateSpace}}
      }
      Sampled Position Offset {bIsDynamic=false, bIsEditable=true, bIsStaticSwitch=false, bIsVisible=true, type {classStructOrEnum {refPath=&Vector3f}}}
      Sampled Position Offset Coordinate Space {
       bIsDynamic=false
       bIsEditable=true
       bIsStaticSwitch=true
       bIsVisible=true
       type {classStructOrEnum {refPath=&ENiagaraSkelMeshTransforms}}
      }
      OffsetAlongNormal {bIsDynamic=false, bIsEditable=false, bIsStaticSwitch=false, bIsVisible=false, type {classStructOrEnum {refPath=&NiagaraBool}}}
      Offset Position Along Sampled Normal {
       bIsDynamic=false
       bIsEditable=false
       bIsStaticSwitch=false
       bIsVisible=false
       type {classStructOrEnum {refPath=&NiagaraFloat}}
      }
      Write To Particle Position {bIsDynamic=false, bIsEditable=false, bIsStaticSwitch=false, bIsVisible=false, type {classStructOrEnum {refPath=&NiagaraBool}}}
      Source Movement Scale {
       bIsDynamic=false
       bIsEditable=false
       bIsStaticSwitch=false
       bIsVisible=false
       type {classStructOrEnum {refPath=&Vector3f}}
      }
      Initialize Particle Position To Sampled Position {bIsDynamic=false, bIsEditable=false, bIsStaticSwitch=false, bIsVisible=false, type {classStructOrEnum {refPath=&NiagaraBool}}}
      Velocity Sampling {
       bIsDynamic=false
       bIsEditable=true
       bIsStaticSwitch=true
       bIsVisible=true
       type {classStructOrEnum {refPath=&ENiagara_AttributeSamplingApplyOutput}}
      }
      Sampled Velocity Coordinate Space {
       bIsDynamic=false
       bIsEditable=true
       bIsStaticSwitch=true
       bIsVisible=true
       type {classStructOrEnum {refPath=&ENiagaraCoordinateSpace}}
      }
      "Inherit Velocity (Scale)" {
       bIsDynamic=false
       bIsEditable=false
       bIsStaticSwitch=false
       bIsVisible=false
       type {classStructOrEnum {refPath=&NiagaraFloat}}
      }
      VelocitySpeedLimit {bIsDynamic=false, bIsEditable=false, bIsStaticSwitch=false, bIsVisible=false, type {classStructOrEnum {refPath=&NiagaraBool}}}
      Inherited Velocity Speed Limit {
       bIsDynamic=false
       bIsEditable=false
       bIsStaticSwitch=false
       bIsVisible=false
       type {classStructOrEnum {refPath=&NiagaraFloat}}
      }
      Orientation Sampling {
       bIsDynamic=false
       bIsEditable=true
       bIsStaticSwitch=true
       bIsVisible=true
       type {classStructOrEnum {refPath=&ENiagara_AttributeSamplingApplyOutput}}
      }
      Sampled Orientation Coordinate Space {
       bIsDynamic=false
       bIsEditable=true
       bIsStaticSwitch=true
       bIsVisible=true
       type {classStructOrEnum {refPath=&ENiagaraCoordinateSpace}}
      }
      Use Sampled Mesh Attribute as Ribbon ID {bIsDynamic=false, bIsEditable=true, bIsStaticSwitch=false, bIsVisible=true, type {classStructOrEnum {refPath=&NiagaraBool}}}
      Sampled Scale Coordinate Space {
       bIsDynamic=false
       bIsEditable=true
       bIsStaticSwitch=true
       bIsVisible=true
       type {classStructOrEnum {refPath=&ENiagaraCoordinateSpace}}
      }
      Interpolate Mesh Sampling {bIsDynamic=false, bIsEditable=false, bIsStaticSwitch=true, bIsVisible=false, type {classStructOrEnum {refPath=&NiagaraBool}}}
      "Interpolate Mesh Sampling (Update)" {bIsDynamic=false, bIsEditable=true, bIsStaticSwitch=true, bIsVisible=true, type {classStructOrEnum {refPath=&NiagaraBool}}}
      MaskSpawnGroup {bIsDynamic=false, bIsEditable=true, bIsStaticSwitch=false, bIsVisible=true, type {classStructOrEnum {refPath=&NiagaraBool}}}
      Spawn Group Mask {bIsDynamic=false, bIsEditable=false, bIsStaticSwitch=false, bIsVisible=true, type {classStructOrEnum {refPath=&NiagaraInt32}}}
     ]
     moduleScript {refPath=&SkeletalMeshLocation}
    }
    {
     bIsSetParametersModule=false
     enabled=false
     moduleName=PointAttractionForce
     inputs [
      Attraction Strength {bIsDynamic=true, bIsEditable=true, bIsStaticSwitch=false, bIsVisible=true, type {classStructOrEnum {refPath=&NiagaraFloat}}}
      Attraction Radius {bIsDynamic=false, bIsEditable=true, bIsStaticSwitch=false, bIsVisible=true, type {classStructOrEnum {refPath=&NiagaraFloat}}}
      Module.Use Falloff {bIsDynamic=false, bIsEditable=true, bIsStaticSwitch=false, bIsVisible=true, type {classStructOrEnum {refPath=&NiagaraBool}}}
      Falloff Exponent {bIsDynamic=false, bIsEditable=true, bIsStaticSwitch=false, bIsVisible=true, type {classStructOrEnum {refPath=&NiagaraFloat}}}
      UseAttractorOffset {bIsDynamic=false, bIsEditable=true, bIsStaticSwitch=false, bIsVisible=true, type {classStructOrEnum {refPath=&NiagaraBool}}}
      Attractor Position Offset {bIsDynamic=false, bIsEditable=false, bIsStaticSwitch=false, bIsVisible=true, type {classStructOrEnum {refPath=&Vector3f}}}
      Attractor Position Offset Coordinate Space {
       bIsDynamic=false
       bIsEditable=false
       bIsStaticSwitch=false
       bIsVisible=false
       type {classStructOrEnum {refPath=&ENiagaraCoordinateSpace}}
      }
      Kill Within Radius {bIsDynamic=false, bIsEditable=true, bIsStaticSwitch=false, bIsVisible=true, type {classStructOrEnum {refPath=&NiagaraBool}}}
      Kill Radius {bIsDynamic=false, bIsEditable=false, bIsStaticSwitch=false, bIsVisible=true, type {classStructOrEnum {refPath=&NiagaraFloat}}}
      Kill Radius Overshoot Correction {
       bIsDynamic=false
       bIsEditable=false
       bIsStaticSwitch=false
       bIsVisible=false
       type {classStructOrEnum {refPath=&NiagaraFloat}}
      }
      Position To Attract {
       bIsDynamic=false
       bIsEditable=true
       bIsStaticSwitch=false
       bIsVisible=true
       type {classStructOrEnum {refPath=&NiagaraPosition}}
      }
      Attractor Position {
       bIsDynamic=false
       bIsEditable=true
       bIsStaticSwitch=false
       bIsVisible=true
       type {classStructOrEnum {refPath=&NiagaraPosition}}
      }
      Debug Every N Particles {
       bIsDynamic=false
       bIsEditable=false
       bIsStaticSwitch=false
       bIsVisible=false
       type {classStructOrEnum {refPath=&NiagaraInt32}}
      }
      Debug Force Line Color {
       bIsDynamic=false
       bIsEditable=false
       bIsStaticSwitch=false
       bIsVisible=false
       type {classStructOrEnum {refPath=&LinearColor}}
      }
      Debug Attractor Radius Color {
       bIsDynamic=false
       bIsEditable=false
       bIsStaticSwitch=false
       bIsVisible=false
       type {classStructOrEnum {refPath=&LinearColor}}
      }
      Debug Force Line Length {
       bIsDynamic=false
       bIsEditable=false
       bIsStaticSwitch=false
       bIsVisible=false
       type {classStructOrEnum {refPath=&NiagaraFloat}}
      }
     ]
     moduleScript {refPath=&PointAttractionForce}
    }
    {
     bIsSetParametersModule=false
     enabled=true
     moduleName=ScaleColor001
     inputs [
      Scale Mode {
       bIsDynamic=false
       bIsEditable=true
       bIsStaticSwitch=true
       bIsVisible=true
       type {classStructOrEnum {refPath=&ENiagaraScaleColorMode}}
      }
      Linear Color Curve {
       bIsDynamic=false
       bIsEditable=false
       bIsStaticSwitch=false
       bIsVisible=false
       type {classStructOrEnum {refPath=&NiagaraDataInterfaceColorCurve}}
      }
      ScaleRGB {bIsDynamic=false, bIsEditable=true, bIsStaticSwitch=false, bIsVisible=true, type {classStructOrEnum {refPath=&NiagaraBool}}}
      Scale RGB {bIsDynamic=false, bIsEditable=true, bIsStaticSwitch=false, bIsVisible=true, type {classStructOrEnum {refPath=&Vector3f}}}
      ScaleRGBA {bIsDynamic=false, bIsEditable=false, bIsStaticSwitch=false, bIsVisible=false, type {classStructOrEnum {refPath=&NiagaraBool}}}
      Scale RGBA {
       bIsDynamic=false
       bIsEditable=false
       bIsStaticSwitch=false
       bIsVisible=false
       type {classStructOrEnum {refPath=&Vector4f}}
      }
      Curve Index {
       bIsDynamic=false
       bIsEditable=false
       bIsStaticSwitch=false
       bIsVisible=false
       type {classStructOrEnum {refPath=&NiagaraFloat}}
      }
      ScaleA {bIsDynamic=false, bIsEditable=true, bIsStaticSwitch=false, bIsVisible=true, type {classStructOrEnum {refPath=&NiagaraBool}}}
      Scale Alpha {bIsDynamic=false, bIsEditable=true, bIsStaticSwitch=false, bIsVisible=true, type {classStructOrEnum {refPath=&NiagaraFloat}}}
      Color Value To Scale {
       bIsDynamic=false
       bIsEditable=true
       bIsStaticSwitch=false
       bIsVisible=true
       type {classStructOrEnum {refPath=&LinearColor}}
      }
     ]
     moduleScript {refPath=&ScaleColor}
    }
    {
     bIsSetParametersModule=false
     enabled=true
     moduleName=SolveForcesAndVelocity
     inputs [
      Clamp Velocity {bIsDynamic=false, bIsEditable=true, bIsStaticSwitch=false, bIsVisible=true, type {classStructOrEnum {refPath=&NiagaraBool}}}
      Speed Limit {bIsDynamic=false, bIsEditable=false, bIsStaticSwitch=false, bIsVisible=true, type {classStructOrEnum {refPath=&NiagaraFloat}}}
      Limit Acceleration {bIsDynamic=false, bIsEditable=true, bIsStaticSwitch=false, bIsVisible=true, type {classStructOrEnum {refPath=&NiagaraBool}}}
      Acceleration Limit {bIsDynamic=false, bIsEditable=false, bIsStaticSwitch=false, bIsVisible=true, type {classStructOrEnum {refPath=&NiagaraFloat}}}
      Manually Enable Rotational Solver {bIsDynamic=false, bIsEditable=true, bIsStaticSwitch=false, bIsVisible=true, type {classStructOrEnum {refPath=&NiagaraBool}}}
      Write to Intrinsic Properties {bIsDynamic=false, bIsEditable=true, bIsStaticSwitch=false, bIsVisible=true, type {classStructOrEnum {refPath=&NiagaraBool}}}
      Write to Presolve Properties {bIsDynamic=false, bIsEditable=true, bIsStaticSwitch=true, bIsVisible=true, type {classStructOrEnum {refPath=&NiagaraBool}}}
      DeltaTime {bIsDynamic=false, bIsEditable=true, bIsStaticSwitch=false, bIsVisible=true, type {classStructOrEnum {refPath=&NiagaraFloat}}}
      Mass {bIsDynamic=false, bIsEditable=true, bIsStaticSwitch=false, bIsVisible=true, type {classStructOrEnum {refPath=&NiagaraFloat}}}
      Position {
       bIsDynamic=false
       bIsEditable=true
       bIsStaticSwitch=false
       bIsVisible=true
       type {classStructOrEnum {refPath=&NiagaraPosition}}
      }
      Velocity {bIsDynamic=false, bIsEditable=true, bIsStaticSwitch=false, bIsVisible=true, type {classStructOrEnum {refPath=&Vector3f}}}
      PreviousVelocity {bIsDynamic=false, bIsEditable=true, bIsStaticSwitch=false, bIsVisible=true, type {classStructOrEnum {refPath=&Vector3f}}}
      Force {bIsDynamic=false, bIsEditable=true, bIsStaticSwitch=false, bIsVisible=true, type {classStructOrEnum {refPath=&Vector3f}}}
      Mesh Orientation {bIsDynamic=false, bIsEditable=false, bIsStaticSwitch=false, bIsVisible=false, type {classStructOrEnum {refPath=&Quat4f}}}
      Rotational Velocity {
       bIsDynamic=false
       bIsEditable=false
       bIsStaticSwitch=false
       bIsVisible=false
       type {classStructOrEnum {refPath=&Vector3f}}
      }
      Rotational Force {
       bIsDynamic=false
       bIsEditable=false
       bIsStaticSwitch=false
       bIsVisible=false
       type {classStructOrEnum {refPath=&Vector3f}}
      }
      Rotational Inertia {
       bIsDynamic=false
       bIsEditable=false
       bIsStaticSwitch=false
       bIsVisible=false
       type {classStructOrEnum {refPath=&NiagaraFloat}}
      }
     ]
     moduleScript {refPath=&SolveForcesAndVelocity}
    }
   ]
  }
  rendererClasses [{refPath=&NiagaraRibbonRendererProperties}]
  renderers [{rendererIndex=0, rendererClass {refPath=&NiagaraRibbonRendererProperties}}]
 }
}
summary {
 systemName=NS_Emote_RealCrown_Glows
 emitters [
  {
   bEnabled=false
   emitterName=HandGlow
   rendererClasses [{refPath=&NiagaraSpriteRendererProperties}, {refPath=&NiagaraLightRendererProperties}]
  }
  {
   bEnabled=false
   emitterName=HandGlow001
   rendererClasses [{refPath=&NiagaraSpriteRendererProperties}, {refPath=&NiagaraLightRendererProperties}]
  }
  {
   bEnabled=false
   emitterName=HandGlow002
   rendererClasses [{refPath=&NiagaraSpriteRendererProperties}, {refPath=&NiagaraLightRendererProperties}]
  }
  {
   bEnabled=false
   emitterName=HandGlow003
   rendererClasses [{refPath=&NiagaraSpriteRendererProperties}, {refPath=&NiagaraLightRendererProperties}]
  }
  {
   bEnabled=false
   emitterName=HandGlow004
   rendererClasses [{refPath=&NiagaraSpriteRendererProperties}, {refPath=&NiagaraLightRendererProperties}]
  }
  {emitterName=Glows, rendererClasses [{refPath=&NiagaraSpriteRendererProperties}]}
  {emitterName=Trails, rendererClasses [{refPath=&NiagaraRibbonRendererProperties}]}
 ]
 userVariables [
  User.bIsFrontEnd {
   description=""
   defaultValue {struct {refPath=&NiagaraBool}, value {value=0}}
   type {classStructOrEnum {refPath=&NiagaraBool}}
  }
  User.bIsFrontEndPreview {
   description=""
   defaultValue {struct {refPath=&NiagaraBool}, value {value=0}}
   type {classStructOrEnum {refPath=&NiagaraBool}}
  }
  User.RibbonIDOffset {
   description=""
   defaultValue {struct {refPath=&NiagaraInt32}, value {value=0}}
   type {classStructOrEnum {refPath=&NiagaraInt32}}
  }
  User.SkeletalMesh {
   description=""
   defaultValue {
    struct {refPath=/Script/NiagaraEditor.NiagaraExt_VariableValue_DataInterface}
    value {
     dataInterface {refPath=&self:NiagaraDataInterfaceSkeletalMesh_0}
     dataInterfaceClass {refPath=&NiagaraDataInterfaceSkeletalMesh}
    }
   }
   type {classStructOrEnum {refPath=&NiagaraDataInterfaceSkeletalMesh}}
  }
 ]
}
